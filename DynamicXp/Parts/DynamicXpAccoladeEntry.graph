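# List-entry widget for one Accolade Entry Object in DynamicXp, showing its trigger rate, interval, trigger count and last XP.
# OnListItemObjectSet fills the name and rate, then binds the object's carry-over, trigger and error events.
# The spin box writes TriggerMult and refreshes the time-between-triggers text, while carry-over drives a background progress bar.
# Errors only set a flag, which the next trigger event consumes to tint the bar with the error color.
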WidgetBlueprint /DynamicXp/Parts/DynamicXpAccoladeEntry
graphs [
 EventGraph {nodes [
   n1 = "On Value Committed (SpinBox_FreqMult)"!() -> (OutputDelegate, then, InValue:float, CommitMethod:ETextCommit Enum=Default) {is=title K2Node_ComponentBoundEvent}
   n2 = Get Accolade Entry Object() -> (Accolade Entry Object:Accolade Entry Object Object Reference)
   > Is Valid(InputObject=@n2)
   > n3 = Set TriggerMult!(execute, TriggerMult:double=@n1.InValue, self:Accolade Entry Object Object Reference=@n2) -> (then, Output_Get:double)
   UpdateSecondsView(execute=[@n5.then, @n3.then])
   n4 = "On Value Changed (SpinBox_FreqMult)"!() -> (OutputDelegate, then, InValue:float) {is=title K2Node_ComponentBoundEvent}
   > Is Valid(InputObject=@n2)
   > n5 = Set TriggerMult(TriggerMult=@n4.InValue, self=@n2)
   n6 = OnTrigger_Event() {is=title K2Node_CustomEvent}
   > UpdateTiggerCounter()
   > n7 = Branch(Condition=Get HadErrorOnTrigger() -> (HadErrorOnTrigger:bool))
   n8 = Get ProgressBar_Back() -> (ProgressBar_Back:Progress Bar Object Reference)
   > SetFillColorAndOpacity!(
    execute
    self:Progress Bar Object Reference=@n8
    InColor:Linear Color Structure=Get ErrorBarColor() -> (ErrorBarColor:Linear Color Structure)
   ) -> (then)
   > Set HadErrorOnTrigger(HadErrorOnTrigger:bool=false) -> (Output_Get:bool)
   SetFillColorAndOpacity(execute=@n7.else, self=@n8, InColor=Get NormalBarColor() -> (NormalBarColor:Linear Color Structure))
   n9 = OnTriggerError() {is=title K2Node_CustomEvent}
   > Set HadErrorOnTrigger(HadErrorOnTrigger=true)
   n10 = OnCarryOverChange_Event() {is=title K2Node_CustomEvent}
   UpdateProgressBar(execute=[@n14.then, @n10.then])
   n11 = Event OnListItemObjectSet()
   > n12 = Cast To AccoladeEntryObject!(execute, Object=@n11.ListItemObject) -> (then, CastFailed, AsAccolade Entry Object:Accolade Entry Object Object Reference) {is=title K2Node_DynamicCast}
   > n13 = Set Accolade Entry Object(Accolade Entry Object:Accolade Entry Object Object Reference=@n12) -> (Output_Get:Accolade Entry Object Object Reference)
   > SetColorAndOpacity(
    self=Get textField_Accolade() -> (textField_Accolade:Common Text Object Reference)
    InColorAndOpacity=GetAccoladeColor!(self:Accolade Entry Object Object Reference=@n13) -> (ColorFromId:Slate Color Structure)
   )
   > Bind Event to On Carry Over Change(self:Accolade Entry Object Object Reference=@n13, Delegate=@n10) {is=title K2Node_AddDelegate}
   > Bind Event to On Trigger Event(self:Accolade Entry Object Object Reference=@n13, Delegate=@n6) {is=title K2Node_AddDelegate}
   > Bind Event to On Trigger Error(self:Accolade Entry Object Object Reference=@n13, Delegate=@n9) {is=title K2Node_AddDelegate}
   > "SetText (Text)"(
    self=Get textField_Accolade()
    InText=GetAccoladeDisplayName!(self:Accolade Entry Object Object Reference=@n12, Full:bool=false) -> (Name:Text)
   )
   > "SetText (Text)"(
    self=Get textField_Triggers() -> (textField_Triggers:Common Text Object Reference)
    InText="To Text (Float)"(
     Value=GetAdjustedTriggersPerMin!(self:Accolade Entry Object Object Reference=@n12) -> (TriggersPerMin:double, SecondsToTrigger:double)
     MinimumIntegralDigits=1
     MaximumIntegralDigits=324
     MinimumFractionalDigits=2
     MaximumFractionalDigits=3
    )
   )
   > n14 = UpdateSecondsView()
   Event PreConstruct()
   Event Construct()
  ]}
 UpdateProgressBar {nodes [
   UpdateProgressBar() {is=title K2Node_FunctionEntry}
   n1 = Get Accolade Entry Object()
   n2 = GetAdjustedTriggersPerMin(self=@n1)
   > n3 = Branch(Condition="float > float"(A=@n2, B=0.0001))
   n4 = Get ProgressBar_Back()
   > n5 = SetPercent!(
    execute
    self:Progress Bar Object Reference=@n4
    InPercent:float=float * float(A=float / float(A=@n2, B=60), B=Get CarryOver!(self:Accolade Entry Object Object Reference=@n1) -> (CarryOver:double)) {is=title K2Node_PromotableOperator}
   ) -> (then)
   Reroute Node(InputPin=[@n6.then, @n5.then])
   n6 = SetPercent(execute=@n3.else, self=@n4, InPercent=0)
  ]}
 UpdateSecondsView {nodes [
   UpdateSecondsView() {is=title K2Node_FunctionEntry}
   n1 = GetAdjustedTriggersPerMin(self=Get Accolade Entry Object())
   > n2 = Branch(Condition="float > float"(A=@n1))
   n3 = float / float(A=60, B=@n1)
   n4 = Fraction!(A:double=@n3) -> (ReturnValue:double)
   n5 = MakeTimespan!(Days:int=0, Hours:int=0, Minutes:int=0, Seconds:int=Truncate(A=@n3), Milliseconds:int=Truncate(A=float * float(A=@n3, B=@n4))) -> (ReturnValue:Timespan Structure)
   n6 = Get Seconds!(A:Timespan Structure=@n5) -> (ReturnValue:int=0) {is=title K2Node_CallFunction}
   n7 = Get Hours!(A:Timespan Structure=@n5) -> (ReturnValue:int=0) {is=title K2Node_CallFunction}
   n8 = Get Minutes!(A:Timespan Structure=@n5) -> (ReturnValue:int=0) {is=title K2Node_CallFunction}
   n9 = Get textField_Seconds() -> (textField_Seconds:Common Text Object Reference)
   > "SetText (Text)"(
    self=@n9
    InText=Format Text!(
     Format="INVTEXT(\"{h}h {m}m {s}s.{ms}\")"
     h:Text="To Text (Integer)"(Value=@n7, MaximumIntegralDigits=3)
     m:Text="To Text (Integer)"(Value=@n8, MinimumIntegralDigits=2, MaximumIntegralDigits=2)
     s:Text="To Text (Integer)"(Value=@n6, MinimumIntegralDigits=2, MaximumIntegralDigits=2)
     ms:int=Get Milliseconds!(A:Timespan Structure=@n5) -> (ReturnValue:int=0) {is=title K2Node_CallFunction}
    ) -> (Result)
   )
   "SetText (Text)"(execute=@n2.else, self=@n9, InText="LOCTEXT(\"-\")")
  ]}
 UpdateTiggerCounter {nodes [
   UpdateTiggerCounter() {is=title K2Node_FunctionEntry}
   n1 = Get Accolade Entry Object()
   > n2 = GetWasTriggered!(execute, self:Accolade Entry Object Object Reference=@n1) -> (then, Triggered:bool)
   > Branch(Condition=@n2)
   > "SetText (Text)"(
    self=Get textField_LastXp() -> (textField_LastXp:Common Text Object Reference)
    InText=Format Text!(
     Format="LOCTEXT(\"{Times} : {Last}\")"
     Times:int=Get Triggers!(self:Accolade Entry Object Object Reference=@n1) -> (Triggers:int=0)
     Last:int=Get LastXpValue!(self:Accolade Entry Object Object Reference=@n1) -> (LastXpValue:int=0)
    ) -> (Result)
   )
  ]}
]
variables [
 Accolade Entry Object {type=AccoladeEntryObject_C, default=null}
 ErrorBarColor {type=LinearColor, default=<rgba 1 0 0 0.2202>}
 HadErrorOnTrigger {type=boolean}
 NormalBarColor {type=LinearColor, default=<rgba 0 0.5 1 0.2202>}
]
widgets {
 info {inheritedWidgetCount=0, parentClass=&UserWidget, rootWidgetClass=&Overlay, widgetCount=8}
 widgets [
  Overlay_0 {widgetClassPath=&Overlay, widget=Overlay_0}
  ProgressBar_Back {
   widgetClassPath=/Script/UMG.ProgressBar
   bIsVariable=true
   parent=Overlay_0
   widget=ProgressBar_Back
   properties {
    bIsMarquee=true
    fillColorAndOpacity=<rgba 0 0.5 1 0.2202>
    widgetStyle {
     enableFillAnimation=false
     backgroundImage {
      bIsDynamicallyLoaded=false
      drawAs=NoDrawType
      imageType=FullColor
      mirroring=NoMirror
      resourceName=../../../Engine/Content/Slate/Common/ProgressBar_Background.png
      resourceObject=null
      tiling=NoTile
      imageSize=<0 0>
      margin=<ltrb 0.4167 0.4167 0.4167 0.4167>
      outlineSettings {
       bUseBrushTransparency=false
       roundingType=HalfHeightRadius
       width=0
       color {colorUseRule=UseColor_Specified, specifiedColor=<rgba 0 0 0 0>}
       cornerRadii=<0 0 0 0>
      }
      tintColor {colorUseRule=UseColor_Specified, specifiedColor=<rgba 1 1 1 1>}
      uVRegion {bIsValid=false, max=<0 0>, min=<0 0>}
     }
     fillImage {
      bIsDynamicallyLoaded=false
      drawAs=Box
      imageType=FullColor
      mirroring=NoMirror
      resourceName=../../../Engine/Content/Slate/Common/ProgressBar_Fill.png
      resourceObject=null
      tiling=NoTile
      imageSize=<0 0>
      margin=<ltrb 0.4167 0.4167 0.4167 0.4167>
      outlineSettings {
       bUseBrushTransparency=false
       roundingType=HalfHeightRadius
       width=0
       color {colorUseRule=UseColor_Specified, specifiedColor=<rgba 0 0 0 0>}
       cornerRadii=<0 0 0 0>
      }
      tintColor {colorUseRule=UseColor_Specified, specifiedColor=<rgba 1 1 1 1>}
      uVRegion {bIsValid=false, max=<0 0>, min=<0 0>}
     }
     marqueeImage {
      bIsDynamicallyLoaded=false
      drawAs=Image
      imageType=FullColor
      mirroring=NoMirror
      resourceName=../../../Engine/Content/Slate/Common/ProgressBar_Marquee.png
      resourceObject=null
      tiling=Horizontal
      imageSize=<20 12>
      margin=<ltrb 0 0 0 0>
      outlineSettings {
       bUseBrushTransparency=false
       roundingType=HalfHeightRadius
       width=0
       color {colorUseRule=UseColor_Specified, specifiedColor=<rgba 0 0 0 0>}
       cornerRadii=<0 0 0 0>
      }
      tintColor {colorUseRule=UseColor_Specified, specifiedColor=<rgba 1 1 1 1>}
      uVRegion {bIsValid=false, max=<0 0>, min=<0 0>}
     }
    }
   }
   slot {class=OverlaySlot, properties {horizontalAlignment=HAlign_Fill, verticalAlignment=VAlign_Fill}}
  }
  HorizontalBox_29 {
   widgetClassPath=&HorizontalBox
   parent=Overlay_0
   widget=HorizontalBox_29
   slot {class=OverlaySlot, properties {horizontalAlignment=HAlign_Fill, verticalAlignment=VAlign_Fill}}
  }
  textField_Accolade {
   widgetClassPath=&CommonTextBlock
   bIsVariable=true
   parent=HorizontalBox_29
   widget=textField_Accolade
   properties {
    minDesiredWidth=100
    text=GUID
    toolTipText=Prefix GUID of Accolade
    font {
     bForceMonospaced=false
     bMaterialIsStencil=false
     fontMaterial=null
     fontObject=&NotoSans
     letterSpacing=0
     monospacedWidth=1
     size=9
     skewAmount=0
     typefaceFontName=NotoSans-Regular
     outlineSettings {
      bApplyOutlineToDropShadows=false
      bMiteredCorners=false
      bSeparateFillAlpha=false
      outlineMaterial=null
      outlineSize=0
      outlineColor=<rgba 0 0 0 1>
     }
    }
    shadowColorAndOpacity=<rgba 0 0 0 0.5>
   }
   slot {
    class=HorizontalBoxSlot
    properties {horizontalAlignment=HAlign_Left, verticalAlignment=VAlign_Center, padding=<ltrb 2 0 2 1>, size {sizeRule=Fill, value=2}}
   }
  }
  textField_Triggers {
   widgetClassPath=&CommonTextBlock
   bIsVariable=true
   parent=HorizontalBox_29
   widget=textField_Triggers
   properties {
    minDesiredWidth=100
    text=10.0000
    toolTipText=Prefix GUID of Accolade
    font {
     bForceMonospaced=false
     bMaterialIsStencil=false
     fontMaterial=null
     fontObject=&NotoSans
     letterSpacing=0
     monospacedWidth=1
     size=9
     skewAmount=0
     typefaceFontName=NotoSans-Regular
     outlineSettings {
      bApplyOutlineToDropShadows=false
      bMiteredCorners=false
      bSeparateFillAlpha=false
      outlineMaterial=null
      outlineSize=0
      outlineColor=<rgba 0 0 0 1>
     }
    }
    shadowColorAndOpacity=<rgba 0 0 0 0.5>
   }
   slot {
    class=HorizontalBoxSlot
    properties {horizontalAlignment=HAlign_Left, verticalAlignment=VAlign_Center, padding=<ltrb 2 0 2 1>, size {sizeRule=Fill, value=1}}
   }
  }
  SpinBox_FreqMult {
   widgetClassPath=/Script/UMG.SpinBox
   bIsVariable=true
   parent=HorizontalBox_29
   widget=SpinBox_FreqMult
   properties {
    bOverride_MaxValue=true
    bOverride_MinValue=true
    maxFractionalDigits=2
    maxValue=10000
    toolTipText=A multiplier to modify Triggers Per Minute when in run mode.
    value=1
    foregroundColor {colorUseRule=UseColor_Specified, specifiedColor=<rgba 1 1 1 1>}
    widgetStyle {
     activeBackgroundBrush {
      bIsDynamicallyLoaded=false
      drawAs=NoDrawType
      imageType=NoImage
      mirroring=NoMirror
      resourceName=None
      resourceObject=null
      tiling=NoTile
      imageSize=<0 0>
      margin=<ltrb 0 0 0 0>
      outlineSettings {
       bUseBrushTransparency=false
       roundingType=HalfHeightRadius
       width=0
       color {colorUseRule=UseColor_Specified, specifiedColor=<rgba 0 0 0 0>}
       cornerRadii=<0 0 0 0>
      }
      tintColor {colorUseRule=UseColor_Specified, specifiedColor=<rgba 0.0048 0.0048 0.0048 1>}
      uVRegion {bIsValid=false, max=<0 0>, min=<0 0>}
     }
     activeFillBrush {
      bIsDynamicallyLoaded=false
      drawAs=Box
      imageType=FullColor
      mirroring=NoMirror
      resourceName=../../../Engine/Content/Slate/Common/Spinbox_Fill.png
      resourceObject=null
      tiling=NoTile
      imageSize=<0 0>
      margin=<ltrb 0.25 0.25 0.5 0.25>
      outlineSettings {
       bUseBrushTransparency=false
       roundingType=HalfHeightRadius
       width=0
       color {colorUseRule=UseColor_Specified, specifiedColor=<rgba 0 0 0 0>}
       cornerRadii=<0 0 0 0>
      }
      tintColor {colorUseRule=UseColor_Specified, specifiedColor=<rgba 0.0048 0.0048 0.0048 1>}
      uVRegion {bIsValid=false, max=<0 0>, min=<0 0>}
     }
     arrowsImage {
      bIsDynamicallyLoaded=false
      drawAs=Image
      imageType=FullColor
      mirroring=NoMirror
      resourceName=../../../Engine/Content/Slate/Common/SpinArrows.png
      resourceObject=null
      tiling=NoTile
      imageSize=<12 12>
      margin=<ltrb 0 0 0 0>
      outlineSettings {
       bUseBrushTransparency=false
       roundingType=HalfHeightRadius
       width=0
       color {colorUseRule=UseColor_Specified, specifiedColor=<rgba 0 0 0 0>}
       cornerRadii=<0 0 0 0>
      }
      tintColor {colorUseRule=UseColor_Specified, specifiedColor=<rgba 0.0048 0.0048 0.0048 1>}
      uVRegion {bIsValid=false, max=<0 0>, min=<0 0>}
     }
     backgroundBrush {
      bIsDynamicallyLoaded=false
      drawAs=Box
      imageType=FullColor
      mirroring=NoMirror
      resourceName=../../../Engine/Content/Slate/Common/Spinbox.png
      resourceObject=null
      tiling=NoTile
      imageSize=<0 0>
      margin=<ltrb 0.25 0.25 0.25 0.25>
      outlineSettings {
       bUseBrushTransparency=false
       roundingType=HalfHeightRadius
       width=0
       color {colorUseRule=UseColor_Specified, specifiedColor=<rgba 0 0 0 0>}
       cornerRadii=<0 0 0 0>
      }
      tintColor {colorUseRule=UseColor_Specified, specifiedColor=<rgba 0.0048 0.0048 0.0048 0.375>}
      uVRegion {bIsValid=false, max=<0 0>, min=<0 0>}
     }
     foregroundColor {colorUseRule=UseColor_Specified, specifiedColor=<rgba 0.4729 0.4729 0.4729 1>}
     hoveredBackgroundBrush {
      bIsDynamicallyLoaded=false
      drawAs=Box
      imageType=FullColor
      mirroring=NoMirror
      resourceName=../../../Engine/Content/Slate/Common/Spinbox_Hovered.png
      resourceObject=null
      tiling=NoTile
      imageSize=<0 0>
      margin=<ltrb 0.25 0.25 0.25 0.25>
      outlineSettings {
       bUseBrushTransparency=false
       roundingType=HalfHeightRadius
       width=0
       color {colorUseRule=UseColor_Specified, specifiedColor=<rgba 0 0 0 0>}
       cornerRadii=<0 0 0 0>
      }
      tintColor {colorUseRule=UseColor_Specified, specifiedColor=<rgba 0.0048 0.0048 0.0048 1>}
      uVRegion {bIsValid=false, max=<0 0>, min=<0 0>}
     }
     hoveredFillBrush {
      bIsDynamicallyLoaded=false
      drawAs=NoDrawType
      imageType=NoImage
      mirroring=NoMirror
      resourceName=None
      resourceObject=null
      tiling=NoTile
      imageSize=<0 0>
      margin=<ltrb 0 0 0 0>
      outlineSettings {
       bUseBrushTransparency=false
       roundingType=HalfHeightRadius
       width=0
       color {colorUseRule=UseColor_Specified, specifiedColor=<rgba 0 0 0 0>}
       cornerRadii=<0 0 0 0>
      }
      tintColor {colorUseRule=UseColor_Specified, specifiedColor=<rgba 0.0048 0.0048 0.0048 1>}
      uVRegion {bIsValid=false, max=<0 0>, min=<0 0>}
     }
     inactiveFillBrush {
      bIsDynamicallyLoaded=false
      drawAs=Box
      imageType=FullColor
      mirroring=NoMirror
      resourceName=../../../Engine/Content/Slate/Common/Spinbox_Fill_Hovered.png
      resourceObject=null
      tiling=NoTile
      imageSize=<0 0>
      margin=<ltrb 0.25 0.25 0.25 0.25>
      outlineSettings {
       bUseBrushTransparency=false
       roundingType=HalfHeightRadius
       width=0
       color {colorUseRule=UseColor_Specified, specifiedColor=<rgba 0 0 0 0>}
       cornerRadii=<0 0 0 0>
      }
      tintColor {colorUseRule=UseColor_Specified, specifiedColor=<rgba 0.0048 0.0048 0.0048 1>}
      uVRegion {bIsValid=false, max=<0 0>, min=<0 0>}
     }
     insetPadding=<ltrb 0 0 0 0>
     textPadding=<ltrb 1 2 1 2>
    }
   }
   slot {
    class=HorizontalBoxSlot
    properties {horizontalAlignment=HAlign_Right, verticalAlignment=VAlign_Center, padding=<ltrb 2 0 2 1>, size {sizeRule=Fill, value=1}}
   }
  }
  textField_Seconds {
   widgetClassPath=&CommonTextBlock
   bIsVariable=true
   parent=HorizontalBox_29
   widget=textField_Seconds
   properties {
    minDesiredWidth=100
    text=00:00:00:00:00
    toolTipText=Prefix GUID of Accolade
    font {
     bForceMonospaced=false
     bMaterialIsStencil=false
     fontMaterial=null
     fontObject=&NotoSans
     letterSpacing=0
     monospacedWidth=1
     size=9
     skewAmount=0
     typefaceFontName=NotoSans-Regular
     outlineSettings {
      bApplyOutlineToDropShadows=false
      bMiteredCorners=false
      bSeparateFillAlpha=false
      outlineMaterial=null
      outlineSize=0
      outlineColor=<rgba 0 0 0 1>
     }
    }
    shadowColorAndOpacity=<rgba 0 0 0 0.5>
   }
   slot {class=HorizontalBoxSlot, properties {verticalAlignment=VAlign_Center, padding=<ltrb 2 0 2 1>, size {sizeRule=Fill, value=1}}}
  }
  textField_LastXp {
   widgetClassPath=&CommonTextBlock
   bIsVariable=true
   parent=HorizontalBox_29
   widget=textField_LastXp
   properties {
    justification=Right
    minDesiredWidth=100
    text=-
    toolTipText=Prefix GUID of Accolade
    font {
     bForceMonospaced=false
     bMaterialIsStencil=false
     fontMaterial=null
     fontObject=&NotoSans
     letterSpacing=0
     monospacedWidth=1
     size=9
     skewAmount=0
     typefaceFontName=NotoSans-Regular
     outlineSettings {
      bApplyOutlineToDropShadows=false
      bMiteredCorners=false
      bSeparateFillAlpha=false
      outlineMaterial=null
      outlineSize=0
      outlineColor=<rgba 0 0 0 1>
     }
    }
    shadowColorAndOpacity=<rgba 0 0 0 0.5>
   }
   slot {
    class=HorizontalBoxSlot
    properties {horizontalAlignment=HAlign_Right, verticalAlignment=VAlign_Center, padding=<ltrb 2 0 2 1>, size {sizeRule=Fill, value=1}}
   }
  }
 ]
}
