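# Editor utility widget that simulates Dynamic XP for creator islands, letting designers run, pause or step a player session and trigger accolades by hand.
# Island stats are queried through a dynamic_xp.py bridge, converted to an accolade map, and fed with tuning into a FortDynamicXpCalculator.
# Tick advances simulated time, throttling it on slow frames, and reloads the script on request.
# A SaveGame persists island codes, credentials and data; results print to a recycled-row log.
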EditorUtilityWidgetBlueprint /DynamicXp/DynamicXpCalculator
animations [NewAnimation {
  binding_count=0
  length_seconds=5
  object=&self:NewAnimation
  sequence {bindings [], display_rate {denominator=1, numerator=20}, tick_resolution {denominator=1, numerator=60000}, tracks []}
 }]
graphs [
 AddOutputLine {nodes [
   n1 = AddOutputLine!() -> (then, InText:Text, Color:Slate Color Structure, ForceScrollToEnd:bool=~) {is=title K2Node_FunctionEntry}
   n2 = Get OutputLog() -> (OutputLog:Vertical Box Object Reference)
   > n3 = Branch(Condition="Equal (Integer)"(A=GetChildrenCount!(self:Panel Widget Object Reference=@n2) -> (ReturnValue:int=0)))
   > n4 = For Loop(LastIndex=3000)
   n5 = Cast To DynamicXpLogEntry!(
    execute=[@n3.else, @n4.Completed]
    Object=GetChildAt!(self:Panel Widget Object Reference=@n2, Index:int=0) -> (ReturnValue:Widget Object Reference)
   ) -> (then, CastFailed, AsDynamic Xp Log Entry:Dynamic Xp Log Entry Object Reference) {is=title K2Node_DynamicCast}
   > n6 = Set NewLine(NewLine:Dynamic Xp Log Entry Object Reference=@n5) -> (Output_Get:Dynamic Xp Log Entry Object Reference)
   > RemoveChildAt!(execute, self:Panel Widget Object Reference=@n2, Index:int=0) -> (then, ReturnValue:bool)
   > SetLogLine!(execute, self:Dynamic Xp Log Entry Object Reference=@n6, LogLine:Text=@n1, ColorAndOpacity:Slate Color Structure=@n1.Color) -> (then)
   > AddChildToVerticalBox(self=@n2, Content=@n6)
   > Branch(Condition=@n1.ForceScrollToEnd)
   n7 = Get OutputScroll() -> (OutputScroll:Scroll Box Object Reference)
   > SetScrollOffset!(
    execute
    self:Scroll Box Object Reference=@n7
    NewScrollOffset:float=float + float(A=GetScrollOffsetOfEnd!(self:Scroll Box Object Reference=@n7) -> (ReturnValue:float), B=100)
   ) -> (then)
   n8 = Construct Object from Class(execute=@n4.LoopBody, Class=/DynamicXp/Parts/DynamicXpLogEntry.DynamicXpLogEntry_C) -> (ReturnValue:Dynamic Xp Log Entry Object Reference)
   > AddChildToVerticalBox(self=@n2, Content=@n8)
  ]}
 BuildAccoladeMapFromXpStats {nodes [
   BuildAccoladeMapFromXpStats() {is=title K2Node_FunctionEntry}
   > n1 = For Each Loop!(
    Exec
    Array:Array of Dynamic Xp Aggregated Accolade Stats Structures=Get!() -> (
     XpStats_UniqueGameId:String
     XpStats_UniqueGameVersion:int
     XpStats_CalibrationResult:Dynamic Xp Calibration Result Structure
     XpStats_Accolades:Array of Dynamic Xp Aggregated Accolade Stats Structures
     XpStats_MeanSessionLengthMinutes:double=~
     XpStats_PublishData:Array of Accolade Publish Data Structures
    ).XpStats_Accolades
   ) -> (LoopBody, Array Element_AccoladeID:String, Array Element_TriggersPerMinute:double=~, Array Index, Completed)
   n2 = For Each Loop!(
    Exec=@n1.Completed
    Array:Array of Accolade Publish Data Structures=Get Publish Data() -> (Publish Data:Array of Accolade Publish Data Structures)
   ) -> (
    LoopBody
    Array Element_ID:String
    Array Element_Name:String
    Array Element_Description:String
    Array Element_Weight:Gameplay Tag Structure
    Array Index
    Completed
   )
   PopulateGroupAccolades!(execute=@n2.Completed, self, Group:String) -> (then)
   > PopulateActiveGroupsComboBox()
   n3 = Get AccoladesMap() -> (AccoladesMap:Map of Strings to Dynamic Xp Accolade Structures)
   n4 = Set members in Dynamic Xp Accolade!(
    execute:Exec=@n2.LoopBody
    StructRef:"Dynamic Xp Accolade Structure (by ref)"=Find(TargetMap:Map of Strings to Dynamic Xp Accolade Structures=@n3, Key=@n2) -> (Value:Dynamic Xp Accolade Structure)
    Name_10_#1:String=@n2
    CreatorName_9_#2:String=@n2.Array Element_Name
   ) -> (then:Exec, StructOut:"Dynamic Xp Accolade Structure (by ref)") {is=title K2Node_SetFieldsInStruct}
   > Add!(
    execute
    TargetMap:Map of Strings to Dynamic Xp Accolade Structures=@n3
    Key:"String (by ref)"=@n2
    Value:"Dynamic Xp Accolade Structure (by ref)"=@n4
   ) -> (then) {is=title K2Node_CallFunction}
   n5 = GetAccoladeNameAndGrouping!(execute=@n1.LoopBody, FromId:String=@n1, None:%=Get Calculator() -> (Calculator:Fort Dynamic Xp Calculator Object Reference)) -> (then, AccoladeName:String, Grouping:String)
   n6 = Find(TargetMap=@n3, Key=@n5)
   > n7 = Set TempNumGroups(TempNumGroups:int=Break Dynamic Xp Accolade!(DynamicXpAccolade:"Dynamic Xp Accolade Structure (by ref)"=@n6) -> (
     Name_10_#1:String
     CreatorName_9_#2:String
     NumGroups_8_#3:int
    ) {is=title K2Node_BreakStruct}.NumGroups_8_#3) -> (Output_Get:int=0)
   > n8 = Branch(Condition="float > float"(A=@n1.Array Element_TriggersPerMinute))
   > n9 = Increment Int(Value=@n7)
   n10 = Set members in Dynamic Xp Accolade!(
    execute=[@n8.else, @n9."  "]
    StructRef=@n6
    Name_10_#1=@n5
    NumGroups_8_#3:int=Get TempNumGroups() -> (TempNumGroups:int=0)
   ) -> (then, StructOut)
   > Add(TargetMap=@n3, Key=@n5, Value=@n10)
  ]}
 Button Events {nodes [
   n1 = PopulateActiveGroupsComboBox()
   n2 = Get XpStats() -> (XpStats:Dynamic Xp Aggregated Stats Structure)
   AddOutputLine!(
    execute=[@n38.then, @n1.then]
    self
    InText:Text="To Text (String)"(InString=Break Dynamic Xp Aggregated Stats!(DynamicXpAggregatedStats:"Dynamic Xp Aggregated Stats Structure (by ref)"=@n2) -> (
      UniqueGameId:String
      UniqueGameVersion:int
      CalibrationResult:Dynamic Xp Calibration Result Structure
      Accolades:Array of Dynamic Xp Aggregated Accolade Stats Structures
      MeanSessionLengthMinutes:double=~
      PublishData:Array of Accolade Publish Data Structures
     ) {is=title K2Node_BreakStruct})
    Color_SpecifiedColor:Linear Color Structure=(R=0,G=0.3352,B=1,A=1)
    Color_ColorUseRule:ESlateColorStylingMode Enum
    ForceScrollToEnd:bool=true
   ) -> (then) {is=title K2Node_CallFunction}
   > n3 = To Array!(execute, A:Set of Strings=Get AllAccolades() -> (AllAccolades:Set of Strings)) -> (then, Result:Array of Strings)
   > n4 = For Each Loop!(Exec, Array:Array of Strings=@n3) -> (LoopBody, Array Element:String, Array Index, Completed)
   UpdateProgressBars(execute=@n4.Completed)
   > AddOutputLine!(execute, self, InText="LOCTEXT(\"Started Game\")", Color:Slate Color Structure, ForceScrollToEnd=true) -> (then)
   n5 = GetAccoladeTuning!(execute=@n4.LoopBody, self, AccoladeName:String=@n4) -> (then, AccoladeTuning:Dynamic Xp Creator Accolade Tuning Structure)
   n6 = Get Calculator()
   > AddAccolade!(
    execute
    self:Fort Dynamic Xp Calculator Object Reference=@n6
    AccoladeId:String=@n4
    Tuning:"Dynamic Xp Creator Accolade Tuning Structure (by ref)"=@n5
   ) -> (then)
   n7 = PopulateActiveGroupsComboBox()
   AddOutputLine(execute=[@n7.then, @n43.then], InText="LOCTEXT(\"Ended Game\")", ForceScrollToEnd=true)
   "On Clicked (bttnQueryIsland)"!() -> (OutputDelegate, then, widget:User Widget Object Reference, event:Pointer Event Structure) {is=title K2Node_ComponentBoundEvent}
   n8 = Get inputUserName() -> (inputUserName:Text Box Object Reference)
   n9 = Get Args() -> (Args:Map of Strings to Strings)
   > Add(
    TargetMap:Map of Strings to Strings=@n9
    Key=user_name
    Value:"String (by ref)"="To String (Text)"(InText="GetText (Text Box)"!(self:Text Box Object Reference=@n8) -> (ReturnValue:Text))
   )
   n10 = Get inputPassword() -> (inputPassword:Text Box Object Reference)
   > Add(TargetMap=@n9, Key=password, Value="To String (Text)"(InText="GetText (Text Box)"(self=@n10)))
   n11 = GetSelectedOption!(self:"Combo Box (String) Object Reference"=Get ComboBoxIslandCodes() -> (ComboBoxIslandCodes:"Combo Box (String) Object Reference")) -> (ReturnValue:String)
   > Add(TargetMap=@n9, Key=island_code, Value=@n11)
   n12 = Get intputIslandVersion() -> (intputIslandVersion:Text Box Object Reference)
   > Add(TargetMap=@n9, Key=island_version, Value="To String (Text)"(InText="GetText (Text Box)"(self=@n12)))
   > Add(
    TargetMap=@n9
    Key=enviroment
    Value=GetSelectedOption(self=Get ComboBox_DevOrProd() -> (ComboBox_DevOrProd:"Combo Box (String) Object Reference"))
   )
   > n13 = CallPythonWithObj!(execute, self, Command:String=query_island_meta_data, Args:Map of Strings to Strings=@n9) -> (then, Result:String)
   n14 = Get SaveGame() -> (SaveGame:Dynamic Xp Save Object Reference)
   > Add(
    TargetMap=Get IslandCodes!(self:Dynamic Xp Save Object Reference=@n14) -> (IslandCodes:Map of Strings to Strings)
    Key=@n11
    Value=Get IslandTitle() -> (IslandTitle:String)
   )
   > SaveGameToSlot!(
    execute
    SaveGameObject:Save Game Object Reference=@n14
    SlotName:String=Get SaveGameSlotName() -> (SaveGameSlotName:String)
    UserIndex:int=0
   ) -> (then, ReturnValue:bool)
   > PopulateIslandComboBox!(execute, self, Selected:String=@n11) -> (then)
   > n15 = JsonStrToStats!(execute, JsonStr:String=@n13, None:%=Get Calculator()) -> (then, Stats:Dynamic Xp Aggregated Stats Structure, PublishData:Array of Accolade Publish Data Structures, ReturnValue:bool)
   > Branch(Condition=@n15.ReturnValue)
   > Set Publish Data(Publish Data:Array of Accolade Publish Data Structures=@n15.PublishData) -> (Output_Get:Array of Accolade Publish Data Structures)
   > Set XpStats(XpStats:Dynamic Xp Aggregated Stats Structure=@n15) -> (Output_Get:Dynamic Xp Aggregated Stats Structure)
   > BuildAccoladeMapFromXpStats() {is=title K2Node_CallFunction}
   n16 = "On Item Clicked (ListView_Accolades)"!() -> (OutputDelegate, then, Item:Object Reference) {is=title K2Node_ComponentBoundEvent}
   > n17 = Cast To AccoladeEntryObject!(execute, Object=@n16.Item) -> (then, CastFailed, AsAccolade Entry Object:Accolade Entry Object Object Reference) {is=title K2Node_DynamicCast}
   n18 = Get ClickTime() -> (ClickTime:double)
   > n19 = Branch(Condition="float > float"(A=@n18))
   > n20 = Set AccoladeClickTime(AccoladeClickTime:double=@n18) -> (Output_Get:double)
   TriggerAccolade!(
    execute=[@n20.then, @n21.then]
    self
    SecondsPassed:double=Get AccoladeClickTime() -> (AccoladeClickTime:double)
    AccoladeEntry:Accolade Entry Object Object Reference=@n17
   ) -> (then)
   > UpdateProgressBars()
   n21 = Set AccoladeClickTime(
    execute=@n19.else
    AccoladeClickTime=GetAdjustedTriggersPerMin!(self:Accolade Entry Object Object Reference=@n17) -> (TriggersPerMin:double, SecondsToTrigger:double).SecondsToTrigger
   )
   n22 = "On Selection Changed (ComboBoxIslandCodes)"!() -> (OutputDelegate, then, SelectedItem:String, SelectionType:ESelectInfo Enum) {is=title K2Node_ComponentBoundEvent}
   > SetIslandTitle!(execute, self, Key:"String (by ref)"=@n22.SelectedItem) -> (then)
   n23 = Get SaveGame()
   > Set SelectedIslandCode!(execute, SelectedIslandCode:String=@n22.SelectedItem, self:Dynamic Xp Save Object Reference=@n23) -> (then, Output_Get:String)
   n24 = Get SaveGameSlotName()
   > SaveGameToSlot(SaveGameObject=@n23, SlotName=@n24, UserIndex=0)
   "On Clicked (bttnRemoveIsland)"!() -> (OutputDelegate, then, widget:User Widget Object Reference, event:Pointer Event Structure) {is=title K2Node_ComponentBoundEvent}
   > n25 = Switch on Int!(
    execute:Exec
    Selection:int=GetActiveWidgetIndex!(self:Widget Switcher Object Reference=Get SwitcherAddIslandCode() -> (SwitcherAddIslandCode:Widget Switcher Object Reference)) -> (ReturnValue:int=0)
   ) -> (Default:Exec, 0:Exec, 1:Exec)
   n26 = Get SaveGame()
   Remove!(
    execute=@n25.0
    TargetMap:Map of Strings to Strings=Get IslandCodes(self=@n26)
    Key:"String (by ref)"=GetSelectedOption(self=Get ComboBoxIslandCodes())
   ) -> (then, ReturnValue:bool)
   > SaveGameToSlot(SaveGameObject=@n26, SlotName=@n24, UserIndex=0)
   > PopulateIslandComboBox()
   n27 = Get SwitcherAddIslandCode()
   SetActiveWidgetIndex(execute=[@n35.then, @n34.then, @n31.then, @n25.1], self=@n27, Index=0)
   n28 = Get intputIslandCode() -> (intputIslandCode:Text Box Object Reference)
   > "SetText (Text Box)"!(execute, self:Text Box Object Reference=@n28, InText:Text) -> (then)
   "On Clicked (bttn_copyLog)"!() -> (OutputDelegate, then, widget:User Widget Object Reference, event:Pointer Event Structure) {is=title K2Node_ComponentBoundEvent}
   > CopyLog()
   n29 = "On Text Committed (intputIslandCode)"!() -> (OutputDelegate, then, Text:"Text (by ref)", CommitMethod:ETextCommit Enum=Default) {is=title K2Node_ComponentBoundEvent}
   n30 = "GetText (Text Box)"(self=@n28)
   n31 = Branch(execute=[@n29.then, @n37.1], Condition=TextIsEmpty(InText=@n30))
   n32 = "To String (Text)"(InText=@n30)
   n33 = Get IslandCodes(self=Get SaveGame())
   n34 = Branch(execute=@n31.else, Condition=Contains!(TargetMap:Map of Strings to Strings=@n33, Key:"String (by ref)"=@n32) -> (ReturnValue))
   Add(execute=@n34.else, TargetMap=@n33, Key=@n32, Value=ISLAND NOT YET QUIERIED.)
   > n35 = PopulateIslandComboBox(Selected=@n32)
   "On Clicked (bttn_help)"!() -> (OutputDelegate, then, widget:User Widget Object Reference, event:Pointer Event Structure) {is=title K2Node_ComponentBoundEvent}
   > ShowHelp()
   "On Clicked (bttnDataAssetChange)"!() -> (OutputDelegate, then, widget:User Widget Object Reference, event:Pointer Event Structure) {is=title K2Node_ComponentBoundEvent}
   > InitializeCalculatorConstData()
   "On Clicked (bttnIslandChange)"!() -> (OutputDelegate, then, widget:User Widget Object Reference, event:Pointer Event Structure) {is=title K2Node_ComponentBoundEvent}
   > InitializeCalculatorXpStats()
   "On Clicked (bttnFillDataGather)"!() -> (OutputDelegate, then, widget:User Widget Object Reference, event:Pointer Event Structure) {is=title K2Node_ComponentBoundEvent}
   > Set members in Dynamic Xp Aggregated Stats!(execute:Exec, StructRef:"Dynamic Xp Aggregated Stats Structure (by ref)"=Get XpStats()) -> (then:Exec, StructOut:"Dynamic Xp Aggregated Stats Structure (by ref)") {is=title K2Node_SetFieldsInStruct}
   "On Clicked (bttnSaveDesignerTuning)"!() -> (OutputDelegate, then, widget:User Widget Object Reference, event:Pointer Event Structure) {is=title K2Node_ComponentBoundEvent}
   n36 = Get SaveGame()
   > Set CappedSessionLenMinutes!(
    execute
    CappedSessionLenMinutes:double=Get CappedSessionLenMinutes() -> (CappedSessionLenMinutes:double)
    self:Dynamic Xp Save Object Reference=@n36
   ) -> (then, Output_Get:double)
   > Set XpBudgetPerMin!(
    execute
    XpBudgetPerMin:double="To Float (Integer)"(InInt=Get XpBudgetPerMin() -> (XpBudgetPerMin:int=0))
    self:Dynamic Xp Save Object Reference=@n36
   ) -> (then, Output_Get:double)
   > Set CircuitBreaker!(
    execute
    CircuitBreaker:Dynamic Xp Circuit Breaker Structure=Get CircuitBreaker() -> (CircuitBreaker:Dynamic Xp Circuit Breaker Structure)
    self:Dynamic Xp Save Object Reference=@n36
   ) -> (then, Output_Get:Dynamic Xp Circuit Breaker Structure)
   > Set AfkPenalty!(
    execute
    AfkPenalty:Dynamic Xp Afk Penalty Structure=Get AfkPenalty() -> (AfkPenalty:Dynamic Xp Afk Penalty Structure)
    self:Dynamic Xp Save Object Reference=@n36
   ) -> (then, Output_Get:Dynamic Xp Afk Penalty Structure)
   "On Clicked (bttnLoadDesignerTuning)"!() -> (OutputDelegate, then, widget:User Widget Object Reference, event:Pointer Event Structure) {is=title K2Node_ComponentBoundEvent}
   > Set CappedSessionLenMinutes!(execute, CappedSessionLenMinutes=Get CappedSessionLenMinutes!(self:Dynamic Xp Save Object Reference=@n36) -> (CappedSessionLenMinutes)) -> (then, Output_Get)
   > Set CappedSessionLenMinutes(CappedSessionLenMinutes=Get XpBudgetPerMin!(self:Dynamic Xp Save Object Reference=@n36) -> (XpBudgetPerMin:double))
   > Set CircuitBreaker!(execute, CircuitBreaker=Get CircuitBreaker!(self:Dynamic Xp Save Object Reference=@n36) -> (CircuitBreaker)) -> (then, Output_Get)
   "On Clicked (btnClearOutput)"!() -> (OutputDelegate, then, widget:User Widget Object Reference, event:Pointer Event Structure) {is=title K2Node_ComponentBoundEvent}
   > ClearChildren(self=Get OutputLog())
   "On Clicked (bttnLoadAnalyticsData)"!() -> (OutputDelegate, then, widget:User Widget Object Reference, event:Pointer Event Structure) {is=title K2Node_ComponentBoundEvent}
   > Set XpStats(XpStats=Get XpStats!(self:Dynamic Xp Save Object Reference=@n36) -> (XpStats))
   > Set PlayerSessionStart(PlayerSessionStart:Dynamic Xp Player Session Structure=Get PlayerSession!(self:Dynamic Xp Save Object Reference=@n36) -> (PlayerSession:Dynamic Xp Player Session Structure)) -> (Output_Get:Dynamic Xp Player Session Structure)
   > Set McpProfileStats(McpProfileStats:Dynamic Xp Profile Stats Structure=Get McpProfileStats!(self:Dynamic Xp Save Object Reference=@n36) -> (McpProfileStats:Dynamic Xp Profile Stats Structure)) -> (Output_Get:Dynamic Xp Profile Stats Structure)
   "On Clicked (bttnSaveAnalyticsData)"!() -> (OutputDelegate, then, widget:User Widget Object Reference, event:Pointer Event Structure) {is=title K2Node_ComponentBoundEvent}
   > Set XpStats!(execute, XpStats=Get XpStats!() -> (XpStats), self:Dynamic Xp Save Object Reference=@n36) -> (then, Output_Get)
   > Set PlayerSession!(
    execute
    PlayerSession:Dynamic Xp Player Session Structure=Get PlayerSessionStart() -> (PlayerSessionStart:Dynamic Xp Player Session Structure)
    self:Dynamic Xp Save Object Reference=@n36
   ) -> (then, Output_Get:Dynamic Xp Player Session Structure)
   > Set McpProfileStats!(execute, McpProfileStats=Get McpProfileStats!() -> (McpProfileStats), self:Dynamic Xp Save Object Reference=@n36) -> (then, Output_Get)
   > SaveGameToSlot(SaveGameObject=@n36, SlotName=Get SaveGameSlotName(), UserIndex=0)
   "On Clicked (bttn_ResetPlayerSession)"!() -> (OutputDelegate, then, widget:User Widget Object Reference, event:Pointer Event Structure) {is=title K2Node_ComponentBoundEvent}
   > Set PlayerSessionStart(PlayerSessionStart=Get DefaultSessionStart() -> (DefaultSessionStart:Dynamic Xp Player Session Structure))
   "On Clicked (bttnAddIslandCode)"!() -> (OutputDelegate, then, widget:User Widget Object Reference, event:Pointer Event Structure) {is=title K2Node_ComponentBoundEvent}
   > n37 = Switch on Int(Selection=GetActiveWidgetIndex(self=@n27) -> (ReturnValue=0))
   SetActiveWidgetIndex(execute=@n37.0, self=@n27, Index=1)
   "On Clicked (bttnStartGame)"!() -> (OutputDelegate, then, widget:User Widget Object Reference, event:Pointer Event Structure) {is=title K2Node_ComponentBoundEvent}
   > Set PlayerSession!(execute, PlayerSession=Get PlayerSessionStart()) -> (then, Output_Get)
   > CopyMcpProfileStatsToPlayerSession()
   > SetActiveWidget(
    self:Widget Switcher Object Reference=Get StartStopSwitcher() -> (StartStopSwitcher:Widget Switcher Object Reference)
    Widget=Get EndGamePanel() -> (EndGamePanel:Horizontal Box Object Reference)
   )
   > InitializeCalculatorConstData()
   > InitializeCalculatorXpStats()
   > n38 = OnRoundStartForPlayer!(
    execute
    self:Fort Dynamic Xp Calculator Object Reference=@n6
    PlayerSession:"Dynamic Xp Player Session Structure (by ref)"=Get PlayerSession!() -> (PlayerSession)
    CurrentTime:float=0
   ) -> (then)
   "On Clicked (btnPopulateData)"!() -> (OutputDelegate, then, widget:User Widget Object Reference, event:Pointer Event Structure) {is=title K2Node_ComponentBoundEvent}
   n39 = Get Design Settings Data Asset() -> (Design Settings Data Asset:Creative Dynamic Xp Data Asset Object Reference)
   > Branch(Condition=IsValid(Object=@n39))
   "On Clicked (bttnEndGame)"!() -> (OutputDelegate, then, widget:User Widget Object Reference, event:Pointer Event Structure) {is=title K2Node_ComponentBoundEvent}
   > Set GameIsRunning(GameIsRunning:bool=false) -> (Output_Get:bool)
   > SetActiveWidget(
    self=Get PlaySwitcher() -> (PlaySwitcher:Widget Switcher Object Reference)
    Widget=Get bttnRun() -> (bttnRun:TA Generic 01 Button Object Reference)
   )
   > SetActiveWidget(self=Get StartStopSwitcher(), Widget=Get StartGamePanel() -> (StartGamePanel:Horizontal Box Object Reference))
   n40 = Get Calculator()
   > OnRoundEndForPlayer!(
    execute
    self:Fort Dynamic Xp Calculator Object Reference=@n40
    PlayerSession:"Dynamic Xp Player Session Structure (by ref)"=Get PlayerSession()
    CurrentTime:float=0
   ) -> (then)
   > n41 = To Array(A=Get AllAccolades())
   > n42 = For Each Loop(Array=@n41)
   n43 = Set GameTime(execute=@n42.Completed, GameTime:double=0) -> (Output_Get:double)
   RemoveAccolade!(execute=@n42.LoopBody, self:Fort Dynamic Xp Calculator Object Reference=@n40, AccoladeId:String=@n42) -> (then)
   "On Clicked (bttnRun)"!() -> (OutputDelegate, then, widget:User Widget Object Reference, event:Pointer Event Structure) {is=title K2Node_ComponentBoundEvent}
   n44 = Get PlaySwitcher()
   > SetActiveWidget(self=@n44, Widget=Get bttnPause() -> (bttnPause:TA Generic 01 Button Object Reference))
   > Set GameIsRunning(GameIsRunning=true)
   > Set LastTimeSeconds(LastTimeSeconds:double=GetTimeSeconds!(WorldContextObject:Object Reference) -> (ReturnValue:double)) -> (Output_Get:double)
   > CopyMcpProfileStatsToPlayerSession()
   > AddOutputLine(InText="LOCTEXT(\"Simulating Sesson\")", ForceScrollToEnd=true)
   "On Clicked (bttnPause)"!() -> (OutputDelegate, then, widget:User Widget Object Reference, event:Pointer Event Structure) {is=title K2Node_ComponentBoundEvent}
   > SetActiveWidget(self=@n44, Widget=Get bttnRun())
   > Set GameIsRunning(GameIsRunning=false)
   > AddOutputLine(InText="LOCTEXT(\"Paused Sesson\")", ForceScrollToEnd=true)
   "On Clicked (bttnStep)"!() -> (OutputDelegate, then, widget:User Widget Object Reference, event:Pointer Event Structure) {is=title K2Node_ComponentBoundEvent}
   > CopyMcpProfileStatsToPlayerSession()
   n45 = Get StepTime() -> (StepTime:double)
   > Simulate Time Passed!(execute, self, SecondsPassed:double=@n45, OnlyAccolade:String) -> (then)
   > AddOutputLine(
    InText=Format Text!(
     Format="LOCTEXT(\"Stepped {secconds}\")"
     secconds:Text="To Text (Float)"(Value=@n45, MinimumIntegralDigits=1, MaximumIntegralDigits=324, MinimumFractionalDigits=0, MaximumFractionalDigits=3)
    ) -> (Result)
    ForceScrollToEnd=true
   )
   n46 = "On Check State Changed (CheckBox_Afk)"!() -> (OutputDelegate, then, bIsChecked:bool) {is=title K2Node_ComponentBoundEvent}
   > n47 = Branch(Condition=@n46.bIsChecked)
   > AddOutputLine(InText="LOCTEXT(\"AFK Turned ON\")", ForceScrollToEnd=true)
   AddOutputLine(execute=@n47.else, InText="LOCTEXT(\"AFK Turned OFF\")", ForceScrollToEnd=true)
   Set CappedSessionLenMinutes(CappedSessionLenMinutes=GetCappedSessionLength!(self:Creative Dynamic Xp Data Asset Object Reference=@n39) -> (ReturnValue:float))
   K2Node_Composite_0 = Get Time Until Remained AFK!(self:Creative Dynamic Xp Data Asset Object Reference=@n39) -> (ReturnValue_TimeUntilRemainedAFK:double=~) {is=title K2Node_Composite}
   > Set!(
    execute
    AfkPenalty_AfkThresholdSecs:float=@K2Node_Composite_0
    AfkPenalty_AfkMinRewardMult:float=GetAfkMinRewardMult!(self:Creative Dynamic Xp Data Asset Object Reference=@n39) -> (ReturnValue:float)
   ) -> (then, Output_Get:Dynamic Xp Afk Penalty Structure)
   > Set UseLinksCalibrationResult(UseLinksCalibrationResult:bool=Get bUseLinksCalibrationResult!(self:Creative Dynamic Xp Data Asset Object Reference=@n39) -> (bUseLinksCalibrationResult:bool)) -> (Output_Get:bool)
   > AddOutputLine!(
    execute
    self
    InText="LOCTEXT(\"Initialized Data from Data Asset\")"
    Color_SpecifiedColor:Linear Color Structure=(R=0.1869,G=1,B=0,A=1)
    Color_ColorUseRule:ESlateColorStylingMode Enum
    ForceScrollToEnd=true
   ) -> (then)
   Get Variance!(self:Creative Dynamic Xp Data Asset Object Reference=@n39) -> (Variance:Structure)
   Get CalibrationRequirements!(self:Creative Dynamic Xp Data Asset Object Reference=@n39) -> (CalibrationRequirements:Structure)
   Get Variance(self:Dynamic Xp Save Object Reference=@n36)
   Get Calibration!(self:Dynamic Xp Save Object Reference=@n36) -> (Calibration:Structure)
   Set CircuitBreaker(CircuitBreaker=Get CircuitBreaker(self:Creative Dynamic Xp Data Asset Object Reference=@n39))
   Set AfkPenalty!(execute, AfkPenalty=Get AfkPenalty!(self:Dynamic Xp Save Object Reference=@n36) -> (AfkPenalty)) -> (then, Output_Get)
  ]}
 Get Time Until Remained AFK {
  path=Button Events.K2Node_Composite_0.Get Time Until Remained AFK
  nodes [Outputs!(ReturnValue_TimeUntilRemainedAFK:double=GetAFKDetectionParameters!(self:Creative Dynamic Xp Data Asset Object Reference=Inputs!() -> (self:Creative Dynamic Xp Data Asset Object Reference)) -> (
     ReturnValue_TimeUntilConfirmedAFK:float=~
     ReturnValue_TimeUntilRemainedAFK:float=~
     ReturnValue_MinimumMovementPerSecond:float=~
     ReturnValue_MinimumCameraRotationPerSecond:float=~
     ReturnValue_MaxTimeWithoutInteracting:float=~
     ReturnValue_MaxTimeWithoutInputChange:float=~
     ReturnValue_SendInputChangeInterval:float=~
     ReturnValue_InputActionsToTrack:Array of Names
     ReturnValue_ValidInteractionsGameplayEventTags:Gameplay Tag Container Structure
     ReturnValue_ValidInteractionsGameplayEffectsTags:Gameplay Tag Container Structure
     ReturnValue_RoboticLoop:AFKDetection Robotic Loop Parameters Structure
    ).ReturnValue_TimeUntilRemainedAFK)]
 }
 CallPythonWithObj {nodes [
   n1 = CallPythonWithObj!() -> (then, Command:String, Args:Map of Strings to Strings) {is=title K2Node_FunctionEntry}
   > n2 = DoCommand!(
    execute
    self:Fort Dynamic Xp Calculator Object Reference=Get Calculator()
    Caller:Object Reference=Self-Reference()
    Command:String=@n1
    Args:Map of Strings to Strings=@n1.Args
   ) -> (then, ReturnValue:String)
   > Return Node!(execute, Result:String=@n2)
  ]}
 CopyLog {nodes [
   CopyLog() {is=title K2Node_FunctionEntry}
   > n1 = For Each Loop(Array:Array of Widget Object References=GetAllChildren(self=Get OutputLog())) -> (Array Element:Widget Object Reference)
   > n2 = Cast To DynamicXpLogEntry(Object=@n1)
   > n3 = GetLogLine!(execute, self:Dynamic Xp Log Entry Object Reference=@n2) -> (then, LogLineText:Text)
   > n4 = Branch(Condition=TextIsEmpty(InText=@n3))
   PrintText(execute=@n4.else, InText=@n3, bPrintToScreen=false)
  ]}
 CopyMcpProfileStatsToPlayerSession {nodes [
   CopyMcpProfileStatsToPlayerSession() {is=title K2Node_FunctionEntry}
   > SetXpProfileStats!(
    execute
    PlayerSession:"Dynamic Xp Player Session Structure (by ref)"=Get PlayerSession()
    Stats:"Dynamic Xp Profile Stats Structure (by ref)"=Get McpProfileStats()
    None:%=Get Calculator()
   ) -> (then)
  ]}
 CreateSaveGame {nodes [
   Inputs!() -> (execute:Exec)
   > n1 = LoadGameFromSlot!(execute, SlotName:String=Get SaveGameSlotName(), UserIndex:int=0) -> (then, ReturnValue:Save Game Object Reference)
   > n2 = Cast To DynamicXpSave!(execute, Object=@n1) -> (then, CastFailed, AsDynamic Xp Save:Dynamic Xp Save Object Reference) {is=title K2Node_DynamicCast}
   > n3 = Set SaveGame(SaveGame:Dynamic Xp Save Object Reference=@n2) -> (Output_Get:Dynamic Xp Save Object Reference)
   > "SetText (Text Box)"(self=Get inputPassword(), InText=Get Password!(self:Dynamic Xp Save Object Reference=@n3) -> (Password:Text))
   > "SetText (Text Box)"(self=Get inputUserName(), InText=Get UserName!(self:Dynamic Xp Save Object Reference=@n3) -> (UserName:Text))
   > n4 = PopulateIslandComboBox(Selected=Get SelectedIslandCode!(self:Dynamic Xp Save Object Reference=@n3) -> (SelectedIslandCode:String))
   Outputs!(then:Exec=[@n4.then, @n6.then])
   n5 = CreateSaveGameObject!(execute=@n2.CastFailed, SaveGameClass:Save Game Class Reference=/DynamicXp/Parts/DynamicXpSave.DynamicXpSave_C) -> (then, ReturnValue:Dynamic Xp Save Object Reference)
   > n6 = Set SaveGame(SaveGame=@n5)
  ]}
 EventGraph {nodes [
   n1 = "On Property Changed (DesignerDataAssetView)"!() -> (OutputDelegate, then, PropertyName:Name=None) {is=title K2Node_ComponentBoundEvent}
   validateButtonState(execute=[@n14.then, @n1.then, @n6.then, @n8.then])
   n2 = "On Property Changed (XpStats_DetailView)"!() -> (OutputDelegate, then, PropertyName:Name=None) {is=title K2Node_ComponentBoundEvent}
   > Branch(Condition=OR Boolean(A="Equal (Name)"(A=@n2.PropertyName, B=AccoladeID), B="Equal (Name)"(A=@n2.PropertyName, B=TriggersPerMinute)))
   > PopulateActiveGroupsComboBox()
   > UpdateProgressBars()
   n3 = "On Check State Changed (CheckBox_ShowAccolades)"!() -> (OutputDelegate, then, bIsChecked:bool) {is=title K2Node_ComponentBoundEvent}
   > Set ShowAccoladeTriggers(ShowAccoladeTriggers:bool=@n3.bIsChecked) -> (Output_Get:bool)
   n4 = "On Check State Changed (CheckBox_FAllowExploits)"!() -> (OutputDelegate, then, bIsChecked:bool) {is=title K2Node_ComponentBoundEvent}
   > n5 = Branch(Condition=@n4.bIsChecked)
   > ExecuteConsoleCommand(Command="DynamicXp.AllowExploits  true")
   ExecuteConsoleCommand(execute=@n5.else, Command=DynamicXp.AllowExploits false)
   n6 = "On Text Changed (inputPassword)"!() -> (OutputDelegate, then, Text:"Text (by ref)") {is=title K2Node_ComponentBoundEvent}
   "On Clicked (bttn_reload_script)"!() -> (OutputDelegate, then, widget:User Widget Object Reference, event:Pointer Event Structure) {is=title K2Node_ComponentBoundEvent}
   > Set Calculator(Calculator:Fort Dynamic Xp Calculator Object Reference) -> (Output_Get:Fort Dynamic Xp Calculator Object Reference)
   > Set ReloadPythonIfNeeded(ReloadPythonIfNeeded:bool=true) -> (Output_Get:bool)
   > SetColorAndOpacity(
    self:User Widget Object Reference=Get bttn_reload_script() -> (bttn_reload_script:TA Generic 01 Button Object Reference)
    InColorAndOpacity:Linear Color Structure=(R=0,G=0.5744,B=0.9896,A=1)
   )
   n7 = "On Selection Changed (ComboBox_Groups)"!() -> (OutputDelegate, then, SelectedItem:String, SelectionType:ESelectInfo Enum) {is=title K2Node_ComponentBoundEvent}
   > PopulateGroupAccolades(Group=@n7.SelectedItem)
   n8 = "On Text Changed (inputUserName)"!() -> (OutputDelegate, then, Text:"Text (by ref)") {is=title K2Node_ComponentBoundEvent}
   n9 = "On Check State Changed (CheckBox_ForcePass)"!() -> (OutputDelegate, then, bIsChecked:bool) {is=title K2Node_ComponentBoundEvent}
   > n10 = Branch(Condition=@n9.bIsChecked)
   > ExecuteConsoleCommand(Command="DynamicXp.ForceLive  true")
   ExecuteConsoleCommand(execute=@n10.else, Command=DynamicXp.ForceLive false)
   n11 = "On Selection Changed (ComboBoxLog)"!() -> (OutputDelegate, then, SelectedItem:String, SelectionType:ESelectInfo Enum) {is=title K2Node_ComponentBoundEvent}
   > SetLogLevel!(execute, self, Level:String=@n11.SelectedItem) -> (then)
   n12 = "On Text Committed (inputPassword)"!() -> (OutputDelegate, then, Text:"Text (by ref)", CommitMethod:ETextCommit Enum=Default) {is=title K2Node_ComponentBoundEvent}
   n13 = Get SaveGame()
   n14 = SaveGameToSlot(execute=[@n16.then, @n36.then, @n12.then], SaveGameObject=@n13, SlotName=Get SaveGameSlotName(), UserIndex=0)
   "On Property Changed (DetailsView_Time)"!() -> (OutputDelegate, then, PropertyName:Name=None) {is=title K2Node_ComponentBoundEvent}
   > UpdateTriggerTimeOnAccolades!(execute, self, Force:bool=false) -> (then)
   n15 = "On Text Committed (inputUserName)"!() -> (OutputDelegate, then, Text:"Text (by ref)", CommitMethod:ETextCommit Enum=Default) {is=title K2Node_ComponentBoundEvent}
   > n16 = Set UserName!(execute, UserName:Text=@n15.Text, self:Dynamic Xp Save Object Reference=@n13) -> (then, Output_Get:Text)
   n17 = Event Tick()
   > Parent: Tick!(execute, MyGeometry:Geometry Structure, InDeltaTime:float=@n17.InDeltaTime) -> (then) {is=title K2Node_CallParentFunction}
   > n18 = Is Valid(InputObject=Get Calculator())
   n19 = Branch(execute=@n18.Is Not Valid, Condition=Get ReloadPythonIfNeeded() -> (ReloadPythonIfNeeded:bool))
   > AddOutputLine!(execute, self, InText="LOCTEXT(\"loading dynamic_xp.py\")", Color:Slate Color Structure, ForceScrollToEnd=false) -> (then)
   > n20 = "Execute Python Command (Advanced)"!(
    execute
    PythonCommand:String=dynamic_xp.py
    ExecutionMode:EPythonCommandExecutionMode Enum=ExecuteFile
    FileExecutionScope:EPythonFileExecutionScope Enum=Private
   ) -> (then, CommandResult:String, LogOutput:Array of Python Log Output Entry Structures, ReturnValue:bool)
   > InitializeCalculatorDelegates()
   > Set ReloadPythonIfNeeded(ReloadPythonIfNeeded=false)
   > n21 = For Each Loop!(Exec, Array:Array of Python Log Output Entry Structures=@n20.LogOutput) -> (LoopBody, Array Element_Type:EPythonLogOutputType Enum, Array Element_Output:String, Array Index, Completed)
   AddOutputLine(execute=@n21.Completed, InText="To Text (String)"(InString=Append(A=Result:, B=@n20)), ForceScrollToEnd=true)
   n22 = Get bttn_reload_script()
   > SetColorAndOpacity(self=@n22, InColorAndOpacity=(R=0.9896,G=0.9896,B=0.9896,A=1))
   AddOutputLine(execute=@n21.LoopBody, InText="To Text (String)"(InString=@n21.Array Element_Output), ForceScrollToEnd=true)
   SetColorAndOpacity(execute=@n19.else, self=@n22, InColorAndOpacity=(R=0.9896,G=0,B=0.0293,A=1))
   n23 = Sequence(execute=@n18.Is Valid)
   n24 = Get PyMessage() -> (PyMessage:String)
   > n25 = Branch(Condition=IsEmpty(InString=@n24))
   AddOutputLine!(
    execute=@n25.else
    self
    InText="To Text (String)"(InString=@n24)
    Color_SpecifiedColor:Linear Color Structure=(R=0.4858,G=0.0512,B=1,A=1)
    Color_ColorUseRule:ESlateColorStylingMode Enum
    ForceScrollToEnd=true
   ) -> (then)
   > Set PyMessage(PyMessage:String) -> (Output_Get:String)
   n26 = Set LastTimeSeconds(execute=@n23.then_1, LastTimeSeconds=Get NowTimeSeconds() -> (NowTimeSeconds:double))
   > n27 = Set NowTimeSeconds(NowTimeSeconds:double=GetTimeSeconds()) -> (Output_Get:double)
   > Branch(Condition=Get GameIsRunning() -> (GameIsRunning:bool))
   > n28 = Branch(Condition="float > float"(A=float - float(A=@n27, B=@n26), B=0.3))
   > Set StepTime(StepTime:double=float * float(A=Get StepTime(), B=0.9) {is=title K2Node_PromotableOperator}) -> (Output_Get:double)
   > n29 = Set FrameTimeMult(FrameTimeMult:double=0.1) -> (Output_Get:double)
   n30 = Get NowTimeSeconds()
   n31 = Get LastTimeSeconds() -> (LastTimeSeconds:double)
   n32 = Set SeccondsPassThisFrame(
    execute=[@n34.then, @n29.then]
    SeccondsPassThisFrame:double=float * float!(A=float - float(A=@n30, B=@n31), B=Get StepTime(), C:double=Get FrameTimeMult() -> (FrameTimeMult:double)) -> (ReturnValue)
   ) -> (Output_Get:double)
   > Simulate Time Passed(SecondsPassed=@n32)
   n33 = Get FrameTimeMult()
   n34 = Set FrameTimeMult(
    execute=@n28.else
    FrameTimeMult="Min (Float)"(A:double=float + float(A=@n33, B=0.1), B:double=1) -> (ReturnValue:double) {is=title K2Node_CommutativeAssociativeBinaryOperator}
   )
   Event PreConstruct()
   n35 = Self-Reference()
   > SetObject(self=Get XpStats_DetailView() -> (XpStats_DetailView:Details View Object Reference), NewObject=@n35)
   > SetObject(self=Get DesignerDataAssetView() -> (DesignerDataAssetView:Single Property View Object Reference), NewObject=@n35)
   > SetObject(self=Get DetailsView_Time() -> (DetailsView_Time:Details View Object Reference), NewObject=@n35)
   Event Construct()
   > Parent: Construct()
   > SetScrollbarVisibility!(
    execute
    self:List View Base Object Reference=Get ListView_Accolades() -> (ListView_Accolades:Common List View Object Reference)
    InVisibility:ESlateVisibility Enum=Hidden
   ) -> (then)
   > SetLogLevel(Level=GetSelectedOption(self=Get ComboBoxLog() -> (ComboBoxLog:"Combo Box (String) Object Reference")))
   > Create Save Game!(execute:Exec) -> (then:Exec) {is=title K2Node_MacroInstance}
   > UpdateProgressBars()
   > ShowHelp()
   > validateButtonState()
   n36 = Set Password!(execute, Password:Text=@n12.Text, self:Dynamic Xp Save Object Reference=@n13) -> (then, Output_Get:Text)
   Event Destruct()
  ]}
 FindAccoladeEntryObject {nodes [
   n1 = FindAccoladeEntryObject!() -> (then, AccoladeName:String) {is=title K2Node_FunctionEntry}
   > n2 = For Each Loop!(
    Exec
    Array:Array of Object References=GetListItems!(self:List View Object Reference=Get ListView_Accolades()) -> (ReturnValue:Array of Object References)
   ) -> (LoopBody, Array Element:Object Reference, Array Index, Completed)
   > n3 = Cast To AccoladeEntryObject(Object=@n2)
   > Branch(Condition=StartsWith(
     SourceString=Get!(self:Accolade Entry Object Object Reference=@n3) -> (
      Accolade_Name_10_#1:String
      Accolade_CreatorName_9_#2:String
      Accolade_NumGroups_8_#3:int
     )
     InPrefix=@n1
    ))
   > n4 = Set EntryObj(EntryObj:Accolade Entry Object Object Reference=@n3) -> (Output_Get:Accolade Entry Object Object Reference)
   Return Node!(
    execute=[@n4.then, @n2.Completed]
    AccoladeEntry:Accolade Entry Object Object Reference=Get EntryObj() -> (EntryObj:Accolade Entry Object Object Reference)
   )
  ]}
 GetAccoladeTuning {nodes [
   n1 = GetAccoladeTuning!() -> (then, AccoladeName:String) {is=title K2Node_FunctionEntry}
   n2 = Get Design Settings Data Asset()
   > n3 = Is Valid(InputObject=@n2)
   > n4 = For Each Loop with Break!(Exec:Exec, Array:Array of Accolade Publish Data Structures=Get Publish Data(), Break:Exec=@n6.then) -> (
    LoopBody:Exec
    Array Element_ID:String
    Array Element_Name:String
    Array Element_Description:String
    Array Element_Weight:Gameplay Tag Structure
    Array Index:int
    Completed:Exec
   )
   n5 = Find!(
    TargetMap:Map of Gameplay Tag Structures to Dynamic Xp Creator Accolade Tuning Structures=Get CreatorTuning!(self:Creative Dynamic Xp Data Asset Object Reference=@n2) -> (CreatorTuning:Map of Gameplay Tag Structures to Dynamic Xp Creator Accolade Tuning Structures)
    Key:"Gameplay Tag Structure (by ref)"=@n4.Array Element_Weight
   ) -> (Value_Weight:float=~, ReturnValue)
   > Branch(Condition=AND Boolean(A="Equal Exactly (String)"(A=@n1, B=@n4), B=@n5.ReturnValue))
   > n6 = Set!(execute, AccoladeTuningValue_Weight:float=@n5) -> (then, Output_Get:Dynamic Xp Creator Accolade Tuning Structure)
   Return Node!(
    execute=[@n4.Completed, @n3.Is Not Valid]
    AccoladeTuning:Dynamic Xp Creator Accolade Tuning Structure=Get AccoladeTuningValue() -> (AccoladeTuningValue:Dynamic Xp Creator Accolade Tuning Structure)
   )
  ]}
 InitializeCalculatorConstData {nodes [
   InitializeCalculatorConstData() {is=title K2Node_FunctionEntry}
   > n1 = GetDataTableRowNames(Table=/Game/Balance/DataTables/TPMToWeightMultiplier)
   > n2 = For Each Loop(Array:Array of Names=@n1) -> (Array Element:Name)
   InitializeConstantData!(
    execute=@n2.Completed
    self:Fort Dynamic Xp Calculator Object Reference=Get Calculator()
    CappedSessionLenMinutesIn:float=Get CappedSessionLenMinutes!() -> (CappedSessionLenMinutes)
    CircuitBreakerValues:"Dynamic Xp Circuit Breaker Structure (by ref)"=Get CircuitBreaker!() -> (CircuitBreaker)
    Afk:"Dynamic Xp Afk Penalty Structure (by ref)"=Get AfkPenalty!() -> (AfkPenalty)
    TPMToWeightMultipliers:Array of TPMTo Weight Multiplier Structures=Get TPMToWeightMultiplier() -> (TPMToWeightMultiplier:Array of TPMTo Weight Multiplier Structures)
    bShouldRunCircuitBreaker:bool=false
   ) -> (then)
   > SetXpBudget!(execute, self:Fort Dynamic Xp Calculator Object Reference=Get Calculator(), NewXpBudget:int=Get XpBudgetPerMin!() -> (XpBudgetPerMin:int=0)) -> (then)
   n3 = Get Data Table Row TPMToWeightMultiplier!(execute:Exec=@n2.LoopBody, DataTable:Data Table Object Reference=&TPMToWeightMultiplier, RowName:Name=@n2) -> (then:Exec, RowNotFound:Exec, ReturnValue:TPMTo Weight Multiplier Structure) {is=title K2Node_GetDataTableRow}
   > Add(
    TargetArray:Array of TPMTo Weight Multiplier Structures=Get TPMToWeightMultiplier()
    NewItem:"TPMTo Weight Multiplier Structure (by ref)"=@n3
   ) {is=title K2Node_CallArrayFunction}
  ]}
 InitializeCalculatorDelegates {nodes [
   InitializeCalculatorDelegates() {is=title K2Node_FunctionEntry}
   n1 = Get Calculator()
   > Bind Event to On Exceeded Potential Cap(self:Fort Dynamic Xp Calculator Object Reference=@n1, Delegate=Create Event()) {is=title K2Node_AddDelegate}
   > Bind Event to On Exceeded Trigger Limit(self:Fort Dynamic Xp Calculator Object Reference=@n1, Delegate=Create Event()) {is=title K2Node_AddDelegate}
  ]}
 InitializeCalculatorXpStats {nodes [
   InitializeCalculatorXpStats() {is=title K2Node_FunctionEntry}
   n1 = Get XpStats()
   n2 = Get Calculator()
   > InitializeXpStats!(
    execute
    self:Fort Dynamic Xp Calculator Object Reference=@n2
    AggregatedStats:"Dynamic Xp Aggregated Stats Structure (by ref)"=@n1
    CurrentGameVersion:int=Break Dynamic Xp Aggregated Stats(DynamicXpAggregatedStats=@n1) -> (MeanSessionLengthMinutes=~).UniqueGameVersion
   ) -> (then)
   > Set XpStats!(
    execute
    XpStats=GetXpStats!(self:Fort Dynamic Xp Calculator Object Reference=@n2) -> (ReturnValue:"Dynamic Xp Aggregated Stats Structure (by ref)")
   ) -> (then, Output_Get)
   > BuildAccoladeMapFromXpStats()
  ]}
 OnExceededPotentialCap {nodes [
   n1 = OnExceededPotentialCap!() -> (then, Stats:"Dynamic Xp Session Accolade Stats Structure (by ref)", ResultingPotential:float=~, TotalXpCap:float=~) {is=title K2Node_FunctionEntry}
   n2 = Break Dynamic Xp Session Accolade Stats!(DynamicXpSessionAccoladeStats:"Dynamic Xp Session Accolade Stats Structure (by ref)"=@n1) -> (AccoladeId:String, AccoladeName:String, Grouping:String, Triggers:int, GrantedXp:int, PotentialXp:int) {is=title K2Node_BreakStruct}
   > ShowTriggerError!(
    execute
    self
    AccoladeName:String=@n2.AccoladeName
    ErrorMsg:Text=Format Text!(
     Format="LOCTEXT(\"{Accolade} Exceeded PotentialXP Cap {XP} > {Cap}\")"
     Accolade:String=@n2.AccoladeName
     XP:float=@n1.ResultingPotential
     Cap:float=@n1.TotalXpCap
    ) -> (Result)
   ) -> (then)
   > Return Node!(execute)
  ]}
 OnExceededTriggerLimit {nodes [
   n1 = OnExceededTriggerLimit!() -> (
    then
    Stats_AccoladeId:String
    Stats_AccoladeName:String
    Stats_Grouping:String
    Stats_Triggers:int
    Stats_GrantedXp:int
    Stats_PotentialXp:int
    Triggers:int
    TriggerLimit:int
   ) {is=title K2Node_FunctionEntry}
   > ShowTriggerError(
    AccoladeName=@n1.Stats_AccoladeName
    ErrorMsg=Format Text!(
     Format="INVTEXT(\"{Accolade}  Exceeded Trigger Limit {Triggers} > {Limit}\")"
     Accolade=@n1.Stats_AccoladeName
     Triggers:int=@n1.Triggers
     Limit:int=@n1.TriggerLimit
    ) -> (Result)
   )
   > Return Node()
  ]}
 PopulateActiveGroupsComboBox {nodes [
   PopulateActiveGroupsComboBox() {is=title K2Node_FunctionEntry}
   n1 = Get ComboBox_Groups() -> (ComboBox_Groups:"Combo Box (String) Object Reference")
   > n2 = Set OrigIndex(OrigIndex:int=GetSelectedIndex!(self:"Combo Box (String) Object Reference"=@n1) -> (ReturnValue:int=0)) -> (Output_Get:int=0)
   > ClearOptions(self:"Combo Box (String) Object Reference"=@n1)
   > Clear!(execute, TargetSet:Set of Strings=Get Groups() -> (Groups:Set of Strings)) -> (then) {is=title K2Node_CallFunction}
   > Clear(TargetSet=Get AllAccolades())
   K2Node_Composite_1 = Accolades!() -> (XpStats_Accolades:Array of Dynamic Xp Aggregated Accolade Stats Structures) {is=title K2Node_Composite}
   > n3 = For Each Loop!(Exec, Array:Array of Dynamic Xp Aggregated Accolade Stats Structures=@K2Node_Composite_1) -> (LoopBody, Array Element_AccoladeID:String, Array Element_TriggersPerMinute:double=~, Array Index, Completed)
   n4 = Branch(execute=@n3.Completed, Condition="integer >= integer"(A=@n2))
   SetSelectedIndex!(execute=[@n4.then, @n5.then], self:"Combo Box (String) Object Reference"=@n1, Index:int=Get OrigIndex() -> (OrigIndex:int=0)) -> (then)
   n5 = Set OrigIndex(execute=@n4.else, OrigIndex=0) -> (Output_Get=0)
   n6 = GetAccoladeNameAndGrouping(execute=@n3.LoopBody, FromId=@n3)
   > n7 = Set Grouping(Grouping:String=@n6.Grouping) -> (Output_Get:String)
   > Add!(execute, TargetSet:Set of Strings=Get AllAccolades(), NewItem:"String (by ref)"=@n6) -> (then) {is=title K2Node_CallFunction}
   n8 = Get Groups()
   > n9 = Branch(Condition=Contains Item!(TargetSet:Set of Strings=@n8, ItemToFind:"String (by ref)"=@n7) -> (ReturnValue:bool) {is=title K2Node_CallFunction})
   Add(execute=@n9.else, TargetSet=@n8, NewItem=@n7)
   > Set NumValidAccolades(NumValidAccolades:int=0) -> (Output_Get:int=0)
   K2Node_Composite_0 = Accolades_2_2!() -> (XpStats_Accolades:Array of Dynamic Xp Aggregated Accolade Stats Structures) {is=title K2Node_Composite}
   > n10 = For Each Loop(Array=@K2Node_Composite_0) -> (Array Element_TriggersPerMinute=~)
   n11 = Get NumValidAccolades() -> (NumValidAccolades:int=0)
   AddOption!(
    execute=@n10.Completed
    self:"Combo Box (String) Object Reference"=@n1
    Option:String=Append!(A=@n7, B=Get GroupSeperator() -> (GroupSeperator:String), C:String="To String (Integer)"(InInt=@n11)) -> (ReturnValue)
   ) -> (then)
   n12 = GetAccoladeNameAndGrouping(execute=@n10.LoopBody, FromId=@n10)
   > Branch(Condition="Equal Exactly (String)"(A=@n7, B=@n12.Grouping))
   > Branch(Condition="float > float"(A=@n10.Array Element_TriggersPerMinute))
   > Increment Int(Value=Get NumValidAccolades() -> (NumValidAccolades=0))
  ]}
 Accolades_2_2 {
  path=PopulateActiveGroupsComboBox.K2Node_Composite_0.Accolades_2_2
  nodes [
   Inputs!()
   Outputs!(XpStats_Accolades:Array of Dynamic Xp Aggregated Accolade Stats Structures=Get!() -> (
     XpStats_UniqueGameId:String
     XpStats_UniqueGameVersion:int
     XpStats_CalibrationResult:Dynamic Xp Calibration Result Structure
     XpStats_Accolades:Array of Dynamic Xp Aggregated Accolade Stats Structures
     XpStats_MeanSessionLengthMinutes:double=~
     XpStats_PublishData:Array of Accolade Publish Data Structures
    ).XpStats_Accolades)
  ]
 }
 Accolades {
  path=PopulateActiveGroupsComboBox.K2Node_Composite_1.Accolades
  nodes [
   Inputs()
   Outputs(XpStats_Accolades=Get() -> (XpStats_MeanSessionLengthMinutes=~).XpStats_Accolades)
  ]
 }
 PopulateGroupAccolades {nodes [
   n1 = PopulateGroupAccolades!() -> (then, Group:String) {is=title K2Node_FunctionEntry}
   > ClearListItems(self=Get ListView_Accolades())
   K2Node_Composite_0 = GetXpStatsAccolades!() -> (XpStats_Accolades:Array of Dynamic Xp Aggregated Accolade Stats Structures) {is=title K2Node_Composite}
   > n2 = For Each Loop(Array=@K2Node_Composite_0) -> (Array Element_TriggersPerMinute=~)
   n3 = Get Entries() -> (Entries:Array of Accolade Entry Object Object References)
   Set List Items!(execute=@n2.Completed, self:List View Object Reference=Get ListView_Accolades(), InListItems:Array of Object References=@n3) -> (then) {is=title K2Node_CallFunction}
   > UpdateTriggerTimeOnAccolades(Force=true)
   n4 = GetAccoladeNameAndGrouping(execute=@n2.LoopBody, FromId=@n2)
   > Branch(Condition="Equal Exactly (String)"(A=@n4.Grouping, B=StripGrouping!(SourceString:String=@n1) -> (LeftS:String)))
   > n5 = Construct Object from Class!(
    execute
    Class=/DynamicXp/Parts/AccoladeEntryObject.AccoladeEntryObject_C
    self=Self-Reference()
    TriggersPerMin:double=@n2.Array Element_TriggersPerMinute
    Accolade:Dynamic Xp Accolade Structure=Find!(TargetMap:Map of Strings to Dynamic Xp Accolade Structures=Get AccoladesMap(), Key:"String (by ref)"=@n4) -> (Value:Dynamic Xp Accolade Structure, ReturnValue)
   ) -> (then, ReturnValue:Accolade Entry Object Object Reference)
   > n6 = For Each Loop with Break!(Exec, Array:Array of Accolade Entry Object Object References=@n3, Break=@n7.then) -> (LoopBody, Array Element:Accolade Entry Object Object Reference, Array Index, Completed)
   Add Unique(
    execute=@n6.Completed
    TargetArray:Array of Accolade Entry Object Object References=@n3
    NewItem:"Accolade Entry Object Object Reference (by ref)"=@n5
   )
   Branch(
    execute=@n6.LoopBody
    Condition="float > float"(A=@n2.Array Element_TriggersPerMinute, B=Get TriggersPerMin!(self:Accolade Entry Object Object Reference=@n6) -> (TriggersPerMin:double))
   )
   > n7 = Insert!(
    execute
    TargetArray:Array of Accolade Entry Object Object References=@n3
    NewItem:"Accolade Entry Object Object Reference (by ref)"=@n5
    Index:int=@n6.Array Index
   ) -> (then) {is=title K2Node_CallArrayFunction}
  ]}
 GetXpStatsAccolades {
  path=PopulateGroupAccolades.K2Node_Composite_0.GetXpStatsAccolades
  nodes [
   Inputs()
   Outputs(XpStats_Accolades=Get() -> (XpStats_MeanSessionLengthMinutes=~).XpStats_Accolades)
  ]
 }
 PopulateIslandComboBox {nodes [
   n1 = PopulateIslandComboBox!() -> (then, Selected:String) {is=title K2Node_FunctionEntry}
   n2 = Get ComboBoxIslandCodes()
   > ClearOptions(self=@n2)
   > n3 = Keys!(execute, TargetMap:Map of Strings to Strings=Get IslandCodes(self=Get SaveGame())) -> (then, Keys:Array of Strings)
   > n4 = For Each Loop!(Exec, Array:Array of Strings=@n3) -> (LoopBody, Array Element:String, Array Index, Completed)
   n5 = Branch(execute=@n4.Completed, Condition=NOT Boolean(A=IsEmpty(InString=@n1)))
   > SetSelectedOption!(execute, self:"Combo Box (String) Object Reference"=@n2, Option:String=@n1) -> (then)
   SetSelectedIndex(execute=@n5.else, self=@n2, Index=0)
   AddOption(execute=@n4.LoopBody, self=[@n2, @n2], Option=@n4)
  ]}
 RunPythonScript {nodes [
   n1 = RunPythonScript!() -> (then, Command:String, CommandArgs:String) {is=title K2Node_FunctionEntry}
   > n2 = "Execute Python Command (Advanced)"(
    PythonCommand=Append!(A=Get ScriptPath() -> (ScriptPath:String), B=" ", C=@n1, D:String, E:String=@n1.CommandArgs) -> (ReturnValue)
    ExecutionMode=ExecuteFile
    FileExecutionScope=Private
   )
   > n3 = Branch(Condition=@n2.ReturnValue)
   > n4 = For Each Loop(Array:Array of Python Log Output Entry Structures=@n2.LogOutput) -> (Array Element:Python Log Output Entry Structure)
   > Set PyMessage(PyMessage=Append!(
     A=Get PyMessage()
     B="\r\n"
     C=Break Python Log Output Entry!(PythonLogOutputEntry:"Python Log Output Entry Structure (by ref)"=@n4) -> (Type:EPythonLogOutputType Enum, Output:String) {is=title K2Node_BreakStruct}.Output
    ) -> (ReturnValue))
   Set PyMessage(execute=@n3.else, PyMessage=@n2)
  ]}
 SetIslandTitle {nodes [
   n1 = SetIslandTitle!() -> (then, Key:"String (by ref)") {is=title K2Node_FunctionEntry}
   n2 = Get IslandCodes(self=Get SaveGame())
   > "SetText (Text)"(
    self=Get textFieldIslandName() -> (textFieldIslandName:Common Text Object Reference)
    InText="To Text (String)"(InString=Find(TargetMap:Map of Strings to Strings=@n2, Key=@n1) -> (Value:String))
   )
  ]}
 SetLogLevel {nodes [
   n1 = SetLogLevel!() -> (then, Level:String) {is=title K2Node_FunctionEntry}
   > ExecuteConsoleCommand(Command=Append!(A="Log LogDynamicXp ", B=@n1) -> (ReturnValue))
  ]}
 SetTextBoxFromCommandLine {nodes [
   n1 = SetTextBoxFromCommandLine!() -> (then, Key:String, TargetBox:Text Box Object Reference) {is=title K2Node_FunctionEntry}
   > n2 = GetCommandLine!(execute) -> (then, ReturnValue:String)
   > n3 = ParseCommandLine!(execute, InCmdLine:String=@n2) -> (then, OutTokens:Array of Strings, OutSwitches:Array of Strings, OutParams:Map of Strings to Strings)
   n4 = "GetText (Text Box)"(self=@n1.TargetBox)
   > Branch(Condition=AND Boolean(A=Contains(TargetMap=@n3.OutParams, Key=@n1), B=TextIsEmpty(InText=@n4)))
   > "SetText (Text Box)"(self=@n1.TargetBox, InText="To Text (String)"(InString=Find(TargetMap=@n3.OutParams, Key=@n1)))
  ]}
 ShowHelp {nodes [
   ShowHelp() {is=title K2Node_FunctionEntry}
   > AddOutputLine(
    InText="INVTEXT(\"General Help:\\r\\nMouse Over most buttons for functionality.\\r\\n\")"
    Color_SpecifiedColor=(R=0,G=1,B=0.4362,A=1)
    ForceScrollToEnd=false
   )
   > AddOutputLine!(
    execute
    self
    InText="INVTEXT(\"Typically you will: \\r\\n1. Select and Copy the data asset.  \\r\\n2. Fill in name, password, and Island code and hit Query.\\r\\n3. Select Active Group\\r\\n4. Press Start\\r\\n5. Press Run, Step, or click accolade to trigger.\\r\\n6. End\\r\\n\")"
    Color:Slate Color Structure=Get Help Color() -> (Help Color:Slate Color Structure)
    ForceScrollToEnd=false
   ) -> (then)
   > AddOutputLine(
    InText="INVTEXT(\"The bars at the bottom are updated from the simulation and affect how much actual XP will be granted when it triggers.  When an accolade triggers, some information will be displayed in this log.  More information is displayed in the Output Log depending on the value of the “Output Log” below. \\r\\n\\r\\nThe format of the log line is:\\r\\n\\r\\n<Triggers>. <Game Time> <Accolade GUID/Name> XP: <Potential XP> * <Multiplier> + <Catchup> -> <Granted XP>\\r\\n\")"
    Color=Get Help Color()
    ForceScrollToEnd=true
   )
  ]}
 ShowTriggerError {nodes [
   n1 = ShowTriggerError!() -> (then, AccoladeName:String, ErrorMsg:Text) {is=title K2Node_FunctionEntry}
   > n2 = FindAccoladeEntryObject!(execute, self, AccoladeName:String=@n1) -> (then, AccoladeEntry:Accolade Entry Object Object Reference) {is=title K2Node_CallFunction}
   > n3 = Branch(Condition=IsValid(Object=@n2))
   > AddOutputLine!(
    execute
    self
    InText=@n1.ErrorMsg
    Color_SpecifiedColor:Linear Color Structure=(R=1,G=0,B=0,A=1)
    Color_ColorUseRule:ESlateColorStylingMode Enum
    ForceScrollToEnd=false
   ) -> (then)
   > ShowTriggerError!(execute, self:Accolade Entry Object Object Reference=@n2) -> (then) {is=title K2Node_CallFunction}
   AddOutputLine(
    execute=@n3.else
    InText=Format Text!(Format="INVTEXT(\"{Msg} ERROR NOT FOUND\")", Msg:Text=@n1.ErrorMsg) -> (Result)
    Color_SpecifiedColor=(R=1,G=0,B=0,A=1)
    ForceScrollToEnd=false
   )
  ]}
 Simulate Time Passed {nodes [
   n1 = Simulate Time Passed!() -> (then, SecondsPassed:double=~, OnlyAccolade:String) {is=title K2Node_FunctionEntry}
   > n2 = For Each Loop(Array:Array of Object References=GetListItems(self=Get ListView_Accolades())) -> (Array Element:Object Reference)
   > n3 = Cast To AccoladeEntryObject(Object=@n2)
   n4 = Get!(self:Accolade Entry Object Object Reference=@n3) -> (
    Accolade_Name_10_#1:String
    Accolade_CreatorName_9_#2:String
    Accolade_NumGroups_8_#3:int
   )
   > Branch(Condition=OR Boolean(A=IsEmpty(InString=@n1.OnlyAccolade), B=StartsWith(SourceString=@n4, InPrefix=@n1.OnlyAccolade)))
   > n5 = GetNumTiggers!(execute, self:Accolade Entry Object Object Reference=@n3, SecondsPassed:double=@n1) -> (then, NumTriggers:int=0)
   > n6 = Set NumTriggers(NumTriggers:int=@n5) -> (Output_Get:int=0)
   > For Loop(FirstIndex=1, LastIndex=@n6)
   n7 = Get AccoladesToTrigger() -> (AccoladesToTrigger:Array of Accolade Entry Object Object References)
   > Add(TargetArray:Array of Accolade Entry Object Object References=@n7, NewItem:"Accolade Entry Object Object Reference (by ref)"=@n3) {is=title K2Node_CallArrayFunction}
   n8 = Length(TargetArray:Array of Accolade Entry Object Object References=@n7)
   n9 = Branch(execute=@n2.Completed, Condition="integer > integer"(A=@n8))
   > n10 = Set SecsPerAccolade(SecsPerAccolade:double=float / float(A=@n1, B:int=@n8)) -> (Output_Get:double)
   > n11 = For Each Loop(Array:Array of Accolade Entry Object Object References=@n7) -> (Array Element:Accolade Entry Object Object Reference)
   UpdateProgressBars(execute=@n11.Completed)
   TriggerAccolade(execute=@n11.LoopBody, SecondsPassed=@n10, AccoladeEntry=@n11)
   UpdateGameTime!(execute=@n9.else, self, AddSeconds:double=@n1) -> (then, NewGameTime:double, Timespan:Timespan Structure)
  ]}
 StripGrouping {nodes [
   n1 = StripGrouping!() -> (then, SourceString:String) {is=title K2Node_FunctionEntry}
   > Return Node!(execute, LeftS:String=Split(SourceString=@n1, InStr=Get GroupSeperator(), SearchDir=FromStart))
  ]}
 TriggerAccolade {nodes [
   n1 = TriggerAccolade!() -> (then, SecondsPassed:double=~, AccoladeEntry:Accolade Entry Object Object Reference) {is=title K2Node_FunctionEntry}
   > n2 = UpdateGameTime(AddSeconds=@n1)
   n3 = Get ComboBox_Groups()
   > n4 = MakeAccoladeId!(
    execute
    AccoladeName:String=GetAccoladeName!(self:Accolade Entry Object Object Reference=@n1.AccoladeEntry) -> (Name:String)
    AccoladeGrouping:String=StripGrouping(SourceString=GetSelectedOption(self=@n3)) {is=title K2Node_CallFunction}
   ) -> (then, ReturnValue:String)
   n5 = Get PlayerSession()
   > n6 = OnTriggerCalculateXp!(
    execute
    self:Fort Dynamic Xp Calculator Object Reference=Get Calculator()
    PlayerSession:"Dynamic Xp Player Session Structure (by ref)"=@n5
    AccoladeId:String=@n4
    CurrentTime:float=@n2
    CurrentWeek:int=Get CurrentWeek() -> (CurrentWeek:int=0)
   ) -> (then, ReturnValue:int=0)
   n7 = Get Calculator()
   n8 = GetXpProfileStats!(PlayerSession:"Dynamic Xp Player Session Structure (by ref)"=Get PlayerSession(), None:%=@n7) -> (
    ReturnValue_Timespan:float=~
    ReturnValue_BucketXp:int
    ReturnValue_BankXp:int
    ReturnValue_BankXpMult:float=~
    ReturnValue_BoosterBucketXp:int
    ReturnValue_BoosterXpMult:float=~
    ReturnValue_WeeklyExcessXpMult:float=~
    ReturnValue_CurrentWeekXp:int
    ReturnValue_CurrentWeek:int
   )
   K2Node_Composite_0 = PlayerSessionAfk!() -> (PlayerSession_AfkRewardMult:double=~) {is=title K2Node_Composite}
   > n9 = Set LastFullXpMult(LastFullXpMult:double=float * float!(
     A=@K2Node_Composite_0
     B=@n8.ReturnValue_BankXpMult
     C:float=@n8.ReturnValue_WeeklyExcessXpMult
     D:float=GetXpVarianceMult!(self:Fort Dynamic Xp Calculator Object Reference=@n7) -> (ReturnValue:float)
    ) -> (ReturnValue)) -> (Output_Get:double)
   n10 = Get TriggerCount() -> (TriggerCount:int=0)
   > Increment Int(Value=@n10)
   > OnTrigger!(execute, self:Accolade Entry Object Object Reference=@n1.AccoladeEntry, XpAmount:int=@n6) -> (then)
   > Branch(Condition=Get ShowAccoladeTriggers() -> (ShowAccoladeTriggers:bool))
   n11 = Break Dynamic Xp Calculation Record!(DynamicXpCalculationRecord:"Dynamic Xp Calculation Record Structure (by ref)"=Break Dynamic Xp Player Session!(DynamicXpPlayerSession:"Dynamic Xp Player Session Structure (by ref)"=@n5) -> (
     QueuedAccolade:Array of Dynamic Xp Queued Accolade Structures
     LastVoteGuid:Guid Structure
     TotalXpGained:int
     TotalXpPotential:int
     TotalAccoladeXpBudgetNormalized:float=~
     SessionStartTime:float=~
     SessionEndTime:float=~
     LastProgressExportTime:float=~
     BankXpReductionTimer:float=~
     bInMiddleOfRound:bool=~
     bHitXpThreshold:bool=~
     bPrevHitXpThreshold:bool=~
     bHitExcessXpCutoff:bool=~
     bPrevHitExcessXpCutoff:bool=~
     TimeSpanOutliers:int
     LastAccoladeCalculationRecord:Dynamic Xp Calculation Record Structure
     bReportedTimeSpanOutlier:bool=~
     TimeFrameHardCapBreachTimes:"Array of Float (single-precision)s"
     TimeFrameHardCapBreachCount:int
     TimeFrameHardCapXpWithheld:int
     WeeklyExcessXpWithheldThisSession:float=~
     AfkRewardMult:float=~
     AccoladeTriggersWhileAfk:int
     AfkPenaltyXpWithheldThisSession:float=~
     NumTimesBoosted:int
    ) {is=title K2Node_BreakStruct}.LastAccoladeCalculationRecord) -> (
    CurrentTime:float=~
    CurrentWeek:int
    PotentialXp:int
    CategoryWeightMult:float=~
    VarianceMult:float=~
    BaseXp:float=~
    UncappedXpReward:int
    BankXpMult:float=~
    AfkMult:float=~
    WeeklyCapMult:float=~
    BoosterXpMult:float=~
    BoostedXp:float=~
    BankXpWithheld:float=~
    AfkXpWithheld:float=~
    WeeklyCappedXpWithheld:float=~
    TimeFrameHardCapXpWithheld:int
    FinalXpReward:int
   ) {is=title K2Node_BreakStruct}
   n12 = GetAccoladeDisplayName!(self:Accolade Entry Object Object Reference=@n1.AccoladeEntry, Full:bool=false) -> (Name:Text)
   > AddOutputLine!(
    execute
    self
    InText=Format Text!(
     Format="INVTEXT(\"{triggernum}. {time}  {Accolade}   XP: Potential({Potential}) * {Mult} + Boosted({Boosted}) -> {XP} \")"
     triggernum:int=@n10
     time:Text=GetTimespanAsText!(
      Timespan:"Timespan Structure (by ref)"=@n2.Timespan
      bShowHoursWhenZero:bool=true
      bShowSeconds:bool=true
      FormatType:ETimespanAsTextFormat Enum=DaysHoursMinutesSeconds
      UnitDisplayType:ETimespanUnitDisplayFormat Enum=Abbreviated
      bUseSentenceCase:bool=false
     ) -> (ReturnValue:Text)
     Accolade:String="To String (Text)"(InText=@n12)
     Potential:int=@n11.PotentialXp
     Mult:Text="To Text (Float)"(Value=@n9, MinimumIntegralDigits=1, MaximumIntegralDigits=324, MinimumFractionalDigits=0, MaximumFractionalDigits=3)
     Boosted:float=@n11.BoostedXp
     XP:int=@n6
    ) -> (Result)
    Color:Slate Color Structure=GetAccoladeColor!(self:Accolade Entry Object Object Reference=@n1.AccoladeEntry) -> (ColorFromId:Slate Color Structure)
    ForceScrollToEnd=false
   ) -> (then)
   "GetText (Text Box)"(self=Get TextBoxTotalMult() -> (TextBoxTotalMult:Text Box Object Reference))
  ]}
 PlayerSessionAfk {
  path=TriggerAccolade.K2Node_Composite_0.PlayerSessionAfk
  nodes [
   Inputs()
   Outputs!(PlayerSession_AfkRewardMult:double=Get!() -> (
     PlayerSession_QueuedAccolade:Array of Dynamic Xp Queued Accolade Structures
     PlayerSession_LastVoteGuid:Guid Structure
     PlayerSession_TotalXpGained:int
     PlayerSession_TotalXpPotential:int
     PlayerSession_TotalAccoladeXpBudgetNormalized:float=~
     PlayerSession_SessionStartTime:float=~
     PlayerSession_SessionEndTime:float=~
     PlayerSession_LastProgressExportTime:float=~
     PlayerSession_BankXpReductionTimer:float=~
     PlayerSession_bInMiddleOfRound:bool=~
     PlayerSession_bHitXpThreshold:bool=~
     PlayerSession_bPrevHitXpThreshold:bool=~
     PlayerSession_bHitExcessXpCutoff:bool=~
     PlayerSession_bPrevHitExcessXpCutoff:bool=~
     PlayerSession_TimeSpanOutliers:int
     PlayerSession_LastAccoladeCalculationRecord:Dynamic Xp Calculation Record Structure
     PlayerSession_bReportedTimeSpanOutlier:bool=~
     PlayerSession_TimeFrameHardCapBreachTimes:"Array of Float (single-precision)s"
     PlayerSession_TimeFrameHardCapBreachCount:int
     PlayerSession_TimeFrameHardCapXpWithheld:int
     PlayerSession_WeeklyExcessXpWithheldThisSession:float=~
     PlayerSession_AfkRewardMult:float=~
     PlayerSession_AccoladeTriggersWhileAfk:int
     PlayerSession_AfkPenaltyXpWithheldThisSession:float=~
     PlayerSession_NumTimesBoosted:int
    ).PlayerSession_AfkRewardMult)
  ]
 }
 UpdateGameTime {nodes [
   n1 = UpdateGameTime!() -> (then, AddSeconds:double=~) {is=title K2Node_FunctionEntry}
   > n2 = Set GameTime(GameTime=float + float(A=@n1, B=Get GameTime() -> (GameTime:double)))
   > K2Node_Composite_0 = GameTimeDisplaySetter!(execute, GameTimeDisplay_Seconds:int=Truncate(A=@n2)) -> (then:Exec, Output_Get:Timespan Structure) {is=title K2Node_Composite}
   > n3 = Set GameTimeDisplay(GameTimeDisplay:Timespan Structure=@K2Node_Composite_0) -> (Output_Get:Timespan Structure)
   > n4 = Branch(Condition=IsChecked(self=Get CheckBox_Afk() -> (CheckBox_Afk:Check Box Object Reference)))
   > n5 = Set AfkTimer(AfkTimer:double=float + float(A=Get AfkTimer() -> (AfkTimer:double), B=@n1)) -> (Output_Get:double)
   DebugOverrideTimeSpentAfk!(execute=[@n6.then, @n5.then], PlayerSession:"Dynamic Xp Player Session Structure (by ref)"=Get PlayerSession(), AfkTime:float=Get AfkTimer()) -> (then)
   > Return Node!(execute, NewGameTime:double=@n2, Timespan:Timespan Structure=@n3)
   n6 = Set AfkTimer(execute=@n4.else, AfkTimer=0)
  ]}
 GameTimeDisplaySetter {
  path=UpdateGameTime.K2Node_Composite_0.GameTimeDisplaySetter
  nodes [
   n1 = Inputs!() -> (execute:Exec, GameTimeDisplay_Seconds:int)
   > n2 = Set!(
    execute
    GameTimeDisplay_Days:int=0
    GameTimeDisplay_Hours:int=0
    GameTimeDisplay_Minutes:int=0
    GameTimeDisplay_Seconds:int=@n1
    GameTimeDisplay_Milliseconds:int=0
   ) -> (then, Output_Get:Timespan Structure)
   > Outputs!(then:Exec, Output_Get:Timespan Structure=@n2)
  ]
 }
 UpdateProgressBars {nodes [
   UpdateProgressBars() {is=title K2Node_FunctionEntry}
   K2Node_Composite_0 = Player Session Afk Reward Mult!() -> (PlayerSession_AfkRewardMult:double=~) {is=title K2Node_Composite}
   > SetPercent!(
    execute
    self:Progress Bar Object Reference=Get ProgressBar_Afk() -> (ProgressBar_Afk:Progress Bar Object Reference)
    InPercent:float=@K2Node_Composite_0
   ) -> (then)
   > n1 = Set McpProfileStats!(
    execute
    McpProfileStats=GetXpProfileStats!(PlayerSession=Get PlayerSession()) -> (ReturnValue:"Dynamic Xp Profile Stats Structure (by ref)")
   ) -> (
    then
    Output_Get_Timespan:float=~
    Output_Get_BucketXp:int
    Output_Get_BankXp:int
    Output_Get_BankXpMult:float=~
    Output_Get_BoosterBucketXp:int
    Output_Get_BoosterXpMult:float=~
    Output_Get_WeeklyExcessXpMult:float=~
    Output_Get_CurrentWeekXp:int
    Output_Get_CurrentWeek:int
   )
   > SetPercent(self=Get ProgressBar_Bank() -> (ProgressBar_Bank:Progress Bar Object Reference), InPercent=@n1.Output_Get_BankXpMult)
   > SetPercent(self=Get ProgressBar_WeeklyMult() -> (ProgressBar_WeeklyMult:Progress Bar Object Reference), InPercent=@n1.Output_Get_WeeklyExcessXpMult)
   n2 = Get Calculator()
   > n3 = Branch(Condition=IsValid(Object=@n2))
   n4 = Get ProgressBar_XpVarianceMult() -> (ProgressBar_XpVarianceMult:Progress Bar Object Reference)
   > SetPercent(self=@n4, InPercent=GetXpVarianceMult(self=@n2))
   n5 = Get LastFullXpMult() -> (LastFullXpMult:double)
   n6 = Get TextBoxTotalMult()
   > "SetText (Text Box)"(self=@n6, InText=Format Text!(Format="LOCTEXT(\"{Value}%\")", Value:double=float * float!(A=@n5, B=100) -> (ReturnValue)) -> (Result))
   n7 = GetTimespanOutlierThreshold!(self:Fort Dynamic Xp Calculator Object Reference=@n2) -> (ReturnValue:float)
   n8 = Get progressBarBucket() -> (progressBarBucket:Progress Bar Object Reference)
   > SetPercent(self=@n8, InPercent=float / float(A:int=@n1.Output_Get_BucketXp, B:double="Max (Float)"(A=@n7, B=1)))
   n9 = Get textFieldBucketValue() -> (textFieldBucketValue:Common Text Object Reference)
   > "SetText (Text)"(
    self=@n9
    InText=Format Text!(Format="LOCTEXT(\"{Bucket} / {Threshold}\")", Bucket:int=@n1.Output_Get_BucketXp, Threshold:float=@n7) -> (Result)
   )
   n10 = Get textFieldBoosterBucketValue() -> (textFieldBoosterBucketValue:Common Text Object Reference)
   n11 = Get MinXpThreshold!(self:Fort Dynamic Xp Calculator Object Reference=@n2) -> (MinXpThreshold:int=0)
   > "SetText (Text)"(self=@n10, InText=Format Text(Format="LOCTEXT(\"{Bucket} / {Threshold}\")", Bucket=@n1.Output_Get_BoosterBucketXp, Threshold:int=@n11))
   n12 = Get ProgressBar_XpBooster() -> (ProgressBar_XpBooster:Progress Bar Object Reference)
   > SetPercent(self=@n12, InPercent=float / float(A:double="To Float (Integer)"(InInt=@n1.Output_Get_BoosterBucketXp), B="To Float (Integer)"(InInt=@n11)))
   n13 = "Max (Float)"(A=@n1, B=0)
   K2Node_Composite_1 = CircuiteBreakData!() -> (CircuitBreaker_TimespanOutlierDuration:int, CircuitBreaker_MaxXpBoosterMult:double=~) {is=title K2Node_Composite}
   n14 = "To Float (Integer)"(InInt=int * int(A=@K2Node_Composite_1, B=60))
   n15 = Get ProgressBarTimespan() -> (ProgressBarTimespan:Progress Bar Object Reference)
   > SetPercent(self=@n15, InPercent=float / float(A=@n13, B=@n14))
   n16 = Get textFieldTimespanValue() -> (textFieldTimespanValue:Common Text Object Reference)
   > "SetText (Text)"(self=@n16, InText=Format Text!(Format="LOCTEXT(\"{Timespan} / {TimespanMax}\")", Timespan:double=@n13, TimespanMax:double=@n14) -> (Result))
   n17 = "To Float (Integer)"(InInt=@n1.Output_Get_CurrentWeekXp)
   n18 = GetWeeklyExcessXpThreshold!(self:Fort Dynamic Xp Calculator Object Reference=@n2) -> (ReturnValue:float)
   n19 = Get ProgressBar_Weekly() -> (ProgressBar_Weekly:Progress Bar Object Reference)
   > SetPercent(self=@n19, InPercent=float / float(A=@n17, B=@n18))
   n20 = Get textFieldWeeklyValue() -> (textFieldWeeklyValue:Common Text Object Reference)
   > "SetText (Text)"(self=@n20, InText=Format Text!(Format="LOCTEXT(\"{Cur} / {Threshold}\")", Cur:double=@n17, Threshold:float=@n18) -> (Result))
   n21 = Break Dynamic Xp Player Session(DynamicXpPlayerSession=Get PlayerSession()) -> (
    TotalAccoladeXpBudgetNormalized=~
    SessionStartTime=~
    SessionEndTime=~
    LastProgressExportTime=~
    BankXpReductionTimer=~
    bInMiddleOfRound=~
    bHitXpThreshold=~
    bPrevHitXpThreshold=~
    bHitExcessXpCutoff=~
    bPrevHitExcessXpCutoff=~
    bReportedTimeSpanOutlier=~
    WeeklyExcessXpWithheldThisSession=~
    AfkRewardMult=~
    AfkPenaltyXpWithheldThisSession=~
   )
   n22 = "To Float (Integer)"(InInt=@n21.TotalXpGained)
   n23 = float + float(A=float / float(A=Get GameTime(), B=60), B=0.01)
   n24 = Get textActualXpPerMin() -> (textActualXpPerMin:Common Text Object Reference)
   > "SetText (Text)"(
    self=@n24
    InText="To Text (Float)"(
     Value=float / float(A=@n22, B=@n23)
     MinimumIntegralDigits=1
     MaximumIntegralDigits=324
     MinimumFractionalDigits=0
     MaximumFractionalDigits=0
    )
   )
   n25 = "To Float (Integer)"(InInt=@n21.TotalXpPotential)
   n26 = Get textPotentialXpPerMin() -> (textPotentialXpPerMin:Common Text Object Reference)
   > "SetText (Text)"(
    self=@n26
    InText="To Text (Float)"(
     Value=float / float(A=@n25, B=@n23)
     MinimumIntegralDigits=1
     MaximumIntegralDigits=324
     MinimumFractionalDigits=0
     MaximumFractionalDigits=0
    )
   )
   > n27 = Branch(Condition="float > float"(A=@n1.Output_Get_BoosterXpMult))
   n28 = Get ProgressBar_XpBoosterMult() -> (ProgressBar_XpBoosterMult:Progress Bar Object Reference)
   > SetPercent(self=@n28, InPercent=@n1.Output_Get_BoosterXpMult)
   n29 = "To Float (Integer)"(InInt=@n1.Output_Get_BoosterBucketXp)
   n30 = Get MinXpThreshold(self=Get Calculator()) -> (MinXpThreshold=0)
   n31 = Get textFieldXpBoosterMult() -> (textFieldXpBoosterMult:Common Text Object Reference)
   > "SetText (Text)"(
    self=@n31
    InText=Format Text!(
     Format="LOCTEXT(\"Min({MaxXpBoosterMult},  {Calc})\")"
     MaxXpBoosterMult:double=@K2Node_Composite_1.CircuitBreaker_MaxXpBoosterMult
     Calc:double=float / float(A=@n29, B:int=@n30)
    ) -> (Result)
   )
   SetPercent(execute=@n27.else, self=@n28, InPercent=0)
   > "SetText (Text)"(self=@n31, InText="LOCTEXT(\"NA\")")
   SetPercent(execute=@n3.else, self=@n4, InPercent=0)
   > "SetText (Text Box)"(self=@n6)
   > SetPercent(self=@n8, InPercent=0)
   > "SetText (Text)"(self=@n9)
   > "SetText (Text)"(self=@n10)
   > SetPercent(self=@n12, InPercent=0)
   > SetPercent(self=@n15, InPercent=0)
   > "SetText (Text)"(self=@n16)
   > SetPercent(self=@n19, InPercent=0)
   > "SetText (Text)"(self=@n20)
   > "SetText (Text)"(self=@n24, InText="LOCTEXT(\"-\")")
   > "SetText (Text)"(self=@n26, InText="LOCTEXT(\"-\")")
   > SetPercent(self=@n28, InPercent=0)
   > "SetText (Text)"(self=@n31)
  ]}
 Player Session Afk Reward Mult {
  path=UpdateProgressBars.K2Node_Composite_0.Player Session Afk Reward Mult
  nodes [
   Inputs!()
   Outputs!(PlayerSession_AfkRewardMult:double=Get() -> (
     PlayerSession_TotalAccoladeXpBudgetNormalized=~
     PlayerSession_SessionStartTime=~
     PlayerSession_SessionEndTime=~
     PlayerSession_LastProgressExportTime=~
     PlayerSession_BankXpReductionTimer=~
     PlayerSession_bInMiddleOfRound=~
     PlayerSession_bHitXpThreshold=~
     PlayerSession_bPrevHitXpThreshold=~
     PlayerSession_bHitExcessXpCutoff=~
     PlayerSession_bPrevHitExcessXpCutoff=~
     PlayerSession_bReportedTimeSpanOutlier=~
     PlayerSession_WeeklyExcessXpWithheldThisSession=~
     PlayerSession_AfkRewardMult=~
     PlayerSession_AfkPenaltyXpWithheldThisSession=~
    ).PlayerSession_AfkRewardMult)
  ]
 }
 CircuiteBreakData {
  path=UpdateProgressBars.K2Node_Composite_1.CircuiteBreakData
  nodes [
   Inputs()
   n1 = Get!() -> (
    CircuitBreaker_TimespanOutlierMult:float=~
    CircuitBreaker_TimespanOutlierDuration:int
    CircuitBreaker_BankXpRemovalMult:float=~
    CircuitBreaker_BankXpRemovalTimeIntervalSecs:float=~
    CircuitBreaker_ExcessXpThresholdMult:float=~
    CircuitBreaker_XpCutoffThresholdMult:float=~
    CircuitBreaker_WeeklyExcessXpThreshold:int
    CircuitBreaker_WeeklyXpCutoffThreshold:int
    CircuitBreaker_XpCutoffRewardMult:float=~
    CircuitBreaker_FrequencyThresholdMult:float=~
    CircuitBreaker_MinXpThresholdMult:float=~
    CircuitBreaker_MaxXpBoosterMult:float=~
    CircuitBreaker_TimeFrameHardCapSettings:Dynamic Xp Time Frame Circuit Breaker Settings Structure
   )
   Outputs!(
    CircuitBreaker_TimespanOutlierDuration:int=@n1.CircuitBreaker_TimespanOutlierDuration
    CircuitBreaker_MaxXpBoosterMult:double=@n1.CircuitBreaker_MaxXpBoosterMult
   )
  ]
 }
 UpdateTriggerTimeOnAccolades {nodes [
   n1 = UpdateTriggerTimeOnAccolades!() -> (then, Force:bool=~) {is=title K2Node_FunctionEntry}
   n2 = Get LastOverallFreqMult() -> (LastOverallFreqMult:double)
   n3 = Get OverallFreqMult() -> (OverallFreqMult:double)
   > Branch(Condition=OR Boolean(A=@n1, B="Not Equal (Float)"(A=@n2, B=@n3)))
   > Set LastOverallFreqMult(LastOverallFreqMult:double=@n3) -> (Output_Get:double)
   > n4 = For Each Loop(Array:Array of Object References=GetDisplayedEntryWidgets!(self:List View Base Object Reference=Get ListView_Accolades()) -> (ReturnValue:Array of User Widget Object References)) -> (Array Element:Object Reference)
   > n5 = Cast To DynamicXpAccoladeEntry!(execute, Object=@n4) -> (then, CastFailed, AsDynamic Xp Accolade Entry:Dynamic Xp Accolade Entry Object Reference) {is=title K2Node_DynamicCast}
   > UpdateSecondsView(self:Dynamic Xp Accolade Entry Object Reference=@n5)
  ]}
 reset {nodes [
   reset() {is=title K2Node_FunctionEntry}
   > "SetText (Text Box)"(self=[Get inputPassword(), Get inputUserName()])
   > SetIsEnabled(self=Get bttnQueryIsland() -> (bttnQueryIsland:TA Generic 01 Button Object Reference), bInIsEnabled=false)
  ]}
 validateButtonState {nodes [
   validateButtonState() {is=title K2Node_FunctionEntry}
   n1 = IsEmpty(InString=GetSelectedOption(self=Get ComboBoxIslandCodes()))
   n2 = TextIsEmpty(InText="GetText (Text Box)"(self=Get inputPassword()))
   n3 = TextIsEmpty(InText="GetText (Text Box)"(self=Get inputUserName()))
   > SetIsEnabled(
    self=Get bttnQueryIsland()
    bInIsEnabled=AND Boolean!(A=NOT Boolean(A=@n3), B=NOT Boolean(A=@n2), C:bool=NOT Boolean(A=@n1)) -> (ReturnValue)
   )
   > SetIsEnabled(
    self=Get btnPopulateData() -> (btnPopulateData:TA Generic 01 Button Object Reference)
    bInIsEnabled=IsValid(Object=Get Design Settings Data Asset())
   )
   > SetIsEnabled(self=Get bttnAddIslandCode() -> (bttnAddIslandCode:TA Generic 01 Button Object Reference), bInIsEnabled=true)
   > n4 = GetConsoleVariableBoolValue!(execute, VariableName:String=DynamicXp.ForceLive) -> (then, ReturnValue:bool)
   > n5 = Branch(Condition=@n4)
   n6 = Get CheckBox_ForcePass() -> (CheckBox_ForcePass:Check Box Object Reference)
   > n7 = SetCheckedState!(execute, self:Check Box Object Reference=@n6, InCheckedState:ECheckBoxState Enum=Checked) -> (then)
   n8 = Get intputIslandCode()
   n9 = Branch(execute=[@n11.then, @n7.then], Condition=TextIsEmpty(InText="GetText (Text Box)"(self=@n8)))
   n10 = Get bttnAddIslandCode()
   > SetVisibility(self=@n10, InVisibility=SelfHitTestInvisible)
   SetVisibility(execute=@n9.else, self=@n10, InVisibility=Visible)
   n11 = SetCheckedState(execute=@n5.else, self=@n6, InCheckedState=Unchecked)
   n12 = Get intputIslandCode()
   NOT Boolean(A=TextIsEmpty(InText="GetText (Text Box)"(self=@n12)))
  ]}
]
variables [
 AccoladeClickTime {type=number, default=10}
 AccoladesMap {type="map<DynamicXpAccolade>", default {}}
 AfkPenalty {type=DynamicXpAfkPenalty, default {afkMinRewardMult=0, afkThresholdSecs=0}}
 AfkTimer {type=number}
 AllAccolades {type="array<string>", default []}
 Args {type="map<string>", default {}}
 Calculator {type=FortDynamicXpCalculator, default=null}
 CappedSessionLenMinutes {type=number, default=60}
 CircuitBreaker {
  type=DynamicXpCircuitBreaker
  default {
   bankXpRemovalMult=0.1
   bankXpRemovalTimeIntervalSecs=60
   excessXpThresholdMult=180
   frequencyThresholdMult=20
   maxXpBoosterMult=1
   minXpThresholdMult=0.5
   timespanOutlierDuration=10
   timespanOutlierMult=2
   weeklyExcessXpThreshold=0
   weeklyXpCutoffThreshold=0
   xpCutoffRewardMult=0.2
   xpCutoffThresholdMult=360
   timeFrameHardCapSettings {bucketDurationInSeconds=60, budgetMultiplier=4, timeFrameBuckets=10}
  }
 }
 ClickTime {type=number, default=1}
 CurrentWeek {type=integer, default=1}
 DefaultSessionStart {
  type=DynamicXpPlayerSession
  default {
   accoladeTriggersWhileAfk=0
   afkPenaltyXpWithheldThisSession=0
   afkRewardMult=0
   bHitExcessXpCutoff=false
   bHitXpThreshold=false
   bInMiddleOfRound=false
   bPrevHitExcessXpCutoff=false
   bPrevHitXpThreshold=false
   bReportedTimeSpanOutlier=false
   bankXpReductionTimer=0
   endMatchPersistence=null
   lastProgressExportTime=0
   lastVoteGuid=#4
   numTimesBoosted=0
   sessionEndTime=0
   sessionStartTime=0
   timeFrameHardCapBreachCount=0
   timeFrameHardCapXpWithheld=0
   timeSpanOutliers=0
   totalAccoladeXpBudgetNormalized=0
   totalXpGained=0
   totalXpPotential=0
   weeklyExcessXpWithheldThisSession=0
   lastAccoladeCalculationRecord {
    afkMult=0
    afkXpWithheld=0
    bankXpMult=0
    bankXpWithheld=0
    baseXp=0
    boostedXp=0
    boosterXpMult=0
    categoryWeightMult=1
    currentTime=0
    currentWeek=0
    finalXpReward=0
    potentialXp=0
    timeFrameHardCapXpWithheld=0
    uncappedXpReward=0
    varianceMult=0
    weeklyCapMult=0
    weeklyCappedXpWithheld=0
   }
   queuedAccolade []
   timeFrameHardCapBreachTimes []
  }
 }
 Design Settings Data Asset {type=CreativeDynamicXpDataAsset, default=null}
 DetailsViewObj {type=DetailsView, default=null}
 FrameTimeMult {type=number, default=1}
 GameIsRunning {type=boolean}
 GameTime {type=number}
 GameTimeDisplay {type=Timespan, default=+00000000.00:00:00.000000000}
 GroupSeperator {type=string, default=" | "}
 IslandTitle {type=string, default=""}
 LastFullXpMult {type=number}
 LastGameTime {type=number}
 LastOverallFreqMult {type=number, default=1}
 LastTimeSeconds {type=number}
 McpProfileStats {
  type=DynamicXpProfileStats
  default {bankXp=0, bankXpMult=1, boosterBucketXp=0, boosterXpMult=1, bucketXp=0, currentWeek=-1, currentWeekXp=0, timespan=-1, weeklyExcessXpMult=1}
 }
 NowTimeSeconds {type=number}
 OverallFreqMult {type=number, default=1}
 PlayerSession {
  type=DynamicXpPlayerSession
  default {
   accoladeTriggersWhileAfk=0
   afkPenaltyXpWithheldThisSession=0
   afkRewardMult=0
   bHitExcessXpCutoff=false
   bHitXpThreshold=false
   bInMiddleOfRound=false
   bPrevHitExcessXpCutoff=false
   bPrevHitXpThreshold=false
   bReportedTimeSpanOutlier=false
   bankXpReductionTimer=0
   endMatchPersistence=null
   lastProgressExportTime=0
   lastVoteGuid=#4
   numTimesBoosted=0
   sessionEndTime=0
   sessionStartTime=0
   timeFrameHardCapBreachCount=0
   timeFrameHardCapXpWithheld=0
   timeSpanOutliers=0
   totalAccoladeXpBudgetNormalized=0
   totalXpGained=0
   totalXpPotential=0
   weeklyExcessXpWithheldThisSession=0
   lastAccoladeCalculationRecord {
    afkMult=0
    afkXpWithheld=0
    bankXpMult=0
    bankXpWithheld=0
    baseXp=0
    boostedXp=0
    boosterXpMult=0
    categoryWeightMult=1
    currentTime=0
    currentWeek=0
    finalXpReward=0
    potentialXp=0
    timeFrameHardCapXpWithheld=0
    uncappedXpReward=0
    varianceMult=0
    weeklyCapMult=0
    weeklyCappedXpWithheld=0
   }
   queuedAccolade []
   timeFrameHardCapBreachTimes []
  }
 }
 PlayerSessionStart {
  type=DynamicXpPlayerSession
  default {
   accoladeTriggersWhileAfk=0
   afkPenaltyXpWithheldThisSession=0
   afkRewardMult=0
   bHitExcessXpCutoff=false
   bHitXpThreshold=false
   bInMiddleOfRound=false
   bPrevHitExcessXpCutoff=false
   bPrevHitXpThreshold=false
   bReportedTimeSpanOutlier=false
   bankXpReductionTimer=0
   endMatchPersistence=null
   lastProgressExportTime=0
   lastVoteGuid=#4
   numTimesBoosted=0
   sessionEndTime=0
   sessionStartTime=0
   timeFrameHardCapBreachCount=0
   timeFrameHardCapXpWithheld=0
   timeSpanOutliers=0
   totalAccoladeXpBudgetNormalized=0
   totalXpGained=0
   totalXpPotential=0
   weeklyExcessXpWithheldThisSession=0
   lastAccoladeCalculationRecord {
    afkMult=0
    afkXpWithheld=0
    bankXpMult=0
    bankXpWithheld=0
    baseXp=0
    boostedXp=0
    boosterXpMult=0
    categoryWeightMult=1
    currentTime=0
    currentWeek=0
    finalXpReward=0
    potentialXp=0
    timeFrameHardCapXpWithheld=0
    uncappedXpReward=0
    varianceMult=0
    weeklyCapMult=0
    weeklyCappedXpWithheld=0
   }
   queuedAccolade []
   timeFrameHardCapBreachTimes []
  }
 }
 Publish Data {type="array<AccoladePublishData>", default []}
 PyMessage {type=string, default=""}
 ReloadPythonIfNeeded {type=boolean, default=true}
 SaveGame {type=DynamicXpSave_C, default=null}
 SaveGameSlotName {type=string, default=DynamicXpCalc}
 ScriptPath {type=string, default=""}
 SeccondsPassThisFrame {type=number}
 ShowAccoladeTriggers {type=boolean, default=true}
 StepTime {type=number, default=10}
 TempNumGroups {type=integer}
 TriggerCount {type=integer}
 UseLinksCalibrationResult {type=boolean, default=true}
 XpBudgetPerMin {type=integer, default=1000}
 XpStats {
  type=DynamicXpAggregatedStats
  default {
   meanSessionLengthMinutes=0
   uniqueGameId=""
   uniqueGameVersion=0
   accolades []
   calibrationResult {ids="", result=None, resultExpiresAt=0001.01.01-00.00.00, xpVarianceMult=0}
   lilAccoladeTPM []
   publishData []
  }
 }
]
widgets {
 info {rootWidgetClass=&VerticalBox, widgetCount=140}
 widgets [
  VerticalBox_0 {widgetClassPath=&VerticalBox, widget=VerticalBox_0}
  HorizontalBoxTop {
   widgetClassPath=&HorizontalBox
   parent=VerticalBox_0
   widget=HorizontalBoxTop
   slot {class=VerticalBoxSlot, properties {size {sizeRule=Fill, value=1}}}
  }
  VerticalBoxLeftControls {
   widgetClassPath=&VerticalBox
   parent=HorizontalBoxTop
   widget=VerticalBoxLeftControls
   slot {class=HorizontalBoxSlot, properties {horizontalAlignment=HAlign_Left}}
  }
  Border_DataAsset {
   widgetClassPath=&Border
   parent=VerticalBoxLeftControls
   widget=Border_DataAsset
   properties {brushColor=<rgba 0.0176 0.0176 0.0176 1>, padding=<ltrb 0 0 0 0>}
   slot {class=VerticalBoxSlot, properties {size {sizeRule=Automatic, value=0.2}}}
  }
  CopyAssetVBox {
   widgetClassPath=&VerticalBox
   parent=Border_DataAsset
   widget=CopyAssetVBox
   slot {class=BorderSlot, properties {padding=<ltrb 0 0 0 0>}}
  }
  HorizontalBox_4 {
   widgetClassPath=&HorizontalBox
   parent=CopyAssetVBox
   widget=HorizontalBox_4
   slot {class=VerticalBoxSlot, properties {size {sizeRule=Fill, value=1}}}
  }
  DesignerDataAssetView {
   widgetClassPath=&SinglePropertyView
   parent=HorizontalBox_4
   widget=DesignerDataAssetView
   properties {propertyName=Design Settings Data Asset}
   slot {class=HorizontalBoxSlot}
  }
  btnPopulateData {
   widgetClassPath=&TA_Generic01_Button
   parent=HorizontalBox_4
   widget=btnPopulateData
   properties {
    label=" Copy"
    toolTipText="Copy: Select Data Asset and copy over to override defaults to use current values. \r\nAfterwards you can edit values in \"Designer Tuning\", Circuit Breaker\", \"Variance\", \"Afk Penalty\" to change settings.  When you later hit the Start button, these are copied to the Calculator."
    buttonColor_Default=<rgba 0 0.6724 1 1>
   }
   slot {class=HorizontalBoxSlot, properties {horizontalAlignment=HAlign_Right, padding=<ltrb 0 0 2 0>, size {sizeRule=Fill, value=1}}}
  }
  Border_76 {
   widgetClassPath=&Border
   parent=CopyAssetVBox
   widget=Border_76
   properties {background {
     bIsDynamicallyLoaded=false
     drawAs=Image
     imageType=NoImage
     mirroring=NoMirror
     resourceName=None
     resourceObject=null
     tiling=NoTile
     imageSize=<32 32>
     margin=<ltrb 0 0 0 0>
     outlineSettings {
      bUseBrushTransparency=false
      roundingType=HalfHeightRadius
      width=0
      color {colorUseRule=UseColor_Specified, specifiedColor=<rgba 0 0 0 0>}
      cornerRadii=<0 0 0 0>
     }
     tintColor {colorUseRule=UseColor_Specified, specifiedColor=<rgba 0.0284 0.0284 0.0284 1>}
     uVRegion {bIsValid=false, max=<0 0>, min=<0 0>}
    }}
   slot {class=VerticalBoxSlot}
  }
  Border_3 {
   widgetClassPath=&Border
   parent=CopyAssetVBox
   widget=Border_3
   properties {brushColor=<rgba 0.0176 0.0176 0.0176 1>, padding=<ltrb 0 0 0 0>}
   slot {class=VerticalBoxSlot, properties {padding=<ltrb 0 2 0 2>}}
  }
  VerticalBox_2 {widgetClassPath=&VerticalBox, parent=Border_3, widget=VerticalBox_2, slot {class=BorderSlot, properties {padding=<ltrb 0 0 0 0>}}}
  HorizontalBox_5 {
   widgetClassPath=&HorizontalBox
   parent=VerticalBox_2
   widget=HorizontalBox_5
   slot {class=VerticalBoxSlot, properties {padding=<ltrb 2 2 2 2>}}
  }
  HorizontalBoxUserName {widgetClassPath=&HorizontalBox, parent=HorizontalBox_5, widget=HorizontalBoxUserName, slot {class=HorizontalBoxSlot}}
  inputUserName {
   widgetClassPath=&EditableTextBox
   bIsVariable=true
   parent=HorizontalBoxUserName
   widget=inputUserName
   properties {
    hintText=Epic User Name
    minimumDesiredWidth=150
    revertTextOnEscape=true
    widgetStyle {
     backgroundColor {colorUseRule=UseColor_Specified, specifiedColor=<rgba 1 1 1 1>}
     backgroundImageFocused {
      bIsDynamicallyLoaded=false
      drawAs=RoundedBox
      imageType=NoImage
      mirroring=NoMirror
      resourceName=None
      resourceObject=null
      tiling=NoTile
      imageSize=<0 0>
      margin=<ltrb 0 0 0 0>
      outlineSettings {
       bUseBrushTransparency=false
       roundingType=FixedRadius
       width=1
       color {colorUseRule=UseColor_ColorTable, specifiedColor=<rgba 0 0 0 0>}
       cornerRadii=<4 4 4 4>
      }
      tintColor {colorUseRule=UseColor_Specified, specifiedColor=<rgba 0.0048 0.0048 0.0048 1>}
      uVRegion {bIsValid=false, max=<0 0>, min=<0 0>}
     }
     backgroundImageHovered {
      bIsDynamicallyLoaded=false
      drawAs=RoundedBox
      imageType=NoImage
      mirroring=NoMirror
      resourceName=None
      resourceObject=null
      tiling=NoTile
      imageSize=<0 0>
      margin=<ltrb 0 0 0 0>
      outlineSettings {
       bUseBrushTransparency=false
       roundingType=FixedRadius
       width=1
       color {colorUseRule=UseColor_ColorTable, specifiedColor=<rgba 0 0 0 0>}
       cornerRadii=<4 4 4 4>
      }
      tintColor {colorUseRule=UseColor_Specified, specifiedColor=<rgba 0.0048 0.0048 0.0048 1>}
      uVRegion {bIsValid=false, max=<0 0>, min=<0 0>}
     }
     backgroundImageNormal {
      bIsDynamicallyLoaded=false
      drawAs=RoundedBox
      imageType=NoImage
      mirroring=NoMirror
      resourceName=None
      resourceObject=null
      tiling=NoTile
      imageSize=<0 0>
      margin=<ltrb 0 0 0 0>
      outlineSettings {
       bUseBrushTransparency=false
       roundingType=FixedRadius
       width=1
       color {colorUseRule=UseColor_ColorTable, specifiedColor=<rgba 0 0 0 0>}
       cornerRadii=<4 4 4 4>
      }
      tintColor {colorUseRule=UseColor_Specified, specifiedColor=<rgba 0.0048 0.0048 0.0048 1>}
      uVRegion {bIsValid=false, max=<0 0>, min=<0 0>}
     }
     backgroundImageReadOnly {
      bIsDynamicallyLoaded=false
      drawAs=RoundedBox
      imageType=NoImage
      mirroring=NoMirror
      resourceName=None
      resourceObject=null
      tiling=NoTile
      imageSize=<0 0>
      margin=<ltrb 0 0 0 0>
      outlineSettings {
       bUseBrushTransparency=false
       roundingType=FixedRadius
       width=1
       color {colorUseRule=UseColor_ColorTable, specifiedColor=<rgba 0 0 0 0>}
       cornerRadii=<4 4 4 4>
      }
      tintColor {colorUseRule=UseColor_Specified, specifiedColor=<rgba 0.0284 0.0284 0.0284 1>}
      uVRegion {bIsValid=false, max=<0 0>, min=<0 0>}
     }
     focusedForegroundColor {colorUseRule=UseColor_Specified, specifiedColor=<rgba 0 0.1998 1 1>}
     foregroundColor {colorUseRule=UseColor_Specified, specifiedColor=<rgba 0.5271 0.5271 0.5271 1>}
     hScrollBarPadding=<ltrb 0 0 0 0>
     padding=<ltrb 8 3 8 4>
     readOnlyForegroundColor {colorUseRule=UseColor_Specified, specifiedColor=<rgba 1 0 1 1>}
     scrollBarStyle {
      thickness=8
      draggedThumbImage {
       bIsDynamicallyLoaded=false
       drawAs=RoundedBox
       imageType=NoImage
       mirroring=NoMirror
       resourceName=None
       resourceObject=null
       tiling=NoTile
       imageSize=<0 0>
       margin=<ltrb 0 0 0 0>
       outlineSettings {
        bUseBrushTransparency=false
        roundingType=FixedRadius
        width=0
        color {colorUseRule=UseColor_Specified, specifiedColor=<rgba 0 0 0 0>}
        cornerRadii=<4 4 4 4>
       }
       tintColor {colorUseRule=UseColor_Specified, specifiedColor=<rgba 0.2159 0.2159 0.2159 1>}
       uVRegion {bIsValid=false, max=<0 0>, min=<0 0>}
      }
      horizontalBackgroundImage {
       bIsDynamicallyLoaded=false
       drawAs=NoDrawType
       imageType=NoImage
       mirroring=NoMirror
       resourceName=None
       resourceObject=null
       tiling=NoTile
       imageSize=<0 0>
       margin=<ltrb 0 0 0 0>
       outlineSettings {
        bUseBrushTransparency=false
        roundingType=HalfHeightRadius
        width=0
        color {colorUseRule=UseColor_Specified, specifiedColor=<rgba 0 0 0 0>}
        cornerRadii=<0 0 0 0>
       }
       tintColor {colorUseRule=UseColor_Specified, specifiedColor=<rgba 1 1 1 1>}
       uVRegion {bIsValid=false, max=<0 0>, min=<0 0>}
      }
      horizontalBottomSlotImage {
       bIsDynamicallyLoaded=false
       drawAs=NoDrawType
       imageType=NoImage
       mirroring=NoMirror
       resourceName=None
       resourceObject=null
       tiling=NoTile
       imageSize=<0 0>
       margin=<ltrb 0 0 0 0>
       outlineSettings {
        bUseBrushTransparency=false
        roundingType=HalfHeightRadius
        width=0
        color {colorUseRule=UseColor_Specified, specifiedColor=<rgba 0 0 0 0>}
        cornerRadii=<0 0 0 0>
       }
       tintColor {colorUseRule=UseColor_Specified, specifiedColor=<rgba 1 1 1 1>}
       uVRegion {bIsValid=false, max=<0 0>, min=<0 0>}
      }
      horizontalTopSlotImage {
       bIsDynamicallyLoaded=false
       drawAs=NoDrawType
       imageType=NoImage
       mirroring=NoMirror
       resourceName=None
       resourceObject=null
       tiling=NoTile
       imageSize=<0 0>
       margin=<ltrb 0 0 0 0>
       outlineSettings {
        bUseBrushTransparency=false
        roundingType=HalfHeightRadius
        width=0
        color {colorUseRule=UseColor_Specified, specifiedColor=<rgba 0 0 0 0>}
        cornerRadii=<0 0 0 0>
       }
       tintColor {colorUseRule=UseColor_Specified, specifiedColor=<rgba 1 1 1 1>}
       uVRegion {bIsValid=false, max=<0 0>, min=<0 0>}
      }
      hoveredThumbImage {
       bIsDynamicallyLoaded=false
       drawAs=RoundedBox
       imageType=NoImage
       mirroring=NoMirror
       resourceName=None
       resourceObject=null
       tiling=NoTile
       imageSize=<0 0>
       margin=<ltrb 0 0 0 0>
       outlineSettings {
        bUseBrushTransparency=false
        roundingType=FixedRadius
        width=0
        color {colorUseRule=UseColor_Specified, specifiedColor=<rgba 0 0 0 0>}
        cornerRadii=<4 4 4 4>
       }
       tintColor {colorUseRule=UseColor_Specified, specifiedColor=<rgba 0.2159 0.2159 0.2159 1>}
       uVRegion {bIsValid=false, max=<0 0>, min=<0 0>}
      }
      normalThumbImage {
       bIsDynamicallyLoaded=false
       drawAs=RoundedBox
       imageType=NoImage
       mirroring=NoMirror
       resourceName=None
       resourceObject=null
       tiling=NoTile
       imageSize=<0 0>
       margin=<ltrb 0 0 0 0>
       outlineSettings {
        bUseBrushTransparency=false
        roundingType=FixedRadius
        width=0
        color {colorUseRule=UseColor_Specified, specifiedColor=<rgba 0 0 0 0>}
        cornerRadii=<4 4 4 4>
       }
       tintColor {colorUseRule=UseColor_Specified, specifiedColor=<rgba 0.0953 0.0953 0.0953 1>}
       uVRegion {bIsValid=false, max=<0 0>, min=<0 0>}
      }
      verticalBackgroundImage {
       bIsDynamicallyLoaded=false
       drawAs=NoDrawType
       imageType=NoImage
       mirroring=NoMirror
       resourceName=None
       resourceObject=null
       tiling=NoTile
       imageSize=<0 0>
       margin=<ltrb 0 0 0 0>
       outlineSettings {
        bUseBrushTransparency=false
        roundingType=HalfHeightRadius
        width=0
        color {colorUseRule=UseColor_Specified, specifiedColor=<rgba 0 0 0 0>}
        cornerRadii=<0 0 0 0>
       }
       tintColor {colorUseRule=UseColor_Specified, specifiedColor=<rgba 1 1 1 1>}
       uVRegion {bIsValid=false, max=<0 0>, min=<0 0>}
      }
      verticalBottomSlotImage {
       bIsDynamicallyLoaded=false
       drawAs=NoDrawType
       imageType=NoImage
       mirroring=NoMirror
       resourceName=None
       resourceObject=null
       tiling=NoTile
       imageSize=<0 0>
       margin=<ltrb 0 0 0 0>
       outlineSettings {
        bUseBrushTransparency=false
        roundingType=HalfHeightRadius
        width=0
        color {colorUseRule=UseColor_Specified, specifiedColor=<rgba 0 0 0 0>}
        cornerRadii=<0 0 0 0>
       }
       tintColor {colorUseRule=UseColor_Specified, specifiedColor=<rgba 1 1 1 1>}
       uVRegion {bIsValid=false, max=<0 0>, min=<0 0>}
      }
      verticalTopSlotImage {
       bIsDynamicallyLoaded=false
       drawAs=NoDrawType
       imageType=NoImage
       mirroring=NoMirror
       resourceName=None
       resourceObject=null
       tiling=NoTile
       imageSize=<0 0>
       margin=<ltrb 0 0 0 0>
       outlineSettings {
        bUseBrushTransparency=false
        roundingType=HalfHeightRadius
        width=0
        color {colorUseRule=UseColor_Specified, specifiedColor=<rgba 0 0 0 0>}
        cornerRadii=<0 0 0 0>
       }
       tintColor {colorUseRule=UseColor_Specified, specifiedColor=<rgba 1 1 1 1>}
       uVRegion {bIsValid=false, max=<0 0>, min=<0 0>}
      }
     }
     textStyle {
      overflowPolicy=Clip
      transformPolicy=None
      colorAndOpacity {colorUseRule=UseColor_Foreground, specifiedColor=<rgba 1 0 1 1>}
      font {
       bForceMonospaced=false
       bMaterialIsStencil=false
       fontMaterial=null
       fontObject=null
       letterSpacing=0
       monospacedWidth=1
       size=10
       skewAmount=0
       typefaceFontName=Regular
       outlineSettings {
        bApplyOutlineToDropShadows=false
        bMiteredCorners=false
        bSeparateFillAlpha=false
        outlineMaterial=null
        outlineSize=0
        outlineColor=<rgba 0 0 0 1>
       }
      }
      highlightColor {colorUseRule=UseColor_ColorTable, specifiedColor=<rgba 0 0 0 0>}
      highlightShape {
       bIsDynamicallyLoaded=false
       drawAs=Image
       imageType=NoImage
       mirroring=NoMirror
       resourceName=None
       resourceObject=null
       tiling=NoTile
       imageSize=<0 0>
       margin=<ltrb 0 0 0 0>
       outlineSettings {
        bUseBrushTransparency=false
        roundingType=HalfHeightRadius
        width=0
        color {colorUseRule=UseColor_Specified, specifiedColor=<rgba 0 0 0 0>}
        cornerRadii=<0 0 0 0>
       }
       tintColor {colorUseRule=UseColor_ColorTable, specifiedColor=<rgba 0 0 0 0>}
       uVRegion {bIsValid=false, max=<0 0>, min=<0 0>}
      }
      selectedBackgroundColor {colorUseRule=UseColor_ColorTable, specifiedColor=<rgba 0 0 0 0>}
      shadowColorAndOpacity=<rgba 0 0 0 1>
      shadowOffset=<0 0>
      strikeBrush {
       bIsDynamicallyLoaded=false
       drawAs=Image
       imageType=NoImage
       mirroring=NoMirror
       resourceName=None
       resourceObject=null
       tiling=NoTile
       imageSize=<32 32>
       margin=<ltrb 0 0 0 0>
       outlineSettings {
        bUseBrushTransparency=false
        roundingType=HalfHeightRadius
        width=0
        color {colorUseRule=UseColor_Specified, specifiedColor=<rgba 0 0 0 0>}
        cornerRadii=<0 0 0 0>
       }
       tintColor {colorUseRule=UseColor_Specified, specifiedColor=<rgba 1 1 1 1>}
       uVRegion {bIsValid=false, max=<0 0>, min=<0 0>}
      }
      underlineBrush {
       bIsDynamicallyLoaded=false
       drawAs=Image
       imageType=NoImage
       mirroring=NoMirror
       resourceName=None
       resourceObject=null
       tiling=NoTile
       imageSize=<32 32>
       margin=<ltrb 0 0 0 0>
       outlineSettings {
        bUseBrushTransparency=false
        roundingType=HalfHeightRadius
        width=0
        color {colorUseRule=UseColor_Specified, specifiedColor=<rgba 0 0 0 0>}
        cornerRadii=<0 0 0 0>
       }
       tintColor {colorUseRule=UseColor_Specified, specifiedColor=<rgba 1 1 1 1>}
       uVRegion {bIsValid=false, max=<0 0>, min=<0 0>}
      }
     }
     vScrollBarPadding=<ltrb 0 0 0 0>
    }
   }
   slot {class=HorizontalBoxSlot, properties {verticalAlignment=VAlign_Center}}
  }
  inputPassword {
   widgetClassPath=&EditableTextBox
   bIsVariable=true
   parent=HorizontalBoxUserName
   widget=inputPassword
   properties {
    hintText=Epic Password
    isPassword=true
    keyboardType=Password
    minimumDesiredWidth=150
    revertTextOnEscape=true
    widgetStyle {
     backgroundColor {colorUseRule=UseColor_Specified, specifiedColor=<rgba 1 1 1 1>}
     backgroundImageFocused {
      bIsDynamicallyLoaded=false
      drawAs=RoundedBox
      imageType=NoImage
      mirroring=NoMirror
      resourceName=None
      resourceObject=null
      tiling=NoTile
      imageSize=<0 0>
      margin=<ltrb 0 0 0 0>
      outlineSettings {
       bUseBrushTransparency=false
       roundingType=FixedRadius
       width=1
       color {colorUseRule=UseColor_ColorTable, specifiedColor=<rgba 0 0 0 0>}
       cornerRadii=<4 4 4 4>
      }
      tintColor {colorUseRule=UseColor_Specified, specifiedColor=<rgba 0.0048 0.0048 0.0048 1>}
      uVRegion {bIsValid=false, max=<0 0>, min=<0 0>}
     }
     backgroundImageHovered {
      bIsDynamicallyLoaded=false
      drawAs=RoundedBox
      imageType=NoImage
      mirroring=NoMirror
      resourceName=None
      resourceObject=null
      tiling=NoTile
      imageSize=<0 0>
      margin=<ltrb 0 0 0 0>
      outlineSettings {
       bUseBrushTransparency=false
       roundingType=FixedRadius
       width=1
       color {colorUseRule=UseColor_ColorTable, specifiedColor=<rgba 0 0 0 0>}
       cornerRadii=<4 4 4 4>
      }
      tintColor {colorUseRule=UseColor_Specified, specifiedColor=<rgba 0.0048 0.0048 0.0048 1>}
      uVRegion {bIsValid=false, max=<0 0>, min=<0 0>}
     }
     backgroundImageNormal {
      bIsDynamicallyLoaded=false
      drawAs=RoundedBox
      imageType=NoImage
      mirroring=NoMirror
      resourceName=None
      resourceObject=null
      tiling=NoTile
      imageSize=<0 0>
      margin=<ltrb 0 0 0 0>
      outlineSettings {
       bUseBrushTransparency=false
       roundingType=FixedRadius
       width=1
       color {colorUseRule=UseColor_ColorTable, specifiedColor=<rgba 0 0 0 0>}
       cornerRadii=<4 4 4 4>
      }
      tintColor {colorUseRule=UseColor_Specified, specifiedColor=<rgba 0.0048 0.0048 0.0048 1>}
      uVRegion {bIsValid=false, max=<0 0>, min=<0 0>}
     }
     backgroundImageReadOnly {
      bIsDynamicallyLoaded=false
      drawAs=RoundedBox
      imageType=NoImage
      mirroring=NoMirror
      resourceName=None
      resourceObject=null
      tiling=NoTile
      imageSize=<0 0>
      margin=<ltrb 0 0 0 0>
      outlineSettings {
       bUseBrushTransparency=false
       roundingType=FixedRadius
       width=1
       color {colorUseRule=UseColor_ColorTable, specifiedColor=<rgba 0 0 0 0>}
       cornerRadii=<4 4 4 4>
      }
      tintColor {colorUseRule=UseColor_Specified, specifiedColor=<rgba 0.0284 0.0284 0.0284 1>}
      uVRegion {bIsValid=false, max=<0 0>, min=<0 0>}
     }
     focusedForegroundColor {colorUseRule=UseColor_Specified, specifiedColor=<rgba 0 0.1998 1 1>}
     foregroundColor {colorUseRule=UseColor_Specified, specifiedColor=<rgba 0.5271 0.5271 0.5271 1>}
     hScrollBarPadding=<ltrb 0 0 0 0>
     padding=<ltrb 0 0 0 0>
     readOnlyForegroundColor {colorUseRule=UseColor_Specified, specifiedColor=<rgba 1 0 1 1>}
     scrollBarStyle {
      thickness=8
      draggedThumbImage {
       bIsDynamicallyLoaded=false
       drawAs=RoundedBox
       imageType=NoImage
       mirroring=NoMirror
       resourceName=None
       resourceObject=null
       tiling=NoTile
       imageSize=<0 0>
       margin=<ltrb 0 0 0 0>
       outlineSettings {
        bUseBrushTransparency=false
        roundingType=FixedRadius
        width=0
        color {colorUseRule=UseColor_Specified, specifiedColor=<rgba 0 0 0 0>}
        cornerRadii=<4 4 4 4>
       }
       tintColor {colorUseRule=UseColor_Specified, specifiedColor=<rgba 0.2159 0.2159 0.2159 1>}
       uVRegion {bIsValid=false, max=<0 0>, min=<0 0>}
      }
      horizontalBackgroundImage {
       bIsDynamicallyLoaded=false
       drawAs=NoDrawType
       imageType=NoImage
       mirroring=NoMirror
       resourceName=None
       resourceObject=null
       tiling=NoTile
       imageSize=<0 0>
       margin=<ltrb 0 0 0 0>
       outlineSettings {
        bUseBrushTransparency=false
        roundingType=HalfHeightRadius
        width=0
        color {colorUseRule=UseColor_Specified, specifiedColor=<rgba 0 0 0 0>}
        cornerRadii=<0 0 0 0>
       }
       tintColor {colorUseRule=UseColor_Specified, specifiedColor=<rgba 1 1 1 1>}
       uVRegion {bIsValid=false, max=<0 0>, min=<0 0>}
      }
      horizontalBottomSlotImage {
       bIsDynamicallyLoaded=false
       drawAs=NoDrawType
       imageType=NoImage
       mirroring=NoMirror
       resourceName=None
       resourceObject=null
       tiling=NoTile
       imageSize=<0 0>
       margin=<ltrb 0 0 0 0>
       outlineSettings {
        bUseBrushTransparency=false
        roundingType=HalfHeightRadius
        width=0
        color {colorUseRule=UseColor_Specified, specifiedColor=<rgba 0 0 0 0>}
        cornerRadii=<0 0 0 0>
       }
       tintColor {colorUseRule=UseColor_Specified, specifiedColor=<rgba 1 1 1 1>}
       uVRegion {bIsValid=false, max=<0 0>, min=<0 0>}
      }
      horizontalTopSlotImage {
       bIsDynamicallyLoaded=false
       drawAs=NoDrawType
       imageType=NoImage
       mirroring=NoMirror
       resourceName=None
       resourceObject=null
       tiling=NoTile
       imageSize=<0 0>
       margin=<ltrb 0 0 0 0>
       outlineSettings {
        bUseBrushTransparency=false
        roundingType=HalfHeightRadius
        width=0
        color {colorUseRule=UseColor_Specified, specifiedColor=<rgba 0 0 0 0>}
        cornerRadii=<0 0 0 0>
       }
       tintColor {colorUseRule=UseColor_Specified, specifiedColor=<rgba 1 1 1 1>}
       uVRegion {bIsValid=false, max=<0 0>, min=<0 0>}
      }
      hoveredThumbImage {
       bIsDynamicallyLoaded=false
       drawAs=RoundedBox
       imageType=NoImage
       mirroring=NoMirror
       resourceName=None
       resourceObject=null
       tiling=NoTile
       imageSize=<0 0>
       margin=<ltrb 0 0 0 0>
       outlineSettings {
        bUseBrushTransparency=false
        roundingType=FixedRadius
        width=0
        color {colorUseRule=UseColor_Specified, specifiedColor=<rgba 0 0 0 0>}
        cornerRadii=<4 4 4 4>
       }
       tintColor {colorUseRule=UseColor_Specified, specifiedColor=<rgba 0.2159 0.2159 0.2159 1>}
       uVRegion {bIsValid=false, max=<0 0>, min=<0 0>}
      }
      normalThumbImage {
       bIsDynamicallyLoaded=false
       drawAs=RoundedBox
       imageType=NoImage
       mirroring=NoMirror
       resourceName=None
       resourceObject=null
       tiling=NoTile
       imageSize=<0 0>
       margin=<ltrb 0 0 0 0>
       outlineSettings {
        bUseBrushTransparency=false
        roundingType=FixedRadius
        width=0
        color {colorUseRule=UseColor_Specified, specifiedColor=<rgba 0 0 0 0>}
        cornerRadii=<4 4 4 4>
       }
       tintColor {colorUseRule=UseColor_Specified, specifiedColor=<rgba 0.0953 0.0953 0.0953 1>}
       uVRegion {bIsValid=false, max=<0 0>, min=<0 0>}
      }
      verticalBackgroundImage {
       bIsDynamicallyLoaded=false
       drawAs=NoDrawType
       imageType=NoImage
       mirroring=NoMirror
       resourceName=None
       resourceObject=null
       tiling=NoTile
       imageSize=<0 0>
       margin=<ltrb 0 0 0 0>
       outlineSettings {
        bUseBrushTransparency=false
        roundingType=HalfHeightRadius
        width=0
        color {colorUseRule=UseColor_Specified, specifiedColor=<rgba 0 0 0 0>}
        cornerRadii=<0 0 0 0>
       }
       tintColor {colorUseRule=UseColor_Specified, specifiedColor=<rgba 1 1 1 1>}
       uVRegion {bIsValid=false, max=<0 0>, min=<0 0>}
      }
      verticalBottomSlotImage {
       bIsDynamicallyLoaded=false
       drawAs=NoDrawType
       imageType=NoImage
       mirroring=NoMirror
       resourceName=None
       resourceObject=null
       tiling=NoTile
       imageSize=<0 0>
       margin=<ltrb 0 0 0 0>
       outlineSettings {
        bUseBrushTransparency=false
        roundingType=HalfHeightRadius
        width=0
        color {colorUseRule=UseColor_Specified, specifiedColor=<rgba 0 0 0 0>}
        cornerRadii=<0 0 0 0>
       }
       tintColor {colorUseRule=UseColor_Specified, specifiedColor=<rgba 1 1 1 1>}
       uVRegion {bIsValid=false, max=<0 0>, min=<0 0>}
      }
      verticalTopSlotImage {
       bIsDynamicallyLoaded=false
       drawAs=NoDrawType
       imageType=NoImage
       mirroring=NoMirror
       resourceName=None
       resourceObject=null
       tiling=NoTile
       imageSize=<0 0>
       margin=<ltrb 0 0 0 0>
       outlineSettings {
        bUseBrushTransparency=false
        roundingType=HalfHeightRadius
        width=0
        color {colorUseRule=UseColor_Specified, specifiedColor=<rgba 0 0 0 0>}
        cornerRadii=<0 0 0 0>
       }
       tintColor {colorUseRule=UseColor_Specified, specifiedColor=<rgba 1 1 1 1>}
       uVRegion {bIsValid=false, max=<0 0>, min=<0 0>}
      }
     }
     textStyle {
      overflowPolicy=Clip
      transformPolicy=None
      colorAndOpacity {colorUseRule=UseColor_Foreground, specifiedColor=<rgba 1 0 1 1>}
      font {
       bForceMonospaced=false
       bMaterialIsStencil=false
       fontMaterial=null
       fontObject=null
       letterSpacing=0
       monospacedWidth=1
       size=10
       skewAmount=0
       typefaceFontName=Regular
       outlineSettings {
        bApplyOutlineToDropShadows=false
        bMiteredCorners=false
        bSeparateFillAlpha=false
        outlineMaterial=null
        outlineSize=0
        outlineColor=<rgba 0 0 0 1>
       }
      }
      highlightColor {colorUseRule=UseColor_ColorTable, specifiedColor=<rgba 0 0 0 0>}
      highlightShape {
       bIsDynamicallyLoaded=false
       drawAs=Image
       imageType=NoImage
       mirroring=NoMirror
       resourceName=None
       resourceObject=null
       tiling=NoTile
       imageSize=<0 0>
       margin=<ltrb 0 0 0 0>
       outlineSettings {
        bUseBrushTransparency=false
        roundingType=HalfHeightRadius
        width=0
        color {colorUseRule=UseColor_Specified, specifiedColor=<rgba 0 0 0 0>}
        cornerRadii=<0 0 0 0>
       }
       tintColor {colorUseRule=UseColor_ColorTable, specifiedColor=<rgba 0 0 0 0>}
       uVRegion {bIsValid=false, max=<0 0>, min=<0 0>}
      }
      selectedBackgroundColor {colorUseRule=UseColor_ColorTable, specifiedColor=<rgba 0 0 0 0>}
      shadowColorAndOpacity=<rgba 0 0 0 1>
      shadowOffset=<0 0>
      strikeBrush {
       bIsDynamicallyLoaded=false
       drawAs=Image
       imageType=NoImage
       mirroring=NoMirror
       resourceName=None
       resourceObject=null
       tiling=NoTile
       imageSize=<32 32>
       margin=<ltrb 0 0 0 0>
       outlineSettings {
        bUseBrushTransparency=false
        roundingType=HalfHeightRadius
        width=0
        color {colorUseRule=UseColor_Specified, specifiedColor=<rgba 0 0 0 0>}
        cornerRadii=<0 0 0 0>
       }
       tintColor {colorUseRule=UseColor_Specified, specifiedColor=<rgba 1 1 1 1>}
       uVRegion {bIsValid=false, max=<0 0>, min=<0 0>}
      }
      underlineBrush {
       bIsDynamicallyLoaded=false
       drawAs=Image
       imageType=NoImage
       mirroring=NoMirror
       resourceName=None
       resourceObject=null
       tiling=NoTile
       imageSize=<32 32>
       margin=<ltrb 0 0 0 0>
       outlineSettings {
        bUseBrushTransparency=false
        roundingType=HalfHeightRadius
        width=0
        color {colorUseRule=UseColor_Specified, specifiedColor=<rgba 0 0 0 0>}
        cornerRadii=<0 0 0 0>
       }
       tintColor {colorUseRule=UseColor_Specified, specifiedColor=<rgba 1 1 1 1>}
       uVRegion {bIsValid=false, max=<0 0>, min=<0 0>}
      }
     }
     vScrollBarPadding=<ltrb 0 0 0 0>
    }
   }
   slot {class=HorizontalBoxSlot, properties {padding=<ltrb 4 0 0 0>}}
  }
  ComboBox_DevOrProd {
   widgetClassPath=/Script/UMG.ComboBoxString
   bIsVariable=true
   parent=HorizontalBoxUserName
   widget=ComboBox_DevOrProd
   properties {
    selectedOption=Prod
    toolTipText="Query Links service on Prod or GameDev envirtonment?  They should be the same if island is from Live or Prod.  If island only in GameDev, then this options needs to be selected or if permissions issue."
    defaultOptions [Prod,GameDev]
    font {
     bForceMonospaced=false
     bMaterialIsStencil=false
     fontMaterial=null
     fontObject=&Roboto
     letterSpacing=0
     monospacedWidth=1
     size=10
     skewAmount=0
     typefaceFontName=Light
     outlineSettings {
      bApplyOutlineToDropShadows=false
      bMiteredCorners=false
      bSeparateFillAlpha=false
      outlineMaterial=null
      outlineSize=0
      outlineColor=<rgba 0 0 0 1>
     }
    }
    foregroundColor {colorUseRule=UseColor_Specified, specifiedColor=<rgba 0.5271 0.5271 0.5271 1>}
    itemStyle {
     bUseParentRowBrush=false
     activeBrush {
      bIsDynamicallyLoaded=false
      drawAs=Image
      imageType=NoImage
      mirroring=NoMirror
      resourceName=None
      resourceObject=null
      tiling=NoTile
      imageSize=<0 0>
      margin=<ltrb 0 0 0 0>
      outlineSettings {
       bUseBrushTransparency=false
       roundingType=HalfHeightRadius
       width=0
       color {colorUseRule=UseColor_Specified, specifiedColor=<rgba 0 0 0 0>}
       cornerRadii=<0 0 0 0>
      }
      tintColor {colorUseRule=UseColor_Specified, specifiedColor=<rgba 0 0.162 0.7454 1>}
      uVRegion {bIsValid=false, max=<0 0>, min=<0 0>}
     }
     activeHighlightedBrush {
      bIsDynamicallyLoaded=false
      drawAs=Image
      imageType=NoImage
      mirroring=NoMirror
      resourceName=None
      resourceObject=null
      tiling=NoTile
      imageSize=<0 0>
      margin=<ltrb 0 0 0 0>
      outlineSettings {
       bUseBrushTransparency=false
       roundingType=HalfHeightRadius
       width=0
       color {colorUseRule=UseColor_Specified, specifiedColor=<rgba 0 0 0 0>}
       cornerRadii=<0 0 0 0>
      }
      tintColor {colorUseRule=UseColor_Specified, specifiedColor=<rgba 0.0252 0.0319 0.0423 1>}
      uVRegion {bIsValid=false, max=<0 0>, min=<0 0>}
     }
     activeHoveredBrush {
      bIsDynamicallyLoaded=false
      drawAs=Image
      imageType=NoImage
      mirroring=NoMirror
      resourceName=None
      resourceObject=null
      tiling=NoTile
      imageSize=<0 0>
      margin=<ltrb 0 0 0 0>
      outlineSettings {
       bUseBrushTransparency=false
       roundingType=HalfHeightRadius
       width=0
       color {colorUseRule=UseColor_Specified, specifiedColor=<rgba 0 0 0 0>}
       cornerRadii=<0 0 0 0>
      }
      tintColor {colorUseRule=UseColor_Specified, specifiedColor=<rgba 0 0.162 0.7454 1>}
      uVRegion {bIsValid=false, max=<0 0>, min=<0 0>}
     }
     dropIndicator_Above {
      bIsDynamicallyLoaded=false
      drawAs=Box
      imageType=FullColor
      mirroring=NoMirror
      resourceName=../../../Engine/Content/Slate/Common/DropZoneIndicator_Above.png
      resourceObject=null
      tiling=NoTile
      imageSize=<0 0>
      margin=<ltrb 0.625 0.625 0 0>
      outlineSettings {
       bUseBrushTransparency=false
       roundingType=HalfHeightRadius
       width=0
       color {colorUseRule=UseColor_Specified, specifiedColor=<rgba 0 0 0 0>}
       cornerRadii=<0 0 0 0>
      }
      tintColor {colorUseRule=UseColor_Specified, specifiedColor=<rgba 0.0091 0.3515 0.9647 1>}
      uVRegion {bIsValid=false, max=<0 0>, min=<0 0>}
     }
     dropIndicator_Below {
      bIsDynamicallyLoaded=false
      drawAs=Box
      imageType=FullColor
      mirroring=NoMirror
      resourceName=../../../Engine/Content/Slate/Common/DropZoneIndicator_Below.png
      resourceObject=null
      tiling=NoTile
      imageSize=<0 0>
      margin=<ltrb 0.625 0 0 0.625>
      outlineSettings {
       bUseBrushTransparency=false
       roundingType=HalfHeightRadius
       width=0
       color {colorUseRule=UseColor_Specified, specifiedColor=<rgba 0 0 0 0>}
       cornerRadii=<0 0 0 0>
      }
      tintColor {colorUseRule=UseColor_Specified, specifiedColor=<rgba 0.0091 0.3515 0.9647 1>}
      uVRegion {bIsValid=false, max=<0 0>, min=<0 0>}
     }
     dropIndicator_Onto {
      bIsDynamicallyLoaded=false
      drawAs=Box
      imageType=FullColor
      mirroring=NoMirror
      resourceName=../../../Engine/Content/Slate/Common/DropZoneIndicator_Onto.png
      resourceObject=null
      tiling=NoTile
      imageSize=<0 0>
      margin=<ltrb 0.25 0.25 0.25 0.25>
      outlineSettings {
       bUseBrushTransparency=false
       roundingType=HalfHeightRadius
       width=0
       color {colorUseRule=UseColor_Specified, specifiedColor=<rgba 0 0 0 0>}
       cornerRadii=<0 0 0 0>
      }
      tintColor {colorUseRule=UseColor_Specified, specifiedColor=<rgba 0.0091 0.3515 0.9647 1>}
      uVRegion {bIsValid=false, max=<0 0>, min=<0 0>}
     }
     evenRowBackgroundBrush {
      bIsDynamicallyLoaded=false
      drawAs=Image
      imageType=NoImage
      mirroring=NoMirror
      resourceName=None
      resourceObject=null
      tiling=NoTile
      imageSize=<0 0>
      margin=<ltrb 0 0 0 0>
      outlineSettings {
       bUseBrushTransparency=false
       roundingType=HalfHeightRadius
       width=0
       color {colorUseRule=UseColor_Specified, specifiedColor=<rgba 0 0 0 0>}
       cornerRadii=<0 0 0 0>
      }
      tintColor {colorUseRule=UseColor_Specified, specifiedColor=<rgba 0.0103 0.0103 0.0103 1>}
      uVRegion {bIsValid=false, max=<0 0>, min=<0 0>}
     }
     evenRowBackgroundHoveredBrush {
      bIsDynamicallyLoaded=false
      drawAs=Image
      imageType=NoImage
      mirroring=NoMirror
      resourceName=None
      resourceObject=null
      tiling=NoTile
      imageSize=<0 0>
      margin=<ltrb 0 0 0 0>
      outlineSettings {
       bUseBrushTransparency=false
       roundingType=HalfHeightRadius
       width=0
       color {colorUseRule=UseColor_Specified, specifiedColor=<rgba 0 0 0 0>}
       cornerRadii=<0 0 0 0>
      }
      tintColor {colorUseRule=UseColor_Specified, specifiedColor=<rgba 0.0176 0.0176 0.0176 1>}
      uVRegion {bIsValid=false, max=<0 0>, min=<0 0>}
     }
     inactiveBrush {
      bIsDynamicallyLoaded=false
      drawAs=Image
      imageType=NoImage
      mirroring=NoMirror
      resourceName=None
      resourceObject=null
      tiling=NoTile
      imageSize=<0 0>
      margin=<ltrb 0 0 0 0>
      outlineSettings {
       bUseBrushTransparency=false
       roundingType=HalfHeightRadius
       width=0
       color {colorUseRule=UseColor_Specified, specifiedColor=<rgba 0 0 0 0>}
       cornerRadii=<0 0 0 0>
      }
      tintColor {colorUseRule=UseColor_Specified, specifiedColor=<rgba 0.0513 0.0953 0.159 1>}
      uVRegion {bIsValid=false, max=<0 0>, min=<0 0>}
     }
     inactiveHighlightedBrush {
      bIsDynamicallyLoaded=false
      drawAs=Image
      imageType=NoImage
      mirroring=NoMirror
      resourceName=None
      resourceObject=null
      tiling=NoTile
      imageSize=<0 0>
      margin=<ltrb 0 0 0 0>
      outlineSettings {
       bUseBrushTransparency=false
       roundingType=HalfHeightRadius
       width=0
       color {colorUseRule=UseColor_Specified, specifiedColor=<rgba 0 0 0 0>}
       cornerRadii=<0 0 0 0>
      }
      tintColor {colorUseRule=UseColor_Specified, specifiedColor=<rgba 0.0252 0.0319 0.0423 1>}
      uVRegion {bIsValid=false, max=<0 0>, min=<0 0>}
     }
     inactiveHoveredBrush {
      bIsDynamicallyLoaded=false
      drawAs=Image
      imageType=NoImage
      mirroring=NoMirror
      resourceName=None
      resourceObject=null
      tiling=NoTile
      imageSize=<0 0>
      margin=<ltrb 0 0 0 0>
      outlineSettings {
       bUseBrushTransparency=false
       roundingType=HalfHeightRadius
       width=0
       color {colorUseRule=UseColor_Specified, specifiedColor=<rgba 0 0 0 0>}
       cornerRadii=<0 0 0 0>
      }
      tintColor {colorUseRule=UseColor_Specified, specifiedColor=<rgba 0.0513 0.0953 0.159 1>}
      uVRegion {bIsValid=false, max=<0 0>, min=<0 0>}
     }
     oddRowBackgroundBrush {
      bIsDynamicallyLoaded=false
      drawAs=Image
      imageType=NoImage
      mirroring=NoMirror
      resourceName=None
      resourceObject=null
      tiling=NoTile
      imageSize=<0 0>
      margin=<ltrb 0 0 0 0>
      outlineSettings {
       bUseBrushTransparency=false
       roundingType=HalfHeightRadius
       width=0
       color {colorUseRule=UseColor_Specified, specifiedColor=<rgba 0 0 0 0>}
       cornerRadii=<0 0 0 0>
      }
      tintColor {colorUseRule=UseColor_Specified, specifiedColor=<rgba 0.0103 0.0103 0.0103 1>}
      uVRegion {bIsValid=false, max=<0 0>, min=<0 0>}
     }
     oddRowBackgroundHoveredBrush {
      bIsDynamicallyLoaded=false
      drawAs=Image
      imageType=NoImage
      mirroring=NoMirror
      resourceName=None
      resourceObject=null
      tiling=NoTile
      imageSize=<0 0>
      margin=<ltrb 0 0 0 0>
      outlineSettings {
       bUseBrushTransparency=false
       roundingType=HalfHeightRadius
       width=0
       color {colorUseRule=UseColor_Specified, specifiedColor=<rgba 0 0 0 0>}
       cornerRadii=<0 0 0 0>
      }
      tintColor {colorUseRule=UseColor_Specified, specifiedColor=<rgba 0.0176 0.0176 0.0176 1>}
      uVRegion {bIsValid=false, max=<0 0>, min=<0 0>}
     }
     parentRowBackgroundBrush {
      bIsDynamicallyLoaded=false
      drawAs=Image
      imageType=NoImage
      mirroring=NoMirror
      resourceName=None
      resourceObject=null
      tiling=NoTile
      imageSize=<32 32>
      margin=<ltrb 0 0 0 0>
      outlineSettings {
       bUseBrushTransparency=false
       roundingType=HalfHeightRadius
       width=0
       color {colorUseRule=UseColor_Specified, specifiedColor=<rgba 0 0 0 0>}
       cornerRadii=<0 0 0 0>
      }
      tintColor {colorUseRule=UseColor_Specified, specifiedColor=<rgba 1 1 1 1>}
      uVRegion {bIsValid=false, max=<0 0>, min=<0 0>}
     }
     parentRowBackgroundHoveredBrush {
      bIsDynamicallyLoaded=false
      drawAs=Image
      imageType=NoImage
      mirroring=NoMirror
      resourceName=None
      resourceObject=null
      tiling=NoTile
      imageSize=<32 32>
      margin=<ltrb 0 0 0 0>
      outlineSettings {
       bUseBrushTransparency=false
       roundingType=HalfHeightRadius
       width=0
       color {colorUseRule=UseColor_Specified, specifiedColor=<rgba 0 0 0 0>}
       cornerRadii=<0 0 0 0>
      }
      tintColor {colorUseRule=UseColor_Specified, specifiedColor=<rgba 1 1 1 1>}
      uVRegion {bIsValid=false, max=<0 0>, min=<0 0>}
     }
     selectedTextColor {colorUseRule=UseColor_Specified, specifiedColor=<rgba 1 1 1 1>}
     selectorFocusedBrush {
      bIsDynamicallyLoaded=false
      drawAs=Border
      imageType=FullColor
      mirroring=NoMirror
      resourceName=../../../Engine/Content/Slate/Common/Selector.png
      resourceObject=null
      tiling=Both
      imageSize=<0 0>
      margin=<ltrb 0.25 0.25 0.25 0.25>
      outlineSettings {
       bUseBrushTransparency=false
       roundingType=HalfHeightRadius
       width=0
       color {colorUseRule=UseColor_Specified, specifiedColor=<rgba 0 0 0 0>}
       cornerRadii=<0 0 0 0>
      }
      tintColor {colorUseRule=UseColor_Specified, specifiedColor=<rgba 0.701 0.225 0.003 1>}
      uVRegion {bIsValid=false, max=<0 0>, min=<0 0>}
     }
     textColor {colorUseRule=UseColor_Specified, specifiedColor=<rgba 0.5271 0.5271 0.5271 1>}
    }
    scrollBarStyle {
     thickness=8
     draggedThumbImage {
      bIsDynamicallyLoaded=false
      drawAs=RoundedBox
      imageType=NoImage
      mirroring=NoMirror
      resourceName=None
      resourceObject=null
      tiling=NoTile
      imageSize=<0 0>
      margin=<ltrb 0 0 0 0>
      outlineSettings {
       bUseBrushTransparency=false
       roundingType=FixedRadius
       width=0
       color {colorUseRule=UseColor_Specified, specifiedColor=<rgba 0 0 0 0>}
       cornerRadii=<4 4 4 4>
      }
      tintColor {colorUseRule=UseColor_Specified, specifiedColor=<rgba 0.2159 0.2159 0.2159 1>}
      uVRegion {bIsValid=false, max=<0 0>, min=<0 0>}
     }
     horizontalBackgroundImage {
      bIsDynamicallyLoaded=false
      drawAs=NoDrawType
      imageType=NoImage
      mirroring=NoMirror
      resourceName=None
      resourceObject=null
      tiling=NoTile
      imageSize=<0 0>
      margin=<ltrb 0 0 0 0>
      outlineSettings {
       bUseBrushTransparency=false
       roundingType=HalfHeightRadius
       width=0
       color {colorUseRule=UseColor_Specified, specifiedColor=<rgba 0 0 0 0>}
       cornerRadii=<0 0 0 0>
      }
      tintColor {colorUseRule=UseColor_Specified, specifiedColor=<rgba 1 1 1 1>}
      uVRegion {bIsValid=false, max=<0 0>, min=<0 0>}
     }
     horizontalBottomSlotImage {
      bIsDynamicallyLoaded=false
      drawAs=NoDrawType
      imageType=NoImage
      mirroring=NoMirror
      resourceName=None
      resourceObject=null
      tiling=NoTile
      imageSize=<0 0>
      margin=<ltrb 0 0 0 0>
      outlineSettings {
       bUseBrushTransparency=false
       roundingType=HalfHeightRadius
       width=0
       color {colorUseRule=UseColor_Specified, specifiedColor=<rgba 0 0 0 0>}
       cornerRadii=<0 0 0 0>
      }
      tintColor {colorUseRule=UseColor_Specified, specifiedColor=<rgba 1 1 1 1>}
      uVRegion {bIsValid=false, max=<0 0>, min=<0 0>}
     }
     horizontalTopSlotImage {
      bIsDynamicallyLoaded=false
      drawAs=NoDrawType
      imageType=NoImage
      mirroring=NoMirror
      resourceName=None
      resourceObject=null
      tiling=NoTile
      imageSize=<0 0>
      margin=<ltrb 0 0 0 0>
      outlineSettings {
       bUseBrushTransparency=false
       roundingType=HalfHeightRadius
       width=0
       color {colorUseRule=UseColor_Specified, specifiedColor=<rgba 0 0 0 0>}
       cornerRadii=<0 0 0 0>
      }
      tintColor {colorUseRule=UseColor_Specified, specifiedColor=<rgba 1 1 1 1>}
      uVRegion {bIsValid=false, max=<0 0>, min=<0 0>}
     }
     hoveredThumbImage {
      bIsDynamicallyLoaded=false
      drawAs=RoundedBox
      imageType=NoImage
      mirroring=NoMirror
      resourceName=None
      resourceObject=null
      tiling=NoTile
      imageSize=<0 0>
      margin=<ltrb 0 0 0 0>
      outlineSettings {
       bUseBrushTransparency=false
       roundingType=FixedRadius
       width=0
       color {colorUseRule=UseColor_Specified, specifiedColor=<rgba 0 0 0 0>}
       cornerRadii=<4 4 4 4>
      }
      tintColor {colorUseRule=UseColor_Specified, specifiedColor=<rgba 0.2159 0.2159 0.2159 1>}
      uVRegion {bIsValid=false, max=<0 0>, min=<0 0>}
     }
     normalThumbImage {
      bIsDynamicallyLoaded=false
      drawAs=RoundedBox
      imageType=NoImage
      mirroring=NoMirror
      resourceName=None
      resourceObject=null
      tiling=NoTile
      imageSize=<0 0>
      margin=<ltrb 0 0 0 0>
      outlineSettings {
       bUseBrushTransparency=false
       roundingType=FixedRadius
       width=0
       color {colorUseRule=UseColor_Specified, specifiedColor=<rgba 0 0 0 0>}
       cornerRadii=<4 4 4 4>
      }
      tintColor {colorUseRule=UseColor_Specified, specifiedColor=<rgba 0.0953 0.0953 0.0953 1>}
      uVRegion {bIsValid=false, max=<0 0>, min=<0 0>}
     }
     verticalBackgroundImage {
      bIsDynamicallyLoaded=false
      drawAs=NoDrawType
      imageType=NoImage
      mirroring=NoMirror
      resourceName=None
      resourceObject=null
      tiling=NoTile
      imageSize=<0 0>
      margin=<ltrb 0 0 0 0>
      outlineSettings {
       bUseBrushTransparency=false
       roundingType=HalfHeightRadius
       width=0
       color {colorUseRule=UseColor_Specified, specifiedColor=<rgba 0 0 0 0>}
       cornerRadii=<0 0 0 0>
      }
      tintColor {colorUseRule=UseColor_Specified, specifiedColor=<rgba 1 1 1 1>}
      uVRegion {bIsValid=false, max=<0 0>, min=<0 0>}
     }
     verticalBottomSlotImage {
      bIsDynamicallyLoaded=false
      drawAs=NoDrawType
      imageType=NoImage
      mirroring=NoMirror
      resourceName=None
      resourceObject=null
      tiling=NoTile
      imageSize=<0 0>
      margin=<ltrb 0 0 0 0>
      outlineSettings {
       bUseBrushTransparency=false
       roundingType=HalfHeightRadius
       width=0
       color {colorUseRule=UseColor_Specified, specifiedColor=<rgba 0 0 0 0>}
       cornerRadii=<0 0 0 0>
      }
      tintColor {colorUseRule=UseColor_Specified, specifiedColor=<rgba 1 1 1 1>}
      uVRegion {bIsValid=false, max=<0 0>, min=<0 0>}
     }
     verticalTopSlotImage {
      bIsDynamicallyLoaded=false
      drawAs=NoDrawType
      imageType=NoImage
      mirroring=NoMirror
      resourceName=None
      resourceObject=null
      tiling=NoTile
      imageSize=<0 0>
      margin=<ltrb 0 0 0 0>
      outlineSettings {
       bUseBrushTransparency=false
       roundingType=HalfHeightRadius
       width=0
       color {colorUseRule=UseColor_Specified, specifiedColor=<rgba 0 0 0 0>}
       cornerRadii=<0 0 0 0>
      }
      tintColor {colorUseRule=UseColor_Specified, specifiedColor=<rgba 1 1 1 1>}
      uVRegion {bIsValid=false, max=<0 0>, min=<0 0>}
     }
    }
    widgetStyle {
     comboButtonStyle {
      downArrowAlign=VAlign_Center
      buttonStyle {
       clickedSlateSound {resourceObject=null}
       disabled {
        bIsDynamicallyLoaded=false
        drawAs=RoundedBox
        imageType=NoImage
        mirroring=NoMirror
        resourceName=None
        resourceObject=null
        tiling=NoTile
        imageSize=<0 0>
        margin=<ltrb 0 0 0 0>
        outlineSettings {
         bUseBrushTransparency=false
         roundingType=FixedRadius
         width=1
         color {colorUseRule=UseColor_ColorTable, specifiedColor=<rgba 0 0 0 0>}
         cornerRadii=<4 4 4 4>
        }
        tintColor {colorUseRule=UseColor_Specified, specifiedColor=<rgba 0.0395 0.0395 0.0395 1>}
        uVRegion {bIsValid=false, max=<0 0>, min=<0 0>}
       }
       disabledForeground {colorUseRule=UseColor_ColorTable, specifiedColor=<rgba 0 0 0 0>}
       hovered {
        bIsDynamicallyLoaded=false
        drawAs=RoundedBox
        imageType=NoImage
        mirroring=NoMirror
        resourceName=None
        resourceObject=null
        tiling=NoTile
        imageSize=<0 0>
        margin=<ltrb 0 0 0 0>
        outlineSettings {
         bUseBrushTransparency=false
         roundingType=FixedRadius
         width=1
         color {colorUseRule=UseColor_ColorTable, specifiedColor=<rgba 0 0 0 0>}
         cornerRadii=<4 4 4 4>
        }
        tintColor {colorUseRule=UseColor_Specified, specifiedColor=<rgba 0.0048 0.0048 0.0048 1>}
        uVRegion {bIsValid=false, max=<0 0>, min=<0 0>}
       }
       hoveredForeground {colorUseRule=UseColor_ColorTable, specifiedColor=<rgba 0 0 0 0>}
       hoveredSlateSound {resourceObject=null}
       normal {
        bIsDynamicallyLoaded=false
        drawAs=RoundedBox
        imageType=NoImage
        mirroring=NoMirror
        resourceName=None
        resourceObject=null
        tiling=NoTile
        imageSize=<0 0>
        margin=<ltrb 0 0 0 0>
        outlineSettings {
         bUseBrushTransparency=false
         roundingType=FixedRadius
         width=1
         color {colorUseRule=UseColor_ColorTable, specifiedColor=<rgba 0 0 0 0>}
         cornerRadii=<4 4 4 4>
        }
        tintColor {colorUseRule=UseColor_Specified, specifiedColor=<rgba 0.0048 0.0048 0.0048 1>}
        uVRegion {bIsValid=false, max=<0 0>, min=<0 0>}
       }
       normalForeground {colorUseRule=UseColor_ColorTable, specifiedColor=<rgba 0 0 0 0>}
       normalPadding=<ltrb 8 1 8 1>
       pressed {
        bIsDynamicallyLoaded=false
        drawAs=RoundedBox
        imageType=NoImage
        mirroring=NoMirror
        resourceName=None
        resourceObject=null
        tiling=NoTile
        imageSize=<0 0>
        margin=<ltrb 0 0 0 0>
        outlineSettings {
         bUseBrushTransparency=false
         roundingType=FixedRadius
         width=1
         color {colorUseRule=UseColor_ColorTable, specifiedColor=<rgba 0 0 0 0>}
         cornerRadii=<4 4 4 4>
        }
        tintColor {colorUseRule=UseColor_Specified, specifiedColor=<rgba 0.0048 0.0048 0.0048 1>}
        uVRegion {bIsValid=false, max=<0 0>, min=<0 0>}
       }
       pressedForeground {colorUseRule=UseColor_ColorTable, specifiedColor=<rgba 0 0 0 0>}
       pressedPadding=<ltrb 8 2 8 0>
       pressedSlateSound {resourceObject=null}
      }
      contentPadding=<ltrb 0 0 0 0>
      downArrowImage {
       bIsDynamicallyLoaded=false
       drawAs=Image
       imageType=Vector
       mirroring=NoMirror
       resourceName=../../../Engine/Content/Slate/Starship/CoreWidgets/ComboBox/wide-chevron-down.svg
       resourceObject=null
       tiling=NoTile
       imageSize=<16 16>
       margin=<ltrb 0 0 0 0>
       outlineSettings {
        bUseBrushTransparency=false
        roundingType=HalfHeightRadius
        width=0
        color {colorUseRule=UseColor_Specified, specifiedColor=<rgba 0 0 0 0>}
        cornerRadii=<0 0 0 0>
       }
       tintColor {colorUseRule=UseColor_Specified, specifiedColor=<rgba 1 1 1 1>}
       uVRegion {bIsValid=false, max=<0 0>, min=<0 0>}
      }
      downArrowPadding=<ltrb 2 2 2 2>
      menuBorderBrush {
       bIsDynamicallyLoaded=false
       drawAs=RoundedBox
       imageType=NoImage
       mirroring=NoMirror
       resourceName=None
       resourceObject=null
       tiling=NoTile
       imageSize=<0 0>
       margin=<ltrb 0 0 0 0>
       outlineSettings {
        bUseBrushTransparency=false
        roundingType=FixedRadius
        width=1
        color {colorUseRule=UseColor_ColorTable, specifiedColor=<rgba 0 0 0 0>}
        cornerRadii=<0 0 0 0>
       }
       tintColor {colorUseRule=UseColor_Specified, specifiedColor=<rgba 0.0395 0.0395 0.0395 1>}
       uVRegion {bIsValid=false, max=<0 0>, min=<0 0>}
      }
      menuBorderPadding=<ltrb 0 0 0 0>
      shadowColorAndOpacity=<rgba 0 0 0 1>
      shadowOffset=<0 0>
     }
     contentPadding=<ltrb 0 0 0 0>
     menuRowPadding=<ltrb 8 3 8 3>
     pressedSlateSound {resourceObject=null}
     selectionChangeSlateSound {resourceObject=null}
    }
   }
   slot {class=HorizontalBoxSlot, properties {padding=<ltrb 4 0 0 0>}}
  }
  HorizontalBox_7 {
   widgetClassPath=&HorizontalBox
   parent=VerticalBox_2
   widget=HorizontalBox_7
   slot {class=VerticalBoxSlot, properties {padding=<ltrb 2 2 2 2>}}
  }
  textFieldIslandNamePrompt_1 {
   widgetClassPath=&CommonTextBlock
   parent=HorizontalBox_7
   widget=textFieldIslandNamePrompt_1
   properties {
    text=Island Code:
    font {
     bForceMonospaced=false
     bMaterialIsStencil=false
     fontMaterial=null
     fontObject=&NotoSans
     letterSpacing=0
     monospacedWidth=1
     size=10
     skewAmount=0
     typefaceFontName=NotoSans-Regular
     outlineSettings {
      bApplyOutlineToDropShadows=false
      bMiteredCorners=false
      bSeparateFillAlpha=false
      outlineMaterial=null
      outlineSize=0
      outlineColor=<rgba 0 0 0 1>
     }
    }
    shadowOffset=<0 0>
   }
   slot {class=HorizontalBoxSlot, properties {padding=<ltrb 5 3 3 3>}}
  }
  SizeBox_2 {
   widgetClassPath=&SizeBox
   parent=HorizontalBox_7
   widget=SizeBox_2
   properties {bOverride_MinDesiredWidth=true, minDesiredWidth=150}
   slot {class=HorizontalBoxSlot, properties {horizontalAlignment=HAlign_Right, padding=<ltrb 4 0 4 0>}}
  }
  SwitcherAddIslandCode {widgetClassPath=&WidgetSwitcher, bIsVariable=true, parent=SizeBox_2, widget=SwitcherAddIslandCode, slot {class=SizeBoxSlot}}
  ComboBoxIslandCodes {
   widgetClassPath=&ComboBoxString
   bIsVariable=true
   parent=SwitcherAddIslandCode
   widget=ComboBoxIslandCodes
   properties {
    contentPadding=<ltrb 0 0 0 0>
    font {
     bForceMonospaced=false
     bMaterialIsStencil=false
     fontMaterial=null
     fontObject=&Roboto
     letterSpacing=0
     monospacedWidth=1
     size=10
     skewAmount=0
     typefaceFontName=Bold
     outlineSettings {
      bApplyOutlineToDropShadows=false
      bMiteredCorners=false
      bSeparateFillAlpha=false
      outlineMaterial=null
      outlineSize=0
      outlineColor=<rgba 0 0 0 1>
     }
    }
    foregroundColor {colorUseRule=UseColor_Specified, specifiedColor=<rgba 0.5271 0.5271 0.5271 1>}
    itemStyle {
     bUseParentRowBrush=false
     activeBrush {
      bIsDynamicallyLoaded=false
      drawAs=Image
      imageType=NoImage
      mirroring=NoMirror
      resourceName=None
      resourceObject=null
      tiling=NoTile
      imageSize=<0 0>
      margin=<ltrb 0 0 0 0>
      outlineSettings {
       bUseBrushTransparency=false
       roundingType=HalfHeightRadius
       width=0
       color {colorUseRule=UseColor_Specified, specifiedColor=<rgba 0 0 0 0>}
       cornerRadii=<0 0 0 0>
      }
      tintColor {colorUseRule=UseColor_Specified, specifiedColor=<rgba 0 0.162 0.7454 1>}
      uVRegion {bIsValid=false, max=<0 0>, min=<0 0>}
     }
     activeHighlightedBrush {
      bIsDynamicallyLoaded=false
      drawAs=Image
      imageType=NoImage
      mirroring=NoMirror
      resourceName=None
      resourceObject=null
      tiling=NoTile
      imageSize=<0 0>
      margin=<ltrb 0 0 0 0>
      outlineSettings {
       bUseBrushTransparency=false
       roundingType=HalfHeightRadius
       width=0
       color {colorUseRule=UseColor_Specified, specifiedColor=<rgba 0 0 0 0>}
       cornerRadii=<0 0 0 0>
      }
      tintColor {colorUseRule=UseColor_Specified, specifiedColor=<rgba 0.0252 0.0319 0.0423 1>}
      uVRegion {bIsValid=false, max=<0 0>, min=<0 0>}
     }
     activeHoveredBrush {
      bIsDynamicallyLoaded=false
      drawAs=Image
      imageType=NoImage
      mirroring=NoMirror
      resourceName=None
      resourceObject=null
      tiling=NoTile
      imageSize=<0 0>
      margin=<ltrb 0 0 0 0>
      outlineSettings {
       bUseBrushTransparency=false
       roundingType=HalfHeightRadius
       width=0
       color {colorUseRule=UseColor_Specified, specifiedColor=<rgba 0 0 0 0>}
       cornerRadii=<0 0 0 0>
      }
      tintColor {colorUseRule=UseColor_Specified, specifiedColor=<rgba 0 0.162 0.7454 1>}
      uVRegion {bIsValid=false, max=<0 0>, min=<0 0>}
     }
     dropIndicator_Above {
      bIsDynamicallyLoaded=false
      drawAs=Box
      imageType=FullColor
      mirroring=NoMirror
      resourceName=../../../Engine/Content/Slate/Common/DropZoneIndicator_Above.png
      resourceObject=null
      tiling=NoTile
      imageSize=<0 0>
      margin=<ltrb 0.625 0.625 0 0>
      outlineSettings {
       bUseBrushTransparency=false
       roundingType=HalfHeightRadius
       width=0
       color {colorUseRule=UseColor_Specified, specifiedColor=<rgba 0 0 0 0>}
       cornerRadii=<0 0 0 0>
      }
      tintColor {colorUseRule=UseColor_Specified, specifiedColor=<rgba 0.0091 0.3515 0.9647 1>}
      uVRegion {bIsValid=false, max=<0 0>, min=<0 0>}
     }
     dropIndicator_Below {
      bIsDynamicallyLoaded=false
      drawAs=Box
      imageType=FullColor
      mirroring=NoMirror
      resourceName=../../../Engine/Content/Slate/Common/DropZoneIndicator_Below.png
      resourceObject=null
      tiling=NoTile
      imageSize=<0 0>
      margin=<ltrb 0.625 0 0 0.625>
      outlineSettings {
       bUseBrushTransparency=false
       roundingType=HalfHeightRadius
       width=0
       color {colorUseRule=UseColor_Specified, specifiedColor=<rgba 0 0 0 0>}
       cornerRadii=<0 0 0 0>
      }
      tintColor {colorUseRule=UseColor_Specified, specifiedColor=<rgba 0.0091 0.3515 0.9647 1>}
      uVRegion {bIsValid=false, max=<0 0>, min=<0 0>}
     }
     dropIndicator_Onto {
      bIsDynamicallyLoaded=false
      drawAs=Box
      imageType=FullColor
      mirroring=NoMirror
      resourceName=../../../Engine/Content/Slate/Common/DropZoneIndicator_Onto.png
      resourceObject=null
      tiling=NoTile
      imageSize=<0 0>
      margin=<ltrb 0.25 0.25 0.25 0.25>
      outlineSettings {
       bUseBrushTransparency=false
       roundingType=HalfHeightRadius
       width=0
       color {colorUseRule=UseColor_Specified, specifiedColor=<rgba 0 0 0 0>}
       cornerRadii=<0 0 0 0>
      }
      tintColor {colorUseRule=UseColor_Specified, specifiedColor=<rgba 0.0091 0.3515 0.9647 1>}
      uVRegion {bIsValid=false, max=<0 0>, min=<0 0>}
     }
     evenRowBackgroundBrush {
      bIsDynamicallyLoaded=false
      drawAs=Image
      imageType=NoImage
      mirroring=NoMirror
      resourceName=None
      resourceObject=null
      tiling=NoTile
      imageSize=<0 0>
      margin=<ltrb 0 0 0 0>
      outlineSettings {
       bUseBrushTransparency=false
       roundingType=HalfHeightRadius
       width=0
       color {colorUseRule=UseColor_Specified, specifiedColor=<rgba 0 0 0 0>}
       cornerRadii=<0 0 0 0>
      }
      tintColor {colorUseRule=UseColor_Specified, specifiedColor=<rgba 0.0103 0.0103 0.0103 1>}
      uVRegion {bIsValid=false, max=<0 0>, min=<0 0>}
     }
     evenRowBackgroundHoveredBrush {
      bIsDynamicallyLoaded=false
      drawAs=Image
      imageType=NoImage
      mirroring=NoMirror
      resourceName=None
      resourceObject=null
      tiling=NoTile
      imageSize=<0 0>
      margin=<ltrb 0 0 0 0>
      outlineSettings {
       bUseBrushTransparency=false
       roundingType=HalfHeightRadius
       width=0
       color {colorUseRule=UseColor_Specified, specifiedColor=<rgba 0 0 0 0>}
       cornerRadii=<0 0 0 0>
      }
      tintColor {colorUseRule=UseColor_Specified, specifiedColor=<rgba 0.0176 0.0176 0.0176 1>}
      uVRegion {bIsValid=false, max=<0 0>, min=<0 0>}
     }
     inactiveBrush {
      bIsDynamicallyLoaded=false
      drawAs=Image
      imageType=NoImage
      mirroring=NoMirror
      resourceName=None
      resourceObject=null
      tiling=NoTile
      imageSize=<0 0>
      margin=<ltrb 0 0 0 0>
      outlineSettings {
       bUseBrushTransparency=false
       roundingType=HalfHeightRadius
       width=0
       color {colorUseRule=UseColor_Specified, specifiedColor=<rgba 0 0 0 0>}
       cornerRadii=<0 0 0 0>
      }
      tintColor {colorUseRule=UseColor_Specified, specifiedColor=<rgba 0.0513 0.0953 0.159 1>}
      uVRegion {bIsValid=false, max=<0 0>, min=<0 0>}
     }
     inactiveHighlightedBrush {
      bIsDynamicallyLoaded=false
      drawAs=Image
      imageType=NoImage
      mirroring=NoMirror
      resourceName=None
      resourceObject=null
      tiling=NoTile
      imageSize=<0 0>
      margin=<ltrb 0 0 0 0>
      outlineSettings {
       bUseBrushTransparency=false
       roundingType=HalfHeightRadius
       width=0
       color {colorUseRule=UseColor_Specified, specifiedColor=<rgba 0 0 0 0>}
       cornerRadii=<0 0 0 0>
      }
      tintColor {colorUseRule=UseColor_Specified, specifiedColor=<rgba 0.0252 0.0319 0.0423 1>}
      uVRegion {bIsValid=false, max=<0 0>, min=<0 0>}
     }
     inactiveHoveredBrush {
      bIsDynamicallyLoaded=false
      drawAs=Image
      imageType=NoImage
      mirroring=NoMirror
      resourceName=None
      resourceObject=null
      tiling=NoTile
      imageSize=<0 0>
      margin=<ltrb 0 0 0 0>
      outlineSettings {
       bUseBrushTransparency=false
       roundingType=HalfHeightRadius
       width=0
       color {colorUseRule=UseColor_Specified, specifiedColor=<rgba 0 0 0 0>}
       cornerRadii=<0 0 0 0>
      }
      tintColor {colorUseRule=UseColor_Specified, specifiedColor=<rgba 0.0513 0.0953 0.159 1>}
      uVRegion {bIsValid=false, max=<0 0>, min=<0 0>}
     }
     oddRowBackgroundBrush {
      bIsDynamicallyLoaded=false
      drawAs=Image
      imageType=NoImage
      mirroring=NoMirror
      resourceName=None
      resourceObject=null
      tiling=NoTile
      imageSize=<0 0>
      margin=<ltrb 0 0 0 0>
      outlineSettings {
       bUseBrushTransparency=false
       roundingType=HalfHeightRadius
       width=0
       color {colorUseRule=UseColor_Specified, specifiedColor=<rgba 0 0 0 0>}
       cornerRadii=<0 0 0 0>
      }
      tintColor {colorUseRule=UseColor_Specified, specifiedColor=<rgba 0.0103 0.0103 0.0103 1>}
      uVRegion {bIsValid=false, max=<0 0>, min=<0 0>}
     }
     oddRowBackgroundHoveredBrush {
      bIsDynamicallyLoaded=false
      drawAs=Image
      imageType=NoImage
      mirroring=NoMirror
      resourceName=None
      resourceObject=null
      tiling=NoTile
      imageSize=<0 0>
      margin=<ltrb 0 0 0 0>
      outlineSettings {
       bUseBrushTransparency=false
       roundingType=HalfHeightRadius
       width=0
       color {colorUseRule=UseColor_Specified, specifiedColor=<rgba 0 0 0 0>}
       cornerRadii=<0 0 0 0>
      }
      tintColor {colorUseRule=UseColor_Specified, specifiedColor=<rgba 0.0176 0.0176 0.0176 1>}
      uVRegion {bIsValid=false, max=<0 0>, min=<0 0>}
     }
     parentRowBackgroundBrush {
      bIsDynamicallyLoaded=false
      drawAs=Image
      imageType=NoImage
      mirroring=NoMirror
      resourceName=None
      resourceObject=null
      tiling=NoTile
      imageSize=<32 32>
      margin=<ltrb 0 0 0 0>
      outlineSettings {
       bUseBrushTransparency=false
       roundingType=HalfHeightRadius
       width=0
       color {colorUseRule=UseColor_Specified, specifiedColor=<rgba 0 0 0 0>}
       cornerRadii=<0 0 0 0>
      }
      tintColor {colorUseRule=UseColor_Specified, specifiedColor=<rgba 1 1 1 1>}
      uVRegion {bIsValid=false, max=<0 0>, min=<0 0>}
     }
     parentRowBackgroundHoveredBrush {
      bIsDynamicallyLoaded=false
      drawAs=Image
      imageType=NoImage
      mirroring=NoMirror
      resourceName=None
      resourceObject=null
      tiling=NoTile
      imageSize=<32 32>
      margin=<ltrb 0 0 0 0>
      outlineSettings {
       bUseBrushTransparency=false
       roundingType=HalfHeightRadius
       width=0
       color {colorUseRule=UseColor_Specified, specifiedColor=<rgba 0 0 0 0>}
       cornerRadii=<0 0 0 0>
      }
      tintColor {colorUseRule=UseColor_Specified, specifiedColor=<rgba 1 1 1 1>}
      uVRegion {bIsValid=false, max=<0 0>, min=<0 0>}
     }
     selectedTextColor {colorUseRule=UseColor_Specified, specifiedColor=<rgba 1 1 1 1>}
     selectorFocusedBrush {
      bIsDynamicallyLoaded=false
      drawAs=Border
      imageType=FullColor
      mirroring=NoMirror
      resourceName=../../../Engine/Content/Slate/Common/Selector.png
      resourceObject=null
      tiling=Both
      imageSize=<0 0>
      margin=<ltrb 0.25 0.25 0.25 0.25>
      outlineSettings {
       bUseBrushTransparency=false
       roundingType=HalfHeightRadius
       width=0
       color {colorUseRule=UseColor_Specified, specifiedColor=<rgba 0 0 0 0>}
       cornerRadii=<0 0 0 0>
      }
      tintColor {colorUseRule=UseColor_Specified, specifiedColor=<rgba 0.701 0.225 0.003 1>}
      uVRegion {bIsValid=false, max=<0 0>, min=<0 0>}
     }
     textColor {colorUseRule=UseColor_Specified, specifiedColor=<rgba 0.5271 0.5271 0.5271 1>}
    }
    scrollBarStyle {
     thickness=8
     draggedThumbImage {
      bIsDynamicallyLoaded=false
      drawAs=RoundedBox
      imageType=NoImage
      mirroring=NoMirror
      resourceName=None
      resourceObject=null
      tiling=NoTile
      imageSize=<0 0>
      margin=<ltrb 0 0 0 0>
      outlineSettings {
       bUseBrushTransparency=false
       roundingType=FixedRadius
       width=0
       color {colorUseRule=UseColor_Specified, specifiedColor=<rgba 0 0 0 0>}
       cornerRadii=<4 4 4 4>
      }
      tintColor {colorUseRule=UseColor_Specified, specifiedColor=<rgba 0.2159 0.2159 0.2159 1>}
      uVRegion {bIsValid=false, max=<0 0>, min=<0 0>}
     }
     horizontalBackgroundImage {
      bIsDynamicallyLoaded=false
      drawAs=NoDrawType
      imageType=NoImage
      mirroring=NoMirror
      resourceName=None
      resourceObject=null
      tiling=NoTile
      imageSize=<0 0>
      margin=<ltrb 0 0 0 0>
      outlineSettings {
       bUseBrushTransparency=false
       roundingType=HalfHeightRadius
       width=0
       color {colorUseRule=UseColor_Specified, specifiedColor=<rgba 0 0 0 0>}
       cornerRadii=<0 0 0 0>
      }
      tintColor {colorUseRule=UseColor_Specified, specifiedColor=<rgba 1 1 1 1>}
      uVRegion {bIsValid=false, max=<0 0>, min=<0 0>}
     }
     horizontalBottomSlotImage {
      bIsDynamicallyLoaded=false
      drawAs=NoDrawType
      imageType=NoImage
      mirroring=NoMirror
      resourceName=None
      resourceObject=null
      tiling=NoTile
      imageSize=<0 0>
      margin=<ltrb 0 0 0 0>
      outlineSettings {
       bUseBrushTransparency=false
       roundingType=HalfHeightRadius
       width=0
       color {colorUseRule=UseColor_Specified, specifiedColor=<rgba 0 0 0 0>}
       cornerRadii=<0 0 0 0>
      }
      tintColor {colorUseRule=UseColor_Specified, specifiedColor=<rgba 1 1 1 1>}
      uVRegion {bIsValid=false, max=<0 0>, min=<0 0>}
     }
     horizontalTopSlotImage {
      bIsDynamicallyLoaded=false
      drawAs=NoDrawType
      imageType=NoImage
      mirroring=NoMirror
      resourceName=None
      resourceObject=null
      tiling=NoTile
      imageSize=<0 0>
      margin=<ltrb 0 0 0 0>
      outlineSettings {
       bUseBrushTransparency=false
       roundingType=HalfHeightRadius
       width=0
       color {colorUseRule=UseColor_Specified, specifiedColor=<rgba 0 0 0 0>}
       cornerRadii=<0 0 0 0>
      }
      tintColor {colorUseRule=UseColor_Specified, specifiedColor=<rgba 1 1 1 1>}
      uVRegion {bIsValid=false, max=<0 0>, min=<0 0>}
     }
     hoveredThumbImage {
      bIsDynamicallyLoaded=false
      drawAs=RoundedBox
      imageType=NoImage
      mirroring=NoMirror
      resourceName=None
      resourceObject=null
      tiling=NoTile
      imageSize=<0 0>
      margin=<ltrb 0 0 0 0>
      outlineSettings {
       bUseBrushTransparency=false
       roundingType=FixedRadius
       width=0
       color {colorUseRule=UseColor_Specified, specifiedColor=<rgba 0 0 0 0>}
       cornerRadii=<4 4 4 4>
      }
      tintColor {colorUseRule=UseColor_Specified, specifiedColor=<rgba 0.2159 0.2159 0.2159 1>}
      uVRegion {bIsValid=false, max=<0 0>, min=<0 0>}
     }
     normalThumbImage {
      bIsDynamicallyLoaded=false
      drawAs=RoundedBox
      imageType=NoImage
      mirroring=NoMirror
      resourceName=None
      resourceObject=null
      tiling=NoTile
      imageSize=<0 0>
      margin=<ltrb 0 0 0 0>
      outlineSettings {
       bUseBrushTransparency=false
       roundingType=FixedRadius
       width=0
       color {colorUseRule=UseColor_Specified, specifiedColor=<rgba 0 0 0 0>}
       cornerRadii=<4 4 4 4>
      }
      tintColor {colorUseRule=UseColor_Specified, specifiedColor=<rgba 0.0953 0.0953 0.0953 1>}
      uVRegion {bIsValid=false, max=<0 0>, min=<0 0>}
     }
     verticalBackgroundImage {
      bIsDynamicallyLoaded=false
      drawAs=NoDrawType
      imageType=NoImage
      mirroring=NoMirror
      resourceName=None
      resourceObject=null
      tiling=NoTile
      imageSize=<0 0>
      margin=<ltrb 0 0 0 0>
      outlineSettings {
       bUseBrushTransparency=false
       roundingType=HalfHeightRadius
       width=0
       color {colorUseRule=UseColor_Specified, specifiedColor=<rgba 0 0 0 0>}
       cornerRadii=<0 0 0 0>
      }
      tintColor {colorUseRule=UseColor_Specified, specifiedColor=<rgba 1 1 1 1>}
      uVRegion {bIsValid=false, max=<0 0>, min=<0 0>}
     }
     verticalBottomSlotImage {
      bIsDynamicallyLoaded=false
      drawAs=NoDrawType
      imageType=NoImage
      mirroring=NoMirror
      resourceName=None
      resourceObject=null
      tiling=NoTile
      imageSize=<0 0>
      margin=<ltrb 0 0 0 0>
      outlineSettings {
       bUseBrushTransparency=false
       roundingType=HalfHeightRadius
       width=0
       color {colorUseRule=UseColor_Specified, specifiedColor=<rgba 0 0 0 0>}
       cornerRadii=<0 0 0 0>
      }
      tintColor {colorUseRule=UseColor_Specified, specifiedColor=<rgba 1 1 1 1>}
      uVRegion {bIsValid=false, max=<0 0>, min=<0 0>}
     }
     verticalTopSlotImage {
      bIsDynamicallyLoaded=false
      drawAs=NoDrawType
      imageType=NoImage
      mirroring=NoMirror
      resourceName=None
      resourceObject=null
      tiling=NoTile
      imageSize=<0 0>
      margin=<ltrb 0 0 0 0>
      outlineSettings {
       bUseBrushTransparency=false
       roundingType=HalfHeightRadius
       width=0
       color {colorUseRule=UseColor_Specified, specifiedColor=<rgba 0 0 0 0>}
       cornerRadii=<0 0 0 0>
      }
      tintColor {colorUseRule=UseColor_Specified, specifiedColor=<rgba 1 1 1 1>}
      uVRegion {bIsValid=false, max=<0 0>, min=<0 0>}
     }
    }
    widgetStyle {
     comboButtonStyle {
      downArrowAlign=VAlign_Center
      buttonStyle {
       clickedSlateSound {resourceObject=null}
       disabled {
        bIsDynamicallyLoaded=false
        drawAs=RoundedBox
        imageType=NoImage
        mirroring=NoMirror
        resourceName=None
        resourceObject=null
        tiling=NoTile
        imageSize=<0 0>
        margin=<ltrb 0 0 0 0>
        outlineSettings {
         bUseBrushTransparency=false
         roundingType=FixedRadius
         width=1
         color {colorUseRule=UseColor_ColorTable, specifiedColor=<rgba 0 0 0 0>}
         cornerRadii=<4 4 4 4>
        }
        tintColor {colorUseRule=UseColor_Specified, specifiedColor=<rgba 0.0395 0.0395 0.0395 1>}
        uVRegion {bIsValid=false, max=<0 0>, min=<0 0>}
       }
       disabledForeground {colorUseRule=UseColor_ColorTable, specifiedColor=<rgba 0 0 0 0>}
       hovered {
        bIsDynamicallyLoaded=false
        drawAs=RoundedBox
        imageType=NoImage
        mirroring=NoMirror
        resourceName=None
        resourceObject=null
        tiling=NoTile
        imageSize=<0 0>
        margin=<ltrb 0 0 0 0>
        outlineSettings {
         bUseBrushTransparency=false
         roundingType=FixedRadius
         width=1
         color {colorUseRule=UseColor_ColorTable, specifiedColor=<rgba 0 0 0 0>}
         cornerRadii=<4 4 4 4>
        }
        tintColor {colorUseRule=UseColor_Specified, specifiedColor=<rgba 0.0048 0.0048 0.0048 1>}
        uVRegion {bIsValid=false, max=<0 0>, min=<0 0>}
       }
       hoveredForeground {colorUseRule=UseColor_ColorTable, specifiedColor=<rgba 0 0 0 0>}
       hoveredSlateSound {resourceObject=null}
       normal {
        bIsDynamicallyLoaded=false
        drawAs=RoundedBox
        imageType=NoImage
        mirroring=NoMirror
        resourceName=None
        resourceObject=null
        tiling=NoTile
        imageSize=<0 0>
        margin=<ltrb 0 0 0 0>
        outlineSettings {
         bUseBrushTransparency=false
         roundingType=FixedRadius
         width=1
         color {colorUseRule=UseColor_ColorTable, specifiedColor=<rgba 0 0 0 0>}
         cornerRadii=<4 4 4 4>
        }
        tintColor {colorUseRule=UseColor_Specified, specifiedColor=<rgba 0.0048 0.0048 0.0048 1>}
        uVRegion {bIsValid=false, max=<0 0>, min=<0 0>}
       }
       normalForeground {colorUseRule=UseColor_ColorTable, specifiedColor=<rgba 0 0 0 0>}
       normalPadding=<ltrb 8 1 8 1>
       pressed {
        bIsDynamicallyLoaded=false
        drawAs=RoundedBox
        imageType=NoImage
        mirroring=NoMirror
        resourceName=None
        resourceObject=null
        tiling=NoTile
        imageSize=<0 0>
        margin=<ltrb 0 0 0 0>
        outlineSettings {
         bUseBrushTransparency=false
         roundingType=FixedRadius
         width=1
         color {colorUseRule=UseColor_ColorTable, specifiedColor=<rgba 0 0 0 0>}
         cornerRadii=<4 4 4 4>
        }
        tintColor {colorUseRule=UseColor_Specified, specifiedColor=<rgba 0.0048 0.0048 0.0048 1>}
        uVRegion {bIsValid=false, max=<0 0>, min=<0 0>}
       }
       pressedForeground {colorUseRule=UseColor_ColorTable, specifiedColor=<rgba 0 0 0 0>}
       pressedPadding=<ltrb 8 2 8 0>
       pressedSlateSound {resourceObject=null}
      }
      contentPadding=<ltrb 0 0 0 0>
      downArrowImage {
       bIsDynamicallyLoaded=false
       drawAs=Image
       imageType=Vector
       mirroring=NoMirror
       resourceName=../../../Engine/Content/Slate/Starship/CoreWidgets/ComboBox/wide-chevron-down.svg
       resourceObject=null
       tiling=NoTile
       imageSize=<16 16>
       margin=<ltrb 0 0 0 0>
       outlineSettings {
        bUseBrushTransparency=false
        roundingType=HalfHeightRadius
        width=0
        color {colorUseRule=UseColor_Specified, specifiedColor=<rgba 0 0 0 0>}
        cornerRadii=<0 0 0 0>
       }
       tintColor {colorUseRule=UseColor_Specified, specifiedColor=<rgba 1 1 1 1>}
       uVRegion {bIsValid=false, max=<0 0>, min=<0 0>}
      }
      downArrowPadding=<ltrb 2 2 2 2>
      menuBorderBrush {
       bIsDynamicallyLoaded=false
       drawAs=RoundedBox
       imageType=NoImage
       mirroring=NoMirror
       resourceName=None
       resourceObject=null
       tiling=NoTile
       imageSize=<0 0>
       margin=<ltrb 0 0 0 0>
       outlineSettings {
        bUseBrushTransparency=false
        roundingType=FixedRadius
        width=1
        color {colorUseRule=UseColor_ColorTable, specifiedColor=<rgba 0 0 0 0>}
        cornerRadii=<0 0 0 0>
       }
       tintColor {colorUseRule=UseColor_Specified, specifiedColor=<rgba 0.0395 0.0395 0.0395 1>}
       uVRegion {bIsValid=false, max=<0 0>, min=<0 0>}
      }
      menuBorderPadding=<ltrb 0 0 0 0>
      shadowColorAndOpacity=<rgba 0 0 0 1>
      shadowOffset=<0 0>
     }
     contentPadding=<ltrb 0 0 0 0>
     menuRowPadding=<ltrb 8 3 8 3>
     pressedSlateSound {resourceObject=null}
     selectionChangeSlateSound {resourceObject=null}
    }
   }
   slot {class=WidgetSwitcherSlot, properties {verticalAlignment=VAlign_Center}}
  }
  intputIslandCode {
   widgetClassPath=&EditableTextBox
   bIsVariable=true
   parent=SwitcherAddIslandCode
   widget=intputIslandCode
   properties {
    hintText=0000-0000-0000
    revertTextOnEscape=true
    toolTipText=Enter new island code and press enter to add to combo box list
    widgetStyle {
     backgroundColor {colorUseRule=UseColor_Specified, specifiedColor=<rgba 1 1 1 1>}
     backgroundImageFocused {
      bIsDynamicallyLoaded=false
      drawAs=RoundedBox
      imageType=NoImage
      mirroring=NoMirror
      resourceName=None
      resourceObject=null
      tiling=NoTile
      imageSize=<0 0>
      margin=<ltrb 0 0 0 0>
      outlineSettings {
       bUseBrushTransparency=false
       roundingType=FixedRadius
       width=1
       color {colorUseRule=UseColor_ColorTable, specifiedColor=<rgba 0 0 0 0>}
       cornerRadii=<4 4 4 4>
      }
      tintColor {colorUseRule=UseColor_Specified, specifiedColor=<rgba 0.0048 0.0048 0.0048 1>}
      uVRegion {bIsValid=false, max=<0 0>, min=<0 0>}
     }
     backgroundImageHovered {
      bIsDynamicallyLoaded=false
      drawAs=RoundedBox
      imageType=NoImage
      mirroring=NoMirror
      resourceName=None
      resourceObject=null
      tiling=NoTile
      imageSize=<0 0>
      margin=<ltrb 0 0 0 0>
      outlineSettings {
       bUseBrushTransparency=false
       roundingType=FixedRadius
       width=1
       color {colorUseRule=UseColor_ColorTable, specifiedColor=<rgba 0 0 0 0>}
       cornerRadii=<4 4 4 4>
      }
      tintColor {colorUseRule=UseColor_Specified, specifiedColor=<rgba 0.0048 0.0048 0.0048 1>}
      uVRegion {bIsValid=false, max=<0 0>, min=<0 0>}
     }
     backgroundImageNormal {
      bIsDynamicallyLoaded=false
      drawAs=RoundedBox
      imageType=NoImage
      mirroring=NoMirror
      resourceName=None
      resourceObject=null
      tiling=NoTile
      imageSize=<0 0>
      margin=<ltrb 0 0 0 0>
      outlineSettings {
       bUseBrushTransparency=false
       roundingType=FixedRadius
       width=1
       color {colorUseRule=UseColor_ColorTable, specifiedColor=<rgba 0 0 0 0>}
       cornerRadii=<4 4 4 4>
      }
      tintColor {colorUseRule=UseColor_Specified, specifiedColor=<rgba 0.0048 0.0048 0.0048 1>}
      uVRegion {bIsValid=false, max=<0 0>, min=<0 0>}
     }
     backgroundImageReadOnly {
      bIsDynamicallyLoaded=false
      drawAs=RoundedBox
      imageType=NoImage
      mirroring=NoMirror
      resourceName=None
      resourceObject=null
      tiling=NoTile
      imageSize=<0 0>
      margin=<ltrb 0 0 0 0>
      outlineSettings {
       bUseBrushTransparency=false
       roundingType=FixedRadius
       width=1
       color {colorUseRule=UseColor_ColorTable, specifiedColor=<rgba 0 0 0 0>}
       cornerRadii=<4 4 4 4>
      }
      tintColor {colorUseRule=UseColor_Specified, specifiedColor=<rgba 0.0284 0.0284 0.0284 1>}
      uVRegion {bIsValid=false, max=<0 0>, min=<0 0>}
     }
     focusedForegroundColor {colorUseRule=UseColor_Specified, specifiedColor=<rgba 0 0.1998 1 1>}
     foregroundColor {colorUseRule=UseColor_Specified, specifiedColor=<rgba 0.5271 0.5271 0.5271 1>}
     hScrollBarPadding=<ltrb 0 0 0 0>
     padding=<ltrb 8 3 8 4>
     readOnlyForegroundColor {colorUseRule=UseColor_Specified, specifiedColor=<rgba 1 0 1 1>}
     scrollBarStyle {
      thickness=8
      draggedThumbImage {
       bIsDynamicallyLoaded=false
       drawAs=RoundedBox
       imageType=NoImage
       mirroring=NoMirror
       resourceName=None
       resourceObject=null
       tiling=NoTile
       imageSize=<0 0>
       margin=<ltrb 0 0 0 0>
       outlineSettings {
        bUseBrushTransparency=false
        roundingType=FixedRadius
        width=0
        color {colorUseRule=UseColor_Specified, specifiedColor=<rgba 0 0 0 0>}
        cornerRadii=<4 4 4 4>
       }
       tintColor {colorUseRule=UseColor_Specified, specifiedColor=<rgba 0.2159 0.2159 0.2159 1>}
       uVRegion {bIsValid=false, max=<0 0>, min=<0 0>}
      }
      horizontalBackgroundImage {
       bIsDynamicallyLoaded=false
       drawAs=NoDrawType
       imageType=NoImage
       mirroring=NoMirror
       resourceName=None
       resourceObject=null
       tiling=NoTile
       imageSize=<0 0>
       margin=<ltrb 0 0 0 0>
       outlineSettings {
        bUseBrushTransparency=false
        roundingType=HalfHeightRadius
        width=0
        color {colorUseRule=UseColor_Specified, specifiedColor=<rgba 0 0 0 0>}
        cornerRadii=<0 0 0 0>
       }
       tintColor {colorUseRule=UseColor_Specified, specifiedColor=<rgba 1 1 1 1>}
       uVRegion {bIsValid=false, max=<0 0>, min=<0 0>}
      }
      horizontalBottomSlotImage {
       bIsDynamicallyLoaded=false
       drawAs=NoDrawType
       imageType=NoImage
       mirroring=NoMirror
       resourceName=None
       resourceObject=null
       tiling=NoTile
       imageSize=<0 0>
       margin=<ltrb 0 0 0 0>
       outlineSettings {
        bUseBrushTransparency=false
        roundingType=HalfHeightRadius
        width=0
        color {colorUseRule=UseColor_Specified, specifiedColor=<rgba 0 0 0 0>}
        cornerRadii=<0 0 0 0>
       }
       tintColor {colorUseRule=UseColor_Specified, specifiedColor=<rgba 1 1 1 1>}
       uVRegion {bIsValid=false, max=<0 0>, min=<0 0>}
      }
      horizontalTopSlotImage {
       bIsDynamicallyLoaded=false
       drawAs=NoDrawType
       imageType=NoImage
       mirroring=NoMirror
       resourceName=None
       resourceObject=null
       tiling=NoTile
       imageSize=<0 0>
       margin=<ltrb 0 0 0 0>
       outlineSettings {
        bUseBrushTransparency=false
        roundingType=HalfHeightRadius
        width=0
        color {colorUseRule=UseColor_Specified, specifiedColor=<rgba 0 0 0 0>}
        cornerRadii=<0 0 0 0>
       }
       tintColor {colorUseRule=UseColor_Specified, specifiedColor=<rgba 1 1 1 1>}
       uVRegion {bIsValid=false, max=<0 0>, min=<0 0>}
      }
      hoveredThumbImage {
       bIsDynamicallyLoaded=false
       drawAs=RoundedBox
       imageType=NoImage
       mirroring=NoMirror
       resourceName=None
       resourceObject=null
       tiling=NoTile
       imageSize=<0 0>
       margin=<ltrb 0 0 0 0>
       outlineSettings {
        bUseBrushTransparency=false
        roundingType=FixedRadius
        width=0
        color {colorUseRule=UseColor_Specified, specifiedColor=<rgba 0 0 0 0>}
        cornerRadii=<4 4 4 4>
       }
       tintColor {colorUseRule=UseColor_Specified, specifiedColor=<rgba 0.2159 0.2159 0.2159 1>}
       uVRegion {bIsValid=false, max=<0 0>, min=<0 0>}
      }
      normalThumbImage {
       bIsDynamicallyLoaded=false
       drawAs=RoundedBox
       imageType=NoImage
       mirroring=NoMirror
       resourceName=None
       resourceObject=null
       tiling=NoTile
       imageSize=<0 0>
       margin=<ltrb 0 0 0 0>
       outlineSettings {
        bUseBrushTransparency=false
        roundingType=FixedRadius
        width=0
        color {colorUseRule=UseColor_Specified, specifiedColor=<rgba 0 0 0 0>}
        cornerRadii=<4 4 4 4>
       }
       tintColor {colorUseRule=UseColor_Specified, specifiedColor=<rgba 0.0953 0.0953 0.0953 1>}
       uVRegion {bIsValid=false, max=<0 0>, min=<0 0>}
      }
      verticalBackgroundImage {
       bIsDynamicallyLoaded=false
       drawAs=NoDrawType
       imageType=NoImage
       mirroring=NoMirror
       resourceName=None
       resourceObject=null
       tiling=NoTile
       imageSize=<0 0>
       margin=<ltrb 0 0 0 0>
       outlineSettings {
        bUseBrushTransparency=false
        roundingType=HalfHeightRadius
        width=0
        color {colorUseRule=UseColor_Specified, specifiedColor=<rgba 0 0 0 0>}
        cornerRadii=<0 0 0 0>
       }
       tintColor {colorUseRule=UseColor_Specified, specifiedColor=<rgba 1 1 1 1>}
       uVRegion {bIsValid=false, max=<0 0>, min=<0 0>}
      }
      verticalBottomSlotImage {
       bIsDynamicallyLoaded=false
       drawAs=NoDrawType
       imageType=NoImage
       mirroring=NoMirror
       resourceName=None
       resourceObject=null
       tiling=NoTile
       imageSize=<0 0>
       margin=<ltrb 0 0 0 0>
       outlineSettings {
        bUseBrushTransparency=false
        roundingType=HalfHeightRadius
        width=0
        color {colorUseRule=UseColor_Specified, specifiedColor=<rgba 0 0 0 0>}
        cornerRadii=<0 0 0 0>
       }
       tintColor {colorUseRule=UseColor_Specified, specifiedColor=<rgba 1 1 1 1>}
       uVRegion {bIsValid=false, max=<0 0>, min=<0 0>}
      }
      verticalTopSlotImage {
       bIsDynamicallyLoaded=false
       drawAs=NoDrawType
       imageType=NoImage
       mirroring=NoMirror
       resourceName=None
       resourceObject=null
       tiling=NoTile
       imageSize=<0 0>
       margin=<ltrb 0 0 0 0>
       outlineSettings {
        bUseBrushTransparency=false
        roundingType=HalfHeightRadius
        width=0
        color {colorUseRule=UseColor_Specified, specifiedColor=<rgba 0 0 0 0>}
        cornerRadii=<0 0 0 0>
       }
       tintColor {colorUseRule=UseColor_Specified, specifiedColor=<rgba 1 1 1 1>}
       uVRegion {bIsValid=false, max=<0 0>, min=<0 0>}
      }
     }
     textStyle {
      overflowPolicy=Clip
      transformPolicy=None
      colorAndOpacity {colorUseRule=UseColor_Foreground, specifiedColor=<rgba 1 0 1 1>}
      font {
       bForceMonospaced=false
       bMaterialIsStencil=false
       fontMaterial=null
       fontObject=null
       letterSpacing=0
       monospacedWidth=1
       size=10
       skewAmount=0
       typefaceFontName=Regular
       outlineSettings {
        bApplyOutlineToDropShadows=false
        bMiteredCorners=false
        bSeparateFillAlpha=false
        outlineMaterial=null
        outlineSize=0
        outlineColor=<rgba 0 0 0 1>
       }
      }
      highlightColor {colorUseRule=UseColor_ColorTable, specifiedColor=<rgba 0 0 0 0>}
      highlightShape {
       bIsDynamicallyLoaded=false
       drawAs=Image
       imageType=NoImage
       mirroring=NoMirror
       resourceName=None
       resourceObject=null
       tiling=NoTile
       imageSize=<0 0>
       margin=<ltrb 0 0 0 0>
       outlineSettings {
        bUseBrushTransparency=false
        roundingType=HalfHeightRadius
        width=0
        color {colorUseRule=UseColor_Specified, specifiedColor=<rgba 0 0 0 0>}
        cornerRadii=<0 0 0 0>
       }
       tintColor {colorUseRule=UseColor_ColorTable, specifiedColor=<rgba 0 0 0 0>}
       uVRegion {bIsValid=false, max=<0 0>, min=<0 0>}
      }
      selectedBackgroundColor {colorUseRule=UseColor_ColorTable, specifiedColor=<rgba 0 0 0 0>}
      shadowColorAndOpacity=<rgba 0 0 0 1>
      shadowOffset=<0 0>
      strikeBrush {
       bIsDynamicallyLoaded=false
       drawAs=Image
       imageType=NoImage
       mirroring=NoMirror
       resourceName=None
       resourceObject=null
       tiling=NoTile
       imageSize=<32 32>
       margin=<ltrb 0 0 0 0>
       outlineSettings {
        bUseBrushTransparency=false
        roundingType=HalfHeightRadius
        width=0
        color {colorUseRule=UseColor_Specified, specifiedColor=<rgba 0 0 0 0>}
        cornerRadii=<0 0 0 0>
       }
       tintColor {colorUseRule=UseColor_Specified, specifiedColor=<rgba 1 1 1 1>}
       uVRegion {bIsValid=false, max=<0 0>, min=<0 0>}
      }
      underlineBrush {
       bIsDynamicallyLoaded=false
       drawAs=Image
       imageType=NoImage
       mirroring=NoMirror
       resourceName=None
       resourceObject=null
       tiling=NoTile
       imageSize=<32 32>
       margin=<ltrb 0 0 0 0>
       outlineSettings {
        bUseBrushTransparency=false
        roundingType=HalfHeightRadius
        width=0
        color {colorUseRule=UseColor_Specified, specifiedColor=<rgba 0 0 0 0>}
        cornerRadii=<0 0 0 0>
       }
       tintColor {colorUseRule=UseColor_Specified, specifiedColor=<rgba 1 1 1 1>}
       uVRegion {bIsValid=false, max=<0 0>, min=<0 0>}
      }
     }
     vScrollBarPadding=<ltrb 0 0 0 0>
    }
   }
   slot {class=WidgetSwitcherSlot, properties {verticalAlignment=VAlign_Center}}
  }
  bttnAddIslandCode {
   widgetClassPath=&TA_Generic01_Button
   parent=HorizontalBox_7
   widget=bttnAddIslandCode
   properties {
    label=New
    toolTipText="New: Add Epic name and password.  Press Island Codes New button,  enter code and New again to enter it. Press New with no code to go back to the Island Codes drop down list. These settings will persist for next time you use it.  "
    buttonColor_Default=<rgba 0.0677 0.0677 0.0677 1>
    foregroundColor {colorUseRule=UseColor_Specified, specifiedColor=<rgba 1 0 0.0787 1>}
   }
   slot {class=HorizontalBoxSlot}
  }
  bttnRemoveIsland {
   widgetClassPath=&TA_Generic01_Button
   parent=HorizontalBox_7
   widget=bttnRemoveIsland
   properties {
    label=Del
    toolTipText=Del: Use to remove Selected code from drop down list.
    buttonColor_Default=<rgba 0.0677 0.0677 0.0677 1>
    foregroundColor {colorUseRule=UseColor_Specified, specifiedColor=<rgba 1 0 0.0787 1>}
   }
   slot {class=HorizontalBoxSlot, properties {horizontalAlignment=HAlign_Right, padding=<ltrb 4 0 20 0>}}
  }
  textFieldIslandVersion {
   widgetClassPath=&CommonTextBlock
   parent=HorizontalBox_7
   widget=textFieldIslandVersion
   properties {
    text=Ver:
    font {
     bForceMonospaced=false
     bMaterialIsStencil=false
     fontMaterial=null
     fontObject=&NotoSans
     letterSpacing=0
     monospacedWidth=1
     size=10
     skewAmount=0
     typefaceFontName=NotoSans-Regular
     outlineSettings {
      bApplyOutlineToDropShadows=false
      bMiteredCorners=false
      bSeparateFillAlpha=false
      outlineMaterial=null
      outlineSize=0
      outlineColor=<rgba 0 0 0 1>
     }
    }
    shadowOffset=<0 0>
   }
   slot {class=HorizontalBoxSlot, properties {horizontalAlignment=HAlign_Left, padding=<ltrb 5 3 3 3>}}
  }
  intputIslandVersion {
   widgetClassPath=&EditableTextBox
   bIsVariable=true
   parent=HorizontalBox_7
   widget=intputIslandVersion
   properties {
    revertTextOnEscape=true
    text=0
    toolTipText="Version to Query, blank for latest"
    widgetStyle {
     backgroundColor {colorUseRule=UseColor_Specified, specifiedColor=<rgba 1 1 1 1>}
     backgroundImageFocused {
      bIsDynamicallyLoaded=false
      drawAs=RoundedBox
      imageType=NoImage
      mirroring=NoMirror
      resourceName=None
      resourceObject=null
      tiling=NoTile
      imageSize=<0 0>
      margin=<ltrb 0 0 0 0>
      outlineSettings {
       bUseBrushTransparency=false
       roundingType=FixedRadius
       width=1
       color {colorUseRule=UseColor_ColorTable, specifiedColor=<rgba 0 0 0 0>}
       cornerRadii=<4 4 4 4>
      }
      tintColor {colorUseRule=UseColor_Specified, specifiedColor=<rgba 0.0048 0.0048 0.0048 1>}
      uVRegion {bIsValid=false, max=<0 0>, min=<0 0>}
     }
     backgroundImageHovered {
      bIsDynamicallyLoaded=false
      drawAs=RoundedBox
      imageType=NoImage
      mirroring=NoMirror
      resourceName=None
      resourceObject=null
      tiling=NoTile
      imageSize=<0 0>
      margin=<ltrb 0 0 0 0>
      outlineSettings {
       bUseBrushTransparency=false
       roundingType=FixedRadius
       width=1
       color {colorUseRule=UseColor_ColorTable, specifiedColor=<rgba 0 0 0 0>}
       cornerRadii=<4 4 4 4>
      }
      tintColor {colorUseRule=UseColor_Specified, specifiedColor=<rgba 0.0048 0.0048 0.0048 1>}
      uVRegion {bIsValid=false, max=<0 0>, min=<0 0>}
     }
     backgroundImageNormal {
      bIsDynamicallyLoaded=false
      drawAs=RoundedBox
      imageType=NoImage
      mirroring=NoMirror
      resourceName=None
      resourceObject=null
      tiling=NoTile
      imageSize=<0 0>
      margin=<ltrb 0 0 0 0>
      outlineSettings {
       bUseBrushTransparency=false
       roundingType=FixedRadius
       width=1
       color {colorUseRule=UseColor_ColorTable, specifiedColor=<rgba 0 0 0 0>}
       cornerRadii=<4 4 4 4>
      }
      tintColor {colorUseRule=UseColor_Specified, specifiedColor=<rgba 0.0048 0.0048 0.0048 1>}
      uVRegion {bIsValid=false, max=<0 0>, min=<0 0>}
     }
     backgroundImageReadOnly {
      bIsDynamicallyLoaded=false
      drawAs=RoundedBox
      imageType=NoImage
      mirroring=NoMirror
      resourceName=None
      resourceObject=null
      tiling=NoTile
      imageSize=<0 0>
      margin=<ltrb 0 0 0 0>
      outlineSettings {
       bUseBrushTransparency=false
       roundingType=FixedRadius
       width=1
       color {colorUseRule=UseColor_ColorTable, specifiedColor=<rgba 0 0 0 0>}
       cornerRadii=<4 4 4 4>
      }
      tintColor {colorUseRule=UseColor_Specified, specifiedColor=<rgba 0.0284 0.0284 0.0284 1>}
      uVRegion {bIsValid=false, max=<0 0>, min=<0 0>}
     }
     focusedForegroundColor {colorUseRule=UseColor_Specified, specifiedColor=<rgba 0 0.1998 1 1>}
     foregroundColor {colorUseRule=UseColor_Specified, specifiedColor=<rgba 0.9635 0.9635 0.9635 1>}
     hScrollBarPadding=<ltrb 0 0 0 0>
     padding=<ltrb 8 3 8 4>
     readOnlyForegroundColor {colorUseRule=UseColor_Specified, specifiedColor=<rgba 1 0 1 1>}
     scrollBarStyle {
      thickness=8
      draggedThumbImage {
       bIsDynamicallyLoaded=false
       drawAs=RoundedBox
       imageType=NoImage
       mirroring=NoMirror
       resourceName=None
       resourceObject=null
       tiling=NoTile
       imageSize=<0 0>
       margin=<ltrb 0 0 0 0>
       outlineSettings {
        bUseBrushTransparency=false
        roundingType=FixedRadius
        width=0
        color {colorUseRule=UseColor_Specified, specifiedColor=<rgba 0 0 0 0>}
        cornerRadii=<4 4 4 4>
       }
       tintColor {colorUseRule=UseColor_Specified, specifiedColor=<rgba 0.2159 0.2159 0.2159 1>}
       uVRegion {bIsValid=false, max=<0 0>, min=<0 0>}
      }
      horizontalBackgroundImage {
       bIsDynamicallyLoaded=false
       drawAs=NoDrawType
       imageType=NoImage
       mirroring=NoMirror
       resourceName=None
       resourceObject=null
       tiling=NoTile
       imageSize=<0 0>
       margin=<ltrb 0 0 0 0>
       outlineSettings {
        bUseBrushTransparency=false
        roundingType=HalfHeightRadius
        width=0
        color {colorUseRule=UseColor_Specified, specifiedColor=<rgba 0 0 0 0>}
        cornerRadii=<0 0 0 0>
       }
       tintColor {colorUseRule=UseColor_Specified, specifiedColor=<rgba 1 1 1 1>}
       uVRegion {bIsValid=false, max=<0 0>, min=<0 0>}
      }
      horizontalBottomSlotImage {
       bIsDynamicallyLoaded=false
       drawAs=NoDrawType
       imageType=NoImage
       mirroring=NoMirror
       resourceName=None
       resourceObject=null
       tiling=NoTile
       imageSize=<0 0>
       margin=<ltrb 0 0 0 0>
       outlineSettings {
        bUseBrushTransparency=false
        roundingType=HalfHeightRadius
        width=0
        color {colorUseRule=UseColor_Specified, specifiedColor=<rgba 0 0 0 0>}
        cornerRadii=<0 0 0 0>
       }
       tintColor {colorUseRule=UseColor_Specified, specifiedColor=<rgba 1 1 1 1>}
       uVRegion {bIsValid=false, max=<0 0>, min=<0 0>}
      }
      horizontalTopSlotImage {
       bIsDynamicallyLoaded=false
       drawAs=NoDrawType
       imageType=NoImage
       mirroring=NoMirror
       resourceName=None
       resourceObject=null
       tiling=NoTile
       imageSize=<0 0>
       margin=<ltrb 0 0 0 0>
       outlineSettings {
        bUseBrushTransparency=false
        roundingType=HalfHeightRadius
        width=0
        color {colorUseRule=UseColor_Specified, specifiedColor=<rgba 0 0 0 0>}
        cornerRadii=<0 0 0 0>
       }
       tintColor {colorUseRule=UseColor_Specified, specifiedColor=<rgba 1 1 1 1>}
       uVRegion {bIsValid=false, max=<0 0>, min=<0 0>}
      }
      hoveredThumbImage {
       bIsDynamicallyLoaded=false
       drawAs=RoundedBox
       imageType=NoImage
       mirroring=NoMirror
       resourceName=None
       resourceObject=null
       tiling=NoTile
       imageSize=<0 0>
       margin=<ltrb 0 0 0 0>
       outlineSettings {
        bUseBrushTransparency=false
        roundingType=FixedRadius
        width=0
        color {colorUseRule=UseColor_Specified, specifiedColor=<rgba 0 0 0 0>}
        cornerRadii=<4 4 4 4>
       }
       tintColor {colorUseRule=UseColor_Specified, specifiedColor=<rgba 0.2159 0.2159 0.2159 1>}
       uVRegion {bIsValid=false, max=<0 0>, min=<0 0>}
      }
      normalThumbImage {
       bIsDynamicallyLoaded=false
       drawAs=RoundedBox
       imageType=NoImage
       mirroring=NoMirror
       resourceName=None
       resourceObject=null
       tiling=NoTile
       imageSize=<0 0>
       margin=<ltrb 0 0 0 0>
       outlineSettings {
        bUseBrushTransparency=false
        roundingType=FixedRadius
        width=0
        color {colorUseRule=UseColor_Specified, specifiedColor=<rgba 0 0 0 0>}
        cornerRadii=<4 4 4 4>
       }
       tintColor {colorUseRule=UseColor_Specified, specifiedColor=<rgba 0.0953 0.0953 0.0953 1>}
       uVRegion {bIsValid=false, max=<0 0>, min=<0 0>}
      }
      verticalBackgroundImage {
       bIsDynamicallyLoaded=false
       drawAs=NoDrawType
       imageType=NoImage
       mirroring=NoMirror
       resourceName=None
       resourceObject=null
       tiling=NoTile
       imageSize=<0 0>
       margin=<ltrb 0 0 0 0>
       outlineSettings {
        bUseBrushTransparency=false
        roundingType=HalfHeightRadius
        width=0
        color {colorUseRule=UseColor_Specified, specifiedColor=<rgba 0 0 0 0>}
        cornerRadii=<0 0 0 0>
       }
       tintColor {colorUseRule=UseColor_Specified, specifiedColor=<rgba 1 1 1 1>}
       uVRegion {bIsValid=false, max=<0 0>, min=<0 0>}
      }
      verticalBottomSlotImage {
       bIsDynamicallyLoaded=false
       drawAs=NoDrawType
       imageType=NoImage
       mirroring=NoMirror
       resourceName=None
       resourceObject=null
       tiling=NoTile
       imageSize=<0 0>
       margin=<ltrb 0 0 0 0>
       outlineSettings {
        bUseBrushTransparency=false
        roundingType=HalfHeightRadius
        width=0
        color {colorUseRule=UseColor_Specified, specifiedColor=<rgba 0 0 0 0>}
        cornerRadii=<0 0 0 0>
       }
       tintColor {colorUseRule=UseColor_Specified, specifiedColor=<rgba 1 1 1 1>}
       uVRegion {bIsValid=false, max=<0 0>, min=<0 0>}
      }
      verticalTopSlotImage {
       bIsDynamicallyLoaded=false
       drawAs=NoDrawType
       imageType=NoImage
       mirroring=NoMirror
       resourceName=None
       resourceObject=null
       tiling=NoTile
       imageSize=<0 0>
       margin=<ltrb 0 0 0 0>
       outlineSettings {
        bUseBrushTransparency=false
        roundingType=HalfHeightRadius
        width=0
        color {colorUseRule=UseColor_Specified, specifiedColor=<rgba 0 0 0 0>}
        cornerRadii=<0 0 0 0>
       }
       tintColor {colorUseRule=UseColor_Specified, specifiedColor=<rgba 1 1 1 1>}
       uVRegion {bIsValid=false, max=<0 0>, min=<0 0>}
      }
     }
     textStyle {
      overflowPolicy=Clip
      transformPolicy=None
      colorAndOpacity {colorUseRule=UseColor_Foreground, specifiedColor=<rgba 1 0 1 1>}
      font {
       bForceMonospaced=false
       bMaterialIsStencil=false
       fontMaterial=null
       fontObject=null
       letterSpacing=0
       monospacedWidth=1
       size=10
       skewAmount=0
       typefaceFontName=Regular
       outlineSettings {
        bApplyOutlineToDropShadows=false
        bMiteredCorners=false
        bSeparateFillAlpha=false
        outlineMaterial=null
        outlineSize=0
        outlineColor=<rgba 0 0 0 1>
       }
      }
      highlightColor {colorUseRule=UseColor_ColorTable, specifiedColor=<rgba 0 0 0 0>}
      highlightShape {
       bIsDynamicallyLoaded=false
       drawAs=Image
       imageType=NoImage
       mirroring=NoMirror
       resourceName=None
       resourceObject=null
       tiling=NoTile
       imageSize=<0 0>
       margin=<ltrb 0 0 0 0>
       outlineSettings {
        bUseBrushTransparency=false
        roundingType=HalfHeightRadius
        width=0
        color {colorUseRule=UseColor_Specified, specifiedColor=<rgba 0 0 0 0>}
        cornerRadii=<0 0 0 0>
       }
       tintColor {colorUseRule=UseColor_ColorTable, specifiedColor=<rgba 0 0 0 0>}
       uVRegion {bIsValid=false, max=<0 0>, min=<0 0>}
      }
      selectedBackgroundColor {colorUseRule=UseColor_ColorTable, specifiedColor=<rgba 0 0 0 0>}
      shadowColorAndOpacity=<rgba 0 0 0 1>
      shadowOffset=<0 0>
      strikeBrush {
       bIsDynamicallyLoaded=false
       drawAs=Image
       imageType=NoImage
       mirroring=NoMirror
       resourceName=None
       resourceObject=null
       tiling=NoTile
       imageSize=<32 32>
       margin=<ltrb 0 0 0 0>
       outlineSettings {
        bUseBrushTransparency=false
        roundingType=HalfHeightRadius
        width=0
        color {colorUseRule=UseColor_Specified, specifiedColor=<rgba 0 0 0 0>}
        cornerRadii=<0 0 0 0>
       }
       tintColor {colorUseRule=UseColor_Specified, specifiedColor=<rgba 1 1 1 1>}
       uVRegion {bIsValid=false, max=<0 0>, min=<0 0>}
      }
      underlineBrush {
       bIsDynamicallyLoaded=false
       drawAs=Image
       imageType=NoImage
       mirroring=NoMirror
       resourceName=None
       resourceObject=null
       tiling=NoTile
       imageSize=<32 32>
       margin=<ltrb 0 0 0 0>
       outlineSettings {
        bUseBrushTransparency=false
        roundingType=HalfHeightRadius
        width=0
        color {colorUseRule=UseColor_Specified, specifiedColor=<rgba 0 0 0 0>}
        cornerRadii=<0 0 0 0>
       }
       tintColor {colorUseRule=UseColor_Specified, specifiedColor=<rgba 1 1 1 1>}
       uVRegion {bIsValid=false, max=<0 0>, min=<0 0>}
      }
     }
     vScrollBarPadding=<ltrb 0 0 0 0>
    }
   }
   slot {class=HorizontalBoxSlot, properties {verticalAlignment=VAlign_Center, padding=<ltrb 0 0 4 0>}}
  }
  bttnQueryIsland {
   widgetClassPath=&TA_Generic01_Button
   parent=HorizontalBox_7
   widget=bttnQueryIsland
   properties {
    label=Query
    toolTipText="Query: Select desired Island from list and press Query to get latest data from Prod.  \r\nThis will populate Active Groups and if any accolade triggers exist for the\r\ncurrent group, they will populate below Active Group."
    buttonColor_Default=<rgba 1 0.1119 0.9131 1>
   }
   slot {class=HorizontalBoxSlot, properties {horizontalAlignment=HAlign_Right, size {sizeRule=Fill, value=1}}}
  }
  HorizontalBox_6 {
   widgetClassPath=&HorizontalBox
   parent=VerticalBox_2
   widget=HorizontalBox_6
   slot {class=VerticalBoxSlot, properties {verticalAlignment=VAlign_Center, padding=<ltrb 2 2 2 2>}}
  }
  textFieldIslandNamePrompt {
   widgetClassPath=&CommonTextBlock
   parent=HorizontalBox_6
   widget=textFieldIslandNamePrompt
   properties {
    text=Island Name:
    font {
     bForceMonospaced=false
     bMaterialIsStencil=false
     fontMaterial=null
     fontObject=&NotoSans
     letterSpacing=0
     monospacedWidth=1
     size=10
     skewAmount=0
     typefaceFontName=NotoSans-Regular
     outlineSettings {
      bApplyOutlineToDropShadows=false
      bMiteredCorners=false
      bSeparateFillAlpha=false
      outlineMaterial=null
      outlineSize=0
      outlineColor=<rgba 0 0 0 1>
     }
    }
    shadowOffset=<0 0>
   }
   slot {class=HorizontalBoxSlot, properties {padding=<ltrb 5 3 3 3>}}
  }
  textFieldIslandName {
   widgetClassPath=&CommonTextBlock
   bIsVariable=true
   parent=HorizontalBox_6
   widget=textFieldIslandName
   properties {
    minDesiredWidth=150
    font {
     bForceMonospaced=false
     bMaterialIsStencil=false
     fontMaterial=null
     fontObject=&NotoSans
     letterSpacing=0
     monospacedWidth=1
     size=10
     skewAmount=0
     typefaceFontName=NotoSans-Regular
     outlineSettings {
      bApplyOutlineToDropShadows=false
      bMiteredCorners=false
      bSeparateFillAlpha=false
      outlineMaterial=null
      outlineSize=0
      outlineColor=<rgba 0 0 0 1>
     }
    }
    shadowOffset=<0 0>
   }
   slot {class=HorizontalBoxSlot, properties {padding=<ltrb 4 4 4 4>}}
  }
  Border_152 {
   widgetClassPath=&Border
   parent=CopyAssetVBox
   widget=Border_152
   properties {background {
     bIsDynamicallyLoaded=false
     drawAs=Image
     imageType=NoImage
     mirroring=NoMirror
     resourceName=None
     resourceObject=null
     tiling=NoTile
     imageSize=<32 32>
     margin=<ltrb 0 0 0 0>
     outlineSettings {
      bUseBrushTransparency=false
      roundingType=HalfHeightRadius
      width=0
      color {colorUseRule=UseColor_Specified, specifiedColor=<rgba 0 0 0 0>}
      cornerRadii=<0 0 0 0>
     }
     tintColor {colorUseRule=UseColor_Specified, specifiedColor=<rgba 0.0284 0.0284 0.0284 1>}
     uVRegion {bIsValid=false, max=<0 0>, min=<0 0>}
    }}
   slot {class=VerticalBoxSlot}
  }
  HorizontalBox_2 {
   widgetClassPath=&HorizontalBox
   parent=CopyAssetVBox
   widget=HorizontalBox_2
   slot {class=VerticalBoxSlot, properties {padding=<ltrb 5 5 0 5>}}
  }
  textField_1 {
   widgetClassPath=&CommonTextBlock
   parent=HorizontalBox_2
   widget=textField_1
   properties {
    minDesiredWidth=150
    text=Play Game:
    font {
     bForceMonospaced=false
     bMaterialIsStencil=false
     fontMaterial=null
     fontObject=&NotoSans
     letterSpacing=0
     monospacedWidth=1
     size=10
     skewAmount=0
     typefaceFontName=NotoSans-Regular
     outlineSettings {
      bApplyOutlineToDropShadows=false
      bMiteredCorners=false
      bSeparateFillAlpha=false
      outlineMaterial=null
      outlineSize=0
      outlineColor=<rgba 0 0 0 1>
     }
    }
    shadowOffset=<0 0>
   }
   slot {class=HorizontalBoxSlot, properties {horizontalAlignment=HAlign_Left, verticalAlignment=VAlign_Center, size {sizeRule=Fill, value=2.5}}}
  }
  StartStopSwitcher {
   widgetClassPath=&WidgetSwitcher
   bIsVariable=true
   parent=HorizontalBox_2
   widget=StartStopSwitcher
   slot {class=HorizontalBoxSlot, properties {horizontalAlignment=HAlign_Right, size {sizeRule=Fill, value=6}}}
  }
  StartGamePanel {
   widgetClassPath=&HorizontalBox
   bIsVariable=true
   parent=StartStopSwitcher
   widget=StartGamePanel
   slot {class=WidgetSwitcherSlot, properties {horizontalAlignment=HAlign_Right}}
  }
  bttnStartGame {
   widgetClassPath=&TA_Generic01_Button
   parent=StartGamePanel
   widget=bttnStartGame
   properties {
    label=Start
    toolTipText="Start: Select desired Active Group (this is typically Team:Class) and press Start.  \r\nThis simulates a player entering the game.  PlayerStartSession is copied over to PlayerSession. At this point you may click accolades to trigger them.\r\n\r\nThe \"Click Time\" is the number of seconds that pass before the accolade fires. User changes in \"Analytics Data->Mcp Profile Stats\" will also be sent to simulator.  \r\n\r\nNOTE: DynamicXp.ForceLive may need to be checked if data is not Live yet."
    buttonColor_Default=<rgba 0.0908 0.4287 0.0242 1>
   }
   slot {class=HorizontalBoxSlot, properties {padding=<ltrb 0 0 2 0>, size {sizeRule=Fill, value=6}}}
  }
  EndGamePanel {
   widgetClassPath=&HorizontalBox
   bIsVariable=true
   parent=StartStopSwitcher
   widget=EndGamePanel
   slot {class=WidgetSwitcherSlot, properties {horizontalAlignment=HAlign_Right}}
  }
  PlaySwitcher {
   widgetClassPath=&WidgetSwitcher
   bIsVariable=true
   parent=EndGamePanel
   widget=PlaySwitcher
   slot {class=HorizontalBoxSlot, properties {horizontalAlignment=HAlign_Center, size {sizeRule=Fill, value=1}}}
  }
  HorizontalBox_1 {
   widgetClassPath=&HorizontalBox
   parent=PlaySwitcher
   widget=HorizontalBox_1
   slot {class=WidgetSwitcherSlot, properties {horizontalAlignment=HAlign_Center}}
  }
  bttnStep {
   widgetClassPath=&TA_Generic01_Button
   parent=HorizontalBox_1
   widget=bttnStep
   properties {
    label=Step
    toolTipText="Step:  Step will just step a tick based off of \"Step Time\".  \r\nThe amount of time passed is \"Step Time\". User changes in \"Analytics Data->Mcp Profile Stats\" will also be sent to simulator."
    buttonColor_Default=<rgba 0.0026 0.024 0.4271 1>
   }
   slot {class=HorizontalBoxSlot, properties {padding=<ltrb 0 0 10 0>}}
  }
  bttnRun {
   widgetClassPath=&TA_Generic01_Button
   parent=HorizontalBox_1
   widget=bttnRun
   properties {
    label=Run
    toolTipText="Run: To automatically play, use the Run button.  \r\nThe speed can be controlled by \"Game Time\".   When running, each Tick will run \"Step Time\".  If the simulation gets too far behind, \"Step Time\" may automatically throttle down.\r\nUser changes in \"Analytics Data->Mcp Profile Stats\" will \r\nalso be sent to simulator."
    buttonColor_Default=<rgba 0 0.4271 0.3884 1>
   }
   slot {class=HorizontalBoxSlot}
  }
  bttnPause {
   widgetClassPath=&TA_Generic01_Button
   parent=PlaySwitcher
   widget=bttnPause
   properties {
    label=Pause
    toolTipText="Pause: When running Pause will show.  Click to pause simlation.  You can run again.  Also good to modify values then hit Run again."
    buttonColor_Default=<rgba 0.4271 0.1854 0 1>
   }
   slot {class=WidgetSwitcherSlot}
  }
  bttnEndGame {
   widgetClassPath=&TA_Generic01_Button
   parent=EndGamePanel
   widget=bttnEndGame
   properties {label=End, toolTipText="End: Simulate player leaving the game.  You may run again.", buttonColor_Default=<rgba 0.4271 0 0.0041 1>}
   slot {class=HorizontalBoxSlot, properties {horizontalAlignment=HAlign_Right, size {sizeRule=Fill, value=0.6}}}
  }
  HorizontalBox_Simulate {
   widgetClassPath=&HorizontalBox
   parent=CopyAssetVBox
   widget=HorizontalBox_Simulate
   slot {class=VerticalBoxSlot, properties {padding=<ltrb 5 0 0 5>}}
  }
  textField {
   widgetClassPath=&CommonTextBlock
   parent=HorizontalBox_Simulate
   widget=textField
   properties {
    text=Simulate:
    font {
     bForceMonospaced=false
     bMaterialIsStencil=false
     fontMaterial=null
     fontObject=&NotoSans
     letterSpacing=0
     monospacedWidth=1
     size=10
     skewAmount=0
     typefaceFontName=NotoSans-Regular
     outlineSettings {
      bApplyOutlineToDropShadows=false
      bMiteredCorners=false
      bSeparateFillAlpha=false
      outlineMaterial=null
      outlineSize=0
      outlineColor=<rgba 0 0 0 1>
     }
    }
    shadowOffset=<0 0>
   }
   slot {
    class=HorizontalBoxSlot
    properties {horizontalAlignment=HAlign_Left, verticalAlignment=VAlign_Center, padding=<ltrb 0 0 10 0>, size {sizeRule=Automatic, value=2.5}}
   }
  }
  bttnDataAssetChange {
   widgetClassPath=&TA_Generic01_Button
   parent=HorizontalBox_Simulate
   widget=bttnDataAssetChange
   properties {
    label=DADs
    toolTipText="DADs: Simulate data asset changes mid game through DADS (Dynamic Asset Data) system.  Modify any values in \"Designer Tuning\" and click this to have them update live.  They also get put in calculator when \"Start\" button is pressed."
    buttonColor_Default=<rgba 0.0908 0.4287 0.0242 1>
   }
   slot {class=HorizontalBoxSlot, properties {padding=<ltrb 0 0 10 0>}}
  }
  bttnIslandChange {
   widgetClassPath=&TA_Generic01_Button
   parent=HorizontalBox_Simulate
   widget=bttnIslandChange
   properties {
    label=XpStats
    toolTipText="Xp Stats: Used to simlate Island Xp Stats being updating in the middle of the game.  Pressing \"Start\" also updates calculator with latest data in Stats."
    buttonColor_Default=<rgba 0.0908 0.4287 0.0242 1>
   }
   slot {class=HorizontalBoxSlot}
  }
  textField_2 {
   widgetClassPath=&CommonTextBlock
   parent=HorizontalBox_Simulate
   widget=textField_2
   properties {
    text=AFK:
    font {
     bForceMonospaced=false
     bMaterialIsStencil=false
     fontMaterial=null
     fontObject=&NotoSans
     letterSpacing=0
     monospacedWidth=1
     size=10
     skewAmount=0
     typefaceFontName=NotoSans-Regular
     outlineSettings {
      bApplyOutlineToDropShadows=false
      bMiteredCorners=false
      bSeparateFillAlpha=false
      outlineMaterial=null
      outlineSize=0
      outlineColor=<rgba 0 0 0 1>
     }
    }
    shadowOffset=<0 0>
   }
   slot {class=HorizontalBoxSlot, properties {verticalAlignment=VAlign_Center, padding=<ltrb 10 0 0 0>, size {sizeRule=Automatic, value=2.5}}}
  }
  CheckBox_Afk {
   widgetClassPath=/Script/UMG.CheckBox
   bIsVariable=true
   parent=HorizontalBox_Simulate
   widget=CheckBox_Afk
   properties {
    toolTipText=Start AFK Timer to simulae player being AFK
    widgetStyle {
     checkBoxType=CheckBox
     backgroundHoveredImage {
      bIsDynamicallyLoaded=false
      drawAs=RoundedBox
      imageType=NoImage
      mirroring=NoMirror
      resourceName=None
      resourceObject=null
      tiling=NoTile
      imageSize=<18 18>
      margin=<ltrb 0 0 0 0>
      outlineSettings {
       bUseBrushTransparency=false
       roundingType=FixedRadius
       width=1
       color {colorUseRule=UseColor_ColorTable, specifiedColor=<rgba 0 0 0 0>}
       cornerRadii=<3 3 3 3>
      }
      tintColor {colorUseRule=UseColor_Specified, specifiedColor=<rgba 0.5052 0.5052 0.5052 1>}
      uVRegion {bIsValid=false, max=<0 0>, min=<0 0>}
     }
     backgroundImage {
      bIsDynamicallyLoaded=false
      drawAs=RoundedBox
      imageType=NoImage
      mirroring=NoMirror
      resourceName=None
      resourceObject=null
      tiling=NoTile
      imageSize=<18 18>
      margin=<ltrb 0 0 0 0>
      outlineSettings {
       bUseBrushTransparency=false
       roundingType=FixedRadius
       width=1
       color {colorUseRule=UseColor_ColorTable, specifiedColor=<rgba 0 0 0 0>}
       cornerRadii=<3 3 3 3>
      }
      tintColor {colorUseRule=UseColor_Specified, specifiedColor=<rgba 0.3419 0.3419 0.3419 1>}
      uVRegion {bIsValid=false, max=<0 0>, min=<0 0>}
     }
     backgroundPressedImage {
      bIsDynamicallyLoaded=false
      drawAs=RoundedBox
      imageType=NoImage
      mirroring=NoMirror
      resourceName=None
      resourceObject=null
      tiling=NoTile
      imageSize=<18 18>
      margin=<ltrb 0 0 0 0>
      outlineSettings {
       bUseBrushTransparency=false
       roundingType=FixedRadius
       width=1
       color {colorUseRule=UseColor_ColorTable, specifiedColor=<rgba 0 0 0 0>}
       cornerRadii=<3 3 3 3>
      }
      tintColor {colorUseRule=UseColor_Specified, specifiedColor=<rgba 0.0048 0.0048 0.0048 1>}
      uVRegion {bIsValid=false, max=<0 0>, min=<0 0>}
     }
     borderBackgroundColor {colorUseRule=UseColor_Specified, specifiedColor=<rgba 0.3419 0.3419 0.3419 1>}
     checkedForeground {colorUseRule=UseColor_Specified, specifiedColor=<rgba 0.3419 0.3419 0.3419 1>}
     checkedHoveredForeground {colorUseRule=UseColor_Specified, specifiedColor=<rgba 0.3419 0.3419 0.3419 1>}
     checkedHoveredImage {
      bIsDynamicallyLoaded=false
      drawAs=Image
      imageType=Vector
      mirroring=NoMirror
      resourceName=../../../Engine/Content/Slate/Starship/CoreWidgets/CheckBox/check.svg
      resourceObject=null
      tiling=NoTile
      imageSize=<16 16>
      margin=<ltrb 0 0 0 0>
      outlineSettings {
       bUseBrushTransparency=false
       roundingType=HalfHeightRadius
       width=0
       color {colorUseRule=UseColor_Specified, specifiedColor=<rgba 0 0 0 0>}
       cornerRadii=<0 0 0 0>
      }
      tintColor {colorUseRule=UseColor_Specified, specifiedColor=<rgba 0.0044 0.2384 1 1>}
      uVRegion {bIsValid=false, max=<0 0>, min=<0 0>}
     }
     checkedImage {
      bIsDynamicallyLoaded=false
      drawAs=Image
      imageType=Vector
      mirroring=NoMirror
      resourceName=../../../Engine/Content/Slate/Starship/CoreWidgets/CheckBox/check.svg
      resourceObject=null
      tiling=NoTile
      imageSize=<16 16>
      margin=<ltrb 0 0 0 0>
      outlineSettings {
       bUseBrushTransparency=false
       roundingType=HalfHeightRadius
       width=0
       color {colorUseRule=UseColor_Specified, specifiedColor=<rgba 0 0 0 0>}
       cornerRadii=<0 0 0 0>
      }
      tintColor {colorUseRule=UseColor_Specified, specifiedColor=<rgba 0 0.162 0.7454 1>}
      uVRegion {bIsValid=false, max=<0 0>, min=<0 0>}
     }
     checkedPressedForeground {colorUseRule=UseColor_Specified, specifiedColor=<rgba 0.3419 0.3419 0.3419 1>}
     checkedPressedImage {
      bIsDynamicallyLoaded=false
      drawAs=Image
      imageType=Vector
      mirroring=NoMirror
      resourceName=../../../Engine/Content/Slate/Starship/CoreWidgets/CheckBox/check.svg
      resourceObject=null
      tiling=NoTile
      imageSize=<16 16>
      margin=<ltrb 0 0 0 0>
      outlineSettings {
       bUseBrushTransparency=false
       roundingType=HalfHeightRadius
       width=0
       color {colorUseRule=UseColor_Specified, specifiedColor=<rgba 0 0 0 0>}
       cornerRadii=<0 0 0 0>
      }
      tintColor {colorUseRule=UseColor_Specified, specifiedColor=<rgba 0 0.0802 0.3515 1>}
      uVRegion {bIsValid=false, max=<0 0>, min=<0 0>}
     }
     checkedSlateSound {resourceObject=null}
     foregroundColor {colorUseRule=UseColor_Specified, specifiedColor=<rgba 0.3419 0.3419 0.3419 1>}
     hoveredForeground {colorUseRule=UseColor_Specified, specifiedColor=<rgba 0.3419 0.3419 0.3419 1>}
     hoveredSlateSound {resourceObject=null}
     padding=<ltrb 2 0 0 0>
     pressedForeground {colorUseRule=UseColor_Specified, specifiedColor=<rgba 0.3419 0.3419 0.3419 1>}
     uncheckedHoveredImage {
      bIsDynamicallyLoaded=false
      drawAs=NoDrawType
      imageType=NoImage
      mirroring=NoMirror
      resourceName=None
      resourceObject=null
      tiling=NoTile
      imageSize=<0 0>
      margin=<ltrb 0 0 0 0>
      outlineSettings {
       bUseBrushTransparency=false
       roundingType=HalfHeightRadius
       width=0
       color {colorUseRule=UseColor_Specified, specifiedColor=<rgba 0 0 0 0>}
       cornerRadii=<0 0 0 0>
      }
      tintColor {colorUseRule=UseColor_Specified, specifiedColor=<rgba 1 1 1 1>}
      uVRegion {bIsValid=false, max=<0 0>, min=<0 0>}
     }
     uncheckedImage {
      bIsDynamicallyLoaded=false
      drawAs=NoDrawType
      imageType=NoImage
      mirroring=NoMirror
      resourceName=None
      resourceObject=null
      tiling=NoTile
      imageSize=<0 0>
      margin=<ltrb 0 0 0 0>
      outlineSettings {
       bUseBrushTransparency=false
       roundingType=HalfHeightRadius
       width=0
       color {colorUseRule=UseColor_Specified, specifiedColor=<rgba 0 0 0 0>}
       cornerRadii=<0 0 0 0>
      }
      tintColor {colorUseRule=UseColor_Specified, specifiedColor=<rgba 1 1 1 1>}
      uVRegion {bIsValid=false, max=<0 0>, min=<0 0>}
     }
     uncheckedPressedImage {
      bIsDynamicallyLoaded=false
      drawAs=NoDrawType
      imageType=NoImage
      mirroring=NoMirror
      resourceName=None
      resourceObject=null
      tiling=NoTile
      imageSize=<0 0>
      margin=<ltrb 0 0 0 0>
      outlineSettings {
       bUseBrushTransparency=false
       roundingType=HalfHeightRadius
       width=0
       color {colorUseRule=UseColor_Specified, specifiedColor=<rgba 0 0 0 0>}
       cornerRadii=<0 0 0 0>
      }
      tintColor {colorUseRule=UseColor_Specified, specifiedColor=<rgba 1 1 1 1>}
      uVRegion {bIsValid=false, max=<0 0>, min=<0 0>}
     }
     uncheckedSlateSound {resourceObject=null}
     undeterminedForeground {colorUseRule=UseColor_Specified, specifiedColor=<rgba 0.3419 0.3419 0.3419 1>}
     undeterminedHoveredImage {
      bIsDynamicallyLoaded=false
      drawAs=Image
      imageType=Vector
      mirroring=NoMirror
      resourceName=../../../Engine/Content/Slate/Starship/CoreWidgets/CheckBox/indeterminate.svg
      resourceObject=null
      tiling=NoTile
      imageSize=<16 16>
      margin=<ltrb 0 0 0 0>
      outlineSettings {
       bUseBrushTransparency=false
       roundingType=HalfHeightRadius
       width=0
       color {colorUseRule=UseColor_Specified, specifiedColor=<rgba 0 0 0 0>}
       cornerRadii=<0 0 0 0>
      }
      tintColor {colorUseRule=UseColor_Specified, specifiedColor=<rgba 0.0044 0.2384 1 1>}
      uVRegion {bIsValid=false, max=<0 0>, min=<0 0>}
     }
     undeterminedImage {
      bIsDynamicallyLoaded=false
      drawAs=Image
      imageType=Vector
      mirroring=NoMirror
      resourceName=../../../Engine/Content/Slate/Starship/CoreWidgets/CheckBox/indeterminate.svg
      resourceObject=null
      tiling=NoTile
      imageSize=<16 16>
      margin=<ltrb 0 0 0 0>
      outlineSettings {
       bUseBrushTransparency=false
       roundingType=HalfHeightRadius
       width=0
       color {colorUseRule=UseColor_Specified, specifiedColor=<rgba 0 0 0 0>}
       cornerRadii=<0 0 0 0>
      }
      tintColor {colorUseRule=UseColor_Specified, specifiedColor=<rgba 0 0.162 0.7454 1>}
      uVRegion {bIsValid=false, max=<0 0>, min=<0 0>}
     }
     undeterminedPressedImage {
      bIsDynamicallyLoaded=false
      drawAs=Image
      imageType=Vector
      mirroring=NoMirror
      resourceName=../../../Engine/Content/Slate/Starship/CoreWidgets/CheckBox/indeterminate.svg
      resourceObject=null
      tiling=NoTile
      imageSize=<16 16>
      margin=<ltrb 0 0 0 0>
      outlineSettings {
       bUseBrushTransparency=false
       roundingType=HalfHeightRadius
       width=0
       color {colorUseRule=UseColor_Specified, specifiedColor=<rgba 0 0 0 0>}
       cornerRadii=<0 0 0 0>
      }
      tintColor {colorUseRule=UseColor_Specified, specifiedColor=<rgba 0 0.0802 0.3515 1>}
      uVRegion {bIsValid=false, max=<0 0>, min=<0 0>}
     }
    }
   }
   slot {class=HorizontalBoxSlot, properties {horizontalAlignment=HAlign_Center, verticalAlignment=VAlign_Center}}
  }
  Border_Accolades {
   widgetClassPath=&Border
   parent=VerticalBoxLeftControls
   widget=Border_Accolades
   properties {brushColor=<rgba 0.0176 0.0176 0.0176 1>, padding=<ltrb 0 0 0 0>}
   slot {class=VerticalBoxSlot, properties {size {sizeRule=Fill, value=8}}}
  }
  VerticalBox_1 {
   widgetClassPath=&VerticalBox
   parent=Border_Accolades
   widget=VerticalBox_1
   properties {clipping=ClipToBounds}
   slot {class=BorderSlot, properties {padding=<ltrb 0 0 0 0>}}
  }
  DetailsView_Time {
   widgetClassPath=&DetailsView
   parent=VerticalBox_1
   widget=DetailsView_Time
   properties {columnWidth=0.4, categoriesToShow [ShowTimeValues, TimeValues]}
   slot {class=VerticalBoxSlot}
  }
  HorizontalBoxGroup {
   widgetClassPath=&HorizontalBox
   parent=VerticalBox_1
   widget=HorizontalBoxGroup
   properties {toolTipText="Active Group: Slect what Group (Team:Class) to simulate player being on.  \r\nThis can be changed at any time, even while running.  The value after | indicates how many accolades in this group have been triggered and will grant XP."}
   slot {class=VerticalBoxSlot, properties {padding=<ltrb 4 2 4 2>}}
  }
  textField_Group {
   widgetClassPath=&CommonTextBlock
   parent=HorizontalBoxGroup
   widget=textField_Group
   properties {
    justification=Right
    minDesiredWidth=150
    text=Active Group
    colorAndOpacity {colorUseRule=UseColor_Specified, specifiedColor=<rgba 0.849 0.849 0.849 1>}
    font {
     bForceMonospaced=false
     bMaterialIsStencil=false
     fontMaterial=null
     fontObject=&NotoSans
     letterSpacing=0
     monospacedWidth=1
     size=9
     skewAmount=0
     typefaceFontName=NotoSans-Regular
     outlineSettings {
      bApplyOutlineToDropShadows=false
      bMiteredCorners=false
      bSeparateFillAlpha=false
      outlineMaterial=null
      outlineSize=0
      outlineColor=<rgba 0 0 0 1>
     }
    }
    shadowColorAndOpacity=<rgba 0 0 0 0.5>
   }
   slot {
    class=HorizontalBoxSlot
    properties {horizontalAlignment=HAlign_Left, verticalAlignment=VAlign_Center, padding=<ltrb 0 0 4 0>, size {sizeRule=Fill, value=1}}
   }
  }
  ComboBox_Groups {
   widgetClassPath=&ComboBoxString
   bIsVariable=true
   parent=HorizontalBoxGroup
   widget=ComboBox_Groups
   properties {
    font {
     bForceMonospaced=false
     bMaterialIsStencil=false
     fontMaterial=null
     fontObject=&Roboto
     letterSpacing=0
     monospacedWidth=1
     size=12
     skewAmount=0
     typefaceFontName=Bold
     outlineSettings {
      bApplyOutlineToDropShadows=false
      bMiteredCorners=false
      bSeparateFillAlpha=false
      outlineMaterial=null
      outlineSize=0
      outlineColor=<rgba 0 0 0 1>
     }
    }
    foregroundColor {colorUseRule=UseColor_Specified, specifiedColor=<rgba 0.5271 0.5271 0.5271 1>}
    itemStyle {
     bUseParentRowBrush=false
     activeBrush {
      bIsDynamicallyLoaded=false
      drawAs=Image
      imageType=NoImage
      mirroring=NoMirror
      resourceName=None
      resourceObject=null
      tiling=NoTile
      imageSize=<0 0>
      margin=<ltrb 0 0 0 0>
      outlineSettings {
       bUseBrushTransparency=false
       roundingType=HalfHeightRadius
       width=0
       color {colorUseRule=UseColor_Specified, specifiedColor=<rgba 0 0 0 0>}
       cornerRadii=<0 0 0 0>
      }
      tintColor {colorUseRule=UseColor_Specified, specifiedColor=<rgba 0 0.162 0.7454 1>}
      uVRegion {bIsValid=false, max=<0 0>, min=<0 0>}
     }
     activeHighlightedBrush {
      bIsDynamicallyLoaded=false
      drawAs=Image
      imageType=NoImage
      mirroring=NoMirror
      resourceName=None
      resourceObject=null
      tiling=NoTile
      imageSize=<0 0>
      margin=<ltrb 0 0 0 0>
      outlineSettings {
       bUseBrushTransparency=false
       roundingType=HalfHeightRadius
       width=0
       color {colorUseRule=UseColor_Specified, specifiedColor=<rgba 0 0 0 0>}
       cornerRadii=<0 0 0 0>
      }
      tintColor {colorUseRule=UseColor_Specified, specifiedColor=<rgba 0.0252 0.0319 0.0423 1>}
      uVRegion {bIsValid=false, max=<0 0>, min=<0 0>}
     }
     activeHoveredBrush {
      bIsDynamicallyLoaded=false
      drawAs=Image
      imageType=NoImage
      mirroring=NoMirror
      resourceName=None
      resourceObject=null
      tiling=NoTile
      imageSize=<0 0>
      margin=<ltrb 0 0 0 0>
      outlineSettings {
       bUseBrushTransparency=false
       roundingType=HalfHeightRadius
       width=0
       color {colorUseRule=UseColor_Specified, specifiedColor=<rgba 0 0 0 0>}
       cornerRadii=<0 0 0 0>
      }
      tintColor {colorUseRule=UseColor_Specified, specifiedColor=<rgba 0 0.162 0.7454 1>}
      uVRegion {bIsValid=false, max=<0 0>, min=<0 0>}
     }
     dropIndicator_Above {
      bIsDynamicallyLoaded=false
      drawAs=Box
      imageType=FullColor
      mirroring=NoMirror
      resourceName=../../../Engine/Content/Slate/Common/DropZoneIndicator_Above.png
      resourceObject=null
      tiling=NoTile
      imageSize=<0 0>
      margin=<ltrb 0.625 0.625 0 0>
      outlineSettings {
       bUseBrushTransparency=false
       roundingType=HalfHeightRadius
       width=0
       color {colorUseRule=UseColor_Specified, specifiedColor=<rgba 0 0 0 0>}
       cornerRadii=<0 0 0 0>
      }
      tintColor {colorUseRule=UseColor_Specified, specifiedColor=<rgba 0.0091 0.3515 0.9647 1>}
      uVRegion {bIsValid=false, max=<0 0>, min=<0 0>}
     }
     dropIndicator_Below {
      bIsDynamicallyLoaded=false
      drawAs=Box
      imageType=FullColor
      mirroring=NoMirror
      resourceName=../../../Engine/Content/Slate/Common/DropZoneIndicator_Below.png
      resourceObject=null
      tiling=NoTile
      imageSize=<0 0>
      margin=<ltrb 0.625 0 0 0.625>
      outlineSettings {
       bUseBrushTransparency=false
       roundingType=HalfHeightRadius
       width=0
       color {colorUseRule=UseColor_Specified, specifiedColor=<rgba 0 0 0 0>}
       cornerRadii=<0 0 0 0>
      }
      tintColor {colorUseRule=UseColor_Specified, specifiedColor=<rgba 0.0091 0.3515 0.9647 1>}
      uVRegion {bIsValid=false, max=<0 0>, min=<0 0>}
     }
     dropIndicator_Onto {
      bIsDynamicallyLoaded=false
      drawAs=Box
      imageType=FullColor
      mirroring=NoMirror
      resourceName=../../../Engine/Content/Slate/Common/DropZoneIndicator_Onto.png
      resourceObject=null
      tiling=NoTile
      imageSize=<0 0>
      margin=<ltrb 0.25 0.25 0.25 0.25>
      outlineSettings {
       bUseBrushTransparency=false
       roundingType=HalfHeightRadius
       width=0
       color {colorUseRule=UseColor_Specified, specifiedColor=<rgba 0 0 0 0>}
       cornerRadii=<0 0 0 0>
      }
      tintColor {colorUseRule=UseColor_Specified, specifiedColor=<rgba 0.0091 0.3515 0.9647 1>}
      uVRegion {bIsValid=false, max=<0 0>, min=<0 0>}
     }
     evenRowBackgroundBrush {
      bIsDynamicallyLoaded=false
      drawAs=Image
      imageType=NoImage
      mirroring=NoMirror
      resourceName=None
      resourceObject=null
      tiling=NoTile
      imageSize=<0 0>
      margin=<ltrb 0 0 0 0>
      outlineSettings {
       bUseBrushTransparency=false
       roundingType=HalfHeightRadius
       width=0
       color {colorUseRule=UseColor_Specified, specifiedColor=<rgba 0 0 0 0>}
       cornerRadii=<0 0 0 0>
      }
      tintColor {colorUseRule=UseColor_Specified, specifiedColor=<rgba 0.0103 0.0103 0.0103 1>}
      uVRegion {bIsValid=false, max=<0 0>, min=<0 0>}
     }
     evenRowBackgroundHoveredBrush {
      bIsDynamicallyLoaded=false
      drawAs=Image
      imageType=NoImage
      mirroring=NoMirror
      resourceName=None
      resourceObject=null
      tiling=NoTile
      imageSize=<0 0>
      margin=<ltrb 0 0 0 0>
      outlineSettings {
       bUseBrushTransparency=false
       roundingType=HalfHeightRadius
       width=0
       color {colorUseRule=UseColor_Specified, specifiedColor=<rgba 0 0 0 0>}
       cornerRadii=<0 0 0 0>
      }
      tintColor {colorUseRule=UseColor_Specified, specifiedColor=<rgba 0.0176 0.0176 0.0176 1>}
      uVRegion {bIsValid=false, max=<0 0>, min=<0 0>}
     }
     inactiveBrush {
      bIsDynamicallyLoaded=false
      drawAs=Image
      imageType=NoImage
      mirroring=NoMirror
      resourceName=None
      resourceObject=null
      tiling=NoTile
      imageSize=<0 0>
      margin=<ltrb 0 0 0 0>
      outlineSettings {
       bUseBrushTransparency=false
       roundingType=HalfHeightRadius
       width=0
       color {colorUseRule=UseColor_Specified, specifiedColor=<rgba 0 0 0 0>}
       cornerRadii=<0 0 0 0>
      }
      tintColor {colorUseRule=UseColor_Specified, specifiedColor=<rgba 0.0513 0.0953 0.159 1>}
      uVRegion {bIsValid=false, max=<0 0>, min=<0 0>}
     }
     inactiveHighlightedBrush {
      bIsDynamicallyLoaded=false
      drawAs=Image
      imageType=NoImage
      mirroring=NoMirror
      resourceName=None
      resourceObject=null
      tiling=NoTile
      imageSize=<0 0>
      margin=<ltrb 0 0 0 0>
      outlineSettings {
       bUseBrushTransparency=false
       roundingType=HalfHeightRadius
       width=0
       color {colorUseRule=UseColor_Specified, specifiedColor=<rgba 0 0 0 0>}
       cornerRadii=<0 0 0 0>
      }
      tintColor {colorUseRule=UseColor_Specified, specifiedColor=<rgba 0.0252 0.0319 0.0423 1>}
      uVRegion {bIsValid=false, max=<0 0>, min=<0 0>}
     }
     inactiveHoveredBrush {
      bIsDynamicallyLoaded=false
      drawAs=Image
      imageType=NoImage
      mirroring=NoMirror
      resourceName=None
      resourceObject=null
      tiling=NoTile
      imageSize=<0 0>
      margin=<ltrb 0 0 0 0>
      outlineSettings {
       bUseBrushTransparency=false
       roundingType=HalfHeightRadius
       width=0
       color {colorUseRule=UseColor_Specified, specifiedColor=<rgba 0 0 0 0>}
       cornerRadii=<0 0 0 0>
      }
      tintColor {colorUseRule=UseColor_Specified, specifiedColor=<rgba 0.0513 0.0953 0.159 1>}
      uVRegion {bIsValid=false, max=<0 0>, min=<0 0>}
     }
     oddRowBackgroundBrush {
      bIsDynamicallyLoaded=false
      drawAs=Image
      imageType=NoImage
      mirroring=NoMirror
      resourceName=None
      resourceObject=null
      tiling=NoTile
      imageSize=<0 0>
      margin=<ltrb 0 0 0 0>
      outlineSettings {
       bUseBrushTransparency=false
       roundingType=HalfHeightRadius
       width=0
       color {colorUseRule=UseColor_Specified, specifiedColor=<rgba 0 0 0 0>}
       cornerRadii=<0 0 0 0>
      }
      tintColor {colorUseRule=UseColor_Specified, specifiedColor=<rgba 0.0103 0.0103 0.0103 1>}
      uVRegion {bIsValid=false, max=<0 0>, min=<0 0>}
     }
     oddRowBackgroundHoveredBrush {
      bIsDynamicallyLoaded=false
      drawAs=Image
      imageType=NoImage
      mirroring=NoMirror
      resourceName=None
      resourceObject=null
      tiling=NoTile
      imageSize=<0 0>
      margin=<ltrb 0 0 0 0>
      outlineSettings {
       bUseBrushTransparency=false
       roundingType=HalfHeightRadius
       width=0
       color {colorUseRule=UseColor_Specified, specifiedColor=<rgba 0 0 0 0>}
       cornerRadii=<0 0 0 0>
      }
      tintColor {colorUseRule=UseColor_Specified, specifiedColor=<rgba 0.0176 0.0176 0.0176 1>}
      uVRegion {bIsValid=false, max=<0 0>, min=<0 0>}
     }
     parentRowBackgroundBrush {
      bIsDynamicallyLoaded=false
      drawAs=Image
      imageType=NoImage
      mirroring=NoMirror
      resourceName=None
      resourceObject=null
      tiling=NoTile
      imageSize=<32 32>
      margin=<ltrb 0 0 0 0>
      outlineSettings {
       bUseBrushTransparency=false
       roundingType=HalfHeightRadius
       width=0
       color {colorUseRule=UseColor_Specified, specifiedColor=<rgba 0 0 0 0>}
       cornerRadii=<0 0 0 0>
      }
      tintColor {colorUseRule=UseColor_Specified, specifiedColor=<rgba 1 1 1 1>}
      uVRegion {bIsValid=false, max=<0 0>, min=<0 0>}
     }
     parentRowBackgroundHoveredBrush {
      bIsDynamicallyLoaded=false
      drawAs=Image
      imageType=NoImage
      mirroring=NoMirror
      resourceName=None
      resourceObject=null
      tiling=NoTile
      imageSize=<32 32>
      margin=<ltrb 0 0 0 0>
      outlineSettings {
       bUseBrushTransparency=false
       roundingType=HalfHeightRadius
       width=0
       color {colorUseRule=UseColor_Specified, specifiedColor=<rgba 0 0 0 0>}
       cornerRadii=<0 0 0 0>
      }
      tintColor {colorUseRule=UseColor_Specified, specifiedColor=<rgba 1 1 1 1>}
      uVRegion {bIsValid=false, max=<0 0>, min=<0 0>}
     }
     selectedTextColor {colorUseRule=UseColor_Specified, specifiedColor=<rgba 1 1 1 1>}
     selectorFocusedBrush {
      bIsDynamicallyLoaded=false
      drawAs=Border
      imageType=FullColor
      mirroring=NoMirror
      resourceName=../../../Engine/Content/Slate/Common/Selector.png
      resourceObject=null
      tiling=Both
      imageSize=<0 0>
      margin=<ltrb 0.25 0.25 0.25 0.25>
      outlineSettings {
       bUseBrushTransparency=false
       roundingType=HalfHeightRadius
       width=0
       color {colorUseRule=UseColor_Specified, specifiedColor=<rgba 0 0 0 0>}
       cornerRadii=<0 0 0 0>
      }
      tintColor {colorUseRule=UseColor_Specified, specifiedColor=<rgba 0.701 0.225 0.003 1>}
      uVRegion {bIsValid=false, max=<0 0>, min=<0 0>}
     }
     textColor {colorUseRule=UseColor_Specified, specifiedColor=<rgba 0.5271 0.5271 0.5271 1>}
    }
    scrollBarStyle {
     thickness=8
     draggedThumbImage {
      bIsDynamicallyLoaded=false
      drawAs=RoundedBox
      imageType=NoImage
      mirroring=NoMirror
      resourceName=None
      resourceObject=null
      tiling=NoTile
      imageSize=<0 0>
      margin=<ltrb 0 0 0 0>
      outlineSettings {
       bUseBrushTransparency=false
       roundingType=FixedRadius
       width=0
       color {colorUseRule=UseColor_Specified, specifiedColor=<rgba 0 0 0 0>}
       cornerRadii=<4 4 4 4>
      }
      tintColor {colorUseRule=UseColor_Specified, specifiedColor=<rgba 0.2159 0.2159 0.2159 1>}
      uVRegion {bIsValid=false, max=<0 0>, min=<0 0>}
     }
     horizontalBackgroundImage {
      bIsDynamicallyLoaded=false
      drawAs=NoDrawType
      imageType=NoImage
      mirroring=NoMirror
      resourceName=None
      resourceObject=null
      tiling=NoTile
      imageSize=<0 0>
      margin=<ltrb 0 0 0 0>
      outlineSettings {
       bUseBrushTransparency=false
       roundingType=HalfHeightRadius
       width=0
       color {colorUseRule=UseColor_Specified, specifiedColor=<rgba 0 0 0 0>}
       cornerRadii=<0 0 0 0>
      }
      tintColor {colorUseRule=UseColor_Specified, specifiedColor=<rgba 1 1 1 1>}
      uVRegion {bIsValid=false, max=<0 0>, min=<0 0>}
     }
     horizontalBottomSlotImage {
      bIsDynamicallyLoaded=false
      drawAs=NoDrawType
      imageType=NoImage
      mirroring=NoMirror
      resourceName=None
      resourceObject=null
      tiling=NoTile
      imageSize=<0 0>
      margin=<ltrb 0 0 0 0>
      outlineSettings {
       bUseBrushTransparency=false
       roundingType=HalfHeightRadius
       width=0
       color {colorUseRule=UseColor_Specified, specifiedColor=<rgba 0 0 0 0>}
       cornerRadii=<0 0 0 0>
      }
      tintColor {colorUseRule=UseColor_Specified, specifiedColor=<rgba 1 1 1 1>}
      uVRegion {bIsValid=false, max=<0 0>, min=<0 0>}
     }
     horizontalTopSlotImage {
      bIsDynamicallyLoaded=false
      drawAs=NoDrawType
      imageType=NoImage
      mirroring=NoMirror
      resourceName=None
      resourceObject=null
      tiling=NoTile
      imageSize=<0 0>
      margin=<ltrb 0 0 0 0>
      outlineSettings {
       bUseBrushTransparency=false
       roundingType=HalfHeightRadius
       width=0
       color {colorUseRule=UseColor_Specified, specifiedColor=<rgba 0 0 0 0>}
       cornerRadii=<0 0 0 0>
      }
      tintColor {colorUseRule=UseColor_Specified, specifiedColor=<rgba 1 1 1 1>}
      uVRegion {bIsValid=false, max=<0 0>, min=<0 0>}
     }
     hoveredThumbImage {
      bIsDynamicallyLoaded=false
      drawAs=RoundedBox
      imageType=NoImage
      mirroring=NoMirror
      resourceName=None
      resourceObject=null
      tiling=NoTile
      imageSize=<0 0>
      margin=<ltrb 0 0 0 0>
      outlineSettings {
       bUseBrushTransparency=false
       roundingType=FixedRadius
       width=0
       color {colorUseRule=UseColor_Specified, specifiedColor=<rgba 0 0 0 0>}
       cornerRadii=<4 4 4 4>
      }
      tintColor {colorUseRule=UseColor_Specified, specifiedColor=<rgba 0.2159 0.2159 0.2159 1>}
      uVRegion {bIsValid=false, max=<0 0>, min=<0 0>}
     }
     normalThumbImage {
      bIsDynamicallyLoaded=false
      drawAs=RoundedBox
      imageType=NoImage
      mirroring=NoMirror
      resourceName=None
      resourceObject=null
      tiling=NoTile
      imageSize=<0 0>
      margin=<ltrb 0 0 0 0>
      outlineSettings {
       bUseBrushTransparency=false
       roundingType=FixedRadius
       width=0
       color {colorUseRule=UseColor_Specified, specifiedColor=<rgba 0 0 0 0>}
       cornerRadii=<4 4 4 4>
      }
      tintColor {colorUseRule=UseColor_Specified, specifiedColor=<rgba 0.0953 0.0953 0.0953 1>}
      uVRegion {bIsValid=false, max=<0 0>, min=<0 0>}
     }
     verticalBackgroundImage {
      bIsDynamicallyLoaded=false
      drawAs=NoDrawType
      imageType=NoImage
      mirroring=NoMirror
      resourceName=None
      resourceObject=null
      tiling=NoTile
      imageSize=<0 0>
      margin=<ltrb 0 0 0 0>
      outlineSettings {
       bUseBrushTransparency=false
       roundingType=HalfHeightRadius
       width=0
       color {colorUseRule=UseColor_Specified, specifiedColor=<rgba 0 0 0 0>}
       cornerRadii=<0 0 0 0>
      }
      tintColor {colorUseRule=UseColor_Specified, specifiedColor=<rgba 1 1 1 1>}
      uVRegion {bIsValid=false, max=<0 0>, min=<0 0>}
     }
     verticalBottomSlotImage {
      bIsDynamicallyLoaded=false
      drawAs=NoDrawType
      imageType=NoImage
      mirroring=NoMirror
      resourceName=None
      resourceObject=null
      tiling=NoTile
      imageSize=<0 0>
      margin=<ltrb 0 0 0 0>
      outlineSettings {
       bUseBrushTransparency=false
       roundingType=HalfHeightRadius
       width=0
       color {colorUseRule=UseColor_Specified, specifiedColor=<rgba 0 0 0 0>}
       cornerRadii=<0 0 0 0>
      }
      tintColor {colorUseRule=UseColor_Specified, specifiedColor=<rgba 1 1 1 1>}
      uVRegion {bIsValid=false, max=<0 0>, min=<0 0>}
     }
     verticalTopSlotImage {
      bIsDynamicallyLoaded=false
      drawAs=NoDrawType
      imageType=NoImage
      mirroring=NoMirror
      resourceName=None
      resourceObject=null
      tiling=NoTile
      imageSize=<0 0>
      margin=<ltrb 0 0 0 0>
      outlineSettings {
       bUseBrushTransparency=false
       roundingType=HalfHeightRadius
       width=0
       color {colorUseRule=UseColor_Specified, specifiedColor=<rgba 0 0 0 0>}
       cornerRadii=<0 0 0 0>
      }
      tintColor {colorUseRule=UseColor_Specified, specifiedColor=<rgba 1 1 1 1>}
      uVRegion {bIsValid=false, max=<0 0>, min=<0 0>}
     }
    }
    widgetStyle {
     comboButtonStyle {
      downArrowAlign=VAlign_Center
      buttonStyle {
       clickedSlateSound {resourceObject=null}
       disabled {
        bIsDynamicallyLoaded=false
        drawAs=RoundedBox
        imageType=NoImage
        mirroring=NoMirror
        resourceName=None
        resourceObject=null
        tiling=NoTile
        imageSize=<0 0>
        margin=<ltrb 0 0 0 0>
        outlineSettings {
         bUseBrushTransparency=false
         roundingType=FixedRadius
         width=1
         color {colorUseRule=UseColor_ColorTable, specifiedColor=<rgba 0 0 0 0>}
         cornerRadii=<4 4 4 4>
        }
        tintColor {colorUseRule=UseColor_Specified, specifiedColor=<rgba 0.0395 0.0395 0.0395 1>}
        uVRegion {bIsValid=false, max=<0 0>, min=<0 0>}
       }
       disabledForeground {colorUseRule=UseColor_ColorTable, specifiedColor=<rgba 0 0 0 0>}
       hovered {
        bIsDynamicallyLoaded=false
        drawAs=RoundedBox
        imageType=NoImage
        mirroring=NoMirror
        resourceName=None
        resourceObject=null
        tiling=NoTile
        imageSize=<0 0>
        margin=<ltrb 0 0 0 0>
        outlineSettings {
         bUseBrushTransparency=false
         roundingType=FixedRadius
         width=1
         color {colorUseRule=UseColor_ColorTable, specifiedColor=<rgba 0 0 0 0>}
         cornerRadii=<4 4 4 4>
        }
        tintColor {colorUseRule=UseColor_Specified, specifiedColor=<rgba 0.0048 0.0048 0.0048 1>}
        uVRegion {bIsValid=false, max=<0 0>, min=<0 0>}
       }
       hoveredForeground {colorUseRule=UseColor_ColorTable, specifiedColor=<rgba 0 0 0 0>}
       hoveredSlateSound {resourceObject=null}
       normal {
        bIsDynamicallyLoaded=false
        drawAs=RoundedBox
        imageType=NoImage
        mirroring=NoMirror
        resourceName=None
        resourceObject=null
        tiling=NoTile
        imageSize=<0 0>
        margin=<ltrb 0 0 0 0>
        outlineSettings {
         bUseBrushTransparency=false
         roundingType=FixedRadius
         width=1
         color {colorUseRule=UseColor_ColorTable, specifiedColor=<rgba 0 0 0 0>}
         cornerRadii=<4 4 4 4>
        }
        tintColor {colorUseRule=UseColor_Specified, specifiedColor=<rgba 0.0048 0.0048 0.0048 1>}
        uVRegion {bIsValid=false, max=<0 0>, min=<0 0>}
       }
       normalForeground {colorUseRule=UseColor_ColorTable, specifiedColor=<rgba 0 0 0 0>}
       normalPadding=<ltrb 8 1 8 1>
       pressed {
        bIsDynamicallyLoaded=false
        drawAs=RoundedBox
        imageType=NoImage
        mirroring=NoMirror
        resourceName=None
        resourceObject=null
        tiling=NoTile
        imageSize=<0 0>
        margin=<ltrb 0 0 0 0>
        outlineSettings {
         bUseBrushTransparency=false
         roundingType=FixedRadius
         width=1
         color {colorUseRule=UseColor_ColorTable, specifiedColor=<rgba 0 0 0 0>}
         cornerRadii=<4 4 4 4>
        }
        tintColor {colorUseRule=UseColor_Specified, specifiedColor=<rgba 0.0048 0.0048 0.0048 1>}
        uVRegion {bIsValid=false, max=<0 0>, min=<0 0>}
       }
       pressedForeground {colorUseRule=UseColor_ColorTable, specifiedColor=<rgba 0 0 0 0>}
       pressedPadding=<ltrb 8 2 8 0>
       pressedSlateSound {resourceObject=null}
      }
      contentPadding=<ltrb 0 0 0 0>
      downArrowImage {
       bIsDynamicallyLoaded=false
       drawAs=Image
       imageType=Vector
       mirroring=NoMirror
       resourceName=../../../Engine/Content/Slate/Starship/CoreWidgets/ComboBox/wide-chevron-down.svg
       resourceObject=null
       tiling=NoTile
       imageSize=<16 16>
       margin=<ltrb 0 0 0 0>
       outlineSettings {
        bUseBrushTransparency=false
        roundingType=HalfHeightRadius
        width=0
        color {colorUseRule=UseColor_Specified, specifiedColor=<rgba 0 0 0 0>}
        cornerRadii=<0 0 0 0>
       }
       tintColor {colorUseRule=UseColor_Specified, specifiedColor=<rgba 1 1 1 1>}
       uVRegion {bIsValid=false, max=<0 0>, min=<0 0>}
      }
      downArrowPadding=<ltrb 2 2 2 2>
      menuBorderBrush {
       bIsDynamicallyLoaded=false
       drawAs=RoundedBox
       imageType=NoImage
       mirroring=NoMirror
       resourceName=None
       resourceObject=null
       tiling=NoTile
       imageSize=<0 0>
       margin=<ltrb 0 0 0 0>
       outlineSettings {
        bUseBrushTransparency=false
        roundingType=FixedRadius
        width=1
        color {colorUseRule=UseColor_ColorTable, specifiedColor=<rgba 0 0 0 0>}
        cornerRadii=<0 0 0 0>
       }
       tintColor {colorUseRule=UseColor_Specified, specifiedColor=<rgba 0.0395 0.0395 0.0395 1>}
       uVRegion {bIsValid=false, max=<0 0>, min=<0 0>}
      }
      menuBorderPadding=<ltrb 0 0 0 0>
      shadowColorAndOpacity=<rgba 0 0 0 1>
      shadowOffset=<0 0>
     }
     contentPadding=<ltrb 0 0 0 0>
     menuRowPadding=<ltrb 8 3 8 3>
     pressedSlateSound {resourceObject=null}
     selectionChangeSlateSound {resourceObject=null}
    }
   }
   slot {class=HorizontalBoxSlot, properties {size {sizeRule=Fill, value=2.5}}}
  }
  SizeBox_Accolades {
   widgetClassPath=&SizeBox
   parent=VerticalBox_1
   widget=SizeBox_Accolades
   properties {heightOverride=123.9655, maxDesiredHeight=260.1522}
   slot {class=VerticalBoxSlot, properties {padding=<ltrb 4 4 4 4>, size {sizeRule=Fill, value=1}}}
  }
  ListView_Accolades {
   widgetClassPath=&CommonListView
   bIsVariable=true
   parent=SizeBox_Accolades
   widget=ListView_Accolades
   properties {
    bAllowDragging=false
    bClearSelectionOnClick=true
    consumeMouseWheel=Always
    entryWidgetClass=/DynamicXp/Parts/DynamicXpAccoladeEntry.DynamicXpAccoladeEntry_C
    numDesignerPreviewEntries=6
    toolTipText="Click row to grant accolade.  The colums are:\r\nGUID/Name (Groups used in)\r\nTiggers Per Minute - From Link Data\r\nTPM Freqency Multiplier - to increase/descrease rate\r\nTime to trigger once. From :TPM, Freq Mult, and overall Freq Mult (above)\r\nTriggers/Last XP"
    scrollBarStyle {
     thickness=8
     draggedThumbImage {
      bIsDynamicallyLoaded=false
      drawAs=RoundedBox
      imageType=NoImage
      mirroring=NoMirror
      resourceName=None
      resourceObject=null
      tiling=NoTile
      imageSize=<0 0>
      margin=<ltrb 0 0 0 0>
      outlineSettings {
       bUseBrushTransparency=false
       roundingType=FixedRadius
       width=0
       color {colorUseRule=UseColor_Specified, specifiedColor=<rgba 0 0 0 0>}
       cornerRadii=<4 4 4 4>
      }
      tintColor {colorUseRule=UseColor_Specified, specifiedColor=<rgba 0.2159 0.2159 0.2159 1>}
      uVRegion {bIsValid=false, max=<0 0>, min=<0 0>}
     }
     horizontalBackgroundImage {
      bIsDynamicallyLoaded=false
      drawAs=NoDrawType
      imageType=NoImage
      mirroring=NoMirror
      resourceName=None
      resourceObject=null
      tiling=NoTile
      imageSize=<0 0>
      margin=<ltrb 0 0 0 0>
      outlineSettings {
       bUseBrushTransparency=false
       roundingType=HalfHeightRadius
       width=0
       color {colorUseRule=UseColor_Specified, specifiedColor=<rgba 0 0 0 0>}
       cornerRadii=<0 0 0 0>
      }
      tintColor {colorUseRule=UseColor_Specified, specifiedColor=<rgba 1 1 1 1>}
      uVRegion {bIsValid=false, max=<0 0>, min=<0 0>}
     }
     horizontalBottomSlotImage {
      bIsDynamicallyLoaded=false
      drawAs=NoDrawType
      imageType=NoImage
      mirroring=NoMirror
      resourceName=None
      resourceObject=null
      tiling=NoTile
      imageSize=<0 0>
      margin=<ltrb 0 0 0 0>
      outlineSettings {
       bUseBrushTransparency=false
       roundingType=HalfHeightRadius
       width=0
       color {colorUseRule=UseColor_Specified, specifiedColor=<rgba 0 0 0 0>}
       cornerRadii=<0 0 0 0>
      }
      tintColor {colorUseRule=UseColor_Specified, specifiedColor=<rgba 1 1 1 1>}
      uVRegion {bIsValid=false, max=<0 0>, min=<0 0>}
     }
     horizontalTopSlotImage {
      bIsDynamicallyLoaded=false
      drawAs=NoDrawType
      imageType=NoImage
      mirroring=NoMirror
      resourceName=None
      resourceObject=null
      tiling=NoTile
      imageSize=<0 0>
      margin=<ltrb 0 0 0 0>
      outlineSettings {
       bUseBrushTransparency=false
       roundingType=HalfHeightRadius
       width=0
       color {colorUseRule=UseColor_Specified, specifiedColor=<rgba 0 0 0 0>}
       cornerRadii=<0 0 0 0>
      }
      tintColor {colorUseRule=UseColor_Specified, specifiedColor=<rgba 1 1 1 1>}
      uVRegion {bIsValid=false, max=<0 0>, min=<0 0>}
     }
     hoveredThumbImage {
      bIsDynamicallyLoaded=false
      drawAs=RoundedBox
      imageType=NoImage
      mirroring=NoMirror
      resourceName=None
      resourceObject=null
      tiling=NoTile
      imageSize=<0 0>
      margin=<ltrb 0 0 0 0>
      outlineSettings {
       bUseBrushTransparency=false
       roundingType=FixedRadius
       width=0
       color {colorUseRule=UseColor_Specified, specifiedColor=<rgba 0 0 0 0>}
       cornerRadii=<4 4 4 4>
      }
      tintColor {colorUseRule=UseColor_Specified, specifiedColor=<rgba 0.2159 0.2159 0.2159 1>}
      uVRegion {bIsValid=false, max=<0 0>, min=<0 0>}
     }
     normalThumbImage {
      bIsDynamicallyLoaded=false
      drawAs=RoundedBox
      imageType=NoImage
      mirroring=NoMirror
      resourceName=None
      resourceObject=null
      tiling=NoTile
      imageSize=<0 0>
      margin=<ltrb 0 0 0 0>
      outlineSettings {
       bUseBrushTransparency=false
       roundingType=FixedRadius
       width=0
       color {colorUseRule=UseColor_Specified, specifiedColor=<rgba 0 0 0 0>}
       cornerRadii=<4 4 4 4>
      }
      tintColor {colorUseRule=UseColor_Specified, specifiedColor=<rgba 0.0953 0.0953 0.0953 1>}
      uVRegion {bIsValid=false, max=<0 0>, min=<0 0>}
     }
     verticalBackgroundImage {
      bIsDynamicallyLoaded=false
      drawAs=NoDrawType
      imageType=NoImage
      mirroring=NoMirror
      resourceName=None
      resourceObject=null
      tiling=NoTile
      imageSize=<0 0>
      margin=<ltrb 0 0 0 0>
      outlineSettings {
       bUseBrushTransparency=false
       roundingType=HalfHeightRadius
       width=0
       color {colorUseRule=UseColor_Specified, specifiedColor=<rgba 0 0 0 0>}
       cornerRadii=<0 0 0 0>
      }
      tintColor {colorUseRule=UseColor_Specified, specifiedColor=<rgba 1 1 1 1>}
      uVRegion {bIsValid=false, max=<0 0>, min=<0 0>}
     }
     verticalBottomSlotImage {
      bIsDynamicallyLoaded=false
      drawAs=NoDrawType
      imageType=NoImage
      mirroring=NoMirror
      resourceName=None
      resourceObject=null
      tiling=NoTile
      imageSize=<0 0>
      margin=<ltrb 0 0 0 0>
      outlineSettings {
       bUseBrushTransparency=false
       roundingType=HalfHeightRadius
       width=0
       color {colorUseRule=UseColor_Specified, specifiedColor=<rgba 0 0 0 0>}
       cornerRadii=<0 0 0 0>
      }
      tintColor {colorUseRule=UseColor_Specified, specifiedColor=<rgba 1 1 1 1>}
      uVRegion {bIsValid=false, max=<0 0>, min=<0 0>}
     }
     verticalTopSlotImage {
      bIsDynamicallyLoaded=false
      drawAs=NoDrawType
      imageType=NoImage
      mirroring=NoMirror
      resourceName=None
      resourceObject=null
      tiling=NoTile
      imageSize=<0 0>
      margin=<ltrb 0 0 0 0>
      outlineSettings {
       bUseBrushTransparency=false
       roundingType=HalfHeightRadius
       width=0
       color {colorUseRule=UseColor_Specified, specifiedColor=<rgba 0 0 0 0>}
       cornerRadii=<0 0 0 0>
      }
      tintColor {colorUseRule=UseColor_Specified, specifiedColor=<rgba 1 1 1 1>}
      uVRegion {bIsValid=false, max=<0 0>, min=<0 0>}
     }
    }
    widgetStyle {backgroundBrush {
      bIsDynamicallyLoaded=false
      drawAs=Image
      imageType=NoImage
      mirroring=NoMirror
      resourceName=None
      resourceObject=null
      tiling=NoTile
      imageSize=<0 0>
      margin=<ltrb 0 0 0 0>
      outlineSettings {
       bUseBrushTransparency=false
       roundingType=HalfHeightRadius
       width=0
       color {colorUseRule=UseColor_Specified, specifiedColor=<rgba 0 0 0 0>}
       cornerRadii=<0 0 0 0>
      }
      tintColor {colorUseRule=UseColor_Specified, specifiedColor=<rgba 0.0103 0.0103 0.0103 1>}
      uVRegion {bIsValid=false, max=<0 0>, min=<0 0>}
     }}
   }
   slot {class=SizeBoxSlot}
  }
  BorderXpPerMin {
   widgetClassPath=&Border
   parent=VerticalBoxLeftControls
   widget=BorderXpPerMin
   properties {
    background {
     bIsDynamicallyLoaded=false
     drawAs=Image
     imageType=NoImage
     mirroring=NoMirror
     resourceName=None
     resourceObject=null
     tiling=NoTile
     imageSize=<32 32>
     margin=<ltrb 0 0 0 0>
     outlineSettings {
      bUseBrushTransparency=false
      roundingType=HalfHeightRadius
      width=0
      color {colorUseRule=UseColor_Specified, specifiedColor=<rgba 0 0 0 0>}
      cornerRadii=<0 0 0 0>
     }
     tintColor {colorUseRule=UseColor_Specified, specifiedColor=<rgba 0.0103 0.0103 0.0103 1>}
     uVRegion {bIsValid=false, max=<0 0>, min=<0 0>}
    }
    padding=<ltrb 4 0 4 0>
   }
   slot {class=VerticalBoxSlot, properties {padding=<ltrb 4 1 4 1>}}
  }
  HorizontalBox_10 {
   widgetClassPath=&HorizontalBox
   parent=BorderXpPerMin
   widget=HorizontalBox_10
   slot {class=BorderSlot, properties {padding=<ltrb 4 0 4 0>}}
  }
  textFieldXpPerMin {
   widgetClassPath=&CommonTextBlock
   parent=HorizontalBox_10
   widget=textFieldXpPerMin
   properties {
    minDesiredWidth=150
    text=Actual XP/Min:
    toolTipText=Actual XP gained per minute
    font {
     bForceMonospaced=false
     bMaterialIsStencil=false
     fontMaterial=null
     fontObject=&NotoSans
     letterSpacing=0
     monospacedWidth=1
     size=9
     skewAmount=0
     typefaceFontName=NotoSans-Regular
     outlineSettings {
      bApplyOutlineToDropShadows=false
      bMiteredCorners=false
      bSeparateFillAlpha=false
      outlineMaterial=null
      outlineSize=0
      outlineColor=<rgba 0 0 0 1>
     }
    }
    shadowColorAndOpacity=<rgba 0 0 0 0.5>
   }
   slot {class=HorizontalBoxSlot, properties {horizontalAlignment=HAlign_Left, verticalAlignment=VAlign_Center, padding=<ltrb 1 1 1 1>}}
  }
  textActualXpPerMin {
   widgetClassPath=&CommonTextBlock
   bIsVariable=true
   parent=HorizontalBox_10
   widget=textActualXpPerMin
   properties {
    minDesiredWidth=150
    text=100/1000
    font {
     bForceMonospaced=false
     bMaterialIsStencil=false
     fontMaterial=null
     fontObject=&NotoSans
     letterSpacing=0
     monospacedWidth=1
     size=9
     skewAmount=0
     typefaceFontName=NotoSans-Regular
     outlineSettings {
      bApplyOutlineToDropShadows=false
      bMiteredCorners=false
      bSeparateFillAlpha=false
      outlineMaterial=null
      outlineSize=0
      outlineColor=<rgba 0 0 0 1>
     }
    }
    shadowColorAndOpacity=<rgba 0 0 0 0.5>
   }
   slot {class=HorizontalBoxSlot, properties {horizontalAlignment=HAlign_Right, verticalAlignment=VAlign_Center, padding=<ltrb 1 1 1 1>}}
  }
  PotentialXpPerMin {
   widgetClassPath=&CommonTextBlock
   parent=HorizontalBox_10
   widget=PotentialXpPerMin
   properties {
    minDesiredWidth=150
    text=Potential XP/Min:
    toolTipText="With no multipliers, what potentially can be earned."
    font {
     bForceMonospaced=false
     bMaterialIsStencil=false
     fontMaterial=null
     fontObject=&NotoSans
     letterSpacing=0
     monospacedWidth=1
     size=9
     skewAmount=0
     typefaceFontName=NotoSans-Regular
     outlineSettings {
      bApplyOutlineToDropShadows=false
      bMiteredCorners=false
      bSeparateFillAlpha=false
      outlineMaterial=null
      outlineSize=0
      outlineColor=<rgba 0 0 0 1>
     }
    }
    shadowColorAndOpacity=<rgba 0 0 0 0.5>
   }
   slot {class=HorizontalBoxSlot, properties {horizontalAlignment=HAlign_Left, verticalAlignment=VAlign_Center, padding=<ltrb 1 1 1 1>}}
  }
  textPotentialXpPerMin {
   widgetClassPath=&CommonTextBlock
   bIsVariable=true
   parent=HorizontalBox_10
   widget=textPotentialXpPerMin
   properties {
    minDesiredWidth=150
    text=100/1000
    font {
     bForceMonospaced=false
     bMaterialIsStencil=false
     fontMaterial=null
     fontObject=&NotoSans
     letterSpacing=0
     monospacedWidth=1
     size=9
     skewAmount=0
     typefaceFontName=NotoSans-Regular
     outlineSettings {
      bApplyOutlineToDropShadows=false
      bMiteredCorners=false
      bSeparateFillAlpha=false
      outlineMaterial=null
      outlineSize=0
      outlineColor=<rgba 0 0 0 1>
     }
    }
    shadowColorAndOpacity=<rgba 0 0 0 0.5>
   }
   slot {class=HorizontalBoxSlot, properties {horizontalAlignment=HAlign_Right, verticalAlignment=VAlign_Center, padding=<ltrb 1 1 1 1>}}
  }
  BorderAccumilation {
   widgetClassPath=&Border
   parent=VerticalBoxLeftControls
   widget=BorderAccumilation
   properties {background {
     bIsDynamicallyLoaded=false
     drawAs=Image
     imageType=NoImage
     mirroring=NoMirror
     resourceName=None
     resourceObject=null
     tiling=NoTile
     imageSize=<32 32>
     margin=<ltrb 0 0 0 0>
     outlineSettings {
      bUseBrushTransparency=false
      roundingType=HalfHeightRadius
      width=0
      color {colorUseRule=UseColor_Specified, specifiedColor=<rgba 0 0 0 0>}
      cornerRadii=<0 0 0 0>
     }
     tintColor {colorUseRule=UseColor_Specified, specifiedColor=<rgba 0.0103 0.0103 0.0103 1>}
     uVRegion {bIsValid=false, max=<0 0>, min=<0 0>}
    }}
   slot {class=VerticalBoxSlot, properties {padding=<ltrb 4 1 4 1>}}
  }
  VerticalBox_183 {widgetClassPath=&VerticalBox, parent=BorderAccumilation, widget=VerticalBox_183, slot {class=BorderSlot}}
  Overlay_4 {
   widgetClassPath=&Overlay
   parent=VerticalBox_183
   widget=Overlay_4
   slot {class=VerticalBoxSlot, properties {size {sizeRule=Fill, value=1}}}
  }
  progressBarBucket {
   widgetClassPath=/Script/UMG.ProgressBar
   bIsVariable=true
   parent=Overlay_4
   widget=progressBarBucket
   properties {
    bIsMarquee=true
    toolTipText=Amount in XP bucket that can affect Bank Mult when player gains too much XP.
    fillColorAndOpacity=<rgba 0 0.5 1 0.2202>
    widgetStyle {
     enableFillAnimation=true
     backgroundImage {
      bIsDynamicallyLoaded=false
      drawAs=NoDrawType
      imageType=NoImage
      mirroring=NoMirror
      resourceName=None
      resourceObject=null
      tiling=NoTile
      imageSize=<0 0>
      margin=<ltrb 0 0 0 0>
      outlineSettings {
       bUseBrushTransparency=false
       roundingType=HalfHeightRadius
       width=0
       color {colorUseRule=UseColor_Specified, specifiedColor=<rgba 0 0 0 0>}
       cornerRadii=<0 0 0 0>
      }
      tintColor {colorUseRule=UseColor_Specified, specifiedColor=<rgba 0.0048 0.0048 0.0048 1>}
      uVRegion {bIsValid=false, max=<0 0>, min=<0 0>}
     }
     fillImage {
      bIsDynamicallyLoaded=false
      drawAs=Image
      imageType=FullColor
      mirroring=NoMirror
      resourceName=../../../Engine/Content/Slate/Starship/CoreWidgets/ProgressBar/ProgressMarquee.png
      resourceObject=null
      tiling=Horizontal
      imageSize=<20 12>
      margin=<ltrb 0 0 0 0>
      outlineSettings {
       bUseBrushTransparency=false
       roundingType=HalfHeightRadius
       width=0
       color {colorUseRule=UseColor_Specified, specifiedColor=<rgba 0 0 0 0>}
       cornerRadii=<0 0 0 0>
      }
      tintColor {colorUseRule=UseColor_Specified, specifiedColor=<rgba 1 1 1 1>}
      uVRegion {bIsValid=false, max=<0 0>, min=<0 0>}
     }
     marqueeImage {
      bIsDynamicallyLoaded=false
      drawAs=Image
      imageType=FullColor
      mirroring=NoMirror
      resourceName=../../../Engine/Content/Slate/Starship/CoreWidgets/ProgressBar/ProgressMarquee.png
      resourceObject=null
      tiling=Horizontal
      imageSize=<20 12>
      margin=<ltrb 0 0 0 0>
      outlineSettings {
       bUseBrushTransparency=false
       roundingType=HalfHeightRadius
       width=0
       color {colorUseRule=UseColor_Specified, specifiedColor=<rgba 0 0 0 0>}
       cornerRadii=<0 0 0 0>
      }
      tintColor {colorUseRule=UseColor_Specified, specifiedColor=<rgba 0.0044 0.2384 1 1>}
      uVRegion {bIsValid=false, max=<0 0>, min=<0 0>}
     }
    }
   }
   slot {class=OverlaySlot, properties {horizontalAlignment=HAlign_Fill, verticalAlignment=VAlign_Fill}}
  }
  textFieldBucket {
   widgetClassPath=&CommonTextBlock
   parent=Overlay_4
   widget=textFieldBucket
   properties {
    minDesiredWidth=150
    text=Bucket:
    toolTipText=Prefix GUID of Accolade
    font {
     bForceMonospaced=false
     bMaterialIsStencil=false
     fontMaterial=null
     fontObject=&NotoSans
     letterSpacing=0
     monospacedWidth=1
     size=9
     skewAmount=0
     typefaceFontName=NotoSans-Regular
     outlineSettings {
      bApplyOutlineToDropShadows=false
      bMiteredCorners=false
      bSeparateFillAlpha=false
      outlineMaterial=null
      outlineSize=0
      outlineColor=<rgba 0 0 0 1>
     }
    }
    shadowColorAndOpacity=<rgba 0 0 0 0.5>
   }
   slot {class=OverlaySlot, properties {verticalAlignment=VAlign_Center, padding=<ltrb 1 1 1 1>}}
  }
  textFieldBucketValue {
   widgetClassPath=&CommonTextBlock
   bIsVariable=true
   parent=Overlay_4
   widget=textFieldBucketValue
   properties {
    minDesiredWidth=150
    text=100/1000
    font {
     bForceMonospaced=false
     bMaterialIsStencil=false
     fontMaterial=null
     fontObject=&NotoSans
     letterSpacing=0
     monospacedWidth=1
     size=9
     skewAmount=0
     typefaceFontName=NotoSans-Regular
     outlineSettings {
      bApplyOutlineToDropShadows=false
      bMiteredCorners=false
      bSeparateFillAlpha=false
      outlineMaterial=null
      outlineSize=0
      outlineColor=<rgba 0 0 0 1>
     }
    }
    shadowColorAndOpacity=<rgba 0 0 0 0.5>
   }
   slot {class=OverlaySlot, properties {horizontalAlignment=HAlign_Right, verticalAlignment=VAlign_Center, padding=<ltrb 1 1 1 1>}}
  }
  Overlay_5 {
   widgetClassPath=&Overlay
   parent=VerticalBox_183
   widget=Overlay_5
   slot {class=VerticalBoxSlot, properties {size {sizeRule=Fill, value=1}}}
  }
  ProgressBarTimespan {
   widgetClassPath=&ProgressBar
   bIsVariable=true
   parent=Overlay_5
   widget=ProgressBarTimespan
   properties {
    bIsMarquee=true
    toolTipText=How far into current time splice used to calulation excess XP penalties.
    fillColorAndOpacity=<rgba 0 0.5 1 0.2202>
    widgetStyle {
     enableFillAnimation=true
     backgroundImage {
      bIsDynamicallyLoaded=false
      drawAs=NoDrawType
      imageType=NoImage
      mirroring=NoMirror
      resourceName=None
      resourceObject=null
      tiling=NoTile
      imageSize=<0 0>
      margin=<ltrb 0 0 0 0>
      outlineSettings {
       bUseBrushTransparency=false
       roundingType=HalfHeightRadius
       width=0
       color {colorUseRule=UseColor_Specified, specifiedColor=<rgba 0 0 0 0>}
       cornerRadii=<0 0 0 0>
      }
      tintColor {colorUseRule=UseColor_Specified, specifiedColor=<rgba 0.0048 0.0048 0.0048 1>}
      uVRegion {bIsValid=false, max=<0 0>, min=<0 0>}
     }
     fillImage {
      bIsDynamicallyLoaded=false
      drawAs=Image
      imageType=FullColor
      mirroring=NoMirror
      resourceName=../../../Engine/Content/Slate/Starship/CoreWidgets/ProgressBar/ProgressMarquee.png
      resourceObject=null
      tiling=Horizontal
      imageSize=<20 12>
      margin=<ltrb 0 0 0 0>
      outlineSettings {
       bUseBrushTransparency=false
       roundingType=HalfHeightRadius
       width=0
       color {colorUseRule=UseColor_Specified, specifiedColor=<rgba 0 0 0 0>}
       cornerRadii=<0 0 0 0>
      }
      tintColor {colorUseRule=UseColor_Specified, specifiedColor=<rgba 1 1 1 1>}
      uVRegion {bIsValid=false, max=<0 0>, min=<0 0>}
     }
     marqueeImage {
      bIsDynamicallyLoaded=false
      drawAs=Image
      imageType=FullColor
      mirroring=NoMirror
      resourceName=../../../Engine/Content/Slate/Starship/CoreWidgets/ProgressBar/ProgressMarquee.png
      resourceObject=null
      tiling=Horizontal
      imageSize=<20 12>
      margin=<ltrb 0 0 0 0>
      outlineSettings {
       bUseBrushTransparency=false
       roundingType=HalfHeightRadius
       width=0
       color {colorUseRule=UseColor_Specified, specifiedColor=<rgba 0 0 0 0>}
       cornerRadii=<0 0 0 0>
      }
      tintColor {colorUseRule=UseColor_Specified, specifiedColor=<rgba 0.0044 0.2384 1 1>}
      uVRegion {bIsValid=false, max=<0 0>, min=<0 0>}
     }
    }
   }
   slot {class=OverlaySlot, properties {horizontalAlignment=HAlign_Fill, verticalAlignment=VAlign_Fill}}
  }
  textFieldTimespan {
   widgetClassPath=&CommonTextBlock
   parent=Overlay_5
   widget=textFieldTimespan
   properties {
    minDesiredWidth=150
    text=Timespan:
    toolTipText=Prefix GUID of Accolade
    font {
     bForceMonospaced=false
     bMaterialIsStencil=false
     fontMaterial=null
     fontObject=&NotoSans
     letterSpacing=0
     monospacedWidth=1
     size=9
     skewAmount=0
     typefaceFontName=NotoSans-Regular
     outlineSettings {
      bApplyOutlineToDropShadows=false
      bMiteredCorners=false
      bSeparateFillAlpha=false
      outlineMaterial=null
      outlineSize=0
      outlineColor=<rgba 0 0 0 1>
     }
    }
    shadowColorAndOpacity=<rgba 0 0 0 0.5>
   }
   slot {class=OverlaySlot, properties {verticalAlignment=VAlign_Center, padding=<ltrb 1 1 1 1>}}
  }
  textFieldTimespanValue {
   widgetClassPath=&CommonTextBlock
   bIsVariable=true
   parent=Overlay_5
   widget=textFieldTimespanValue
   properties {
    minDesiredWidth=150
    text=100/1000
    font {
     bForceMonospaced=false
     bMaterialIsStencil=false
     fontMaterial=null
     fontObject=&NotoSans
     letterSpacing=0
     monospacedWidth=1
     size=9
     skewAmount=0
     typefaceFontName=NotoSans-Regular
     outlineSettings {
      bApplyOutlineToDropShadows=false
      bMiteredCorners=false
      bSeparateFillAlpha=false
      outlineMaterial=null
      outlineSize=0
      outlineColor=<rgba 0 0 0 1>
     }
    }
    shadowColorAndOpacity=<rgba 0 0 0 0.5>
   }
   slot {class=OverlaySlot, properties {horizontalAlignment=HAlign_Right, verticalAlignment=VAlign_Center, padding=<ltrb 1 1 1 1>}}
  }
  Overlay_6 {
   widgetClassPath=&Overlay
   parent=VerticalBox_183
   widget=Overlay_6
   slot {class=VerticalBoxSlot, properties {size {sizeRule=Fill, value=1}}}
  }
  ProgressBar_Weekly {
   widgetClassPath=&ProgressBar
   bIsVariable=true
   parent=Overlay_6
   widget=ProgressBar_Weekly
   properties {
    bIsMarquee=true
    toolTipText=Total amount of XP gained today
    fillColorAndOpacity=<rgba 0 0.5 1 0.2202>
    widgetStyle {
     enableFillAnimation=true
     backgroundImage {
      bIsDynamicallyLoaded=false
      drawAs=NoDrawType
      imageType=NoImage
      mirroring=NoMirror
      resourceName=None
      resourceObject=null
      tiling=NoTile
      imageSize=<0 0>
      margin=<ltrb 0 0 0 0>
      outlineSettings {
       bUseBrushTransparency=false
       roundingType=HalfHeightRadius
       width=0
       color {colorUseRule=UseColor_Specified, specifiedColor=<rgba 0 0 0 0>}
       cornerRadii=<0 0 0 0>
      }
      tintColor {colorUseRule=UseColor_Specified, specifiedColor=<rgba 0.0048 0.0048 0.0048 1>}
      uVRegion {bIsValid=false, max=<0 0>, min=<0 0>}
     }
     fillImage {
      bIsDynamicallyLoaded=false
      drawAs=Image
      imageType=FullColor
      mirroring=NoMirror
      resourceName=../../../Engine/Content/Slate/Starship/CoreWidgets/ProgressBar/ProgressMarquee.png
      resourceObject=null
      tiling=Horizontal
      imageSize=<20 12>
      margin=<ltrb 0 0 0 0>
      outlineSettings {
       bUseBrushTransparency=false
       roundingType=HalfHeightRadius
       width=0
       color {colorUseRule=UseColor_Specified, specifiedColor=<rgba 0 0 0 0>}
       cornerRadii=<0 0 0 0>
      }
      tintColor {colorUseRule=UseColor_Specified, specifiedColor=<rgba 1 1 1 1>}
      uVRegion {bIsValid=false, max=<0 0>, min=<0 0>}
     }
     marqueeImage {
      bIsDynamicallyLoaded=false
      drawAs=Image
      imageType=FullColor
      mirroring=NoMirror
      resourceName=../../../Engine/Content/Slate/Starship/CoreWidgets/ProgressBar/ProgressMarquee.png
      resourceObject=null
      tiling=Horizontal
      imageSize=<20 12>
      margin=<ltrb 0 0 0 0>
      outlineSettings {
       bUseBrushTransparency=false
       roundingType=HalfHeightRadius
       width=0
       color {colorUseRule=UseColor_Specified, specifiedColor=<rgba 0 0 0 0>}
       cornerRadii=<0 0 0 0>
      }
      tintColor {colorUseRule=UseColor_Specified, specifiedColor=<rgba 0.0044 0.2384 1 1>}
      uVRegion {bIsValid=false, max=<0 0>, min=<0 0>}
     }
    }
   }
   slot {class=OverlaySlot, properties {horizontalAlignment=HAlign_Fill, verticalAlignment=VAlign_Fill}}
  }
  textFieldWeekly {
   widgetClassPath=&CommonTextBlock
   parent=Overlay_6
   widget=textFieldWeekly
   properties {
    minDesiredWidth=150
    text=Weekly XP:
    toolTipText=Prefix GUID of Accolade
    font {
     bForceMonospaced=false
     bMaterialIsStencil=false
     fontMaterial=null
     fontObject=&NotoSans
     letterSpacing=0
     monospacedWidth=1
     size=9
     skewAmount=0
     typefaceFontName=NotoSans-Regular
     outlineSettings {
      bApplyOutlineToDropShadows=false
      bMiteredCorners=false
      bSeparateFillAlpha=false
      outlineMaterial=null
      outlineSize=0
      outlineColor=<rgba 0 0 0 1>
     }
    }
    shadowColorAndOpacity=<rgba 0 0 0 0.5>
   }
   slot {class=OverlaySlot, properties {verticalAlignment=VAlign_Center, padding=<ltrb 1 1 1 1>}}
  }
  textFieldWeeklyValue {
   widgetClassPath=&CommonTextBlock
   bIsVariable=true
   parent=Overlay_6
   widget=textFieldWeeklyValue
   properties {
    minDesiredWidth=150
    text=100/1000
    font {
     bForceMonospaced=false
     bMaterialIsStencil=false
     fontMaterial=null
     fontObject=&NotoSans
     letterSpacing=0
     monospacedWidth=1
     size=9
     skewAmount=0
     typefaceFontName=NotoSans-Regular
     outlineSettings {
      bApplyOutlineToDropShadows=false
      bMiteredCorners=false
      bSeparateFillAlpha=false
      outlineMaterial=null
      outlineSize=0
      outlineColor=<rgba 0 0 0 1>
     }
    }
    shadowColorAndOpacity=<rgba 0 0 0 0.5>
   }
   slot {class=OverlaySlot, properties {horizontalAlignment=HAlign_Right, verticalAlignment=VAlign_Center, padding=<ltrb 1 1 1 1>}}
  }
  BorderBooster {
   widgetClassPath=&Border
   parent=VerticalBoxLeftControls
   widget=BorderBooster
   properties {background {
     bIsDynamicallyLoaded=false
     drawAs=Image
     imageType=NoImage
     mirroring=NoMirror
     resourceName=None
     resourceObject=null
     tiling=NoTile
     imageSize=<32 32>
     margin=<ltrb 0 0 0 0>
     outlineSettings {
      bUseBrushTransparency=false
      roundingType=HalfHeightRadius
      width=0
      color {colorUseRule=UseColor_Specified, specifiedColor=<rgba 0 0 0 0>}
      cornerRadii=<0 0 0 0>
     }
     tintColor {colorUseRule=UseColor_Specified, specifiedColor=<rgba 0.0103 0.0103 0.0103 1>}
     uVRegion {bIsValid=false, max=<0 0>, min=<0 0>}
    }}
   slot {class=VerticalBoxSlot, properties {padding=<ltrb 4 1 4 1>}}
  }
  VerticalBox_3 {widgetClassPath=&VerticalBox, parent=BorderBooster, widget=VerticalBox_3, slot {class=BorderSlot}}
  Overlay_7 {
   widgetClassPath=&Overlay
   parent=VerticalBox_3
   widget=Overlay_7
   slot {class=VerticalBoxSlot, properties {size {sizeRule=Fill, value=1}}}
  }
  ProgressBar_XpBooster {
   widgetClassPath=&ProgressBar
   bIsVariable=true
   parent=Overlay_7
   widget=ProgressBar_XpBooster
   properties {
    bIsMarquee=true
    toolTipText=Help in caluclation of Boster Mult
    fillColorAndOpacity=<rgba 0 0.5 1 0.2202>
    widgetStyle {
     enableFillAnimation=true
     backgroundImage {
      bIsDynamicallyLoaded=false
      drawAs=NoDrawType
      imageType=NoImage
      mirroring=NoMirror
      resourceName=None
      resourceObject=null
      tiling=NoTile
      imageSize=<0 0>
      margin=<ltrb 0 0 0 0>
      outlineSettings {
       bUseBrushTransparency=false
       roundingType=HalfHeightRadius
       width=0
       color {colorUseRule=UseColor_Specified, specifiedColor=<rgba 0 0 0 0>}
       cornerRadii=<0 0 0 0>
      }
      tintColor {colorUseRule=UseColor_Specified, specifiedColor=<rgba 0.0048 0.0048 0.0048 1>}
      uVRegion {bIsValid=false, max=<0 0>, min=<0 0>}
     }
     fillImage {
      bIsDynamicallyLoaded=false
      drawAs=Image
      imageType=FullColor
      mirroring=NoMirror
      resourceName=../../../Engine/Content/Slate/Starship/CoreWidgets/ProgressBar/ProgressMarquee.png
      resourceObject=null
      tiling=Horizontal
      imageSize=<20 12>
      margin=<ltrb 0 0 0 0>
      outlineSettings {
       bUseBrushTransparency=false
       roundingType=HalfHeightRadius
       width=0
       color {colorUseRule=UseColor_Specified, specifiedColor=<rgba 0 0 0 0>}
       cornerRadii=<0 0 0 0>
      }
      tintColor {colorUseRule=UseColor_Specified, specifiedColor=<rgba 1 1 1 1>}
      uVRegion {bIsValid=false, max=<0 0>, min=<0 0>}
     }
     marqueeImage {
      bIsDynamicallyLoaded=false
      drawAs=Image
      imageType=FullColor
      mirroring=NoMirror
      resourceName=../../../Engine/Content/Slate/Starship/CoreWidgets/ProgressBar/ProgressMarquee.png
      resourceObject=null
      tiling=Horizontal
      imageSize=<20 12>
      margin=<ltrb 0 0 0 0>
      outlineSettings {
       bUseBrushTransparency=false
       roundingType=HalfHeightRadius
       width=0
       color {colorUseRule=UseColor_Specified, specifiedColor=<rgba 0 0 0 0>}
       cornerRadii=<0 0 0 0>
      }
      tintColor {colorUseRule=UseColor_Specified, specifiedColor=<rgba 0.0044 0.2384 1 1>}
      uVRegion {bIsValid=false, max=<0 0>, min=<0 0>}
     }
    }
   }
   slot {class=OverlaySlot, properties {horizontalAlignment=HAlign_Fill, verticalAlignment=VAlign_Fill}}
  }
  textFieldBoosterBucket {
   widgetClassPath=&CommonTextBlock
   parent=Overlay_7
   widget=textFieldBoosterBucket
   properties {
    minDesiredWidth=150
    text=Booster Bucket:
    toolTipText=Prefix GUID of Accolade
    font {
     bForceMonospaced=false
     bMaterialIsStencil=false
     fontMaterial=null
     fontObject=&NotoSans
     letterSpacing=0
     monospacedWidth=1
     size=9
     skewAmount=0
     typefaceFontName=NotoSans-Regular
     outlineSettings {
      bApplyOutlineToDropShadows=false
      bMiteredCorners=false
      bSeparateFillAlpha=false
      outlineMaterial=null
      outlineSize=0
      outlineColor=<rgba 0 0 0 1>
     }
    }
    shadowColorAndOpacity=<rgba 0 0 0 0.5>
   }
   slot {class=OverlaySlot, properties {verticalAlignment=VAlign_Center, padding=<ltrb 1 1 1 1>}}
  }
  textFieldBoosterBucketValue {
   widgetClassPath=&CommonTextBlock
   bIsVariable=true
   parent=Overlay_7
   widget=textFieldBoosterBucketValue
   properties {
    minDesiredWidth=150
    text=100/1000
    font {
     bForceMonospaced=false
     bMaterialIsStencil=false
     fontMaterial=null
     fontObject=&NotoSans
     letterSpacing=0
     monospacedWidth=1
     size=9
     skewAmount=0
     typefaceFontName=NotoSans-Regular
     outlineSettings {
      bApplyOutlineToDropShadows=false
      bMiteredCorners=false
      bSeparateFillAlpha=false
      outlineMaterial=null
      outlineSize=0
      outlineColor=<rgba 0 0 0 1>
     }
    }
    shadowColorAndOpacity=<rgba 0 0 0 0.5>
   }
   slot {class=OverlaySlot, properties {horizontalAlignment=HAlign_Right, verticalAlignment=VAlign_Center, padding=<ltrb 1 1 1 1>}}
  }
  Overlay_8 {widgetClassPath=&Overlay, parent=VerticalBox_3, widget=Overlay_8, slot {class=VerticalBoxSlot}}
  ProgressBar_XpBoosterMult {
   widgetClassPath=&ProgressBar
   bIsVariable=true
   parent=Overlay_8
   widget=ProgressBar_XpBoosterMult
   properties {
    bIsMarquee=true
    toolTipText="Potential XP is multiplied by this and added to granted XP if player is underperforming.  Future values to be used do not match actual percentage used in last calulation."
    fillColorAndOpacity=<rgba 0 0.5 1 0.2202>
    widgetStyle {
     enableFillAnimation=true
     backgroundImage {
      bIsDynamicallyLoaded=false
      drawAs=NoDrawType
      imageType=NoImage
      mirroring=NoMirror
      resourceName=None
      resourceObject=null
      tiling=NoTile
      imageSize=<0 0>
      margin=<ltrb 0 0 0 0>
      outlineSettings {
       bUseBrushTransparency=false
       roundingType=HalfHeightRadius
       width=0
       color {colorUseRule=UseColor_Specified, specifiedColor=<rgba 0 0 0 0>}
       cornerRadii=<0 0 0 0>
      }
      tintColor {colorUseRule=UseColor_Specified, specifiedColor=<rgba 0.0048 0.0048 0.0048 1>}
      uVRegion {bIsValid=false, max=<0 0>, min=<0 0>}
     }
     fillImage {
      bIsDynamicallyLoaded=false
      drawAs=Image
      imageType=FullColor
      mirroring=NoMirror
      resourceName=../../../Engine/Content/Slate/Starship/CoreWidgets/ProgressBar/ProgressMarquee.png
      resourceObject=null
      tiling=Horizontal
      imageSize=<20 12>
      margin=<ltrb 0 0 0 0>
      outlineSettings {
       bUseBrushTransparency=false
       roundingType=HalfHeightRadius
       width=0
       color {colorUseRule=UseColor_Specified, specifiedColor=<rgba 0 0 0 0>}
       cornerRadii=<0 0 0 0>
      }
      tintColor {colorUseRule=UseColor_Specified, specifiedColor=<rgba 1 1 1 1>}
      uVRegion {bIsValid=false, max=<0 0>, min=<0 0>}
     }
     marqueeImage {
      bIsDynamicallyLoaded=false
      drawAs=Image
      imageType=FullColor
      mirroring=NoMirror
      resourceName=../../../Engine/Content/Slate/Starship/CoreWidgets/ProgressBar/ProgressMarquee.png
      resourceObject=null
      tiling=Horizontal
      imageSize=<20 12>
      margin=<ltrb 0 0 0 0>
      outlineSettings {
       bUseBrushTransparency=false
       roundingType=HalfHeightRadius
       width=0
       color {colorUseRule=UseColor_Specified, specifiedColor=<rgba 0 0 0 0>}
       cornerRadii=<0 0 0 0>
      }
      tintColor {colorUseRule=UseColor_Specified, specifiedColor=<rgba 0.0044 0.2384 1 1>}
      uVRegion {bIsValid=false, max=<0 0>, min=<0 0>}
     }
    }
   }
   slot {class=OverlaySlot, properties {horizontalAlignment=HAlign_Fill, verticalAlignment=VAlign_Fill}}
  }
  txtFieldXpBoosterMult {
   widgetClassPath=&CommonTextBlock
   parent=Overlay_8
   widget=txtFieldXpBoosterMult
   properties {
    minDesiredWidth=150
    text=Xp Booster Mult:
    toolTipText=Prefix GUID of Accolade
    font {
     bForceMonospaced=false
     bMaterialIsStencil=false
     fontMaterial=null
     fontObject=&NotoSans
     letterSpacing=0
     monospacedWidth=1
     size=9
     skewAmount=0
     typefaceFontName=NotoSans-Regular
     outlineSettings {
      bApplyOutlineToDropShadows=false
      bMiteredCorners=false
      bSeparateFillAlpha=false
      outlineMaterial=null
      outlineSize=0
      outlineColor=<rgba 0 0 0 1>
     }
    }
    shadowColorAndOpacity=<rgba 0 0 0 0.5>
   }
   slot {class=OverlaySlot, properties {verticalAlignment=VAlign_Center, padding=<ltrb 1 1 1 1>}}
  }
  textFieldXpBoosterMult {
   widgetClassPath=&CommonTextBlock
   bIsVariable=true
   parent=Overlay_8
   widget=textFieldXpBoosterMult
   properties {
    minDesiredWidth=150
    text=100/1000
    font {
     bForceMonospaced=false
     bMaterialIsStencil=false
     fontMaterial=null
     fontObject=&NotoSans
     letterSpacing=0
     monospacedWidth=1
     size=9
     skewAmount=0
     typefaceFontName=NotoSans-Regular
     outlineSettings {
      bApplyOutlineToDropShadows=false
      bMiteredCorners=false
      bSeparateFillAlpha=false
      outlineMaterial=null
      outlineSize=0
      outlineColor=<rgba 0 0 0 1>
     }
    }
    shadowColorAndOpacity=<rgba 0 0 0 0.5>
   }
   slot {class=OverlaySlot, properties {horizontalAlignment=HAlign_Right, verticalAlignment=VAlign_Center, padding=<ltrb 1 1 1 1>}}
  }
  BorderXpMults {
   widgetClassPath=&Border
   parent=VerticalBoxLeftControls
   widget=BorderXpMults
   properties {background {
     bIsDynamicallyLoaded=false
     drawAs=Image
     imageType=NoImage
     mirroring=NoMirror
     resourceName=None
     resourceObject=null
     tiling=NoTile
     imageSize=<32 32>
     margin=<ltrb 0 0 0 0>
     outlineSettings {
      bUseBrushTransparency=false
      roundingType=HalfHeightRadius
      width=0
      color {colorUseRule=UseColor_Specified, specifiedColor=<rgba 0 0 0 0>}
      cornerRadii=<0 0 0 0>
     }
     tintColor {colorUseRule=UseColor_Specified, specifiedColor=<rgba 0.0103 0.0103 0.0103 1>}
     uVRegion {bIsValid=false, max=<0 0>, min=<0 0>}
    }}
   slot {class=VerticalBoxSlot, properties {padding=<ltrb 4 1 4 1>}}
  }
  Overlay_3 {widgetClassPath=&Overlay, parent=BorderXpMults, widget=Overlay_3, slot {class=BorderSlot}}
  VerticalBox_456 {
   widgetClassPath=&VerticalBox
   parent=Overlay_3
   widget=VerticalBox_456
   slot {class=OverlaySlot, properties {horizontalAlignment=HAlign_Fill, verticalAlignment=VAlign_Fill}}
  }
  Overlay_0 {
   widgetClassPath=&Overlay
   parent=VerticalBox_456
   widget=Overlay_0
   slot {class=VerticalBoxSlot, properties {size {sizeRule=Fill, value=1}}}
  }
  ProgressBar_Afk {
   widgetClassPath=&ProgressBar
   bIsVariable=true
   parent=Overlay_0
   widget=ProgressBar_Afk
   properties {
    bIsMarquee=true
    fillColorAndOpacity=<rgba 0 0.5 1 0.2202>
    widgetStyle {
     enableFillAnimation=true
     backgroundImage {
      bIsDynamicallyLoaded=false
      drawAs=NoDrawType
      imageType=NoImage
      mirroring=NoMirror
      resourceName=None
      resourceObject=null
      tiling=NoTile
      imageSize=<0 0>
      margin=<ltrb 0 0 0 0>
      outlineSettings {
       bUseBrushTransparency=false
       roundingType=HalfHeightRadius
       width=0
       color {colorUseRule=UseColor_Specified, specifiedColor=<rgba 0 0 0 0>}
       cornerRadii=<0 0 0 0>
      }
      tintColor {colorUseRule=UseColor_Specified, specifiedColor=<rgba 0.0048 0.0048 0.0048 1>}
      uVRegion {bIsValid=false, max=<0 0>, min=<0 0>}
     }
     fillImage {
      bIsDynamicallyLoaded=false
      drawAs=Image
      imageType=FullColor
      mirroring=NoMirror
      resourceName=../../../Engine/Content/Slate/Starship/CoreWidgets/ProgressBar/ProgressMarquee.png
      resourceObject=null
      tiling=Horizontal
      imageSize=<20 12>
      margin=<ltrb 0 0 0 0>
      outlineSettings {
       bUseBrushTransparency=false
       roundingType=HalfHeightRadius
       width=0
       color {colorUseRule=UseColor_Specified, specifiedColor=<rgba 0 0 0 0>}
       cornerRadii=<0 0 0 0>
      }
      tintColor {colorUseRule=UseColor_Specified, specifiedColor=<rgba 1 1 1 1>}
      uVRegion {bIsValid=false, max=<0 0>, min=<0 0>}
     }
     marqueeImage {
      bIsDynamicallyLoaded=false
      drawAs=Image
      imageType=FullColor
      mirroring=NoMirror
      resourceName=../../../Engine/Content/Slate/Starship/CoreWidgets/ProgressBar/ProgressMarquee.png
      resourceObject=null
      tiling=Horizontal
      imageSize=<20 12>
      margin=<ltrb 0 0 0 0>
      outlineSettings {
       bUseBrushTransparency=false
       roundingType=HalfHeightRadius
       width=0
       color {colorUseRule=UseColor_Specified, specifiedColor=<rgba 0 0 0 0>}
       cornerRadii=<0 0 0 0>
      }
      tintColor {colorUseRule=UseColor_Specified, specifiedColor=<rgba 0.0044 0.2384 1 1>}
      uVRegion {bIsValid=false, max=<0 0>, min=<0 0>}
     }
    }
   }
   slot {class=OverlaySlot, properties {horizontalAlignment=HAlign_Fill, verticalAlignment=VAlign_Fill}}
  }
  textField_Accolade {
   widgetClassPath=&CommonTextBlock
   parent=Overlay_0
   widget=textField_Accolade
   properties {
    minDesiredWidth=150
    text=AFK Mult
    toolTipText=Prefix GUID of Accolade
    font {
     bForceMonospaced=false
     bMaterialIsStencil=false
     fontMaterial=null
     fontObject=&NotoSans
     letterSpacing=0
     monospacedWidth=1
     size=9
     skewAmount=0
     typefaceFontName=NotoSans-Regular
     outlineSettings {
      bApplyOutlineToDropShadows=false
      bMiteredCorners=false
      bSeparateFillAlpha=false
      outlineMaterial=null
      outlineSize=0
      outlineColor=<rgba 0 0 0 1>
     }
    }
    shadowColorAndOpacity=<rgba 0 0 0 0.5>
   }
   slot {class=OverlaySlot, properties {verticalAlignment=VAlign_Center, padding=<ltrb 1 1 1 1>}}
  }
  Overlay {widgetClassPath=&Overlay, parent=VerticalBox_456, widget=Overlay, slot {class=VerticalBoxSlot}}
  ProgressBar_Bank {
   widgetClassPath=&ProgressBar
   bIsVariable=true
   parent=Overlay
   widget=ProgressBar_Bank
   properties {
    bIsMarquee=true
    fillColorAndOpacity=<rgba 0 0.5 1 0.2202>
    widgetStyle {
     enableFillAnimation=true
     backgroundImage {
      bIsDynamicallyLoaded=false
      drawAs=NoDrawType
      imageType=NoImage
      mirroring=NoMirror
      resourceName=None
      resourceObject=null
      tiling=NoTile
      imageSize=<0 0>
      margin=<ltrb 0 0 0 0>
      outlineSettings {
       bUseBrushTransparency=false
       roundingType=HalfHeightRadius
       width=0
       color {colorUseRule=UseColor_Specified, specifiedColor=<rgba 0 0 0 0>}
       cornerRadii=<0 0 0 0>
      }
      tintColor {colorUseRule=UseColor_Specified, specifiedColor=<rgba 0.0048 0.0048 0.0048 1>}
      uVRegion {bIsValid=false, max=<0 0>, min=<0 0>}
     }
     fillImage {
      bIsDynamicallyLoaded=false
      drawAs=Image
      imageType=FullColor
      mirroring=NoMirror
      resourceName=../../../Engine/Content/Slate/Starship/CoreWidgets/ProgressBar/ProgressMarquee.png
      resourceObject=null
      tiling=Horizontal
      imageSize=<20 12>
      margin=<ltrb 0 0 0 0>
      outlineSettings {
       bUseBrushTransparency=false
       roundingType=HalfHeightRadius
       width=0
       color {colorUseRule=UseColor_Specified, specifiedColor=<rgba 0 0 0 0>}
       cornerRadii=<0 0 0 0>
      }
      tintColor {colorUseRule=UseColor_Specified, specifiedColor=<rgba 1 1 1 1>}
      uVRegion {bIsValid=false, max=<0 0>, min=<0 0>}
     }
     marqueeImage {
      bIsDynamicallyLoaded=false
      drawAs=Image
      imageType=FullColor
      mirroring=NoMirror
      resourceName=../../../Engine/Content/Slate/Starship/CoreWidgets/ProgressBar/ProgressMarquee.png
      resourceObject=null
      tiling=Horizontal
      imageSize=<20 12>
      margin=<ltrb 0 0 0 0>
      outlineSettings {
       bUseBrushTransparency=false
       roundingType=HalfHeightRadius
       width=0
       color {colorUseRule=UseColor_Specified, specifiedColor=<rgba 0 0 0 0>}
       cornerRadii=<0 0 0 0>
      }
      tintColor {colorUseRule=UseColor_Specified, specifiedColor=<rgba 0.0044 0.2384 1 1>}
      uVRegion {bIsValid=false, max=<0 0>, min=<0 0>}
     }
    }
   }
   slot {class=OverlaySlot, properties {horizontalAlignment=HAlign_Fill, verticalAlignment=VAlign_Fill}}
  }
  ProgresssBar_BankMult {
   widgetClassPath=&CommonTextBlock
   bIsVariable=true
   parent=Overlay
   widget=ProgresssBar_BankMult
   properties {
    minDesiredWidth=150
    text=Bank Mult
    toolTipText=Prefix GUID of Accolade
    font {
     bForceMonospaced=false
     bMaterialIsStencil=false
     fontMaterial=null
     fontObject=&NotoSans
     letterSpacing=0
     monospacedWidth=1
     size=9
     skewAmount=0
     typefaceFontName=NotoSans-Regular
     outlineSettings {
      bApplyOutlineToDropShadows=false
      bMiteredCorners=false
      bSeparateFillAlpha=false
      outlineMaterial=null
      outlineSize=0
      outlineColor=<rgba 0 0 0 1>
     }
    }
    shadowColorAndOpacity=<rgba 0 0 0 0.5>
   }
   slot {class=OverlaySlot, properties {verticalAlignment=VAlign_Center, padding=<ltrb 1 1 1 1>}}
  }
  Overlay_1 {
   widgetClassPath=&Overlay
   parent=VerticalBox_456
   widget=Overlay_1
   slot {class=VerticalBoxSlot, properties {size {sizeRule=Fill, value=1}}}
  }
  ProgressBar_WeeklyMult {
   widgetClassPath=&ProgressBar
   bIsVariable=true
   parent=Overlay_1
   widget=ProgressBar_WeeklyMult
   properties {
    bIsMarquee=true
    fillColorAndOpacity=<rgba 0 0.5 1 0.2202>
    widgetStyle {
     enableFillAnimation=true
     backgroundImage {
      bIsDynamicallyLoaded=false
      drawAs=NoDrawType
      imageType=NoImage
      mirroring=NoMirror
      resourceName=None
      resourceObject=null
      tiling=NoTile
      imageSize=<0 0>
      margin=<ltrb 0 0 0 0>
      outlineSettings {
       bUseBrushTransparency=false
       roundingType=HalfHeightRadius
       width=0
       color {colorUseRule=UseColor_Specified, specifiedColor=<rgba 0 0 0 0>}
       cornerRadii=<0 0 0 0>
      }
      tintColor {colorUseRule=UseColor_Specified, specifiedColor=<rgba 0.0048 0.0048 0.0048 1>}
      uVRegion {bIsValid=false, max=<0 0>, min=<0 0>}
     }
     fillImage {
      bIsDynamicallyLoaded=false
      drawAs=Image
      imageType=FullColor
      mirroring=NoMirror
      resourceName=../../../Engine/Content/Slate/Starship/CoreWidgets/ProgressBar/ProgressMarquee.png
      resourceObject=null
      tiling=Horizontal
      imageSize=<20 12>
      margin=<ltrb 0 0 0 0>
      outlineSettings {
       bUseBrushTransparency=false
       roundingType=HalfHeightRadius
       width=0
       color {colorUseRule=UseColor_Specified, specifiedColor=<rgba 0 0 0 0>}
       cornerRadii=<0 0 0 0>
      }
      tintColor {colorUseRule=UseColor_Specified, specifiedColor=<rgba 1 1 1 1>}
      uVRegion {bIsValid=false, max=<0 0>, min=<0 0>}
     }
     marqueeImage {
      bIsDynamicallyLoaded=false
      drawAs=Image
      imageType=FullColor
      mirroring=NoMirror
      resourceName=../../../Engine/Content/Slate/Starship/CoreWidgets/ProgressBar/ProgressMarquee.png
      resourceObject=null
      tiling=Horizontal
      imageSize=<20 12>
      margin=<ltrb 0 0 0 0>
      outlineSettings {
       bUseBrushTransparency=false
       roundingType=HalfHeightRadius
       width=0
       color {colorUseRule=UseColor_Specified, specifiedColor=<rgba 0 0 0 0>}
       cornerRadii=<0 0 0 0>
      }
      tintColor {colorUseRule=UseColor_Specified, specifiedColor=<rgba 0.0044 0.2384 1 1>}
      uVRegion {bIsValid=false, max=<0 0>, min=<0 0>}
     }
    }
   }
   slot {class=OverlaySlot, properties {horizontalAlignment=HAlign_Fill, verticalAlignment=VAlign_Fill}}
  }
  textField_Accolade_1 {
   widgetClassPath=&CommonTextBlock
   parent=Overlay_1
   widget=textField_Accolade_1
   properties {
    minDesiredWidth=150
    text=Weekly Mult
    toolTipText=Prefix GUID of Accolade
    font {
     bForceMonospaced=false
     bMaterialIsStencil=false
     fontMaterial=null
     fontObject=&NotoSans
     letterSpacing=0
     monospacedWidth=1
     size=9
     skewAmount=0
     typefaceFontName=NotoSans-Regular
     outlineSettings {
      bApplyOutlineToDropShadows=false
      bMiteredCorners=false
      bSeparateFillAlpha=false
      outlineMaterial=null
      outlineSize=0
      outlineColor=<rgba 0 0 0 1>
     }
    }
    shadowColorAndOpacity=<rgba 0 0 0 0.5>
   }
   slot {class=OverlaySlot, properties {verticalAlignment=VAlign_Center, padding=<ltrb 1 1 1 1>}}
  }
  Overlay_2 {
   widgetClassPath=&Overlay
   parent=VerticalBox_456
   widget=Overlay_2
   slot {class=VerticalBoxSlot, properties {size {sizeRule=Fill, value=1}}}
  }
  ProgressBar_XpVarianceMult {
   widgetClassPath=&ProgressBar
   bIsVariable=true
   parent=Overlay_2
   widget=ProgressBar_XpVarianceMult
   properties {
    bIsMarquee=true
    fillColorAndOpacity=<rgba 0 0.5 1 0.2202>
    widgetStyle {
     enableFillAnimation=true
     backgroundImage {
      bIsDynamicallyLoaded=false
      drawAs=NoDrawType
      imageType=NoImage
      mirroring=NoMirror
      resourceName=None
      resourceObject=null
      tiling=NoTile
      imageSize=<0 0>
      margin=<ltrb 0 0 0 0>
      outlineSettings {
       bUseBrushTransparency=false
       roundingType=HalfHeightRadius
       width=0
       color {colorUseRule=UseColor_Specified, specifiedColor=<rgba 0 0 0 0>}
       cornerRadii=<0 0 0 0>
      }
      tintColor {colorUseRule=UseColor_Specified, specifiedColor=<rgba 0.0048 0.0048 0.0048 1>}
      uVRegion {bIsValid=false, max=<0 0>, min=<0 0>}
     }
     fillImage {
      bIsDynamicallyLoaded=false
      drawAs=Image
      imageType=FullColor
      mirroring=NoMirror
      resourceName=../../../Engine/Content/Slate/Starship/CoreWidgets/ProgressBar/ProgressMarquee.png
      resourceObject=null
      tiling=Horizontal
      imageSize=<20 12>
      margin=<ltrb 0 0 0 0>
      outlineSettings {
       bUseBrushTransparency=false
       roundingType=HalfHeightRadius
       width=0
       color {colorUseRule=UseColor_Specified, specifiedColor=<rgba 0 0 0 0>}
       cornerRadii=<0 0 0 0>
      }
      tintColor {colorUseRule=UseColor_Specified, specifiedColor=<rgba 1 1 1 1>}
      uVRegion {bIsValid=false, max=<0 0>, min=<0 0>}
     }
     marqueeImage {
      bIsDynamicallyLoaded=false
      drawAs=Image
      imageType=FullColor
      mirroring=NoMirror
      resourceName=../../../Engine/Content/Slate/Starship/CoreWidgets/ProgressBar/ProgressMarquee.png
      resourceObject=null
      tiling=Horizontal
      imageSize=<20 12>
      margin=<ltrb 0 0 0 0>
      outlineSettings {
       bUseBrushTransparency=false
       roundingType=HalfHeightRadius
       width=0
       color {colorUseRule=UseColor_Specified, specifiedColor=<rgba 0 0 0 0>}
       cornerRadii=<0 0 0 0>
      }
      tintColor {colorUseRule=UseColor_Specified, specifiedColor=<rgba 0.0044 0.2384 1 1>}
      uVRegion {bIsValid=false, max=<0 0>, min=<0 0>}
     }
    }
   }
   slot {class=OverlaySlot, properties {horizontalAlignment=HAlign_Fill, verticalAlignment=VAlign_Fill}}
  }
  textField_XpVarianceMult {
   widgetClassPath=&CommonTextBlock
   parent=Overlay_2
   widget=textField_XpVarianceMult
   properties {
    minDesiredWidth=150
    text=XP Variance Mult
    toolTipText=Prefix GUID of Accolade
    font {
     bForceMonospaced=false
     bMaterialIsStencil=false
     fontMaterial=null
     fontObject=&NotoSans
     letterSpacing=0
     monospacedWidth=1
     size=9
     skewAmount=0
     typefaceFontName=NotoSans-Regular
     outlineSettings {
      bApplyOutlineToDropShadows=false
      bMiteredCorners=false
      bSeparateFillAlpha=false
      outlineMaterial=null
      outlineSize=0
      outlineColor=<rgba 0 0 0 1>
     }
    }
    shadowColorAndOpacity=<rgba 0 0 0 0.5>
   }
   slot {class=OverlaySlot, properties {verticalAlignment=VAlign_Center, padding=<ltrb 1 1 1 1>}}
  }
  TextBoxTotalMult {
   widgetClassPath=&EditableTextBox
   bIsVariable=true
   parent=Overlay_3
   widget=TextBoxTotalMult
   properties {
    justification=Right
    text=100%
    widgetStyle {
     backgroundColor {colorUseRule=UseColor_Specified, specifiedColor=<rgba 1 1 1 0>}
     backgroundImageFocused {
      bIsDynamicallyLoaded=false
      drawAs=RoundedBox
      imageType=NoImage
      mirroring=NoMirror
      resourceName=None
      resourceObject=null
      tiling=NoTile
      imageSize=<0 0>
      margin=<ltrb 0 0 0 0>
      outlineSettings {
       bUseBrushTransparency=false
       roundingType=FixedRadius
       width=1
       color {colorUseRule=UseColor_ColorTable, specifiedColor=<rgba 0 0 0 0>}
       cornerRadii=<4 4 4 4>
      }
      tintColor {colorUseRule=UseColor_Specified, specifiedColor=<rgba 0.0048 0.0048 0.0048 0>}
      uVRegion {bIsValid=false, max=<0 0>, min=<0 0>}
     }
     backgroundImageHovered {
      bIsDynamicallyLoaded=false
      drawAs=RoundedBox
      imageType=NoImage
      mirroring=NoMirror
      resourceName=None
      resourceObject=null
      tiling=NoTile
      imageSize=<0 0>
      margin=<ltrb 0 0 0 0>
      outlineSettings {
       bUseBrushTransparency=false
       roundingType=FixedRadius
       width=1
       color {colorUseRule=UseColor_ColorTable, specifiedColor=<rgba 0 0 0 0>}
       cornerRadii=<4 4 4 4>
      }
      tintColor {colorUseRule=UseColor_Specified, specifiedColor=<rgba 0.0048 0.0048 0.0048 0>}
      uVRegion {bIsValid=false, max=<0 0>, min=<0 0>}
     }
     backgroundImageNormal {
      bIsDynamicallyLoaded=false
      drawAs=RoundedBox
      imageType=NoImage
      mirroring=NoMirror
      resourceName=None
      resourceObject=null
      tiling=NoTile
      imageSize=<0 0>
      margin=<ltrb 0 0 0 0>
      outlineSettings {
       bUseBrushTransparency=false
       roundingType=FixedRadius
       width=1
       color {colorUseRule=UseColor_ColorTable, specifiedColor=<rgba 0 0 0 0>}
       cornerRadii=<4 4 4 4>
      }
      tintColor {colorUseRule=UseColor_Specified, specifiedColor=<rgba 0.0048 0.0048 0.0048 0>}
      uVRegion {bIsValid=false, max=<0 0>, min=<0 0>}
     }
     backgroundImageReadOnly {
      bIsDynamicallyLoaded=false
      drawAs=RoundedBox
      imageType=NoImage
      mirroring=NoMirror
      resourceName=None
      resourceObject=null
      tiling=NoTile
      imageSize=<0 0>
      margin=<ltrb 0 0 0 0>
      outlineSettings {
       bUseBrushTransparency=false
       roundingType=FixedRadius
       width=1
       color {colorUseRule=UseColor_ColorTable, specifiedColor=<rgba 0 0 0 0>}
       cornerRadii=<4 4 4 4>
      }
      tintColor {colorUseRule=UseColor_Specified, specifiedColor=<rgba 0.0284 0.0284 0.0284 0>}
      uVRegion {bIsValid=false, max=<0 0>, min=<0 0>}
     }
     focusedForegroundColor {colorUseRule=UseColor_ColorTable, specifiedColor=<rgba 0 0 0 0>}
     foregroundColor {colorUseRule=UseColor_Specified, specifiedColor=<rgba 0.5271 0.5271 0.5271 1>}
     hScrollBarPadding=<ltrb 0 0 0 0>
     padding=<ltrb 0 0 0 0>
     readOnlyForegroundColor {colorUseRule=UseColor_Foreground, specifiedColor=<rgba 1 0 1 0>}
     scrollBarStyle {
      thickness=8
      draggedThumbImage {
       bIsDynamicallyLoaded=false
       drawAs=RoundedBox
       imageType=NoImage
       mirroring=NoMirror
       resourceName=None
       resourceObject=null
       tiling=NoTile
       imageSize=<0 0>
       margin=<ltrb 0 0 0 0>
       outlineSettings {
        bUseBrushTransparency=false
        roundingType=FixedRadius
        width=0
        color {colorUseRule=UseColor_Specified, specifiedColor=<rgba 0 0 0 0>}
        cornerRadii=<4 4 4 4>
       }
       tintColor {colorUseRule=UseColor_Specified, specifiedColor=<rgba 0.2159 0.2159 0.2159 0>}
       uVRegion {bIsValid=false, max=<0 0>, min=<0 0>}
      }
      horizontalBackgroundImage {
       bIsDynamicallyLoaded=false
       drawAs=NoDrawType
       imageType=NoImage
       mirroring=NoMirror
       resourceName=None
       resourceObject=null
       tiling=NoTile
       imageSize=<0 0>
       margin=<ltrb 0 0 0 0>
       outlineSettings {
        bUseBrushTransparency=false
        roundingType=HalfHeightRadius
        width=0
        color {colorUseRule=UseColor_Specified, specifiedColor=<rgba 0 0 0 0>}
        cornerRadii=<0 0 0 0>
       }
       tintColor {colorUseRule=UseColor_Specified, specifiedColor=<rgba 1 1 1 0>}
       uVRegion {bIsValid=false, max=<0 0>, min=<0 0>}
      }
      horizontalBottomSlotImage {
       bIsDynamicallyLoaded=false
       drawAs=NoDrawType
       imageType=NoImage
       mirroring=NoMirror
       resourceName=None
       resourceObject=null
       tiling=NoTile
       imageSize=<0 0>
       margin=<ltrb 0 0 0 0>
       outlineSettings {
        bUseBrushTransparency=false
        roundingType=HalfHeightRadius
        width=0
        color {colorUseRule=UseColor_Specified, specifiedColor=<rgba 0 0 0 0>}
        cornerRadii=<0 0 0 0>
       }
       tintColor {colorUseRule=UseColor_Specified, specifiedColor=<rgba 1 1 1 0>}
       uVRegion {bIsValid=false, max=<0 0>, min=<0 0>}
      }
      horizontalTopSlotImage {
       bIsDynamicallyLoaded=false
       drawAs=NoDrawType
       imageType=NoImage
       mirroring=NoMirror
       resourceName=None
       resourceObject=null
       tiling=NoTile
       imageSize=<0 0>
       margin=<ltrb 0 0 0 0>
       outlineSettings {
        bUseBrushTransparency=false
        roundingType=HalfHeightRadius
        width=0
        color {colorUseRule=UseColor_Specified, specifiedColor=<rgba 0 0 0 0>}
        cornerRadii=<0 0 0 0>
       }
       tintColor {colorUseRule=UseColor_Specified, specifiedColor=<rgba 1 1 1 0>}
       uVRegion {bIsValid=false, max=<0 0>, min=<0 0>}
      }
      hoveredThumbImage {
       bIsDynamicallyLoaded=false
       drawAs=RoundedBox
       imageType=NoImage
       mirroring=NoMirror
       resourceName=None
       resourceObject=null
       tiling=NoTile
       imageSize=<0 0>
       margin=<ltrb 0 0 0 0>
       outlineSettings {
        bUseBrushTransparency=false
        roundingType=FixedRadius
        width=0
        color {colorUseRule=UseColor_Specified, specifiedColor=<rgba 0 0 0 0>}
        cornerRadii=<4 4 4 4>
       }
       tintColor {colorUseRule=UseColor_Specified, specifiedColor=<rgba 0.2159 0.2159 0.2159 0>}
       uVRegion {bIsValid=false, max=<0 0>, min=<0 0>}
      }
      normalThumbImage {
       bIsDynamicallyLoaded=false
       drawAs=RoundedBox
       imageType=NoImage
       mirroring=NoMirror
       resourceName=None
       resourceObject=null
       tiling=NoTile
       imageSize=<0 0>
       margin=<ltrb 0 0 0 0>
       outlineSettings {
        bUseBrushTransparency=false
        roundingType=FixedRadius
        width=0
        color {colorUseRule=UseColor_Specified, specifiedColor=<rgba 0 0 0 0>}
        cornerRadii=<4 4 4 4>
       }
       tintColor {colorUseRule=UseColor_Specified, specifiedColor=<rgba 0.0953 0.0953 0.0953 0>}
       uVRegion {bIsValid=false, max=<0 0>, min=<0 0>}
      }
      verticalBackgroundImage {
       bIsDynamicallyLoaded=false
       drawAs=NoDrawType
       imageType=NoImage
       mirroring=NoMirror
       resourceName=None
       resourceObject=null
       tiling=NoTile
       imageSize=<0 0>
       margin=<ltrb 0 0 0 0>
       outlineSettings {
        bUseBrushTransparency=false
        roundingType=HalfHeightRadius
        width=0
        color {colorUseRule=UseColor_Specified, specifiedColor=<rgba 0 0 0 0>}
        cornerRadii=<0 0 0 0>
       }
       tintColor {colorUseRule=UseColor_Specified, specifiedColor=<rgba 1 1 1 0>}
       uVRegion {bIsValid=false, max=<0 0>, min=<0 0>}
      }
      verticalBottomSlotImage {
       bIsDynamicallyLoaded=false
       drawAs=NoDrawType
       imageType=NoImage
       mirroring=NoMirror
       resourceName=None
       resourceObject=null
       tiling=NoTile
       imageSize=<0 0>
       margin=<ltrb 0 0 0 0>
       outlineSettings {
        bUseBrushTransparency=false
        roundingType=HalfHeightRadius
        width=0
        color {colorUseRule=UseColor_Specified, specifiedColor=<rgba 0 0 0 0>}
        cornerRadii=<0 0 0 0>
       }
       tintColor {colorUseRule=UseColor_Specified, specifiedColor=<rgba 1 1 1 0>}
       uVRegion {bIsValid=false, max=<0 0>, min=<0 0>}
      }
      verticalTopSlotImage {
       bIsDynamicallyLoaded=false
       drawAs=NoDrawType
       imageType=NoImage
       mirroring=NoMirror
       resourceName=None
       resourceObject=null
       tiling=NoTile
       imageSize=<0 0>
       margin=<ltrb 0 0 0 0>
       outlineSettings {
        bUseBrushTransparency=false
        roundingType=HalfHeightRadius
        width=0
        color {colorUseRule=UseColor_Specified, specifiedColor=<rgba 0 0 0 0>}
        cornerRadii=<0 0 0 0>
       }
       tintColor {colorUseRule=UseColor_Specified, specifiedColor=<rgba 1 1 1 0>}
       uVRegion {bIsValid=false, max=<0 0>, min=<0 0>}
      }
     }
     textStyle {
      overflowPolicy=Clip
      transformPolicy=None
      colorAndOpacity {colorUseRule=UseColor_Foreground, specifiedColor=<rgba 1 0 1 1>}
      font {
       bForceMonospaced=false
       bMaterialIsStencil=false
       fontMaterial=null
       fontObject=&NotoSans
       letterSpacing=0
       monospacedWidth=1
       size=20
       skewAmount=0
       typefaceFontName=NotoSans-Regular
       outlineSettings {
        bApplyOutlineToDropShadows=false
        bMiteredCorners=false
        bSeparateFillAlpha=false
        outlineMaterial=null
        outlineSize=0
        outlineColor=<rgba 0 0 0 0>
       }
      }
      highlightColor {colorUseRule=UseColor_ColorTable, specifiedColor=<rgba 0 0 0 0>}
      highlightShape {
       bIsDynamicallyLoaded=false
       drawAs=Image
       imageType=NoImage
       mirroring=NoMirror
       resourceName=None
       resourceObject=null
       tiling=NoTile
       imageSize=<0 0>
       margin=<ltrb 0 0 0 0>
       outlineSettings {
        bUseBrushTransparency=false
        roundingType=HalfHeightRadius
        width=0
        color {colorUseRule=UseColor_Specified, specifiedColor=<rgba 0 0 0 0>}
        cornerRadii=<0 0 0 0>
       }
       tintColor {colorUseRule=UseColor_ColorTable, specifiedColor=<rgba 0 0 0 0>}
       uVRegion {bIsValid=false, max=<0 0>, min=<0 0>}
      }
      selectedBackgroundColor {colorUseRule=UseColor_ColorTable, specifiedColor=<rgba 0 0 0 0>}
      shadowColorAndOpacity=<rgba 0 0 0 1>
      shadowOffset=<0 0>
      strikeBrush {
       bIsDynamicallyLoaded=false
       drawAs=Image
       imageType=NoImage
       mirroring=NoMirror
       resourceName=None
       resourceObject=null
       tiling=NoTile
       imageSize=<32 32>
       margin=<ltrb 0 0 0 0>
       outlineSettings {
        bUseBrushTransparency=false
        roundingType=HalfHeightRadius
        width=0
        color {colorUseRule=UseColor_Specified, specifiedColor=<rgba 0 0 0 0>}
        cornerRadii=<0 0 0 0>
       }
       tintColor {colorUseRule=UseColor_Specified, specifiedColor=<rgba 1 1 1 1>}
       uVRegion {bIsValid=false, max=<0 0>, min=<0 0>}
      }
      underlineBrush {
       bIsDynamicallyLoaded=false
       drawAs=Image
       imageType=NoImage
       mirroring=NoMirror
       resourceName=None
       resourceObject=null
       tiling=NoTile
       imageSize=<32 32>
       margin=<ltrb 0 0 0 0>
       outlineSettings {
        bUseBrushTransparency=false
        roundingType=HalfHeightRadius
        width=0
        color {colorUseRule=UseColor_Specified, specifiedColor=<rgba 0 0 0 0>}
        cornerRadii=<0 0 0 0>
       }
       tintColor {colorUseRule=UseColor_Specified, specifiedColor=<rgba 1 1 1 1>}
       uVRegion {bIsValid=false, max=<0 0>, min=<0 0>}
      }
     }
     vScrollBarPadding=<ltrb 0 0 0 0>
    }
   }
   slot {class=OverlaySlot, properties {horizontalAlignment=HAlign_Right, verticalAlignment=VAlign_Center}}
  }
  VerticalBox_4 {
   widgetClassPath=&VerticalBox
   parent=HorizontalBoxTop
   widget=VerticalBox_4
   slot {class=HorizontalBoxSlot, properties {size {sizeRule=Fill, value=1}}}
  }
  Border_Variables {
   widgetClassPath=&Border
   parent=VerticalBox_4
   widget=Border_Variables
   properties {brushColor=<rgba 0.0103 0.0103 0.0103 1>, padding=<ltrb 4 0 4 0>}
   slot {class=VerticalBoxSlot, properties {size {sizeRule=Fill, value=1}}}
  }
  XpStats_DetailView {
   widgetClassPath=&DetailsView
   parent=Border_Variables
   widget=XpStats_DetailView
   properties {
    bAllowFiltering=true
    columnWidth=0.6
    toolTipText="Field can be updated and take affect in simlation.  Some can be done at different times.  All can be done before hitting Start."
    categoriesToShow [Analytics Data, Designer Tuning]
   }
   slot {class=BorderSlot, properties {padding=<ltrb 4 0 4 0>}}
  }
  Border_0 {
   widgetClassPath=&Border
   parent=VerticalBox_4
   widget=Border_0
   properties {horizontalAlignment=HAlign_Center, brushColor=<rgba 0.0103 0.0103 0.0103 1>}
   slot {class=VerticalBoxSlot, properties {padding=<ltrb 0 0 0 6>}}
  }
  HorizontalBox_8 {
   widgetClassPath=&HorizontalBox
   parent=Border_0
   widget=HorizontalBox_8
   slot {class=BorderSlot, properties {horizontalAlignment=HAlign_Center}}
  }
  textField_PlayerSessionStart {
   widgetClassPath=&CommonTextBlock
   parent=HorizontalBox_8
   widget=textField_PlayerSessionStart
   properties {
    justification=Right
    text=Player Session Start:
    toolTipText="Typically true, just used to test system simulating Island is not yet approved to grant stats"
    colorAndOpacity {colorUseRule=UseColor_Specified, specifiedColor=<rgba 0.849 0.849 0.849 1>}
    font {
     bForceMonospaced=false
     bMaterialIsStencil=false
     fontMaterial=null
     fontObject=&NotoSans
     letterSpacing=0
     monospacedWidth=1
     size=9
     skewAmount=0
     typefaceFontName=NotoSans-Regular
     outlineSettings {
      bApplyOutlineToDropShadows=false
      bMiteredCorners=false
      bSeparateFillAlpha=false
      outlineMaterial=null
      outlineSize=0
      outlineColor=<rgba 0 0 0 1>
     }
    }
    shadowColorAndOpacity=<rgba 0 0 0 0.5>
   }
   slot {class=HorizontalBoxSlot, properties {verticalAlignment=VAlign_Center, padding=<ltrb 10 0 4 0>}}
  }
  bttn_ResetPlayerSession {
   widgetClassPath=&TA_Generic01_Button
   parent=HorizontalBox_8
   widget=bttn_ResetPlayerSession
   properties {
    label=Reset
    toolTipText=Fill in defaults to Xp Stats to pass data gathering other than Accolades.
    buttonColor_Default=<rgba 1 0.8254 0 1>
    buttonStyle_Default {
     clickedSlateSound {resourceObject=null}
     disabled {
      bIsDynamicallyLoaded=false
      drawAs=Box
      imageType=NoImage
      mirroring=NoMirror
      resourceName=None
      resourceObject=/Game/Blueprints/Blutilities/ForTechArt/ArtAssets/Textures/Button02/MI_TA_ButtonBG_02_Down
      tiling=NoTile
      imageSize=<32 32>
      margin=<ltrb 0.25 0.25 0.25 0.25>
      outlineSettings {
       bUseBrushTransparency=false
       roundingType=HalfHeightRadius
       width=0
       color {colorUseRule=UseColor_Specified, specifiedColor=<rgba 0 0 0 0>}
       cornerRadii=<0 0 0 1>
      }
      tintColor {colorUseRule=UseColor_Specified, specifiedColor=<rgba 0.3333 0.3333 0.3333 1>}
      uVRegion {bIsValid=false, max=<0 0>, min=<0 0>}
     }
     disabledForeground {colorUseRule=UseColor_Foreground, specifiedColor=<rgba 1 0 1 1>}
     hovered {
      bIsDynamicallyLoaded=false
      drawAs=Box
      imageType=NoImage
      mirroring=NoMirror
      resourceName=None
      resourceObject=/Game/Blueprints/Blutilities/ForTechArt/ArtAssets/Textures/Button02/MI_TA_ButtonBG_02_Over
      tiling=NoTile
      imageSize=<32 32>
      margin=<ltrb 0.25 0.25 0.25 0.25>
      outlineSettings {
       bUseBrushTransparency=false
       roundingType=HalfHeightRadius
       width=0
       color {colorUseRule=UseColor_Specified, specifiedColor=<rgba 0 0 0 0>}
       cornerRadii=<0 0 0 1>
      }
      tintColor {colorUseRule=UseColor_Specified, specifiedColor=<rgba 0.9 0.9 0.9 1>}
      uVRegion {bIsValid=false, max=<0 0>, min=<0 0>}
     }
     hoveredForeground {colorUseRule=UseColor_Foreground, specifiedColor=<rgba 1 0 1 1>}
     hoveredSlateSound {resourceObject=null}
     normal {
      bIsDynamicallyLoaded=false
      drawAs=Box
      imageType=NoImage
      mirroring=NoMirror
      resourceName=None
      resourceObject=/Game/Blueprints/Blutilities/ForTechArt/ArtAssets/Textures/Button02/M_TA_ButtonBG_02_Up
      tiling=NoTile
      imageSize=<32 32>
      margin=<ltrb 0.25 0.25 0.25 0.25>
      outlineSettings {
       bUseBrushTransparency=false
       roundingType=HalfHeightRadius
       width=0
       color {colorUseRule=UseColor_Specified, specifiedColor=<rgba 0 0 0 0>}
       cornerRadii=<0 0 0 1>
      }
      tintColor {colorUseRule=UseColor_Specified, specifiedColor=<rgba 1 1 1 1>}
      uVRegion {bIsValid=false, max=<0 0>, min=<0 0>}
     }
     normalForeground {colorUseRule=UseColor_Foreground, specifiedColor=<rgba 1 0 1 1>}
     normalPadding=<ltrb 0 0 0 0>
     pressed {
      bIsDynamicallyLoaded=false
      drawAs=Box
      imageType=NoImage
      mirroring=NoMirror
      resourceName=None
      resourceObject=&MI_TA_ButtonBG_02_Down
      tiling=NoTile
      imageSize=<32 32>
      margin=<ltrb 0.25 0.25 0.25 0.25>
      outlineSettings {
       bUseBrushTransparency=false
       roundingType=HalfHeightRadius
       width=0
       color {colorUseRule=UseColor_Specified, specifiedColor=<rgba 0 0 0 0>}
       cornerRadii=<0 0 0 1>
      }
      tintColor {colorUseRule=UseColor_Specified, specifiedColor=<rgba 0.65 0.65 0.65 1>}
      uVRegion {bIsValid=false, max=<0 0>, min=<0 0>}
     }
     pressedForeground {colorUseRule=UseColor_Foreground, specifiedColor=<rgba 1 0 1 1>}
     pressedPadding=<ltrb 0 0 0 0>
     pressedSlateSound {resourceObject=null}
    }
    foregroundColor {colorUseRule=UseColor_Specified, specifiedColor=<rgba 0.8938 1 0 1>}
    iconBrush {
     bIsDynamicallyLoaded=false
     drawAs=Image
     imageType=NoImage
     mirroring=NoMirror
     resourceName=None
     resourceObject=null
     tiling=NoTile
     imageSize=<32 32>
     margin=<ltrb 0 0 0 0>
     outlineSettings {
      bUseBrushTransparency=false
      roundingType=HalfHeightRadius
      width=0
      color {colorUseRule=UseColor_Specified, specifiedColor=<rgba 0 0 0 0>}
      cornerRadii=<0 0 0 1>
     }
     tintColor {colorUseRule=UseColor_Specified, specifiedColor=<rgba 0.0469 0.0469 0.0469 1>}
     uVRegion {bIsValid=false, max=<0 0>, min=<0 0>}
    }
    labelColor_Default=<rgba 0.0729 0.0729 0.0729 1>
   }
   slot {class=HorizontalBoxSlot, properties {padding=<ltrb 2 2 2 2>}}
  }
  textField_XpStats {
   widgetClassPath=&CommonTextBlock
   parent=HorizontalBox_8
   widget=textField_XpStats
   properties {
    justification=Right
    text=Xp Stats:
    toolTipText="Typically true, just used to test system simulating Island is not yet approved to grant stats"
    colorAndOpacity {colorUseRule=UseColor_Specified, specifiedColor=<rgba 0.849 0.849 0.849 1>}
    font {
     bForceMonospaced=false
     bMaterialIsStencil=false
     fontMaterial=null
     fontObject=&NotoSans
     letterSpacing=0
     monospacedWidth=1
     size=9
     skewAmount=0
     typefaceFontName=NotoSans-Regular
     outlineSettings {
      bApplyOutlineToDropShadows=false
      bMiteredCorners=false
      bSeparateFillAlpha=false
      outlineMaterial=null
      outlineSize=0
      outlineColor=<rgba 0 0 0 1>
     }
    }
    shadowColorAndOpacity=<rgba 0 0 0 0.5>
   }
   slot {class=HorizontalBoxSlot, properties {verticalAlignment=VAlign_Center, padding=<ltrb 10 0 4 0>}}
  }
  bttnFillDataGather {
   widgetClassPath=&TA_Generic01_Button
   parent=HorizontalBox_8
   widget=bttnFillDataGather
   properties {
    label=" Fill "
    toolTipText=Fill in defaults to Xp Stats to pass data gathering other than Accolades.
    buttonColor_Default=<rgba 1 0.8254 0 1>
    buttonStyle_Default {
     clickedSlateSound {resourceObject=null}
     disabled {
      bIsDynamicallyLoaded=false
      drawAs=Box
      imageType=NoImage
      mirroring=NoMirror
      resourceName=None
      resourceObject=&MI_TA_ButtonBG_02_Down
      tiling=NoTile
      imageSize=<32 32>
      margin=<ltrb 0.25 0.25 0.25 0.25>
      outlineSettings {
       bUseBrushTransparency=false
       roundingType=HalfHeightRadius
       width=0
       color {colorUseRule=UseColor_Specified, specifiedColor=<rgba 0 0 0 0>}
       cornerRadii=<0 0 0 1>
      }
      tintColor {colorUseRule=UseColor_Specified, specifiedColor=<rgba 0.3333 0.3333 0.3333 1>}
      uVRegion {bIsValid=false, max=<0 0>, min=<0 0>}
     }
     disabledForeground {colorUseRule=UseColor_Foreground, specifiedColor=<rgba 1 0 1 1>}
     hovered {
      bIsDynamicallyLoaded=false
      drawAs=Box
      imageType=NoImage
      mirroring=NoMirror
      resourceName=None
      resourceObject=&MI_TA_ButtonBG_02_Over
      tiling=NoTile
      imageSize=<32 32>
      margin=<ltrb 0.25 0.25 0.25 0.25>
      outlineSettings {
       bUseBrushTransparency=false
       roundingType=HalfHeightRadius
       width=0
       color {colorUseRule=UseColor_Specified, specifiedColor=<rgba 0 0 0 0>}
       cornerRadii=<0 0 0 1>
      }
      tintColor {colorUseRule=UseColor_Specified, specifiedColor=<rgba 0.9 0.9 0.9 1>}
      uVRegion {bIsValid=false, max=<0 0>, min=<0 0>}
     }
     hoveredForeground {colorUseRule=UseColor_Foreground, specifiedColor=<rgba 1 0 1 1>}
     hoveredSlateSound {resourceObject=null}
     normal {
      bIsDynamicallyLoaded=false
      drawAs=Box
      imageType=NoImage
      mirroring=NoMirror
      resourceName=None
      resourceObject=&M_TA_ButtonBG_02_Up
      tiling=NoTile
      imageSize=<32 32>
      margin=<ltrb 0.25 0.25 0.25 0.25>
      outlineSettings {
       bUseBrushTransparency=false
       roundingType=HalfHeightRadius
       width=0
       color {colorUseRule=UseColor_Specified, specifiedColor=<rgba 0 0 0 0>}
       cornerRadii=<0 0 0 1>
      }
      tintColor {colorUseRule=UseColor_Specified, specifiedColor=<rgba 1 1 1 1>}
      uVRegion {bIsValid=false, max=<0 0>, min=<0 0>}
     }
     normalForeground {colorUseRule=UseColor_Foreground, specifiedColor=<rgba 1 0 1 1>}
     normalPadding=<ltrb 0 0 0 0>
     pressed {
      bIsDynamicallyLoaded=false
      drawAs=Box
      imageType=NoImage
      mirroring=NoMirror
      resourceName=None
      resourceObject=&MI_TA_ButtonBG_02_Down
      tiling=NoTile
      imageSize=<32 32>
      margin=<ltrb 0.25 0.25 0.25 0.25>
      outlineSettings {
       bUseBrushTransparency=false
       roundingType=HalfHeightRadius
       width=0
       color {colorUseRule=UseColor_Specified, specifiedColor=<rgba 0 0 0 0>}
       cornerRadii=<0 0 0 1>
      }
      tintColor {colorUseRule=UseColor_Specified, specifiedColor=<rgba 0.65 0.65 0.65 1>}
      uVRegion {bIsValid=false, max=<0 0>, min=<0 0>}
     }
     pressedForeground {colorUseRule=UseColor_Foreground, specifiedColor=<rgba 1 0 1 1>}
     pressedPadding=<ltrb 0 0 0 0>
     pressedSlateSound {resourceObject=null}
    }
    foregroundColor {colorUseRule=UseColor_Specified, specifiedColor=<rgba 0.8938 1 0 1>}
    iconBrush {
     bIsDynamicallyLoaded=false
     drawAs=Image
     imageType=NoImage
     mirroring=NoMirror
     resourceName=None
     resourceObject=null
     tiling=NoTile
     imageSize=<32 32>
     margin=<ltrb 0 0 0 0>
     outlineSettings {
      bUseBrushTransparency=false
      roundingType=HalfHeightRadius
      width=0
      color {colorUseRule=UseColor_Specified, specifiedColor=<rgba 0 0 0 0>}
      cornerRadii=<0 0 0 1>
     }
     tintColor {colorUseRule=UseColor_Specified, specifiedColor=<rgba 0.0469 0.0469 0.0469 1>}
     uVRegion {bIsValid=false, max=<0 0>, min=<0 0>}
    }
    labelColor_Default=<rgba 0.0729 0.0729 0.0729 1>
   }
   slot {class=HorizontalBoxSlot, properties {padding=<ltrb 2 2 2 2>}}
  }
  textField_DesignerTuning {
   widgetClassPath=&CommonTextBlock
   parent=HorizontalBox_8
   widget=textField_DesignerTuning
   properties {
    justification=Right
    text=Designer Tuning:
    toolTipText="Typically true, just used to test system simulating Island is not yet approved to grant stats"
    colorAndOpacity {colorUseRule=UseColor_Specified, specifiedColor=<rgba 0.849 0.849 0.849 1>}
    font {
     bForceMonospaced=false
     bMaterialIsStencil=false
     fontMaterial=null
     fontObject=&NotoSans
     letterSpacing=0
     monospacedWidth=1
     size=9
     skewAmount=0
     typefaceFontName=NotoSans-Regular
     outlineSettings {
      bApplyOutlineToDropShadows=false
      bMiteredCorners=false
      bSeparateFillAlpha=false
      outlineMaterial=null
      outlineSize=0
      outlineColor=<rgba 0 0 0 1>
     }
    }
    shadowColorAndOpacity=<rgba 0 0 0 0.5>
   }
   slot {class=HorizontalBoxSlot, properties {verticalAlignment=VAlign_Center, padding=<ltrb 10 0 4 0>}}
  }
  bttnSaveDesignerTuning {
   widgetClassPath=&TA_Generic01_Button
   parent=HorizontalBox_8
   widget=bttnSaveDesignerTuning
   properties {
    label=Save
    toolTipText=Save All Designer Data and XP Stats data so it could be loaded later.
    buttonColor_Default=<rgba 1 0.8254 0 1>
    buttonStyle_Default {
     clickedSlateSound {resourceObject=null}
     disabled {
      bIsDynamicallyLoaded=false
      drawAs=Box
      imageType=NoImage
      mirroring=NoMirror
      resourceName=None
      resourceObject=&MI_TA_ButtonBG_02_Down
      tiling=NoTile
      imageSize=<32 32>
      margin=<ltrb 0.25 0.25 0.25 0.25>
      outlineSettings {
       bUseBrushTransparency=false
       roundingType=HalfHeightRadius
       width=0
       color {colorUseRule=UseColor_Specified, specifiedColor=<rgba 0 0 0 0>}
       cornerRadii=<0 0 0 1>
      }
      tintColor {colorUseRule=UseColor_Specified, specifiedColor=<rgba 0.3333 0.3333 0.3333 1>}
      uVRegion {bIsValid=false, max=<0 0>, min=<0 0>}
     }
     disabledForeground {colorUseRule=UseColor_Foreground, specifiedColor=<rgba 1 0 1 1>}
     hovered {
      bIsDynamicallyLoaded=false
      drawAs=Box
      imageType=NoImage
      mirroring=NoMirror
      resourceName=None
      resourceObject=&MI_TA_ButtonBG_02_Over
      tiling=NoTile
      imageSize=<32 32>
      margin=<ltrb 0.25 0.25 0.25 0.25>
      outlineSettings {
       bUseBrushTransparency=false
       roundingType=HalfHeightRadius
       width=0
       color {colorUseRule=UseColor_Specified, specifiedColor=<rgba 0 0 0 0>}
       cornerRadii=<0 0 0 1>
      }
      tintColor {colorUseRule=UseColor_Specified, specifiedColor=<rgba 0.9 0.9 0.9 1>}
      uVRegion {bIsValid=false, max=<0 0>, min=<0 0>}
     }
     hoveredForeground {colorUseRule=UseColor_Foreground, specifiedColor=<rgba 1 0 1 1>}
     hoveredSlateSound {resourceObject=null}
     normal {
      bIsDynamicallyLoaded=false
      drawAs=Box
      imageType=NoImage
      mirroring=NoMirror
      resourceName=None
      resourceObject=&M_TA_ButtonBG_02_Up
      tiling=NoTile
      imageSize=<32 32>
      margin=<ltrb 0.25 0.25 0.25 0.25>
      outlineSettings {
       bUseBrushTransparency=false
       roundingType=HalfHeightRadius
       width=0
       color {colorUseRule=UseColor_Specified, specifiedColor=<rgba 0 0 0 0>}
       cornerRadii=<0 0 0 1>
      }
      tintColor {colorUseRule=UseColor_Specified, specifiedColor=<rgba 1 1 1 1>}
      uVRegion {bIsValid=false, max=<0 0>, min=<0 0>}
     }
     normalForeground {colorUseRule=UseColor_Foreground, specifiedColor=<rgba 1 0 1 1>}
     normalPadding=<ltrb 0 0 0 0>
     pressed {
      bIsDynamicallyLoaded=false
      drawAs=Box
      imageType=NoImage
      mirroring=NoMirror
      resourceName=None
      resourceObject=&MI_TA_ButtonBG_02_Down
      tiling=NoTile
      imageSize=<32 32>
      margin=<ltrb 0.25 0.25 0.25 0.25>
      outlineSettings {
       bUseBrushTransparency=false
       roundingType=HalfHeightRadius
       width=0
       color {colorUseRule=UseColor_Specified, specifiedColor=<rgba 0 0 0 0>}
       cornerRadii=<0 0 0 1>
      }
      tintColor {colorUseRule=UseColor_Specified, specifiedColor=<rgba 0.65 0.65 0.65 1>}
      uVRegion {bIsValid=false, max=<0 0>, min=<0 0>}
     }
     pressedForeground {colorUseRule=UseColor_Foreground, specifiedColor=<rgba 1 0 1 1>}
     pressedPadding=<ltrb 0 0 0 0>
     pressedSlateSound {resourceObject=null}
    }
    foregroundColor {colorUseRule=UseColor_Specified, specifiedColor=<rgba 0.8938 1 0 1>}
    iconBrush {
     bIsDynamicallyLoaded=false
     drawAs=Image
     imageType=NoImage
     mirroring=NoMirror
     resourceName=None
     resourceObject=null
     tiling=NoTile
     imageSize=<32 32>
     margin=<ltrb 0 0 0 0>
     outlineSettings {
      bUseBrushTransparency=false
      roundingType=HalfHeightRadius
      width=0
      color {colorUseRule=UseColor_Specified, specifiedColor=<rgba 0 0 0 0>}
      cornerRadii=<0 0 0 1>
     }
     tintColor {colorUseRule=UseColor_Specified, specifiedColor=<rgba 0.0469 0.0469 0.0469 1>}
     uVRegion {bIsValid=false, max=<0 0>, min=<0 0>}
    }
    labelColor_Default=<rgba 0.0729 0.0729 0.0729 1>
   }
   slot {class=HorizontalBoxSlot, properties {padding=<ltrb 2 2 2 2>}}
  }
  bttnLoadDesignerTuning {
   widgetClassPath=&TA_Generic01_Button
   parent=HorizontalBox_8
   widget=bttnLoadDesignerTuning
   properties {
    label=Load
    toolTipText=Load previously saved data
    buttonColor_Default=<rgba 1 0.8254 0 1>
    buttonStyle_Default {
     clickedSlateSound {resourceObject=null}
     disabled {
      bIsDynamicallyLoaded=false
      drawAs=Box
      imageType=NoImage
      mirroring=NoMirror
      resourceName=None
      resourceObject=&MI_TA_ButtonBG_02_Down
      tiling=NoTile
      imageSize=<32 32>
      margin=<ltrb 0.25 0.25 0.25 0.25>
      outlineSettings {
       bUseBrushTransparency=false
       roundingType=HalfHeightRadius
       width=0
       color {colorUseRule=UseColor_Specified, specifiedColor=<rgba 0 0 0 0>}
       cornerRadii=<0 0 0 1>
      }
      tintColor {colorUseRule=UseColor_Specified, specifiedColor=<rgba 0.3333 0.3333 0.3333 1>}
      uVRegion {bIsValid=false, max=<0 0>, min=<0 0>}
     }
     disabledForeground {colorUseRule=UseColor_Foreground, specifiedColor=<rgba 1 0 1 1>}
     hovered {
      bIsDynamicallyLoaded=false
      drawAs=Box
      imageType=NoImage
      mirroring=NoMirror
      resourceName=None
      resourceObject=&MI_TA_ButtonBG_02_Over
      tiling=NoTile
      imageSize=<32 32>
      margin=<ltrb 0.25 0.25 0.25 0.25>
      outlineSettings {
       bUseBrushTransparency=false
       roundingType=HalfHeightRadius
       width=0
       color {colorUseRule=UseColor_Specified, specifiedColor=<rgba 0 0 0 0>}
       cornerRadii=<0 0 0 1>
      }
      tintColor {colorUseRule=UseColor_Specified, specifiedColor=<rgba 0.9 0.9 0.9 1>}
      uVRegion {bIsValid=false, max=<0 0>, min=<0 0>}
     }
     hoveredForeground {colorUseRule=UseColor_Foreground, specifiedColor=<rgba 1 0 1 1>}
     hoveredSlateSound {resourceObject=null}
     normal {
      bIsDynamicallyLoaded=false
      drawAs=Box
      imageType=NoImage
      mirroring=NoMirror
      resourceName=None
      resourceObject=&M_TA_ButtonBG_02_Up
      tiling=NoTile
      imageSize=<32 32>
      margin=<ltrb 0.25 0.25 0.25 0.25>
      outlineSettings {
       bUseBrushTransparency=false
       roundingType=HalfHeightRadius
       width=0
       color {colorUseRule=UseColor_Specified, specifiedColor=<rgba 0 0 0 0>}
       cornerRadii=<0 0 0 1>
      }
      tintColor {colorUseRule=UseColor_Specified, specifiedColor=<rgba 1 1 1 1>}
      uVRegion {bIsValid=false, max=<0 0>, min=<0 0>}
     }
     normalForeground {colorUseRule=UseColor_Foreground, specifiedColor=<rgba 1 0 1 1>}
     normalPadding=<ltrb 0 0 0 0>
     pressed {
      bIsDynamicallyLoaded=false
      drawAs=Box
      imageType=NoImage
      mirroring=NoMirror
      resourceName=None
      resourceObject=&MI_TA_ButtonBG_02_Down
      tiling=NoTile
      imageSize=<32 32>
      margin=<ltrb 0.25 0.25 0.25 0.25>
      outlineSettings {
       bUseBrushTransparency=false
       roundingType=HalfHeightRadius
       width=0
       color {colorUseRule=UseColor_Specified, specifiedColor=<rgba 0 0 0 0>}
       cornerRadii=<0 0 0 1>
      }
      tintColor {colorUseRule=UseColor_Specified, specifiedColor=<rgba 0.65 0.65 0.65 1>}
      uVRegion {bIsValid=false, max=<0 0>, min=<0 0>}
     }
     pressedForeground {colorUseRule=UseColor_Foreground, specifiedColor=<rgba 1 0 1 1>}
     pressedPadding=<ltrb 0 0 0 0>
     pressedSlateSound {resourceObject=null}
    }
    foregroundColor {colorUseRule=UseColor_Specified, specifiedColor=<rgba 0.8938 1 0 1>}
    iconBrush {
     bIsDynamicallyLoaded=false
     drawAs=Image
     imageType=NoImage
     mirroring=NoMirror
     resourceName=None
     resourceObject=null
     tiling=NoTile
     imageSize=<32 32>
     margin=<ltrb 0 0 0 0>
     outlineSettings {
      bUseBrushTransparency=false
      roundingType=HalfHeightRadius
      width=0
      color {colorUseRule=UseColor_Specified, specifiedColor=<rgba 0 0 0 0>}
      cornerRadii=<0 0 0 1>
     }
     tintColor {colorUseRule=UseColor_Specified, specifiedColor=<rgba 0.0469 0.0469 0.0469 1>}
     uVRegion {bIsValid=false, max=<0 0>, min=<0 0>}
    }
    labelColor_Default=<rgba 0.0729 0.0729 0.0729 1>
   }
   slot {class=HorizontalBoxSlot, properties {padding=<ltrb 2 2 2 2>}}
  }
  textField_AnalyticsData {
   widgetClassPath=&CommonTextBlock
   parent=HorizontalBox_8
   widget=textField_AnalyticsData
   properties {
    justification=Right
    text=Analytics Data
    toolTipText="Typically true, just used to test system simulating Island is not yet approved to grant stats"
    colorAndOpacity {colorUseRule=UseColor_Specified, specifiedColor=<rgba 0.849 0.849 0.849 1>}
    font {
     bForceMonospaced=false
     bMaterialIsStencil=false
     fontMaterial=null
     fontObject=&NotoSans
     letterSpacing=0
     monospacedWidth=1
     size=9
     skewAmount=0
     typefaceFontName=NotoSans-Regular
     outlineSettings {
      bApplyOutlineToDropShadows=false
      bMiteredCorners=false
      bSeparateFillAlpha=false
      outlineMaterial=null
      outlineSize=0
      outlineColor=<rgba 0 0 0 1>
     }
    }
    shadowColorAndOpacity=<rgba 0 0 0 0.5>
   }
   slot {class=HorizontalBoxSlot, properties {verticalAlignment=VAlign_Center, padding=<ltrb 10 0 4 0>}}
  }
  bttnSaveAnalyticsData {
   widgetClassPath=&TA_Generic01_Button
   parent=HorizontalBox_8
   widget=bttnSaveAnalyticsData
   properties {
    label=Save
    toolTipText=Save All Designer Data and XP Stats data so it could be loaded later.
    buttonColor_Default=<rgba 1 0.8254 0 1>
    buttonStyle_Default {
     clickedSlateSound {resourceObject=null}
     disabled {
      bIsDynamicallyLoaded=false
      drawAs=Box
      imageType=NoImage
      mirroring=NoMirror
      resourceName=None
      resourceObject=&MI_TA_ButtonBG_02_Down
      tiling=NoTile
      imageSize=<32 32>
      margin=<ltrb 0.25 0.25 0.25 0.25>
      outlineSettings {
       bUseBrushTransparency=false
       roundingType=HalfHeightRadius
       width=0
       color {colorUseRule=UseColor_Specified, specifiedColor=<rgba 0 0 0 0>}
       cornerRadii=<0 0 0 1>
      }
      tintColor {colorUseRule=UseColor_Specified, specifiedColor=<rgba 0.3333 0.3333 0.3333 1>}
      uVRegion {bIsValid=false, max=<0 0>, min=<0 0>}
     }
     disabledForeground {colorUseRule=UseColor_Foreground, specifiedColor=<rgba 1 0 1 1>}
     hovered {
      bIsDynamicallyLoaded=false
      drawAs=Box
      imageType=NoImage
      mirroring=NoMirror
      resourceName=None
      resourceObject=&MI_TA_ButtonBG_02_Over
      tiling=NoTile
      imageSize=<32 32>
      margin=<ltrb 0.25 0.25 0.25 0.25>
      outlineSettings {
       bUseBrushTransparency=false
       roundingType=HalfHeightRadius
       width=0
       color {colorUseRule=UseColor_Specified, specifiedColor=<rgba 0 0 0 0>}
       cornerRadii=<0 0 0 1>
      }
      tintColor {colorUseRule=UseColor_Specified, specifiedColor=<rgba 0.9 0.9 0.9 1>}
      uVRegion {bIsValid=false, max=<0 0>, min=<0 0>}
     }
     hoveredForeground {colorUseRule=UseColor_Foreground, specifiedColor=<rgba 1 0 1 1>}
     hoveredSlateSound {resourceObject=null}
     normal {
      bIsDynamicallyLoaded=false
      drawAs=Box
      imageType=NoImage
      mirroring=NoMirror
      resourceName=None
      resourceObject=&M_TA_ButtonBG_02_Up
      tiling=NoTile
      imageSize=<32 32>
      margin=<ltrb 0.25 0.25 0.25 0.25>
      outlineSettings {
       bUseBrushTransparency=false
       roundingType=HalfHeightRadius
       width=0
       color {colorUseRule=UseColor_Specified, specifiedColor=<rgba 0 0 0 0>}
       cornerRadii=<0 0 0 1>
      }
      tintColor {colorUseRule=UseColor_Specified, specifiedColor=<rgba 1 1 1 1>}
      uVRegion {bIsValid=false, max=<0 0>, min=<0 0>}
     }
     normalForeground {colorUseRule=UseColor_Foreground, specifiedColor=<rgba 1 0 1 1>}
     normalPadding=<ltrb 0 0 0 0>
     pressed {
      bIsDynamicallyLoaded=false
      drawAs=Box
      imageType=NoImage
      mirroring=NoMirror
      resourceName=None
      resourceObject=&MI_TA_ButtonBG_02_Down
      tiling=NoTile
      imageSize=<32 32>
      margin=<ltrb 0.25 0.25 0.25 0.25>
      outlineSettings {
       bUseBrushTransparency=false
       roundingType=HalfHeightRadius
       width=0
       color {colorUseRule=UseColor_Specified, specifiedColor=<rgba 0 0 0 0>}
       cornerRadii=<0 0 0 1>
      }
      tintColor {colorUseRule=UseColor_Specified, specifiedColor=<rgba 0.65 0.65 0.65 1>}
      uVRegion {bIsValid=false, max=<0 0>, min=<0 0>}
     }
     pressedForeground {colorUseRule=UseColor_Foreground, specifiedColor=<rgba 1 0 1 1>}
     pressedPadding=<ltrb 0 0 0 0>
     pressedSlateSound {resourceObject=null}
    }
    foregroundColor {colorUseRule=UseColor_Specified, specifiedColor=<rgba 0.8938 1 0 1>}
    iconBrush {
     bIsDynamicallyLoaded=false
     drawAs=Image
     imageType=NoImage
     mirroring=NoMirror
     resourceName=None
     resourceObject=null
     tiling=NoTile
     imageSize=<32 32>
     margin=<ltrb 0 0 0 0>
     outlineSettings {
      bUseBrushTransparency=false
      roundingType=HalfHeightRadius
      width=0
      color {colorUseRule=UseColor_Specified, specifiedColor=<rgba 0 0 0 0>}
      cornerRadii=<0 0 0 1>
     }
     tintColor {colorUseRule=UseColor_Specified, specifiedColor=<rgba 0.0469 0.0469 0.0469 1>}
     uVRegion {bIsValid=false, max=<0 0>, min=<0 0>}
    }
    labelColor_Default=<rgba 0.0729 0.0729 0.0729 1>
   }
   slot {class=HorizontalBoxSlot, properties {padding=<ltrb 2 2 2 2>}}
  }
  bttnLoadAnalyticsData {
   widgetClassPath=&TA_Generic01_Button
   parent=HorizontalBox_8
   widget=bttnLoadAnalyticsData
   properties {
    label=Load
    toolTipText=Load previously saved data
    buttonColor_Default=<rgba 1 0.8254 0 1>
    buttonStyle_Default {
     clickedSlateSound {resourceObject=null}
     disabled {
      bIsDynamicallyLoaded=false
      drawAs=Box
      imageType=NoImage
      mirroring=NoMirror
      resourceName=None
      resourceObject=&MI_TA_ButtonBG_02_Down
      tiling=NoTile
      imageSize=<32 32>
      margin=<ltrb 0.25 0.25 0.25 0.25>
      outlineSettings {
       bUseBrushTransparency=false
       roundingType=HalfHeightRadius
       width=0
       color {colorUseRule=UseColor_Specified, specifiedColor=<rgba 0 0 0 0>}
       cornerRadii=<0 0 0 1>
      }
      tintColor {colorUseRule=UseColor_Specified, specifiedColor=<rgba 0.3333 0.3333 0.3333 1>}
      uVRegion {bIsValid=false, max=<0 0>, min=<0 0>}
     }
     disabledForeground {colorUseRule=UseColor_Foreground, specifiedColor=<rgba 1 0 1 1>}
     hovered {
      bIsDynamicallyLoaded=false
      drawAs=Box
      imageType=NoImage
      mirroring=NoMirror
      resourceName=None
      resourceObject=&MI_TA_ButtonBG_02_Over
      tiling=NoTile
      imageSize=<32 32>
      margin=<ltrb 0.25 0.25 0.25 0.25>
      outlineSettings {
       bUseBrushTransparency=false
       roundingType=HalfHeightRadius
       width=0
       color {colorUseRule=UseColor_Specified, specifiedColor=<rgba 0 0 0 0>}
       cornerRadii=<0 0 0 1>
      }
      tintColor {colorUseRule=UseColor_Specified, specifiedColor=<rgba 0.9 0.9 0.9 1>}
      uVRegion {bIsValid=false, max=<0 0>, min=<0 0>}
     }
     hoveredForeground {colorUseRule=UseColor_Foreground, specifiedColor=<rgba 1 0 1 1>}
     hoveredSlateSound {resourceObject=null}
     normal {
      bIsDynamicallyLoaded=false
      drawAs=Box
      imageType=NoImage
      mirroring=NoMirror
      resourceName=None
      resourceObject=&M_TA_ButtonBG_02_Up
      tiling=NoTile
      imageSize=<32 32>
      margin=<ltrb 0.25 0.25 0.25 0.25>
      outlineSettings {
       bUseBrushTransparency=false
       roundingType=HalfHeightRadius
       width=0
       color {colorUseRule=UseColor_Specified, specifiedColor=<rgba 0 0 0 0>}
       cornerRadii=<0 0 0 1>
      }
      tintColor {colorUseRule=UseColor_Specified, specifiedColor=<rgba 1 1 1 1>}
      uVRegion {bIsValid=false, max=<0 0>, min=<0 0>}
     }
     normalForeground {colorUseRule=UseColor_Foreground, specifiedColor=<rgba 1 0 1 1>}
     normalPadding=<ltrb 0 0 0 0>
     pressed {
      bIsDynamicallyLoaded=false
      drawAs=Box
      imageType=NoImage
      mirroring=NoMirror
      resourceName=None
      resourceObject=&MI_TA_ButtonBG_02_Down
      tiling=NoTile
      imageSize=<32 32>
      margin=<ltrb 0.25 0.25 0.25 0.25>
      outlineSettings {
       bUseBrushTransparency=false
       roundingType=HalfHeightRadius
       width=0
       color {colorUseRule=UseColor_Specified, specifiedColor=<rgba 0 0 0 0>}
       cornerRadii=<0 0 0 1>
      }
      tintColor {colorUseRule=UseColor_Specified, specifiedColor=<rgba 0.65 0.65 0.65 1>}
      uVRegion {bIsValid=false, max=<0 0>, min=<0 0>}
     }
     pressedForeground {colorUseRule=UseColor_Foreground, specifiedColor=<rgba 1 0 1 1>}
     pressedPadding=<ltrb 0 0 0 0>
     pressedSlateSound {resourceObject=null}
    }
    foregroundColor {colorUseRule=UseColor_Specified, specifiedColor=<rgba 0.8938 1 0 1>}
    iconBrush {
     bIsDynamicallyLoaded=false
     drawAs=Image
     imageType=NoImage
     mirroring=NoMirror
     resourceName=None
     resourceObject=null
     tiling=NoTile
     imageSize=<32 32>
     margin=<ltrb 0 0 0 0>
     outlineSettings {
      bUseBrushTransparency=false
      roundingType=HalfHeightRadius
      width=0
      color {colorUseRule=UseColor_Specified, specifiedColor=<rgba 0 0 0 0>}
      cornerRadii=<0 0 0 1>
     }
     tintColor {colorUseRule=UseColor_Specified, specifiedColor=<rgba 0.0469 0.0469 0.0469 1>}
     uVRegion {bIsValid=false, max=<0 0>, min=<0 0>}
    }
    labelColor_Default=<rgba 0.0729 0.0729 0.0729 1>
   }
   slot {class=HorizontalBoxSlot, properties {padding=<ltrb 2 2 2 2>}}
  }
  BorderResults {
   widgetClassPath=&Border
   parent=VerticalBox_4
   widget=BorderResults
   properties {
    background {
     bIsDynamicallyLoaded=false
     drawAs=Image
     imageType=NoImage
     mirroring=NoMirror
     resourceName=None
     resourceObject=null
     tiling=NoTile
     imageSize=<32 32>
     margin=<ltrb 0 0 0 0>
     outlineSettings {
      bUseBrushTransparency=false
      roundingType=HalfHeightRadius
      width=0
      color {colorUseRule=UseColor_Specified, specifiedColor=<rgba 0 0 0 0>}
      cornerRadii=<0 0 0 0>
     }
     tintColor {colorUseRule=UseColor_Specified, specifiedColor=<rgba 0.0176 0.0176 0.0176 1>}
     uVRegion {bIsValid=false, max=<0 0>, min=<0 0>}
    }
    padding=<ltrb 0 0 0 0>
   }
   slot {class=VerticalBoxSlot, properties {padding=<ltrb 2 2 2 2>, size {sizeRule=Fill, value=0.5}}}
  }
  OutputScroll {
   widgetClassPath=&ScrollBox
   bIsVariable=true
   parent=BorderResults
   widget=OutputScroll
   properties {
    widgetBarStyle {
     thickness=8
     draggedThumbImage {
      bIsDynamicallyLoaded=false
      drawAs=RoundedBox
      imageType=NoImage
      mirroring=NoMirror
      resourceName=None
      resourceObject=null
      tiling=NoTile
      imageSize=<0 0>
      margin=<ltrb 0 0 0 0>
      outlineSettings {
       bUseBrushTransparency=false
       roundingType=FixedRadius
       width=0
       color {colorUseRule=UseColor_Specified, specifiedColor=<rgba 0 0 0 0>}
       cornerRadii=<4 4 4 4>
      }
      tintColor {colorUseRule=UseColor_Specified, specifiedColor=<rgba 0.2159 0.2159 0.2159 1>}
      uVRegion {bIsValid=false, max=<0 0>, min=<0 0>}
     }
     horizontalBackgroundImage {
      bIsDynamicallyLoaded=false
      drawAs=NoDrawType
      imageType=NoImage
      mirroring=NoMirror
      resourceName=None
      resourceObject=null
      tiling=NoTile
      imageSize=<0 0>
      margin=<ltrb 0 0 0 0>
      outlineSettings {
       bUseBrushTransparency=false
       roundingType=HalfHeightRadius
       width=0
       color {colorUseRule=UseColor_Specified, specifiedColor=<rgba 0 0 0 0>}
       cornerRadii=<0 0 0 0>
      }
      tintColor {colorUseRule=UseColor_Specified, specifiedColor=<rgba 1 1 1 1>}
      uVRegion {bIsValid=false, max=<0 0>, min=<0 0>}
     }
     horizontalBottomSlotImage {
      bIsDynamicallyLoaded=false
      drawAs=NoDrawType
      imageType=NoImage
      mirroring=NoMirror
      resourceName=None
      resourceObject=null
      tiling=NoTile
      imageSize=<0 0>
      margin=<ltrb 0 0 0 0>
      outlineSettings {
       bUseBrushTransparency=false
       roundingType=HalfHeightRadius
       width=0
       color {colorUseRule=UseColor_Specified, specifiedColor=<rgba 0 0 0 0>}
       cornerRadii=<0 0 0 0>
      }
      tintColor {colorUseRule=UseColor_Specified, specifiedColor=<rgba 1 1 1 1>}
      uVRegion {bIsValid=false, max=<0 0>, min=<0 0>}
     }
     horizontalTopSlotImage {
      bIsDynamicallyLoaded=false
      drawAs=NoDrawType
      imageType=NoImage
      mirroring=NoMirror
      resourceName=None
      resourceObject=null
      tiling=NoTile
      imageSize=<0 0>
      margin=<ltrb 0 0 0 0>
      outlineSettings {
       bUseBrushTransparency=false
       roundingType=HalfHeightRadius
       width=0
       color {colorUseRule=UseColor_Specified, specifiedColor=<rgba 0 0 0 0>}
       cornerRadii=<0 0 0 0>
      }
      tintColor {colorUseRule=UseColor_Specified, specifiedColor=<rgba 1 1 1 1>}
      uVRegion {bIsValid=false, max=<0 0>, min=<0 0>}
     }
     hoveredThumbImage {
      bIsDynamicallyLoaded=false
      drawAs=RoundedBox
      imageType=NoImage
      mirroring=NoMirror
      resourceName=None
      resourceObject=null
      tiling=NoTile
      imageSize=<0 0>
      margin=<ltrb 0 0 0 0>
      outlineSettings {
       bUseBrushTransparency=false
       roundingType=FixedRadius
       width=0
       color {colorUseRule=UseColor_Specified, specifiedColor=<rgba 0 0 0 0>}
       cornerRadii=<4 4 4 4>
      }
      tintColor {colorUseRule=UseColor_Specified, specifiedColor=<rgba 0.2159 0.2159 0.2159 1>}
      uVRegion {bIsValid=false, max=<0 0>, min=<0 0>}
     }
     normalThumbImage {
      bIsDynamicallyLoaded=false
      drawAs=RoundedBox
      imageType=NoImage
      mirroring=NoMirror
      resourceName=None
      resourceObject=null
      tiling=NoTile
      imageSize=<0 0>
      margin=<ltrb 0 0 0 0>
      outlineSettings {
       bUseBrushTransparency=false
       roundingType=FixedRadius
       width=0
       color {colorUseRule=UseColor_Specified, specifiedColor=<rgba 0 0 0 0>}
       cornerRadii=<4 4 4 4>
      }
      tintColor {colorUseRule=UseColor_Specified, specifiedColor=<rgba 0.0953 0.0953 0.0953 1>}
      uVRegion {bIsValid=false, max=<0 0>, min=<0 0>}
     }
     verticalBackgroundImage {
      bIsDynamicallyLoaded=false
      drawAs=NoDrawType
      imageType=NoImage
      mirroring=NoMirror
      resourceName=None
      resourceObject=null
      tiling=NoTile
      imageSize=<0 0>
      margin=<ltrb 0 0 0 0>
      outlineSettings {
       bUseBrushTransparency=false
       roundingType=HalfHeightRadius
       width=0
       color {colorUseRule=UseColor_Specified, specifiedColor=<rgba 0 0 0 0>}
       cornerRadii=<0 0 0 0>
      }
      tintColor {colorUseRule=UseColor_Specified, specifiedColor=<rgba 1 1 1 1>}
      uVRegion {bIsValid=false, max=<0 0>, min=<0 0>}
     }
     verticalBottomSlotImage {
      bIsDynamicallyLoaded=false
      drawAs=NoDrawType
      imageType=NoImage
      mirroring=NoMirror
      resourceName=None
      resourceObject=null
      tiling=NoTile
      imageSize=<0 0>
      margin=<ltrb 0 0 0 0>
      outlineSettings {
       bUseBrushTransparency=false
       roundingType=HalfHeightRadius
       width=0
       color {colorUseRule=UseColor_Specified, specifiedColor=<rgba 0 0 0 0>}
       cornerRadii=<0 0 0 0>
      }
      tintColor {colorUseRule=UseColor_Specified, specifiedColor=<rgba 1 1 1 1>}
      uVRegion {bIsValid=false, max=<0 0>, min=<0 0>}
     }
     verticalTopSlotImage {
      bIsDynamicallyLoaded=false
      drawAs=NoDrawType
      imageType=NoImage
      mirroring=NoMirror
      resourceName=None
      resourceObject=null
      tiling=NoTile
      imageSize=<0 0>
      margin=<ltrb 0 0 0 0>
      outlineSettings {
       bUseBrushTransparency=false
       roundingType=HalfHeightRadius
       width=0
       color {colorUseRule=UseColor_Specified, specifiedColor=<rgba 0 0 0 0>}
       cornerRadii=<0 0 0 0>
      }
      tintColor {colorUseRule=UseColor_Specified, specifiedColor=<rgba 1 1 1 1>}
      uVRegion {bIsValid=false, max=<0 0>, min=<0 0>}
     }
    }
    widgetStyle {
     barThickness=8
     bottomShadowBrush {
      bIsDynamicallyLoaded=false
      drawAs=Box
      imageType=FullColor
      mirroring=NoMirror
      resourceName=../../../Engine/Content/Slate/Common/ScrollBoxShadowBottom.png
      resourceObject=null
      tiling=NoTile
      imageSize=<16 8>
      margin=<ltrb 0.5 0 0.5 1>
      outlineSettings {
       bUseBrushTransparency=false
       roundingType=HalfHeightRadius
       width=0
       color {colorUseRule=UseColor_Specified, specifiedColor=<rgba 0 0 0 0>}
       cornerRadii=<0 0 0 0>
      }
      tintColor {colorUseRule=UseColor_Specified, specifiedColor=<rgba 1 1 1 1>}
      uVRegion {bIsValid=false, max=<0 0>, min=<0 0>}
     }
     horizontalScrolledContentPadding=<ltrb 0 0 1 0>
     leftShadowBrush {
      bIsDynamicallyLoaded=false
      drawAs=Box
      imageType=FullColor
      mirroring=NoMirror
      resourceName=../../../Engine/Content/Slate/Common/ScrollBoxShadowLeft.png
      resourceObject=null
      tiling=NoTile
      imageSize=<8 16>
      margin=<ltrb 1 0.5 0 0.5>
      outlineSettings {
       bUseBrushTransparency=false
       roundingType=HalfHeightRadius
       width=0
       color {colorUseRule=UseColor_Specified, specifiedColor=<rgba 0 0 0 0>}
       cornerRadii=<0 0 0 0>
      }
      tintColor {colorUseRule=UseColor_Specified, specifiedColor=<rgba 1 1 1 1>}
      uVRegion {bIsValid=false, max=<0 0>, min=<0 0>}
     }
     rightShadowBrush {
      bIsDynamicallyLoaded=false
      drawAs=Box
      imageType=FullColor
      mirroring=NoMirror
      resourceName=../../../Engine/Content/Slate/Common/ScrollBoxShadowRight.png
      resourceObject=null
      tiling=NoTile
      imageSize=<8 16>
      margin=<ltrb 0 0.5 1 0.5>
      outlineSettings {
       bUseBrushTransparency=false
       roundingType=HalfHeightRadius
       width=0
       color {colorUseRule=UseColor_Specified, specifiedColor=<rgba 0 0 0 0>}
       cornerRadii=<0 0 0 0>
      }
      tintColor {colorUseRule=UseColor_Specified, specifiedColor=<rgba 1 1 1 1>}
      uVRegion {bIsValid=false, max=<0 0>, min=<0 0>}
     }
     topShadowBrush {
      bIsDynamicallyLoaded=false
      drawAs=Box
      imageType=FullColor
      mirroring=NoMirror
      resourceName=../../../Engine/Content/Slate/Common/ScrollBoxShadowTop.png
      resourceObject=null
      tiling=NoTile
      imageSize=<16 8>
      margin=<ltrb 0.5 1 0.5 0>
      outlineSettings {
       bUseBrushTransparency=false
       roundingType=HalfHeightRadius
       width=0
       color {colorUseRule=UseColor_Specified, specifiedColor=<rgba 0 0 0 0>}
       cornerRadii=<0 0 0 0>
      }
      tintColor {colorUseRule=UseColor_Specified, specifiedColor=<rgba 1 1 1 1>}
      uVRegion {bIsValid=false, max=<0 0>, min=<0 0>}
     }
     verticalScrolledContentPadding=<ltrb 0 0 0 1>
    }
   }
   slot {class=BorderSlot, properties {padding=<ltrb 0 0 0 0>}}
  }
  OutputLog {widgetClassPath=&VerticalBox, bIsVariable=true, parent=OutputScroll, widget=OutputLog, slot {class=ScrollBoxSlot}}
  Border_Buttons {
   widgetClassPath=&Border
   parent=VerticalBox_4
   widget=Border_Buttons
   properties {
    horizontalAlignment=HAlign_Right
    verticalAlignment=VAlign_Center
    background {
     bIsDynamicallyLoaded=false
     drawAs=RoundedBox
     imageType=NoImage
     mirroring=NoMirror
     resourceName=None
     resourceObject=null
     tiling=NoTile
     imageSize=<32 32>
     margin=<ltrb 0 0 0 0>
     outlineSettings {
      bUseBrushTransparency=false
      roundingType=HalfHeightRadius
      width=0
      color {colorUseRule=UseColor_Specified, specifiedColor=<rgba 0 0 0 0>}
      cornerRadii=<0 0 0 0>
     }
     tintColor {colorUseRule=UseColor_Specified, specifiedColor=<rgba 1 1 1 1>}
     uVRegion {bIsValid=false, max=<0 0>, min=<0 0>}
    }
    brushColor=<rgba 0.0097 0.0097 0.0097 1>
   }
   slot {class=VerticalBoxSlot, properties {horizontalAlignment=HAlign_Right, verticalAlignment=VAlign_Bottom, padding=<ltrb 0.1 0.1 0.1 0.1>}}
  }
  ButtonsHBox {
   widgetClassPath=&HorizontalBox
   parent=Border_Buttons
   widget=ButtonsHBox
   slot {class=BorderSlot, properties {horizontalAlignment=HAlign_Right, verticalAlignment=VAlign_Center}}
  }
  bttn_help {
   widgetClassPath=&TA_Generic01_Button
   parent=ButtonsHBox
   widget=bttn_help
   properties {
    label=Help
    toolTipText=Some help in the log
    buttonColor_Default=<rgba 0.0097 0.0097 0.0097 1>
    buttonStyle_Default {
     clickedSlateSound {resourceObject=null}
     disabled {
      bIsDynamicallyLoaded=false
      drawAs=Box
      imageType=NoImage
      mirroring=NoMirror
      resourceName=None
      resourceObject=&MI_TA_ButtonBG_02_Down
      tiling=NoTile
      imageSize=<32 32>
      margin=<ltrb 0.25 0.25 0.25 0.25>
      outlineSettings {
       bUseBrushTransparency=false
       roundingType=HalfHeightRadius
       width=0
       color {colorUseRule=UseColor_Specified, specifiedColor=<rgba 0 0 0 0>}
       cornerRadii=<0 0 0 1>
      }
      tintColor {colorUseRule=UseColor_Specified, specifiedColor=<rgba 0.3333 0.3333 0.3333 1>}
      uVRegion {bIsValid=false, max=<0 0>, min=<0 0>}
     }
     disabledForeground {colorUseRule=UseColor_Foreground, specifiedColor=<rgba 1 0 1 1>}
     hovered {
      bIsDynamicallyLoaded=false
      drawAs=Box
      imageType=NoImage
      mirroring=NoMirror
      resourceName=None
      resourceObject=&MI_TA_ButtonBG_02_Over
      tiling=NoTile
      imageSize=<32 32>
      margin=<ltrb 0.25 0.25 0.25 0.25>
      outlineSettings {
       bUseBrushTransparency=false
       roundingType=HalfHeightRadius
       width=0
       color {colorUseRule=UseColor_Specified, specifiedColor=<rgba 0 0 0 0>}
       cornerRadii=<0 0 0 1>
      }
      tintColor {colorUseRule=UseColor_Specified, specifiedColor=<rgba 0.9 0.9 0.9 1>}
      uVRegion {bIsValid=false, max=<0 0>, min=<0 0>}
     }
     hoveredForeground {colorUseRule=UseColor_Foreground, specifiedColor=<rgba 1 0 1 1>}
     hoveredSlateSound {resourceObject=null}
     normal {
      bIsDynamicallyLoaded=false
      drawAs=Box
      imageType=NoImage
      mirroring=NoMirror
      resourceName=None
      resourceObject=&M_TA_ButtonBG_02_Up
      tiling=NoTile
      imageSize=<32 32>
      margin=<ltrb 0.25 0.25 0.25 0.25>
      outlineSettings {
       bUseBrushTransparency=false
       roundingType=HalfHeightRadius
       width=0
       color {colorUseRule=UseColor_Specified, specifiedColor=<rgba 0 0 0 0>}
       cornerRadii=<0 0 0 1>
      }
      tintColor {colorUseRule=UseColor_Specified, specifiedColor=<rgba 0.85 0.85 0.85 1>}
      uVRegion {bIsValid=false, max=<0 0>, min=<0 0>}
     }
     normalForeground {colorUseRule=UseColor_Foreground, specifiedColor=<rgba 1 0 1 1>}
     normalPadding=<ltrb 0 0 0 0>
     pressed {
      bIsDynamicallyLoaded=false
      drawAs=Box
      imageType=NoImage
      mirroring=NoMirror
      resourceName=None
      resourceObject=&MI_TA_ButtonBG_02_Down
      tiling=NoTile
      imageSize=<32 32>
      margin=<ltrb 0.25 0.25 0.25 0.25>
      outlineSettings {
       bUseBrushTransparency=false
       roundingType=HalfHeightRadius
       width=0
       color {colorUseRule=UseColor_Specified, specifiedColor=<rgba 0 0 0 0>}
       cornerRadii=<0 0 0 1>
      }
      tintColor {colorUseRule=UseColor_Specified, specifiedColor=<rgba 0.65 0.094 0.0811 0.84>}
      uVRegion {bIsValid=false, max=<0 0>, min=<0 0>}
     }
     pressedForeground {colorUseRule=UseColor_Foreground, specifiedColor=<rgba 1 0 1 1>}
     pressedPadding=<ltrb 0 0 0 0>
     pressedSlateSound {resourceObject=null}
    }
   }
   slot {class=HorizontalBoxSlot, properties {padding=<ltrb 5 3 5 3>}}
  }
  bttn_reload_script {
   widgetClassPath=&TA_Generic01_Button
   parent=ButtonsHBox
   widget=bttn_reload_script
   properties {
    label=Reload Script
    toolTipText="Typically not needed, more for development of tool."
    buttonColor_Default=<rgba 0.0097 0.0097 0.0097 1>
    buttonStyle_Default {
     clickedSlateSound {resourceObject=null}
     disabled {
      bIsDynamicallyLoaded=false
      drawAs=Box
      imageType=NoImage
      mirroring=NoMirror
      resourceName=None
      resourceObject=&MI_TA_ButtonBG_02_Down
      tiling=NoTile
      imageSize=<32 32>
      margin=<ltrb 0.25 0.25 0.25 0.25>
      outlineSettings {
       bUseBrushTransparency=false
       roundingType=HalfHeightRadius
       width=0
       color {colorUseRule=UseColor_Specified, specifiedColor=<rgba 0 0 0 0>}
       cornerRadii=<0 0 0 1>
      }
      tintColor {colorUseRule=UseColor_Specified, specifiedColor=<rgba 0.3333 0.3333 0.3333 1>}
      uVRegion {bIsValid=false, max=<0 0>, min=<0 0>}
     }
     disabledForeground {colorUseRule=UseColor_Foreground, specifiedColor=<rgba 1 0 1 1>}
     hovered {
      bIsDynamicallyLoaded=false
      drawAs=Box
      imageType=NoImage
      mirroring=NoMirror
      resourceName=None
      resourceObject=&MI_TA_ButtonBG_02_Over
      tiling=NoTile
      imageSize=<32 32>
      margin=<ltrb 0.25 0.25 0.25 0.25>
      outlineSettings {
       bUseBrushTransparency=false
       roundingType=HalfHeightRadius
       width=0
       color {colorUseRule=UseColor_Specified, specifiedColor=<rgba 0 0 0 0>}
       cornerRadii=<0 0 0 1>
      }
      tintColor {colorUseRule=UseColor_Specified, specifiedColor=<rgba 0.9 0.9 0.9 1>}
      uVRegion {bIsValid=false, max=<0 0>, min=<0 0>}
     }
     hoveredForeground {colorUseRule=UseColor_Foreground, specifiedColor=<rgba 1 0 1 1>}
     hoveredSlateSound {resourceObject=null}
     normal {
      bIsDynamicallyLoaded=false
      drawAs=Box
      imageType=NoImage
      mirroring=NoMirror
      resourceName=None
      resourceObject=&M_TA_ButtonBG_02_Up
      tiling=NoTile
      imageSize=<32 32>
      margin=<ltrb 0.25 0.25 0.25 0.25>
      outlineSettings {
       bUseBrushTransparency=false
       roundingType=HalfHeightRadius
       width=0
       color {colorUseRule=UseColor_Specified, specifiedColor=<rgba 0 0 0 0>}
       cornerRadii=<0 0 0 1>
      }
      tintColor {colorUseRule=UseColor_Specified, specifiedColor=<rgba 0.85 0.85 0.85 1>}
      uVRegion {bIsValid=false, max=<0 0>, min=<0 0>}
     }
     normalForeground {colorUseRule=UseColor_Foreground, specifiedColor=<rgba 1 0 1 1>}
     normalPadding=<ltrb 0 0 0 0>
     pressed {
      bIsDynamicallyLoaded=false
      drawAs=Box
      imageType=NoImage
      mirroring=NoMirror
      resourceName=None
      resourceObject=&MI_TA_ButtonBG_02_Down
      tiling=NoTile
      imageSize=<32 32>
      margin=<ltrb 0.25 0.25 0.25 0.25>
      outlineSettings {
       bUseBrushTransparency=false
       roundingType=HalfHeightRadius
       width=0
       color {colorUseRule=UseColor_Specified, specifiedColor=<rgba 0 0 0 0>}
       cornerRadii=<0 0 0 1>
      }
      tintColor {colorUseRule=UseColor_Specified, specifiedColor=<rgba 0.65 0.094 0.0811 0.84>}
      uVRegion {bIsValid=false, max=<0 0>, min=<0 0>}
     }
     pressedForeground {colorUseRule=UseColor_Foreground, specifiedColor=<rgba 1 0 1 1>}
     pressedPadding=<ltrb 0 0 0 0>
     pressedSlateSound {resourceObject=null}
    }
   }
   slot {class=HorizontalBoxSlot, properties {padding=<ltrb 5 3 5 3>}}
  }
  bttn_copyLog {
   widgetClassPath=&TA_Generic01_Button
   parent=ButtonsHBox
   widget=bttn_copyLog
   properties {
    label=Copy Log
    toolTipText=Copy current tool log to editor log
    buttonColor_Default=<rgba 0.0097 0.0097 0.0097 1>
    buttonStyle_Default {
     clickedSlateSound {resourceObject=null}
     disabled {
      bIsDynamicallyLoaded=false
      drawAs=Box
      imageType=NoImage
      mirroring=NoMirror
      resourceName=None
      resourceObject=&MI_TA_ButtonBG_02_Down
      tiling=NoTile
      imageSize=<32 32>
      margin=<ltrb 0.25 0.25 0.25 0.25>
      outlineSettings {
       bUseBrushTransparency=false
       roundingType=HalfHeightRadius
       width=0
       color {colorUseRule=UseColor_Specified, specifiedColor=<rgba 0 0 0 0>}
       cornerRadii=<0 0 0 1>
      }
      tintColor {colorUseRule=UseColor_Specified, specifiedColor=<rgba 0.3333 0.3333 0.3333 1>}
      uVRegion {bIsValid=false, max=<0 0>, min=<0 0>}
     }
     disabledForeground {colorUseRule=UseColor_Foreground, specifiedColor=<rgba 1 0 1 1>}
     hovered {
      bIsDynamicallyLoaded=false
      drawAs=Box
      imageType=NoImage
      mirroring=NoMirror
      resourceName=None
      resourceObject=&MI_TA_ButtonBG_02_Over
      tiling=NoTile
      imageSize=<32 32>
      margin=<ltrb 0.25 0.25 0.25 0.25>
      outlineSettings {
       bUseBrushTransparency=false
       roundingType=HalfHeightRadius
       width=0
       color {colorUseRule=UseColor_Specified, specifiedColor=<rgba 0 0 0 0>}
       cornerRadii=<0 0 0 1>
      }
      tintColor {colorUseRule=UseColor_Specified, specifiedColor=<rgba 0.9 0.9 0.9 1>}
      uVRegion {bIsValid=false, max=<0 0>, min=<0 0>}
     }
     hoveredForeground {colorUseRule=UseColor_Foreground, specifiedColor=<rgba 1 0 1 1>}
     hoveredSlateSound {resourceObject=null}
     normal {
      bIsDynamicallyLoaded=false
      drawAs=Box
      imageType=NoImage
      mirroring=NoMirror
      resourceName=None
      resourceObject=&M_TA_ButtonBG_02_Up
      tiling=NoTile
      imageSize=<32 32>
      margin=<ltrb 0.25 0.25 0.25 0.25>
      outlineSettings {
       bUseBrushTransparency=false
       roundingType=HalfHeightRadius
       width=0
       color {colorUseRule=UseColor_Specified, specifiedColor=<rgba 0 0 0 0>}
       cornerRadii=<0 0 0 1>
      }
      tintColor {colorUseRule=UseColor_Specified, specifiedColor=<rgba 0.85 0.85 0.85 1>}
      uVRegion {bIsValid=false, max=<0 0>, min=<0 0>}
     }
     normalForeground {colorUseRule=UseColor_Foreground, specifiedColor=<rgba 1 0 1 1>}
     normalPadding=<ltrb 0 0 0 0>
     pressed {
      bIsDynamicallyLoaded=false
      drawAs=Box
      imageType=NoImage
      mirroring=NoMirror
      resourceName=None
      resourceObject=&MI_TA_ButtonBG_02_Down
      tiling=NoTile
      imageSize=<32 32>
      margin=<ltrb 0.25 0.25 0.25 0.25>
      outlineSettings {
       bUseBrushTransparency=false
       roundingType=HalfHeightRadius
       width=0
       color {colorUseRule=UseColor_Specified, specifiedColor=<rgba 0 0 0 0>}
       cornerRadii=<0 0 0 1>
      }
      tintColor {colorUseRule=UseColor_Specified, specifiedColor=<rgba 0.65 0.094 0.0811 0.84>}
      uVRegion {bIsValid=false, max=<0 0>, min=<0 0>}
     }
     pressedForeground {colorUseRule=UseColor_Foreground, specifiedColor=<rgba 1 0 1 1>}
     pressedPadding=<ltrb 0 0 0 0>
     pressedSlateSound {resourceObject=null}
    }
   }
   slot {class=HorizontalBoxSlot, properties {padding=<ltrb 5 3 5 3>}}
  }
  btnClearOutput {
   widgetClassPath=&TA_Generic01_Button
   parent=ButtonsHBox
   widget=btnClearOutput
   properties {
    label=Clear Log
    buttonColor_Default=<rgba 0.0097 0.0097 0.0097 1>
    buttonStyle_Default {
     clickedSlateSound {resourceObject=null}
     disabled {
      bIsDynamicallyLoaded=false
      drawAs=Box
      imageType=NoImage
      mirroring=NoMirror
      resourceName=None
      resourceObject=&MI_TA_ButtonBG_02_Down
      tiling=NoTile
      imageSize=<32 32>
      margin=<ltrb 0.25 0.25 0.25 0.25>
      outlineSettings {
       bUseBrushTransparency=false
       roundingType=HalfHeightRadius
       width=0
       color {colorUseRule=UseColor_Specified, specifiedColor=<rgba 0 0 0 0>}
       cornerRadii=<0 0 0 1>
      }
      tintColor {colorUseRule=UseColor_Specified, specifiedColor=<rgba 0.3333 0.3333 0.3333 1>}
      uVRegion {bIsValid=false, max=<0 0>, min=<0 0>}
     }
     disabledForeground {colorUseRule=UseColor_Foreground, specifiedColor=<rgba 1 0 1 1>}
     hovered {
      bIsDynamicallyLoaded=false
      drawAs=Box
      imageType=NoImage
      mirroring=NoMirror
      resourceName=None
      resourceObject=&MI_TA_ButtonBG_02_Over
      tiling=NoTile
      imageSize=<32 32>
      margin=<ltrb 0.25 0.25 0.25 0.25>
      outlineSettings {
       bUseBrushTransparency=false
       roundingType=HalfHeightRadius
       width=0
       color {colorUseRule=UseColor_Specified, specifiedColor=<rgba 0 0 0 0>}
       cornerRadii=<0 0 0 1>
      }
      tintColor {colorUseRule=UseColor_Specified, specifiedColor=<rgba 0.9 0.9 0.9 1>}
      uVRegion {bIsValid=false, max=<0 0>, min=<0 0>}
     }
     hoveredForeground {colorUseRule=UseColor_Foreground, specifiedColor=<rgba 1 0 1 1>}
     hoveredSlateSound {resourceObject=null}
     normal {
      bIsDynamicallyLoaded=false
      drawAs=Box
      imageType=NoImage
      mirroring=NoMirror
      resourceName=None
      resourceObject=&M_TA_ButtonBG_02_Up
      tiling=NoTile
      imageSize=<32 32>
      margin=<ltrb 0.25 0.25 0.25 0.25>
      outlineSettings {
       bUseBrushTransparency=false
       roundingType=HalfHeightRadius
       width=0
       color {colorUseRule=UseColor_Specified, specifiedColor=<rgba 0 0 0 0>}
       cornerRadii=<0 0 0 1>
      }
      tintColor {colorUseRule=UseColor_Specified, specifiedColor=<rgba 0.0048 0.0048 0.0048 1>}
      uVRegion {bIsValid=false, max=<0 0>, min=<0 0>}
     }
     normalForeground {colorUseRule=UseColor_Foreground, specifiedColor=<rgba 1 0 1 1>}
     normalPadding=<ltrb 0 0 0 0>
     pressed {
      bIsDynamicallyLoaded=false
      drawAs=Box
      imageType=NoImage
      mirroring=NoMirror
      resourceName=None
      resourceObject=&MI_TA_ButtonBG_02_Down
      tiling=NoTile
      imageSize=<32 32>
      margin=<ltrb 0.25 0.25 0.25 0.25>
      outlineSettings {
       bUseBrushTransparency=false
       roundingType=HalfHeightRadius
       width=0
       color {colorUseRule=UseColor_Specified, specifiedColor=<rgba 0 0 0 0>}
       cornerRadii=<0 0 0 1>
      }
      tintColor {colorUseRule=UseColor_Specified, specifiedColor=<rgba 0.65 0.65 0.65 1>}
      uVRegion {bIsValid=false, max=<0 0>, min=<0 0>}
     }
     pressedForeground {colorUseRule=UseColor_Foreground, specifiedColor=<rgba 1 0 1 1>}
     pressedPadding=<ltrb 0 0 0 0>
     pressedSlateSound {resourceObject=null}
    }
   }
   slot {class=HorizontalBoxSlot, properties {padding=<ltrb 5 3 5 3>}}
  }
  HorizontalBox_3 {
   widgetClassPath=&HorizontalBox
   parent=ButtonsHBox
   widget=HorizontalBox_3
   slot {class=HorizontalBoxSlot, properties {horizontalAlignment=HAlign_Right, verticalAlignment=VAlign_Center, padding=<ltrb 4 0 4 0>}}
  }
  textField_Log {
   widgetClassPath=&CommonTextBlock
   parent=HorizontalBox_3
   widget=textField_Log
   properties {
    justification=Right
    text=Output Log:
    toolTipText="Change verbosity of output in Editor Log (not the log above.)"
    colorAndOpacity {colorUseRule=UseColor_Specified, specifiedColor=<rgba 0.849 0.849 0.849 1>}
    font {
     bForceMonospaced=false
     bMaterialIsStencil=false
     fontMaterial=null
     fontObject=&NotoSans
     letterSpacing=0
     monospacedWidth=1
     size=9
     skewAmount=0
     typefaceFontName=NotoSans-Regular
     outlineSettings {
      bApplyOutlineToDropShadows=false
      bMiteredCorners=false
      bSeparateFillAlpha=false
      outlineMaterial=null
      outlineSize=0
      outlineColor=<rgba 0 0 0 1>
     }
    }
    shadowColorAndOpacity=<rgba 0 0 0 0.5>
   }
   slot {class=HorizontalBoxSlot, properties {horizontalAlignment=HAlign_Center, verticalAlignment=VAlign_Center}}
  }
  SizeBox_0 {
   widgetClassPath=&SizeBox
   parent=HorizontalBox_3
   widget=SizeBox_0
   properties {bOverride_MinDesiredWidth=true, minDesiredWidth=90}
   slot {class=HorizontalBoxSlot, properties {size {sizeRule=Automatic, value=2}}}
  }
  ComboBoxLog {
   widgetClassPath=&ComboBoxString
   bIsVariable=true
   parent=SizeBox_0
   widget=ComboBoxLog
   properties {
    selectedOption=Verbose
    toolTipText="Change verbosity of output in Editor Log (not the log above.)"
    defaultOptions [Off,Error,Warn,Log,Verbose]
    font {
     bForceMonospaced=false
     bMaterialIsStencil=false
     fontMaterial=null
     fontObject=&Roboto
     letterSpacing=0
     monospacedWidth=1
     size=10
     skewAmount=0
     typefaceFontName=Bold
     outlineSettings {
      bApplyOutlineToDropShadows=false
      bMiteredCorners=false
      bSeparateFillAlpha=false
      outlineMaterial=null
      outlineSize=0
      outlineColor=<rgba 0 0 0 1>
     }
    }
    foregroundColor {colorUseRule=UseColor_Specified, specifiedColor=<rgba 0.5271 0.5271 0.5271 1>}
    itemStyle {
     bUseParentRowBrush=false
     activeBrush {
      bIsDynamicallyLoaded=false
      drawAs=Image
      imageType=NoImage
      mirroring=NoMirror
      resourceName=None
      resourceObject=null
      tiling=NoTile
      imageSize=<0 0>
      margin=<ltrb 0 0 0 0>
      outlineSettings {
       bUseBrushTransparency=false
       roundingType=HalfHeightRadius
       width=0
       color {colorUseRule=UseColor_Specified, specifiedColor=<rgba 0 0 0 0>}
       cornerRadii=<0 0 0 0>
      }
      tintColor {colorUseRule=UseColor_Specified, specifiedColor=<rgba 0 0.162 0.7454 1>}
      uVRegion {bIsValid=false, max=<0 0>, min=<0 0>}
     }
     activeHighlightedBrush {
      bIsDynamicallyLoaded=false
      drawAs=Image
      imageType=NoImage
      mirroring=NoMirror
      resourceName=None
      resourceObject=null
      tiling=NoTile
      imageSize=<0 0>
      margin=<ltrb 0 0 0 0>
      outlineSettings {
       bUseBrushTransparency=false
       roundingType=HalfHeightRadius
       width=0
       color {colorUseRule=UseColor_Specified, specifiedColor=<rgba 0 0 0 0>}
       cornerRadii=<0 0 0 0>
      }
      tintColor {colorUseRule=UseColor_Specified, specifiedColor=<rgba 0.0252 0.0319 0.0423 1>}
      uVRegion {bIsValid=false, max=<0 0>, min=<0 0>}
     }
     activeHoveredBrush {
      bIsDynamicallyLoaded=false
      drawAs=Image
      imageType=NoImage
      mirroring=NoMirror
      resourceName=None
      resourceObject=null
      tiling=NoTile
      imageSize=<0 0>
      margin=<ltrb 0 0 0 0>
      outlineSettings {
       bUseBrushTransparency=false
       roundingType=HalfHeightRadius
       width=0
       color {colorUseRule=UseColor_Specified, specifiedColor=<rgba 0 0 0 0>}
       cornerRadii=<0 0 0 0>
      }
      tintColor {colorUseRule=UseColor_Specified, specifiedColor=<rgba 0 0.162 0.7454 1>}
      uVRegion {bIsValid=false, max=<0 0>, min=<0 0>}
     }
     dropIndicator_Above {
      bIsDynamicallyLoaded=false
      drawAs=Box
      imageType=FullColor
      mirroring=NoMirror
      resourceName=../../../Engine/Content/Slate/Common/DropZoneIndicator_Above.png
      resourceObject=null
      tiling=NoTile
      imageSize=<0 0>
      margin=<ltrb 0.625 0.625 0 0>
      outlineSettings {
       bUseBrushTransparency=false
       roundingType=HalfHeightRadius
       width=0
       color {colorUseRule=UseColor_Specified, specifiedColor=<rgba 0 0 0 0>}
       cornerRadii=<0 0 0 0>
      }
      tintColor {colorUseRule=UseColor_Specified, specifiedColor=<rgba 0.0091 0.3515 0.9647 1>}
      uVRegion {bIsValid=false, max=<0 0>, min=<0 0>}
     }
     dropIndicator_Below {
      bIsDynamicallyLoaded=false
      drawAs=Box
      imageType=FullColor
      mirroring=NoMirror
      resourceName=../../../Engine/Content/Slate/Common/DropZoneIndicator_Below.png
      resourceObject=null
      tiling=NoTile
      imageSize=<0 0>
      margin=<ltrb 0.625 0 0 0.625>
      outlineSettings {
       bUseBrushTransparency=false
       roundingType=HalfHeightRadius
       width=0
       color {colorUseRule=UseColor_Specified, specifiedColor=<rgba 0 0 0 0>}
       cornerRadii=<0 0 0 0>
      }
      tintColor {colorUseRule=UseColor_Specified, specifiedColor=<rgba 0.0091 0.3515 0.9647 1>}
      uVRegion {bIsValid=false, max=<0 0>, min=<0 0>}
     }
     dropIndicator_Onto {
      bIsDynamicallyLoaded=false
      drawAs=Box
      imageType=FullColor
      mirroring=NoMirror
      resourceName=../../../Engine/Content/Slate/Common/DropZoneIndicator_Onto.png
      resourceObject=null
      tiling=NoTile
      imageSize=<0 0>
      margin=<ltrb 0.25 0.25 0.25 0.25>
      outlineSettings {
       bUseBrushTransparency=false
       roundingType=HalfHeightRadius
       width=0
       color {colorUseRule=UseColor_Specified, specifiedColor=<rgba 0 0 0 0>}
       cornerRadii=<0 0 0 0>
      }
      tintColor {colorUseRule=UseColor_Specified, specifiedColor=<rgba 0.0091 0.3515 0.9647 1>}
      uVRegion {bIsValid=false, max=<0 0>, min=<0 0>}
     }
     evenRowBackgroundBrush {
      bIsDynamicallyLoaded=false
      drawAs=Image
      imageType=NoImage
      mirroring=NoMirror
      resourceName=None
      resourceObject=null
      tiling=NoTile
      imageSize=<0 0>
      margin=<ltrb 0 0 0 0>
      outlineSettings {
       bUseBrushTransparency=false
       roundingType=HalfHeightRadius
       width=0
       color {colorUseRule=UseColor_Specified, specifiedColor=<rgba 0 0 0 0>}
       cornerRadii=<0 0 0 0>
      }
      tintColor {colorUseRule=UseColor_Specified, specifiedColor=<rgba 0.0103 0.0103 0.0103 1>}
      uVRegion {bIsValid=false, max=<0 0>, min=<0 0>}
     }
     evenRowBackgroundHoveredBrush {
      bIsDynamicallyLoaded=false
      drawAs=Image
      imageType=NoImage
      mirroring=NoMirror
      resourceName=None
      resourceObject=null
      tiling=NoTile
      imageSize=<0 0>
      margin=<ltrb 0 0 0 0>
      outlineSettings {
       bUseBrushTransparency=false
       roundingType=HalfHeightRadius
       width=0
       color {colorUseRule=UseColor_Specified, specifiedColor=<rgba 0 0 0 0>}
       cornerRadii=<0 0 0 0>
      }
      tintColor {colorUseRule=UseColor_Specified, specifiedColor=<rgba 0.0176 0.0176 0.0176 1>}
      uVRegion {bIsValid=false, max=<0 0>, min=<0 0>}
     }
     inactiveBrush {
      bIsDynamicallyLoaded=false
      drawAs=Image
      imageType=NoImage
      mirroring=NoMirror
      resourceName=None
      resourceObject=null
      tiling=NoTile
      imageSize=<0 0>
      margin=<ltrb 0 0 0 0>
      outlineSettings {
       bUseBrushTransparency=false
       roundingType=HalfHeightRadius
       width=0
       color {colorUseRule=UseColor_Specified, specifiedColor=<rgba 0 0 0 0>}
       cornerRadii=<0 0 0 0>
      }
      tintColor {colorUseRule=UseColor_Specified, specifiedColor=<rgba 0.0513 0.0953 0.159 1>}
      uVRegion {bIsValid=false, max=<0 0>, min=<0 0>}
     }
     inactiveHighlightedBrush {
      bIsDynamicallyLoaded=false
      drawAs=Image
      imageType=NoImage
      mirroring=NoMirror
      resourceName=None
      resourceObject=null
      tiling=NoTile
      imageSize=<0 0>
      margin=<ltrb 0 0 0 0>
      outlineSettings {
       bUseBrushTransparency=false
       roundingType=HalfHeightRadius
       width=0
       color {colorUseRule=UseColor_Specified, specifiedColor=<rgba 0 0 0 0>}
       cornerRadii=<0 0 0 0>
      }
      tintColor {colorUseRule=UseColor_Specified, specifiedColor=<rgba 0.0252 0.0319 0.0423 1>}
      uVRegion {bIsValid=false, max=<0 0>, min=<0 0>}
     }
     inactiveHoveredBrush {
      bIsDynamicallyLoaded=false
      drawAs=Image
      imageType=NoImage
      mirroring=NoMirror
      resourceName=None
      resourceObject=null
      tiling=NoTile
      imageSize=<0 0>
      margin=<ltrb 0 0 0 0>
      outlineSettings {
       bUseBrushTransparency=false
       roundingType=HalfHeightRadius
       width=0
       color {colorUseRule=UseColor_Specified, specifiedColor=<rgba 0 0 0 0>}
       cornerRadii=<0 0 0 0>
      }
      tintColor {colorUseRule=UseColor_Specified, specifiedColor=<rgba 0.0513 0.0953 0.159 1>}
      uVRegion {bIsValid=false, max=<0 0>, min=<0 0>}
     }
     oddRowBackgroundBrush {
      bIsDynamicallyLoaded=false
      drawAs=Image
      imageType=NoImage
      mirroring=NoMirror
      resourceName=None
      resourceObject=null
      tiling=NoTile
      imageSize=<0 0>
      margin=<ltrb 0 0 0 0>
      outlineSettings {
       bUseBrushTransparency=false
       roundingType=HalfHeightRadius
       width=0
       color {colorUseRule=UseColor_Specified, specifiedColor=<rgba 0 0 0 0>}
       cornerRadii=<0 0 0 0>
      }
      tintColor {colorUseRule=UseColor_Specified, specifiedColor=<rgba 0.0103 0.0103 0.0103 1>}
      uVRegion {bIsValid=false, max=<0 0>, min=<0 0>}
     }
     oddRowBackgroundHoveredBrush {
      bIsDynamicallyLoaded=false
      drawAs=Image
      imageType=NoImage
      mirroring=NoMirror
      resourceName=None
      resourceObject=null
      tiling=NoTile
      imageSize=<0 0>
      margin=<ltrb 0 0 0 0>
      outlineSettings {
       bUseBrushTransparency=false
       roundingType=HalfHeightRadius
       width=0
       color {colorUseRule=UseColor_Specified, specifiedColor=<rgba 0 0 0 0>}
       cornerRadii=<0 0 0 0>
      }
      tintColor {colorUseRule=UseColor_Specified, specifiedColor=<rgba 0.0176 0.0176 0.0176 1>}
      uVRegion {bIsValid=false, max=<0 0>, min=<0 0>}
     }
     parentRowBackgroundBrush {
      bIsDynamicallyLoaded=false
      drawAs=Image
      imageType=NoImage
      mirroring=NoMirror
      resourceName=None
      resourceObject=null
      tiling=NoTile
      imageSize=<32 32>
      margin=<ltrb 0 0 0 0>
      outlineSettings {
       bUseBrushTransparency=false
       roundingType=HalfHeightRadius
       width=0
       color {colorUseRule=UseColor_Specified, specifiedColor=<rgba 0 0 0 0>}
       cornerRadii=<0 0 0 0>
      }
      tintColor {colorUseRule=UseColor_Specified, specifiedColor=<rgba 1 1 1 1>}
      uVRegion {bIsValid=false, max=<0 0>, min=<0 0>}
     }
     parentRowBackgroundHoveredBrush {
      bIsDynamicallyLoaded=false
      drawAs=Image
      imageType=NoImage
      mirroring=NoMirror
      resourceName=None
      resourceObject=null
      tiling=NoTile
      imageSize=<32 32>
      margin=<ltrb 0 0 0 0>
      outlineSettings {
       bUseBrushTransparency=false
       roundingType=HalfHeightRadius
       width=0
       color {colorUseRule=UseColor_Specified, specifiedColor=<rgba 0 0 0 0>}
       cornerRadii=<0 0 0 0>
      }
      tintColor {colorUseRule=UseColor_Specified, specifiedColor=<rgba 1 1 1 1>}
      uVRegion {bIsValid=false, max=<0 0>, min=<0 0>}
     }
     selectedTextColor {colorUseRule=UseColor_Specified, specifiedColor=<rgba 1 1 1 1>}
     selectorFocusedBrush {
      bIsDynamicallyLoaded=false
      drawAs=Border
      imageType=FullColor
      mirroring=NoMirror
      resourceName=../../../Engine/Content/Slate/Common/Selector.png
      resourceObject=null
      tiling=Both
      imageSize=<0 0>
      margin=<ltrb 0.25 0.25 0.25 0.25>
      outlineSettings {
       bUseBrushTransparency=false
       roundingType=HalfHeightRadius
       width=0
       color {colorUseRule=UseColor_Specified, specifiedColor=<rgba 0 0 0 0>}
       cornerRadii=<0 0 0 0>
      }
      tintColor {colorUseRule=UseColor_Specified, specifiedColor=<rgba 0.701 0.225 0.003 1>}
      uVRegion {bIsValid=false, max=<0 0>, min=<0 0>}
     }
     textColor {colorUseRule=UseColor_Specified, specifiedColor=<rgba 0.5271 0.5271 0.5271 1>}
    }
    scrollBarStyle {
     thickness=8
     draggedThumbImage {
      bIsDynamicallyLoaded=false
      drawAs=RoundedBox
      imageType=NoImage
      mirroring=NoMirror
      resourceName=None
      resourceObject=null
      tiling=NoTile
      imageSize=<0 0>
      margin=<ltrb 0 0 0 0>
      outlineSettings {
       bUseBrushTransparency=false
       roundingType=FixedRadius
       width=0
       color {colorUseRule=UseColor_Specified, specifiedColor=<rgba 0 0 0 0>}
       cornerRadii=<4 4 4 4>
      }
      tintColor {colorUseRule=UseColor_Specified, specifiedColor=<rgba 0.2159 0.2159 0.2159 1>}
      uVRegion {bIsValid=false, max=<0 0>, min=<0 0>}
     }
     horizontalBackgroundImage {
      bIsDynamicallyLoaded=false
      drawAs=NoDrawType
      imageType=NoImage
      mirroring=NoMirror
      resourceName=None
      resourceObject=null
      tiling=NoTile
      imageSize=<0 0>
      margin=<ltrb 0 0 0 0>
      outlineSettings {
       bUseBrushTransparency=false
       roundingType=HalfHeightRadius
       width=0
       color {colorUseRule=UseColor_Specified, specifiedColor=<rgba 0 0 0 0>}
       cornerRadii=<0 0 0 0>
      }
      tintColor {colorUseRule=UseColor_Specified, specifiedColor=<rgba 1 1 1 1>}
      uVRegion {bIsValid=false, max=<0 0>, min=<0 0>}
     }
     horizontalBottomSlotImage {
      bIsDynamicallyLoaded=false
      drawAs=NoDrawType
      imageType=NoImage
      mirroring=NoMirror
      resourceName=None
      resourceObject=null
      tiling=NoTile
      imageSize=<0 0>
      margin=<ltrb 0 0 0 0>
      outlineSettings {
       bUseBrushTransparency=false
       roundingType=HalfHeightRadius
       width=0
       color {colorUseRule=UseColor_Specified, specifiedColor=<rgba 0 0 0 0>}
       cornerRadii=<0 0 0 0>
      }
      tintColor {colorUseRule=UseColor_Specified, specifiedColor=<rgba 1 1 1 1>}
      uVRegion {bIsValid=false, max=<0 0>, min=<0 0>}
     }
     horizontalTopSlotImage {
      bIsDynamicallyLoaded=false
      drawAs=NoDrawType
      imageType=NoImage
      mirroring=NoMirror
      resourceName=None
      resourceObject=null
      tiling=NoTile
      imageSize=<0 0>
      margin=<ltrb 0 0 0 0>
      outlineSettings {
       bUseBrushTransparency=false
       roundingType=HalfHeightRadius
       width=0
       color {colorUseRule=UseColor_Specified, specifiedColor=<rgba 0 0 0 0>}
       cornerRadii=<0 0 0 0>
      }
      tintColor {colorUseRule=UseColor_Specified, specifiedColor=<rgba 1 1 1 1>}
      uVRegion {bIsValid=false, max=<0 0>, min=<0 0>}
     }
     hoveredThumbImage {
      bIsDynamicallyLoaded=false
      drawAs=RoundedBox
      imageType=NoImage
      mirroring=NoMirror
      resourceName=None
      resourceObject=null
      tiling=NoTile
      imageSize=<0 0>
      margin=<ltrb 0 0 0 0>
      outlineSettings {
       bUseBrushTransparency=false
       roundingType=FixedRadius
       width=0
       color {colorUseRule=UseColor_Specified, specifiedColor=<rgba 0 0 0 0>}
       cornerRadii=<4 4 4 4>
      }
      tintColor {colorUseRule=UseColor_Specified, specifiedColor=<rgba 0.2159 0.2159 0.2159 1>}
      uVRegion {bIsValid=false, max=<0 0>, min=<0 0>}
     }
     normalThumbImage {
      bIsDynamicallyLoaded=false
      drawAs=RoundedBox
      imageType=NoImage
      mirroring=NoMirror
      resourceName=None
      resourceObject=null
      tiling=NoTile
      imageSize=<0 0>
      margin=<ltrb 0 0 0 0>
      outlineSettings {
       bUseBrushTransparency=false
       roundingType=FixedRadius
       width=0
       color {colorUseRule=UseColor_Specified, specifiedColor=<rgba 0 0 0 0>}
       cornerRadii=<4 4 4 4>
      }
      tintColor {colorUseRule=UseColor_Specified, specifiedColor=<rgba 0.0953 0.0953 0.0953 1>}
      uVRegion {bIsValid=false, max=<0 0>, min=<0 0>}
     }
     verticalBackgroundImage {
      bIsDynamicallyLoaded=false
      drawAs=NoDrawType
      imageType=NoImage
      mirroring=NoMirror
      resourceName=None
      resourceObject=null
      tiling=NoTile
      imageSize=<0 0>
      margin=<ltrb 0 0 0 0>
      outlineSettings {
       bUseBrushTransparency=false
       roundingType=HalfHeightRadius
       width=0
       color {colorUseRule=UseColor_Specified, specifiedColor=<rgba 0 0 0 0>}
       cornerRadii=<0 0 0 0>
      }
      tintColor {colorUseRule=UseColor_Specified, specifiedColor=<rgba 1 1 1 1>}
      uVRegion {bIsValid=false, max=<0 0>, min=<0 0>}
     }
     verticalBottomSlotImage {
      bIsDynamicallyLoaded=false
      drawAs=NoDrawType
      imageType=NoImage
      mirroring=NoMirror
      resourceName=None
      resourceObject=null
      tiling=NoTile
      imageSize=<0 0>
      margin=<ltrb 0 0 0 0>
      outlineSettings {
       bUseBrushTransparency=false
       roundingType=HalfHeightRadius
       width=0
       color {colorUseRule=UseColor_Specified, specifiedColor=<rgba 0 0 0 0>}
       cornerRadii=<0 0 0 0>
      }
      tintColor {colorUseRule=UseColor_Specified, specifiedColor=<rgba 1 1 1 1>}
      uVRegion {bIsValid=false, max=<0 0>, min=<0 0>}
     }
     verticalTopSlotImage {
      bIsDynamicallyLoaded=false
      drawAs=NoDrawType
      imageType=NoImage
      mirroring=NoMirror
      resourceName=None
      resourceObject=null
      tiling=NoTile
      imageSize=<0 0>
      margin=<ltrb 0 0 0 0>
      outlineSettings {
       bUseBrushTransparency=false
       roundingType=HalfHeightRadius
       width=0
       color {colorUseRule=UseColor_Specified, specifiedColor=<rgba 0 0 0 0>}
       cornerRadii=<0 0 0 0>
      }
      tintColor {colorUseRule=UseColor_Specified, specifiedColor=<rgba 1 1 1 1>}
      uVRegion {bIsValid=false, max=<0 0>, min=<0 0>}
     }
    }
    widgetStyle {
     comboButtonStyle {
      downArrowAlign=VAlign_Center
      buttonStyle {
       clickedSlateSound {resourceObject=null}
       disabled {
        bIsDynamicallyLoaded=false
        drawAs=RoundedBox
        imageType=NoImage
        mirroring=NoMirror
        resourceName=None
        resourceObject=null
        tiling=NoTile
        imageSize=<0 0>
        margin=<ltrb 0 0 0 0>
        outlineSettings {
         bUseBrushTransparency=false
         roundingType=FixedRadius
         width=1
         color {colorUseRule=UseColor_ColorTable, specifiedColor=<rgba 0 0 0 0>}
         cornerRadii=<4 4 4 4>
        }
        tintColor {colorUseRule=UseColor_Specified, specifiedColor=<rgba 0.0395 0.0395 0.0395 1>}
        uVRegion {bIsValid=false, max=<0 0>, min=<0 0>}
       }
       disabledForeground {colorUseRule=UseColor_ColorTable, specifiedColor=<rgba 0 0 0 0>}
       hovered {
        bIsDynamicallyLoaded=false
        drawAs=RoundedBox
        imageType=NoImage
        mirroring=NoMirror
        resourceName=None
        resourceObject=null
        tiling=NoTile
        imageSize=<0 0>
        margin=<ltrb 0 0 0 0>
        outlineSettings {
         bUseBrushTransparency=false
         roundingType=FixedRadius
         width=1
         color {colorUseRule=UseColor_ColorTable, specifiedColor=<rgba 0 0 0 0>}
         cornerRadii=<4 4 4 4>
        }
        tintColor {colorUseRule=UseColor_Specified, specifiedColor=<rgba 0.0048 0.0048 0.0048 1>}
        uVRegion {bIsValid=false, max=<0 0>, min=<0 0>}
       }
       hoveredForeground {colorUseRule=UseColor_ColorTable, specifiedColor=<rgba 0 0 0 0>}
       hoveredSlateSound {resourceObject=null}
       normal {
        bIsDynamicallyLoaded=false
        drawAs=RoundedBox
        imageType=NoImage
        mirroring=NoMirror
        resourceName=None
        resourceObject=null
        tiling=NoTile
        imageSize=<0 0>
        margin=<ltrb 0 0 0 0>
        outlineSettings {
         bUseBrushTransparency=false
         roundingType=FixedRadius
         width=1
         color {colorUseRule=UseColor_ColorTable, specifiedColor=<rgba 0 0 0 0>}
         cornerRadii=<4 4 4 4>
        }
        tintColor {colorUseRule=UseColor_Specified, specifiedColor=<rgba 0.0048 0.0048 0.0048 1>}
        uVRegion {bIsValid=false, max=<0 0>, min=<0 0>}
       }
       normalForeground {colorUseRule=UseColor_ColorTable, specifiedColor=<rgba 0 0 0 0>}
       normalPadding=<ltrb 8 1 8 1>
       pressed {
        bIsDynamicallyLoaded=false
        drawAs=RoundedBox
        imageType=NoImage
        mirroring=NoMirror
        resourceName=None
        resourceObject=null
        tiling=NoTile
        imageSize=<0 0>
        margin=<ltrb 0 0 0 0>
        outlineSettings {
         bUseBrushTransparency=false
         roundingType=FixedRadius
         width=1
         color {colorUseRule=UseColor_ColorTable, specifiedColor=<rgba 0 0 0 0>}
         cornerRadii=<4 4 4 4>
        }
        tintColor {colorUseRule=UseColor_Specified, specifiedColor=<rgba 0.0048 0.0048 0.0048 1>}
        uVRegion {bIsValid=false, max=<0 0>, min=<0 0>}
       }
       pressedForeground {colorUseRule=UseColor_ColorTable, specifiedColor=<rgba 0 0 0 0>}
       pressedPadding=<ltrb 8 2 8 0>
       pressedSlateSound {resourceObject=null}
      }
      contentPadding=<ltrb 0 0 0 0>
      downArrowImage {
       bIsDynamicallyLoaded=false
       drawAs=Image
       imageType=Vector
       mirroring=NoMirror
       resourceName=../../../Engine/Content/Slate/Starship/CoreWidgets/ComboBox/wide-chevron-down.svg
       resourceObject=null
       tiling=NoTile
       imageSize=<16 16>
       margin=<ltrb 0 0 0 0>
       outlineSettings {
        bUseBrushTransparency=false
        roundingType=HalfHeightRadius
        width=0
        color {colorUseRule=UseColor_Specified, specifiedColor=<rgba 0 0 0 0>}
        cornerRadii=<0 0 0 0>
       }
       tintColor {colorUseRule=UseColor_Specified, specifiedColor=<rgba 1 1 1 1>}
       uVRegion {bIsValid=false, max=<0 0>, min=<0 0>}
      }
      downArrowPadding=<ltrb 2 2 2 2>
      menuBorderBrush {
       bIsDynamicallyLoaded=false
       drawAs=RoundedBox
       imageType=NoImage
       mirroring=NoMirror
       resourceName=None
       resourceObject=null
       tiling=NoTile
       imageSize=<0 0>
       margin=<ltrb 0 0 0 0>
       outlineSettings {
        bUseBrushTransparency=false
        roundingType=FixedRadius
        width=1
        color {colorUseRule=UseColor_ColorTable, specifiedColor=<rgba 0 0 0 0>}
        cornerRadii=<0 0 0 0>
       }
       tintColor {colorUseRule=UseColor_Specified, specifiedColor=<rgba 0.0395 0.0395 0.0395 1>}
       uVRegion {bIsValid=false, max=<0 0>, min=<0 0>}
      }
      menuBorderPadding=<ltrb 0 0 0 0>
      shadowColorAndOpacity=<rgba 0 0 0 1>
      shadowOffset=<0 0>
     }
     contentPadding=<ltrb 0 0 0 0>
     menuRowPadding=<ltrb 8 3 8 3>
     pressedSlateSound {resourceObject=null}
     selectionChangeSlateSound {resourceObject=null}
    }
   }
   slot {class=SizeBoxSlot}
  }
  HorizontalBox {
   widgetClassPath=&HorizontalBox
   parent=ButtonsHBox
   widget=HorizontalBox
   slot {class=HorizontalBoxSlot, properties {horizontalAlignment=HAlign_Center, verticalAlignment=VAlign_Center, padding=<ltrb 2 0 0 0>}}
  }
  textField_ForcePass_1 {
   widgetClassPath=&CommonTextBlock
   parent=HorizontalBox
   widget=textField_ForcePass_1
   properties {
    justification=Right
    minDesiredWidth=150
    text=Show Accolade Triggers:
    toolTipText="Disable to allow faster simulation.  "
    colorAndOpacity {colorUseRule=UseColor_Specified, specifiedColor=<rgba 0.849 0.849 0.849 1>}
    font {
     bForceMonospaced=false
     bMaterialIsStencil=false
     fontMaterial=null
     fontObject=&NotoSans
     letterSpacing=0
     monospacedWidth=1
     size=9
     skewAmount=0
     typefaceFontName=NotoSans-Regular
     outlineSettings {
      bApplyOutlineToDropShadows=false
      bMiteredCorners=false
      bSeparateFillAlpha=false
      outlineMaterial=null
      outlineSize=0
      outlineColor=<rgba 0 0 0 1>
     }
    }
    shadowColorAndOpacity=<rgba 0 0 0 0.5>
   }
   slot {class=HorizontalBoxSlot, properties {horizontalAlignment=HAlign_Right, verticalAlignment=VAlign_Center, padding=<ltrb 5 0 4 0>}}
  }
  CheckBox_ShowAccolades {
   widgetClassPath=&CheckBox
   bIsVariable=true
   parent=HorizontalBox
   widget=CheckBox_ShowAccolades
   properties {
    checkedState=Checked
    toolTipText="Disable to allow faster simulation.  "
    widgetStyle {
     checkBoxType=CheckBox
     backgroundHoveredImage {
      bIsDynamicallyLoaded=false
      drawAs=RoundedBox
      imageType=NoImage
      mirroring=NoMirror
      resourceName=None
      resourceObject=null
      tiling=NoTile
      imageSize=<18 18>
      margin=<ltrb 0 0 0 0>
      outlineSettings {
       bUseBrushTransparency=false
       roundingType=FixedRadius
       width=1
       color {colorUseRule=UseColor_ColorTable, specifiedColor=<rgba 0 0 0 0>}
       cornerRadii=<3 3 3 3>
      }
      tintColor {colorUseRule=UseColor_Specified, specifiedColor=<rgba 0.5052 0.5052 0.5052 1>}
      uVRegion {bIsValid=false, max=<0 0>, min=<0 0>}
     }
     backgroundImage {
      bIsDynamicallyLoaded=false
      drawAs=RoundedBox
      imageType=NoImage
      mirroring=NoMirror
      resourceName=None
      resourceObject=null
      tiling=NoTile
      imageSize=<18 18>
      margin=<ltrb 0 0 0 0>
      outlineSettings {
       bUseBrushTransparency=false
       roundingType=FixedRadius
       width=1
       color {colorUseRule=UseColor_ColorTable, specifiedColor=<rgba 0 0 0 0>}
       cornerRadii=<3 3 3 3>
      }
      tintColor {colorUseRule=UseColor_Specified, specifiedColor=<rgba 0.3419 0.3419 0.3419 1>}
      uVRegion {bIsValid=false, max=<0 0>, min=<0 0>}
     }
     backgroundPressedImage {
      bIsDynamicallyLoaded=false
      drawAs=RoundedBox
      imageType=NoImage
      mirroring=NoMirror
      resourceName=None
      resourceObject=null
      tiling=NoTile
      imageSize=<18 18>
      margin=<ltrb 0 0 0 0>
      outlineSettings {
       bUseBrushTransparency=false
       roundingType=FixedRadius
       width=1
       color {colorUseRule=UseColor_ColorTable, specifiedColor=<rgba 0 0 0 0>}
       cornerRadii=<3 3 3 3>
      }
      tintColor {colorUseRule=UseColor_Specified, specifiedColor=<rgba 0.0048 0.0048 0.0048 1>}
      uVRegion {bIsValid=false, max=<0 0>, min=<0 0>}
     }
     borderBackgroundColor {colorUseRule=UseColor_Specified, specifiedColor=<rgba 0.3419 0.3419 0.3419 1>}
     checkedForeground {colorUseRule=UseColor_Specified, specifiedColor=<rgba 0.3419 0.3419 0.3419 1>}
     checkedHoveredForeground {colorUseRule=UseColor_Specified, specifiedColor=<rgba 0.3419 0.3419 0.3419 1>}
     checkedHoveredImage {
      bIsDynamicallyLoaded=false
      drawAs=Image
      imageType=Vector
      mirroring=NoMirror
      resourceName=../../../Engine/Content/Slate/Starship/CoreWidgets/CheckBox/check.svg
      resourceObject=null
      tiling=NoTile
      imageSize=<16 16>
      margin=<ltrb 0 0 0 0>
      outlineSettings {
       bUseBrushTransparency=false
       roundingType=HalfHeightRadius
       width=0
       color {colorUseRule=UseColor_Specified, specifiedColor=<rgba 0 0 0 0>}
       cornerRadii=<0 0 0 0>
      }
      tintColor {colorUseRule=UseColor_Specified, specifiedColor=<rgba 0.0044 0.2384 1 1>}
      uVRegion {bIsValid=false, max=<0 0>, min=<0 0>}
     }
     checkedImage {
      bIsDynamicallyLoaded=false
      drawAs=Image
      imageType=Vector
      mirroring=NoMirror
      resourceName=../../../Engine/Content/Slate/Starship/CoreWidgets/CheckBox/check.svg
      resourceObject=null
      tiling=NoTile
      imageSize=<16 16>
      margin=<ltrb 0 0 0 0>
      outlineSettings {
       bUseBrushTransparency=false
       roundingType=HalfHeightRadius
       width=0
       color {colorUseRule=UseColor_Specified, specifiedColor=<rgba 0 0 0 0>}
       cornerRadii=<0 0 0 0>
      }
      tintColor {colorUseRule=UseColor_Specified, specifiedColor=<rgba 0 0.162 0.7454 1>}
      uVRegion {bIsValid=false, max=<0 0>, min=<0 0>}
     }
     checkedPressedForeground {colorUseRule=UseColor_Specified, specifiedColor=<rgba 0.3419 0.3419 0.3419 1>}
     checkedPressedImage {
      bIsDynamicallyLoaded=false
      drawAs=Image
      imageType=Vector
      mirroring=NoMirror
      resourceName=../../../Engine/Content/Slate/Starship/CoreWidgets/CheckBox/check.svg
      resourceObject=null
      tiling=NoTile
      imageSize=<16 16>
      margin=<ltrb 0 0 0 0>
      outlineSettings {
       bUseBrushTransparency=false
       roundingType=HalfHeightRadius
       width=0
       color {colorUseRule=UseColor_Specified, specifiedColor=<rgba 0 0 0 0>}
       cornerRadii=<0 0 0 0>
      }
      tintColor {colorUseRule=UseColor_Specified, specifiedColor=<rgba 0 0.0802 0.3515 1>}
      uVRegion {bIsValid=false, max=<0 0>, min=<0 0>}
     }
     checkedSlateSound {resourceObject=null}
     foregroundColor {colorUseRule=UseColor_Specified, specifiedColor=<rgba 0.3419 0.3419 0.3419 1>}
     hoveredForeground {colorUseRule=UseColor_Specified, specifiedColor=<rgba 0.3419 0.3419 0.3419 1>}
     hoveredSlateSound {resourceObject=null}
     padding=<ltrb 2 0 0 0>
     pressedForeground {colorUseRule=UseColor_Specified, specifiedColor=<rgba 0.3419 0.3419 0.3419 1>}
     uncheckedHoveredImage {
      bIsDynamicallyLoaded=false
      drawAs=NoDrawType
      imageType=NoImage
      mirroring=NoMirror
      resourceName=None
      resourceObject=null
      tiling=NoTile
      imageSize=<0 0>
      margin=<ltrb 0 0 0 0>
      outlineSettings {
       bUseBrushTransparency=false
       roundingType=HalfHeightRadius
       width=0
       color {colorUseRule=UseColor_Specified, specifiedColor=<rgba 0 0 0 0>}
       cornerRadii=<0 0 0 0>
      }
      tintColor {colorUseRule=UseColor_Specified, specifiedColor=<rgba 1 1 1 1>}
      uVRegion {bIsValid=false, max=<0 0>, min=<0 0>}
     }
     uncheckedImage {
      bIsDynamicallyLoaded=false
      drawAs=NoDrawType
      imageType=NoImage
      mirroring=NoMirror
      resourceName=None
      resourceObject=null
      tiling=NoTile
      imageSize=<0 0>
      margin=<ltrb 0 0 0 0>
      outlineSettings {
       bUseBrushTransparency=false
       roundingType=HalfHeightRadius
       width=0
       color {colorUseRule=UseColor_Specified, specifiedColor=<rgba 0 0 0 0>}
       cornerRadii=<0 0 0 0>
      }
      tintColor {colorUseRule=UseColor_Specified, specifiedColor=<rgba 1 1 1 1>}
      uVRegion {bIsValid=false, max=<0 0>, min=<0 0>}
     }
     uncheckedPressedImage {
      bIsDynamicallyLoaded=false
      drawAs=NoDrawType
      imageType=NoImage
      mirroring=NoMirror
      resourceName=None
      resourceObject=null
      tiling=NoTile
      imageSize=<0 0>
      margin=<ltrb 0 0 0 0>
      outlineSettings {
       bUseBrushTransparency=false
       roundingType=HalfHeightRadius
       width=0
       color {colorUseRule=UseColor_Specified, specifiedColor=<rgba 0 0 0 0>}
       cornerRadii=<0 0 0 0>
      }
      tintColor {colorUseRule=UseColor_Specified, specifiedColor=<rgba 1 1 1 1>}
      uVRegion {bIsValid=false, max=<0 0>, min=<0 0>}
     }
     uncheckedSlateSound {resourceObject=null}
     undeterminedForeground {colorUseRule=UseColor_Specified, specifiedColor=<rgba 0.3419 0.3419 0.3419 1>}
     undeterminedHoveredImage {
      bIsDynamicallyLoaded=false
      drawAs=Image
      imageType=Vector
      mirroring=NoMirror
      resourceName=../../../Engine/Content/Slate/Starship/CoreWidgets/CheckBox/indeterminate.svg
      resourceObject=null
      tiling=NoTile
      imageSize=<16 16>
      margin=<ltrb 0 0 0 0>
      outlineSettings {
       bUseBrushTransparency=false
       roundingType=HalfHeightRadius
       width=0
       color {colorUseRule=UseColor_Specified, specifiedColor=<rgba 0 0 0 0>}
       cornerRadii=<0 0 0 0>
      }
      tintColor {colorUseRule=UseColor_Specified, specifiedColor=<rgba 0.0044 0.2384 1 1>}
      uVRegion {bIsValid=false, max=<0 0>, min=<0 0>}
     }
     undeterminedImage {
      bIsDynamicallyLoaded=false
      drawAs=Image
      imageType=Vector
      mirroring=NoMirror
      resourceName=../../../Engine/Content/Slate/Starship/CoreWidgets/CheckBox/indeterminate.svg
      resourceObject=null
      tiling=NoTile
      imageSize=<16 16>
      margin=<ltrb 0 0 0 0>
      outlineSettings {
       bUseBrushTransparency=false
       roundingType=HalfHeightRadius
       width=0
       color {colorUseRule=UseColor_Specified, specifiedColor=<rgba 0 0 0 0>}
       cornerRadii=<0 0 0 0>
      }
      tintColor {colorUseRule=UseColor_Specified, specifiedColor=<rgba 0 0.162 0.7454 1>}
      uVRegion {bIsValid=false, max=<0 0>, min=<0 0>}
     }
     undeterminedPressedImage {
      bIsDynamicallyLoaded=false
      drawAs=Image
      imageType=Vector
      mirroring=NoMirror
      resourceName=../../../Engine/Content/Slate/Starship/CoreWidgets/CheckBox/indeterminate.svg
      resourceObject=null
      tiling=NoTile
      imageSize=<16 16>
      margin=<ltrb 0 0 0 0>
      outlineSettings {
       bUseBrushTransparency=false
       roundingType=HalfHeightRadius
       width=0
       color {colorUseRule=UseColor_Specified, specifiedColor=<rgba 0 0 0 0>}
       cornerRadii=<0 0 0 0>
      }
      tintColor {colorUseRule=UseColor_Specified, specifiedColor=<rgba 0 0.0802 0.3515 1>}
      uVRegion {bIsValid=false, max=<0 0>, min=<0 0>}
     }
    }
   }
   slot {class=HorizontalBoxSlot}
  }
  HorizontalBox_0 {
   widgetClassPath=&HorizontalBox
   parent=ButtonsHBox
   widget=HorizontalBox_0
   slot {class=HorizontalBoxSlot, properties {horizontalAlignment=HAlign_Center, verticalAlignment=VAlign_Center, padding=<ltrb 2 0 0 0>}}
  }
  textField_ForcePass {
   widgetClassPath=&CommonTextBlock
   parent=HorizontalBox_0
   widget=textField_ForcePass
   properties {
    justification=Right
    minDesiredWidth=150
    text=DynamicXp.ForceLive:
    toolTipText="Set if data is not live (most likely needed if you do not see any payout.)  Must be set before Start."
    colorAndOpacity {colorUseRule=UseColor_Specified, specifiedColor=<rgba 0.849 0.849 0.849 1>}
    font {
     bForceMonospaced=false
     bMaterialIsStencil=false
     fontMaterial=null
     fontObject=&NotoSans
     letterSpacing=0
     monospacedWidth=1
     size=9
     skewAmount=0
     typefaceFontName=NotoSans-Regular
     outlineSettings {
      bApplyOutlineToDropShadows=false
      bMiteredCorners=false
      bSeparateFillAlpha=false
      outlineMaterial=null
      outlineSize=0
      outlineColor=<rgba 0 0 0 1>
     }
    }
    shadowColorAndOpacity=<rgba 0 0 0 0.5>
   }
   slot {class=HorizontalBoxSlot, properties {horizontalAlignment=HAlign_Right, verticalAlignment=VAlign_Center, padding=<ltrb 5 0 4 0>}}
  }
  CheckBox_ForcePass {
   widgetClassPath=&CheckBox
   bIsVariable=true
   parent=HorizontalBox_0
   widget=CheckBox_ForcePass
   properties {
    toolTipText="Set if data is not live (most likely needed if you do not see any payout.)  Must be set before Start."
    widgetStyle {
     checkBoxType=CheckBox
     backgroundHoveredImage {
      bIsDynamicallyLoaded=false
      drawAs=RoundedBox
      imageType=NoImage
      mirroring=NoMirror
      resourceName=None
      resourceObject=null
      tiling=NoTile
      imageSize=<18 18>
      margin=<ltrb 0 0 0 0>
      outlineSettings {
       bUseBrushTransparency=false
       roundingType=FixedRadius
       width=1
       color {colorUseRule=UseColor_ColorTable, specifiedColor=<rgba 0 0 0 0>}
       cornerRadii=<3 3 3 3>
      }
      tintColor {colorUseRule=UseColor_Specified, specifiedColor=<rgba 0.5052 0.5052 0.5052 1>}
      uVRegion {bIsValid=false, max=<0 0>, min=<0 0>}
     }
     backgroundImage {
      bIsDynamicallyLoaded=false
      drawAs=RoundedBox
      imageType=NoImage
      mirroring=NoMirror
      resourceName=None
      resourceObject=null
      tiling=NoTile
      imageSize=<18 18>
      margin=<ltrb 0 0 0 0>
      outlineSettings {
       bUseBrushTransparency=false
       roundingType=FixedRadius
       width=1
       color {colorUseRule=UseColor_ColorTable, specifiedColor=<rgba 0 0 0 0>}
       cornerRadii=<3 3 3 3>
      }
      tintColor {colorUseRule=UseColor_Specified, specifiedColor=<rgba 0.3419 0.3419 0.3419 1>}
      uVRegion {bIsValid=false, max=<0 0>, min=<0 0>}
     }
     backgroundPressedImage {
      bIsDynamicallyLoaded=false
      drawAs=RoundedBox
      imageType=NoImage
      mirroring=NoMirror
      resourceName=None
      resourceObject=null
      tiling=NoTile
      imageSize=<18 18>
      margin=<ltrb 0 0 0 0>
      outlineSettings {
       bUseBrushTransparency=false
       roundingType=FixedRadius
       width=1
       color {colorUseRule=UseColor_ColorTable, specifiedColor=<rgba 0 0 0 0>}
       cornerRadii=<3 3 3 3>
      }
      tintColor {colorUseRule=UseColor_Specified, specifiedColor=<rgba 0.0048 0.0048 0.0048 1>}
      uVRegion {bIsValid=false, max=<0 0>, min=<0 0>}
     }
     borderBackgroundColor {colorUseRule=UseColor_Specified, specifiedColor=<rgba 0.3419 0.3419 0.3419 1>}
     checkedForeground {colorUseRule=UseColor_Specified, specifiedColor=<rgba 0.3419 0.3419 0.3419 1>}
     checkedHoveredForeground {colorUseRule=UseColor_Specified, specifiedColor=<rgba 0.3419 0.3419 0.3419 1>}
     checkedHoveredImage {
      bIsDynamicallyLoaded=false
      drawAs=Image
      imageType=Vector
      mirroring=NoMirror
      resourceName=../../../Engine/Content/Slate/Starship/CoreWidgets/CheckBox/check.svg
      resourceObject=null
      tiling=NoTile
      imageSize=<16 16>
      margin=<ltrb 0 0 0 0>
      outlineSettings {
       bUseBrushTransparency=false
       roundingType=HalfHeightRadius
       width=0
       color {colorUseRule=UseColor_Specified, specifiedColor=<rgba 0 0 0 0>}
       cornerRadii=<0 0 0 0>
      }
      tintColor {colorUseRule=UseColor_Specified, specifiedColor=<rgba 0.0044 0.2384 1 1>}
      uVRegion {bIsValid=false, max=<0 0>, min=<0 0>}
     }
     checkedImage {
      bIsDynamicallyLoaded=false
      drawAs=Image
      imageType=Vector
      mirroring=NoMirror
      resourceName=../../../Engine/Content/Slate/Starship/CoreWidgets/CheckBox/check.svg
      resourceObject=null
      tiling=NoTile
      imageSize=<16 16>
      margin=<ltrb 0 0 0 0>
      outlineSettings {
       bUseBrushTransparency=false
       roundingType=HalfHeightRadius
       width=0
       color {colorUseRule=UseColor_Specified, specifiedColor=<rgba 0 0 0 0>}
       cornerRadii=<0 0 0 0>
      }
      tintColor {colorUseRule=UseColor_Specified, specifiedColor=<rgba 0 0.162 0.7454 1>}
      uVRegion {bIsValid=false, max=<0 0>, min=<0 0>}
     }
     checkedPressedForeground {colorUseRule=UseColor_Specified, specifiedColor=<rgba 0.3419 0.3419 0.3419 1>}
     checkedPressedImage {
      bIsDynamicallyLoaded=false
      drawAs=Image
      imageType=Vector
      mirroring=NoMirror
      resourceName=../../../Engine/Content/Slate/Starship/CoreWidgets/CheckBox/check.svg
      resourceObject=null
      tiling=NoTile
      imageSize=<16 16>
      margin=<ltrb 0 0 0 0>
      outlineSettings {
       bUseBrushTransparency=false
       roundingType=HalfHeightRadius
       width=0
       color {colorUseRule=UseColor_Specified, specifiedColor=<rgba 0 0 0 0>}
       cornerRadii=<0 0 0 0>
      }
      tintColor {colorUseRule=UseColor_Specified, specifiedColor=<rgba 0 0.0802 0.3515 1>}
      uVRegion {bIsValid=false, max=<0 0>, min=<0 0>}
     }
     checkedSlateSound {resourceObject=null}
     foregroundColor {colorUseRule=UseColor_Specified, specifiedColor=<rgba 0.3419 0.3419 0.3419 1>}
     hoveredForeground {colorUseRule=UseColor_Specified, specifiedColor=<rgba 0.3419 0.3419 0.3419 1>}
     hoveredSlateSound {resourceObject=null}
     padding=<ltrb 2 0 0 0>
     pressedForeground {colorUseRule=UseColor_Specified, specifiedColor=<rgba 0.3419 0.3419 0.3419 1>}
     uncheckedHoveredImage {
      bIsDynamicallyLoaded=false
      drawAs=NoDrawType
      imageType=NoImage
      mirroring=NoMirror
      resourceName=None
      resourceObject=null
      tiling=NoTile
      imageSize=<0 0>
      margin=<ltrb 0 0 0 0>
      outlineSettings {
       bUseBrushTransparency=false
       roundingType=HalfHeightRadius
       width=0
       color {colorUseRule=UseColor_Specified, specifiedColor=<rgba 0 0 0 0>}
       cornerRadii=<0 0 0 0>
      }
      tintColor {colorUseRule=UseColor_Specified, specifiedColor=<rgba 1 1 1 1>}
      uVRegion {bIsValid=false, max=<0 0>, min=<0 0>}
     }
     uncheckedImage {
      bIsDynamicallyLoaded=false
      drawAs=NoDrawType
      imageType=NoImage
      mirroring=NoMirror
      resourceName=None
      resourceObject=null
      tiling=NoTile
      imageSize=<0 0>
      margin=<ltrb 0 0 0 0>
      outlineSettings {
       bUseBrushTransparency=false
       roundingType=HalfHeightRadius
       width=0
       color {colorUseRule=UseColor_Specified, specifiedColor=<rgba 0 0 0 0>}
       cornerRadii=<0 0 0 0>
      }
      tintColor {colorUseRule=UseColor_Specified, specifiedColor=<rgba 1 1 1 1>}
      uVRegion {bIsValid=false, max=<0 0>, min=<0 0>}
     }
     uncheckedPressedImage {
      bIsDynamicallyLoaded=false
      drawAs=NoDrawType
      imageType=NoImage
      mirroring=NoMirror
      resourceName=None
      resourceObject=null
      tiling=NoTile
      imageSize=<0 0>
      margin=<ltrb 0 0 0 0>
      outlineSettings {
       bUseBrushTransparency=false
       roundingType=HalfHeightRadius
       width=0
       color {colorUseRule=UseColor_Specified, specifiedColor=<rgba 0 0 0 0>}
       cornerRadii=<0 0 0 0>
      }
      tintColor {colorUseRule=UseColor_Specified, specifiedColor=<rgba 1 1 1 1>}
      uVRegion {bIsValid=false, max=<0 0>, min=<0 0>}
     }
     uncheckedSlateSound {resourceObject=null}
     undeterminedForeground {colorUseRule=UseColor_Specified, specifiedColor=<rgba 0.3419 0.3419 0.3419 1>}
     undeterminedHoveredImage {
      bIsDynamicallyLoaded=false
      drawAs=Image
      imageType=Vector
      mirroring=NoMirror
      resourceName=../../../Engine/Content/Slate/Starship/CoreWidgets/CheckBox/indeterminate.svg
      resourceObject=null
      tiling=NoTile
      imageSize=<16 16>
      margin=<ltrb 0 0 0 0>
      outlineSettings {
       bUseBrushTransparency=false
       roundingType=HalfHeightRadius
       width=0
       color {colorUseRule=UseColor_Specified, specifiedColor=<rgba 0 0 0 0>}
       cornerRadii=<0 0 0 0>
      }
      tintColor {colorUseRule=UseColor_Specified, specifiedColor=<rgba 0.0044 0.2384 1 1>}
      uVRegion {bIsValid=false, max=<0 0>, min=<0 0>}
     }
     undeterminedImage {
      bIsDynamicallyLoaded=false
      drawAs=Image
      imageType=Vector
      mirroring=NoMirror
      resourceName=../../../Engine/Content/Slate/Starship/CoreWidgets/CheckBox/indeterminate.svg
      resourceObject=null
      tiling=NoTile
      imageSize=<16 16>
      margin=<ltrb 0 0 0 0>
      outlineSettings {
       bUseBrushTransparency=false
       roundingType=HalfHeightRadius
       width=0
       color {colorUseRule=UseColor_Specified, specifiedColor=<rgba 0 0 0 0>}
       cornerRadii=<0 0 0 0>
      }
      tintColor {colorUseRule=UseColor_Specified, specifiedColor=<rgba 0 0.162 0.7454 1>}
      uVRegion {bIsValid=false, max=<0 0>, min=<0 0>}
     }
     undeterminedPressedImage {
      bIsDynamicallyLoaded=false
      drawAs=Image
      imageType=Vector
      mirroring=NoMirror
      resourceName=../../../Engine/Content/Slate/Starship/CoreWidgets/CheckBox/indeterminate.svg
      resourceObject=null
      tiling=NoTile
      imageSize=<16 16>
      margin=<ltrb 0 0 0 0>
      outlineSettings {
       bUseBrushTransparency=false
       roundingType=HalfHeightRadius
       width=0
       color {colorUseRule=UseColor_Specified, specifiedColor=<rgba 0 0 0 0>}
       cornerRadii=<0 0 0 0>
      }
      tintColor {colorUseRule=UseColor_Specified, specifiedColor=<rgba 0 0.0802 0.3515 1>}
      uVRegion {bIsValid=false, max=<0 0>, min=<0 0>}
     }
    }
   }
   slot {class=HorizontalBoxSlot}
  }
  HorizontalBox_9 {
   widgetClassPath=&HorizontalBox
   parent=ButtonsHBox
   widget=HorizontalBox_9
   slot {class=HorizontalBoxSlot, properties {horizontalAlignment=HAlign_Center, verticalAlignment=VAlign_Center, padding=<ltrb 2 0 0 0>}}
  }
  textField_ForcePass_2 {
   widgetClassPath=&CommonTextBlock
   parent=HorizontalBox_9
   widget=textField_ForcePass_2
   properties {
    justification=Right
    minDesiredWidth=150
    text=DynamicXp.AllowExploits:
    toolTipText="Set if data is not live (most likely needed if you do not see any payout.)  Must be set before Start."
    colorAndOpacity {colorUseRule=UseColor_Specified, specifiedColor=<rgba 0.849 0.849 0.849 1>}
    font {
     bForceMonospaced=false
     bMaterialIsStencil=false
     fontMaterial=null
     fontObject=&NotoSans
     letterSpacing=0
     monospacedWidth=1
     size=9
     skewAmount=0
     typefaceFontName=NotoSans-Regular
     outlineSettings {
      bApplyOutlineToDropShadows=false
      bMiteredCorners=false
      bSeparateFillAlpha=false
      outlineMaterial=null
      outlineSize=0
      outlineColor=<rgba 0 0 0 1>
     }
    }
    shadowColorAndOpacity=<rgba 0 0 0 0.5>
   }
   slot {class=HorizontalBoxSlot, properties {horizontalAlignment=HAlign_Right, verticalAlignment=VAlign_Center, padding=<ltrb 5 0 4 0>}}
  }
  CheckBox_FAllowExploits {
   widgetClassPath=&CheckBox
   bIsVariable=true
   parent=HorizontalBox_9
   widget=CheckBox_FAllowExploits
   properties {
    toolTipText="Set if data is not live (most likely needed if you do not see any payout.)  Must be set before Start."
    widgetStyle {
     checkBoxType=CheckBox
     backgroundHoveredImage {
      bIsDynamicallyLoaded=false
      drawAs=RoundedBox
      imageType=NoImage
      mirroring=NoMirror
      resourceName=None
      resourceObject=null
      tiling=NoTile
      imageSize=<18 18>
      margin=<ltrb 0 0 0 0>
      outlineSettings {
       bUseBrushTransparency=false
       roundingType=FixedRadius
       width=1
       color {colorUseRule=UseColor_ColorTable, specifiedColor=<rgba 0 0 0 0>}
       cornerRadii=<3 3 3 3>
      }
      tintColor {colorUseRule=UseColor_Specified, specifiedColor=<rgba 0.5052 0.5052 0.5052 1>}
      uVRegion {bIsValid=false, max=<0 0>, min=<0 0>}
     }
     backgroundImage {
      bIsDynamicallyLoaded=false
      drawAs=RoundedBox
      imageType=NoImage
      mirroring=NoMirror
      resourceName=None
      resourceObject=null
      tiling=NoTile
      imageSize=<18 18>
      margin=<ltrb 0 0 0 0>
      outlineSettings {
       bUseBrushTransparency=false
       roundingType=FixedRadius
       width=1
       color {colorUseRule=UseColor_ColorTable, specifiedColor=<rgba 0 0 0 0>}
       cornerRadii=<3 3 3 3>
      }
      tintColor {colorUseRule=UseColor_Specified, specifiedColor=<rgba 0.3419 0.3419 0.3419 1>}
      uVRegion {bIsValid=false, max=<0 0>, min=<0 0>}
     }
     backgroundPressedImage {
      bIsDynamicallyLoaded=false
      drawAs=RoundedBox
      imageType=NoImage
      mirroring=NoMirror
      resourceName=None
      resourceObject=null
      tiling=NoTile
      imageSize=<18 18>
      margin=<ltrb 0 0 0 0>
      outlineSettings {
       bUseBrushTransparency=false
       roundingType=FixedRadius
       width=1
       color {colorUseRule=UseColor_ColorTable, specifiedColor=<rgba 0 0 0 0>}
       cornerRadii=<3 3 3 3>
      }
      tintColor {colorUseRule=UseColor_Specified, specifiedColor=<rgba 0.0048 0.0048 0.0048 1>}
      uVRegion {bIsValid=false, max=<0 0>, min=<0 0>}
     }
     borderBackgroundColor {colorUseRule=UseColor_Specified, specifiedColor=<rgba 0.3419 0.3419 0.3419 1>}
     checkedForeground {colorUseRule=UseColor_Specified, specifiedColor=<rgba 0.3419 0.3419 0.3419 1>}
     checkedHoveredForeground {colorUseRule=UseColor_Specified, specifiedColor=<rgba 0.3419 0.3419 0.3419 1>}
     checkedHoveredImage {
      bIsDynamicallyLoaded=false
      drawAs=Image
      imageType=Vector
      mirroring=NoMirror
      resourceName=../../../Engine/Content/Slate/Starship/CoreWidgets/CheckBox/check.svg
      resourceObject=null
      tiling=NoTile
      imageSize=<16 16>
      margin=<ltrb 0 0 0 0>
      outlineSettings {
       bUseBrushTransparency=false
       roundingType=HalfHeightRadius
       width=0
       color {colorUseRule=UseColor_Specified, specifiedColor=<rgba 0 0 0 0>}
       cornerRadii=<0 0 0 0>
      }
      tintColor {colorUseRule=UseColor_Specified, specifiedColor=<rgba 0.0044 0.2384 1 1>}
      uVRegion {bIsValid=false, max=<0 0>, min=<0 0>}
     }
     checkedImage {
      bIsDynamicallyLoaded=false
      drawAs=Image
      imageType=Vector
      mirroring=NoMirror
      resourceName=../../../Engine/Content/Slate/Starship/CoreWidgets/CheckBox/check.svg
      resourceObject=null
      tiling=NoTile
      imageSize=<16 16>
      margin=<ltrb 0 0 0 0>
      outlineSettings {
       bUseBrushTransparency=false
       roundingType=HalfHeightRadius
       width=0
       color {colorUseRule=UseColor_Specified, specifiedColor=<rgba 0 0 0 0>}
       cornerRadii=<0 0 0 0>
      }
      tintColor {colorUseRule=UseColor_Specified, specifiedColor=<rgba 0 0.162 0.7454 1>}
      uVRegion {bIsValid=false, max=<0 0>, min=<0 0>}
     }
     checkedPressedForeground {colorUseRule=UseColor_Specified, specifiedColor=<rgba 0.3419 0.3419 0.3419 1>}
     checkedPressedImage {
      bIsDynamicallyLoaded=false
      drawAs=Image
      imageType=Vector
      mirroring=NoMirror
      resourceName=../../../Engine/Content/Slate/Starship/CoreWidgets/CheckBox/check.svg
      resourceObject=null
      tiling=NoTile
      imageSize=<16 16>
      margin=<ltrb 0 0 0 0>
      outlineSettings {
       bUseBrushTransparency=false
       roundingType=HalfHeightRadius
       width=0
       color {colorUseRule=UseColor_Specified, specifiedColor=<rgba 0 0 0 0>}
       cornerRadii=<0 0 0 0>
      }
      tintColor {colorUseRule=UseColor_Specified, specifiedColor=<rgba 0 0.0802 0.3515 1>}
      uVRegion {bIsValid=false, max=<0 0>, min=<0 0>}
     }
     checkedSlateSound {resourceObject=null}
     foregroundColor {colorUseRule=UseColor_Specified, specifiedColor=<rgba 0.3419 0.3419 0.3419 1>}
     hoveredForeground {colorUseRule=UseColor_Specified, specifiedColor=<rgba 0.3419 0.3419 0.3419 1>}
     hoveredSlateSound {resourceObject=null}
     padding=<ltrb 2 0 0 0>
     pressedForeground {colorUseRule=UseColor_Specified, specifiedColor=<rgba 0.3419 0.3419 0.3419 1>}
     uncheckedHoveredImage {
      bIsDynamicallyLoaded=false
      drawAs=NoDrawType
      imageType=NoImage
      mirroring=NoMirror
      resourceName=None
      resourceObject=null
      tiling=NoTile
      imageSize=<0 0>
      margin=<ltrb 0 0 0 0>
      outlineSettings {
       bUseBrushTransparency=false
       roundingType=HalfHeightRadius
       width=0
       color {colorUseRule=UseColor_Specified, specifiedColor=<rgba 0 0 0 0>}
       cornerRadii=<0 0 0 0>
      }
      tintColor {colorUseRule=UseColor_Specified, specifiedColor=<rgba 1 1 1 1>}
      uVRegion {bIsValid=false, max=<0 0>, min=<0 0>}
     }
     uncheckedImage {
      bIsDynamicallyLoaded=false
      drawAs=NoDrawType
      imageType=NoImage
      mirroring=NoMirror
      resourceName=None
      resourceObject=null
      tiling=NoTile
      imageSize=<0 0>
      margin=<ltrb 0 0 0 0>
      outlineSettings {
       bUseBrushTransparency=false
       roundingType=HalfHeightRadius
       width=0
       color {colorUseRule=UseColor_Specified, specifiedColor=<rgba 0 0 0 0>}
       cornerRadii=<0 0 0 0>
      }
      tintColor {colorUseRule=UseColor_Specified, specifiedColor=<rgba 1 1 1 1>}
      uVRegion {bIsValid=false, max=<0 0>, min=<0 0>}
     }
     uncheckedPressedImage {
      bIsDynamicallyLoaded=false
      drawAs=NoDrawType
      imageType=NoImage
      mirroring=NoMirror
      resourceName=None
      resourceObject=null
      tiling=NoTile
      imageSize=<0 0>
      margin=<ltrb 0 0 0 0>
      outlineSettings {
       bUseBrushTransparency=false
       roundingType=HalfHeightRadius
       width=0
       color {colorUseRule=UseColor_Specified, specifiedColor=<rgba 0 0 0 0>}
       cornerRadii=<0 0 0 0>
      }
      tintColor {colorUseRule=UseColor_Specified, specifiedColor=<rgba 1 1 1 1>}
      uVRegion {bIsValid=false, max=<0 0>, min=<0 0>}
     }
     uncheckedSlateSound {resourceObject=null}
     undeterminedForeground {colorUseRule=UseColor_Specified, specifiedColor=<rgba 0.3419 0.3419 0.3419 1>}
     undeterminedHoveredImage {
      bIsDynamicallyLoaded=false
      drawAs=Image
      imageType=Vector
      mirroring=NoMirror
      resourceName=../../../Engine/Content/Slate/Starship/CoreWidgets/CheckBox/indeterminate.svg
      resourceObject=null
      tiling=NoTile
      imageSize=<16 16>
      margin=<ltrb 0 0 0 0>
      outlineSettings {
       bUseBrushTransparency=false
       roundingType=HalfHeightRadius
       width=0
       color {colorUseRule=UseColor_Specified, specifiedColor=<rgba 0 0 0 0>}
       cornerRadii=<0 0 0 0>
      }
      tintColor {colorUseRule=UseColor_Specified, specifiedColor=<rgba 0.0044 0.2384 1 1>}
      uVRegion {bIsValid=false, max=<0 0>, min=<0 0>}
     }
     undeterminedImage {
      bIsDynamicallyLoaded=false
      drawAs=Image
      imageType=Vector
      mirroring=NoMirror
      resourceName=../../../Engine/Content/Slate/Starship/CoreWidgets/CheckBox/indeterminate.svg
      resourceObject=null
      tiling=NoTile
      imageSize=<16 16>
      margin=<ltrb 0 0 0 0>
      outlineSettings {
       bUseBrushTransparency=false
       roundingType=HalfHeightRadius
       width=0
       color {colorUseRule=UseColor_Specified, specifiedColor=<rgba 0 0 0 0>}
       cornerRadii=<0 0 0 0>
      }
      tintColor {colorUseRule=UseColor_Specified, specifiedColor=<rgba 0 0.162 0.7454 1>}
      uVRegion {bIsValid=false, max=<0 0>, min=<0 0>}
     }
     undeterminedPressedImage {
      bIsDynamicallyLoaded=false
      drawAs=Image
      imageType=Vector
      mirroring=NoMirror
      resourceName=../../../Engine/Content/Slate/Starship/CoreWidgets/CheckBox/indeterminate.svg
      resourceObject=null
      tiling=NoTile
      imageSize=<16 16>
      margin=<ltrb 0 0 0 0>
      outlineSettings {
       bUseBrushTransparency=false
       roundingType=HalfHeightRadius
       width=0
       color {colorUseRule=UseColor_Specified, specifiedColor=<rgba 0 0 0 0>}
       cornerRadii=<0 0 0 0>
      }
      tintColor {colorUseRule=UseColor_Specified, specifiedColor=<rgba 0 0.0802 0.3515 1>}
      uVRegion {bIsValid=false, max=<0 0>, min=<0 0>}
     }
    }
   }
   slot {class=HorizontalBoxSlot}
  }
 ]
}
