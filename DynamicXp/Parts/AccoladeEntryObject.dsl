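; AccoladeEntryObject represents one Dynamic XP accolade inside a DynamicXpCalculator, tracking its trigger rate, trigger count and last XP value.
; Trigger rate scales TriggersPerMin by its own multiplier and the outer calculator's frequency multiplier; elapsed seconds become whole triggers, with leftover time carried over.
; OnTrigger records XP, sets a read-once WasTriggered flag and fires an event, alongside carry-over and error events.
; Display helpers format the name and derive a fallback color from its hex characters.
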(fn GetAccoladeColor
  (bind _color_specifiedcolor (Utilities|Struct|BreakSlateColor (Variables|GetColor_SpecifiedColor_R)))
  (bind (_color_specifiedcolor_r _color_specifiedcolor_g _color_specifiedcolor_b _color_specifiedcolor_a) (Math|Color|BreakColor _color_specifiedcolor))
  (bind _accolade_name_10_1e75c25e40038301260fb7b3b06d38f2 (Utilities|Struct|BreakDynamicXpAccolade (Variables|GetAccolade_Name_10_1E75C25E40038301260FB7B3B06D38F2)))
  (if (<= _color_specifiedcolor_a 0)
    (Variables|SetColor_SpecifiedColor_R (Utilities|Struct|MakeSlateColor (Math|Color|MakeColor (CallFunction|HextoPercent _accolade_name_10_1e75c25e40038301260fb7b3b06d38f2 0 2) (CallFunction|HextoPercent _accolade_name_10_1e75c25e40038301260fb7b3b06d38f2 2 2) (CallFunction|HextoPercent _accolade_name_10_1e75c25e40038301260fb7b3b06d38f2 4 2)))))
  (return (Variables|Meta|GetColor)))

(fn GetAccoladeDisplayName
  (params (Full :type Boolean))
  (bind (_accolade_name_10_1e75c25e40038301260fb7b3b06d38f2 _accolade_creatorname_9_829b33e7472f3ef2cd4dedbe2e515dc3 _accolade_numgroups_8_1d366eb24d1f201d9117a7ae4faf5bf0) (Utilities|Struct|BreakDynamicXpAccolade (Variables|GetAccolade_Name_10_1E75C25E40038301260FB7B3B06D38F2)))
  (if Full
    _
    (elif (Utilities|String|IsEmpty _accolade_creatorname_9_829b33e7472f3ef2cd4dedbe2e515dc3)
      (LocalVariables|SetNumLetters 16)
      (else
        (LocalVariables|SetNumLetters 4))))
  (bind _output_get (LocalVariables|SetReturnValue (Utilities|Text|FormatText "NSLOCTEXT(\"[614F29E44F365C2914D0569F639D5F94]\", \"9B6081D541ACD6C4B095EEA0BA43F70F\", \"{Id}({Groups}) {CreatorsName}\")" (Utilities|String|Left _accolade_name_10_1e75c25e40038301260fb7b3b06d38f2 (LocalVariables|GetNumLetters)) _accolade_numgroups_8_1d366eb24d1f201d9117a7ae4faf5bf0 _accolade_creatorname_9_829b33e7472f3ef2cd4dedbe2e515dc3)))
  (return _output_get))

(fn GetAccoladeName
  (bind _accolade_name_10_1e75c25e40038301260fb7b3b06d38f2 (Utilities|Struct|BreakDynamicXpAccolade (Variables|GetAccolade_Name_10_1E75C25E40038301260FB7B3B06D38F2)))
  (return _accolade_name_10_1e75c25e40038301260fb7b3b06d38f2))

(fn GetAdjustedTriggersPerMin
  (bind _asdynamic_xp_calculator (Utilities|Casting|CastToDynamicXpCalculator (Utilities|GetOuterObject self))
    (:then
      _)
    (:CastFailed
      _))
  (bind _output_get (LocalVariables|SetCalculatedTriggersPerMin (Utilities|Operators|Multiply (Variables|/DynamicXp/DynamicXpCalculator.DynamicXpCalculator_C|GetOverallFreqMult _asdynamic_xp_calculator) (Variables|Meta|GetTriggersPerMin) (Variables|Running|GetTriggerMult))))
  (if (> _output_get 0)
    (LocalVariables|SetSecondsToTriggerRet (/ 6e+01 _output_get))
    (else
      (LocalVariables|SetSecondsToTriggerRet 0.0)))
  (return _output_get (LocalVariables|GetSecondsToTriggerRet)))

(fn GetNumTiggers
  (params (SecondsPassed :type Float%20%28double-precision%29))
  (bind _triggerspermin (CallFunction|GetAdjustedTriggersPerMin))
  (bind _triggerspersec (LocalVariables|GetTriggersPerSec))
  (bind _output_get (LocalVariables|SetTriggersPerSec (/ _triggerspermin 6e+01)))
  (if (> _output_get 0)
    (LocalVariables|SetOutString (Utilities|String|Append "GetNumTriigers : TPM=" (Utilities|String|ToString(Float) _triggerspermin) "  TPS=" (Utilities|String|ToString(Float) _output_get) " Seconds=" (Utilities|String|ToString(Float) SecondsPassed) " Carry Over=" (Utilities|String|ToString(Float) (Variables|Running|GetCarryOver))))
    (bind _output_get_2 (Variables|Running|SetCarryOver (+ SecondsPassed (Variables|Running|GetCarryOver))))
    (bind _output_get_3 (LocalVariables|SetTiggers (Math|Float|Truncate (* _output_get_2 _triggerspersec))))
    (LocalVariables|SetOutString (Utilities|String|Append (LocalVariables|GetOutString) " New CarryOver=" (Utilities|String|ToString(Float) _output_get_2) " Triggers=" (Utilities|String|ToString(Integer) _output_get_3)))
    (if (>= _output_get_3 1)
      (Variables|Running|SetCarryOver (- (Variables|Running|GetCarryOver) (/ _output_get_3 _triggerspersec)))
      (else
        _))
    (Default|CallOnCarryOverChange)
    (else
      (LocalVariables|SetTiggers 0)))
  (return _output_get_3))

(fn GetWasTriggered
  (if (Variables|Default|GetWasTriggered)
    (Variables|Default|SetWasTriggered false)
    (return true)
    (else
      (return false))))

(fn HexToPercent
  (params (SourceString :type String) (Index :type Integer) (Digits :type Integer))
  (bind _returnvalue (Utilities|String|ToUpper SourceString))
  (for _index (range Index (+ (+ Index Digits) 1))
    (LocalVariables|SetMax (* (LocalVariables|GetMax) 16))
    (LocalVariables|SetAccumulator (* (LocalVariables|GetAccumulator) 16))
    (if (< (- (Utilities|String|GetCharacterasNumber _returnvalue _index) 48) 10)
      (LocalVariables|SetAccumulator (+ (LocalVariables|GetAccumulator) (- (Utilities|String|GetCharacterasNumber _returnvalue _index) 48)))
      (else
        (LocalVariables|SetAccumulator (+ (LocalVariables|GetAccumulator) (- (Utilities|String|GetCharacterasNumber _returnvalue _index) 55))))))
  (return (Math|Float|Clamp(Float) (/ (Math|Conversions|ToFloat(Integer) (LocalVariables|GetAccumulator)) (Math|Conversions|ToFloat(Integer) (LocalVariables|GetMax))) 0.1)))

(fn OnTrigger
  (params (XpAmount :type Integer))
  (Variables|Running|SetLastXpValue XpAmount)
  (Math|Integer|IncrementInt (Variables|Running|GetTriggers)
    (:%20%20
      (Variables|Default|SetWasTriggered true)
      (Default|CallOnTriggerEvent))))

(fn ShowTriggerError
  (Default|CallOnTriggerError))
